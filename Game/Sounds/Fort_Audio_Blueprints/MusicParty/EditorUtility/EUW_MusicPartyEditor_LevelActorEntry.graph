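# Editor utility widget for one row in the Music Party Editor's level actor list.
# When bound to an entry object, it registers itself there, applies the assigned color, and builds either a group title or an indented actor/controller/speaker row, dimming some actors.
# The collapse checkbox toggles two levels of child rows.
# Clicks go to the parent editor, while hover and selection drive tinting, a "Make Child" hint, and mouse-release hover actions outside PIE.
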WidgetBlueprint /Game/Sounds/Fort_Audio_Blueprints/MusicParty/EditorUtility/EUW_MusicPartyEditor_LevelActorEntry
graphs [
 EventGraph {nodes [
   "On Check State Changed (CollapseCheckbox)"!() -> (OutputDelegate, then, bIsChecked:bool) {is=title K2Node_ComponentBoundEvent}
   > OnCheckboxChanged()
   Refresh Display() {is=title K2Node_CustomEvent}
   > n1 = Branch(Condition=Get isTitleRow!(self:Music Party Editor Level Actor Entry Object Reference=Get EntryObjectCasted() -> (EntryObjectCasted:Music Party Editor Level Actor Entry Object Reference)) -> (isTitleRow:bool))
   > SetupTitleRow()
   SetupLevelActorRow(execute=@n1.else)
   n2 = Event OnListItemObjectSet()
   > n3 = Set EntryObject(EntryObject:Object Reference=@n2.ListItemObject) -> (Output_Get:Object Reference)
   > n4 = Cast To MusicPartyEditor_LevelActorEntry!(execute, Object=@n3) -> (then, CastFailed, AsMusic Party Editor Level Actor Entry:Music Party Editor Level Actor Entry Object Reference) {is=title K2Node_DynamicCast}
   > n5 = Set EntryObjectCasted(EntryObjectCasted:Music Party Editor Level Actor Entry Object Reference=@n4) -> (Output_Get:Music Party Editor Level Actor Entry Object Reference)
   > Set Widget!(
    execute
    Widget:EUW Music Party Editor Level Actor Entry Object Reference=Self-Reference()
    self:Music Party Editor Level Actor Entry Object Reference=@n5
   ) -> (then, Output_Get:EUW Music Party Editor Level Actor Entry Object Reference)
   > Refresh Display() {is=title K2Node_CallFunction}
   n6 = Get!(self:Music Party Editor Level Actor Entry Object Reference=@n5) -> (AssignedColor_R:float=~, AssignedColor_G:float=~, AssignedColor_B:float=~, AssignedColor_A:float=~)
   > Branch(Condition="Not Equal (Float)"(B=@n6.AssignedColor_A))
   n7 = Get ColorBorder() -> (ColorBorder:Border Object Reference)
   > n8 = Set!(
    execute
    ContentColorAndOpacity_R:float=@n6
    ContentColorAndOpacity_G:float=@n6.AssignedColor_G
    ContentColorAndOpacity_B:float=@n6.AssignedColor_B
    ContentColorAndOpacity_A:float=1
    self:Border Object Reference=@n7
   ) -> (then, Output_Get:Linear Color Structure)
   > SetBrushColor(self=@n7, InBrushColor=@n8)
   n9 = Event OnMouseEnter()
   > Parent: OnMouseEnter!(execute, MyGeometry:Geometry Structure=@n9.MyGeometry, MouseEvent:"Pointer Event Structure (by ref)"=@n9.MouseEvent) -> (then) {is=title K2Node_CallParentFunction}
   n10 = Get Selected() -> (Selected:bool)
   n11 = Get isPIE!(self:EUW Music Party Editor Object Reference=Get Parent() -> (Parent:EUW Music Party Editor Object Reference)) -> (isPIE:bool)
   n12 = Get EntryObjectCasted()
   n13 = Get isTitleRow(self=@n12)
   n14 = Get hasPropComponent!(self:Music Party Editor Level Actor Entry Object Reference=@n12) -> (hasPropComponent:bool)
   > n15 = Set isConditionallyHovered(isConditionallyHovered:bool=AND Boolean!(A=OR Boolean(A=@n13, B=@n14), B=NOT Boolean(A=@n10), C:bool=NOT Boolean(A=@n11)) -> (ReturnValue)) -> (Output_Get:bool)
   > GetTint!(
    execute
    self
    Selected:bool=Get Selected()
    Hovered:bool=@n15
    MouseDown:bool=Get MouseButtonIsDown!(self:EUW Music Party Editor Object Reference=Get Parent()) -> (MouseButtonIsDown:bool)
   ) -> (then)
   n16 = Event OnMouseLeave!() -> (OutputDelegate, then, MouseEvent:"Pointer Event Structure (by ref)") {is=title K2Node_Event}
   > Parent: OnMouseLeave!(execute, MouseEvent:"Pointer Event Structure (by ref)"=@n16.MouseEvent) -> (then) {is=title K2Node_CallParentFunction}
   > n17 = Set isConditionallyHovered(isConditionallyHovered=false)
   > GetTint(Selected=Get Selected(), Hovered=@n17, MouseDown=Get MouseButtonIsDown(self=Get Parent()))
  ]}
 GetTint {nodes [
   n1 = GetTint!() -> (then, Selected:bool=~, Hovered:bool=~, MouseDown:bool=~) {is=title K2Node_FunctionEntry}
   n2 = Get MouseDown() -> (MouseDown:bool)
   n3 = Get isConditionallyHovered() -> (isConditionallyHovered:bool)
   > SetVisibility(
    self=Get TextAddChild() -> (TextAddChild:Text Block Object Reference)
    InVisibility=Select(Option 0=Collapsed, Option 1=Visible, Index=AND Boolean!(A=@n3, B=@n2) -> (ReturnValue))
   )
   n4 = SelectFloat(A=0.65, B=0, bPickA=@n1.Hovered)
   n5 = Get ColorBorder()
   n6 = Get!(self:Border Object Reference=@n5) -> (ContentColorAndOpacity_R:float=~, ContentColorAndOpacity_G:float=~, ContentColorAndOpacity_B:float=~, ContentColorAndOpacity_A:float=~)
   > SetBrushColor!(
    execute
    self=@n5
    InBrushColor_R:float=@n6
    InBrushColor_G:float=@n6.ContentColorAndOpacity_G
    InBrushColor_B:float=@n6.ContentColorAndOpacity_B
    InBrushColor_A:float=Select(Option 0:double=SelectFloat(A=0.1, B=0, bPickA=@n1), Option 1:double=SelectFloat(A=0.35, B=@n4, bPickA=@n1), Index=@n1.MouseDown) -> (ReturnValue:double)
   ) -> (then)
  ]}
 OnCheckboxChanged {nodes [
   OnCheckboxChanged() {is=title K2Node_FunctionEntry}
   > Branch(Condition=IsVisible())
   > Set Checked(Checked:bool=IsChecked(self=Get CollapseCheckbox() -> (CollapseCheckbox:Check Box Object Reference))) -> (Output_Get:bool)
   > SetRenderTransformAngle!(
    execute
    self:Widget Object Reference=Get CollapseCheckbox()
    Angle:float=Select(Option 0:float=0, Option 1:float=90, Index=Get Checked() -> (Checked:bool)) -> (ReturnValue:float)
   ) -> (then)
   > n1 = For Each Loop(Array:Array of Music Party Editor Level Actor Entry Object References=Get ChildWidgets!(self:Music Party Editor Level Actor Entry Object Reference=Get EntryObjectCasted()) -> (ChildWidgets:Array of Music Party Editor Level Actor Entry Object References)) -> (Array Element:Music Party Editor Level Actor Entry Object Reference)
   n2 = Get Widget!(self:Music Party Editor Level Actor Entry Object Reference=@n1) -> (Widget:EUW Music Party Editor Level Actor Entry Object Reference)
   > Is Valid(InputObject=@n2)
   > SetVisibility(
    self=@n2
    InVisibility=Select(Option 0:ESlateVisibility Enum=Collapsed, Option 1:ESlateVisibility Enum=Visible, Index=Get Checked()) -> (ReturnValue:ESlateVisibility Enum)
   )
   > n3 = For Each Loop(Array=Get ChildWidgets(self=@n1))
   n4 = Get Widget(self=@n3)
   > Is Valid(InputObject=@n4)
   n5 = IsChecked(self=Get CollapseCheckbox!(self:EUW Music Party Editor Level Actor Entry Object Reference=@n2) -> (CollapseCheckbox))
   > SetVisibility(self=@n4, InVisibility=Select(Option 0=Collapsed, Option 1=Select(Option 0=Collapsed, Option 1=Visible, Index=@n5), Index=Get Checked()))
  ]}
 OnMouseButtonDown_0 {nodes [
   OnMouseButtonDown_0!() -> (then, MyGeometry:Geometry Structure, MouseEvent:"Pointer Event Structure (by ref)") {is=title K2Node_FunctionEntry}
   > n1 = Branch(Condition=Get isTitleRow(self=Get EntryObjectCasted()))
   n2 = Get LevelActorEntries!(self:EUW Music Party Editor Object Reference=Get Parent()) -> (LevelActorEntries:List View Object Reference)
   n3 = Branch(
    execute=@n1.else
    Condition=Contains Item(
     TargetArray:Array of User Widget Object References=GetDisplayedEntryWidgets!(self:List View Base Object Reference=@n2) -> (ReturnValue:Array of User Widget Object References)
     ItemToFind:"User Widget Object Reference (by ref)"=Self-Reference()
    )
   )
   n4 = Self-Reference()
   n5 = Get Parent()
   > n6 = OnEntryClicked!(
    execute
    self:EUW Music Party Editor Object Reference=@n5
    Widget:EUW Music Party Editor Level Actor Entry Object Reference=@n4
    DoubleClick:bool=false
    ForceAdditiveSelect:bool=false
   ) -> (then)
   Return Node!(
    execute=[@n6.then, @n7.then]
    ReturnValue:Event Reply Structure=Make Event Reply!() -> (EventReply:Event Reply Structure) {is=title K2Node_MakeStruct}
   )
   n7 = OnPIEEntryClicked!(
    execute=@n3.else
    self:EUW Music Party Editor Object Reference=@n5
    Widget:EUW Music Party Editor Level Actor Entry Object Reference=@n4
    DoubleClick:bool=false
    ForceAdditiveSelect:bool=false
   ) -> (then)
  ]}
 OnMouseButtonUp {nodes [
   n1 = OnMouseButtonUp!() -> (then, MyGeometry:Geometry Structure, MouseEvent:"Pointer Event Structure (by ref)") {is=title K2Node_FunctionEntry}
   > n2 = Parent: OnMouseButtonUp!(execute, MyGeometry:Geometry Structure=@n1, MouseEvent:"Pointer Event Structure (by ref)"=@n1.MouseEvent) -> (then, ReturnValue:Event Reply Structure) {is=title K2Node_CallParentFunction}
   n3 = Get isPIE(self=Get Parent())
   n4 = Get Selected()
   > n5 = Branch(Condition=AND Boolean!(A=Get isConditionallyHovered(), B=NOT Boolean(A=@n4), C:bool=NOT Boolean(A=@n3)) -> (ReturnValue))
   > PerformHoverAction!(
    execute
    self:EUW Music Party Editor Object Reference=Get Parent()
    LevelActorEntry:Music Party Editor Level Actor Entry Object Reference=Get EntryObjectCasted()
   ) -> (then)
   > Return Node(ReturnValue=@n2)
   Return Node(execute=@n5.else, ReturnValue=@n2)
  ]}
 OnMouseDoubleClick {nodes [
   OnMouseDoubleClick!() -> (then, MyGeometry:Geometry Structure, MouseEvent:"Pointer Event Structure (by ref)") {is=title K2Node_FunctionEntry}
   > OnEntryClicked(self=Get Parent(), Widget=Self-Reference(), DoubleClick=true, ForceAdditiveSelect=false)
   > Return Node(ReturnValue=Make Event Reply())
  ]}
 OnPostInitialize {nodes [
   n1 = OnPostInitialize!() -> (then, Parent:EUW Music Party Editor Object Reference) {is=title K2Node_FunctionEntry}
   > Set Parent(Parent:EUW Music Party Editor Object Reference=@n1) -> (Output_Get:EUW Music Party Editor Object Reference)
  ]}
 OnSelectionChanged {nodes [
   n1 = OnSelectionChanged!() -> (then, Selected:bool=~) {is=title K2Node_FunctionEntry}
   > Set Selected(Selected:bool=@n1) -> (Output_Get:bool)
   > GetTint(Selected=@n1, Hovered=IsHovered(), MouseDown=Get MouseButtonIsDown(self=Get Parent())) {is=title K2Node_CallFunction}
  ]}
 SetupLevelActorRow {nodes [
   SetupLevelActorRow() {is=title K2Node_FunctionEntry}
   > Set ToolTipText(ToolTipText:Text=MakeLiteralText(), None=Self-Reference()) -> (Output_Get:Text)
   > SetColumnFill!(
    execute
    self:Grid Panel Object Reference=Get GridPanel() -> (GridPanel:Grid Panel Object Reference)
    ColumnIndex:int=0
    Coefficient:float=0.33
   ) -> (then)
   > "SetText (Text)"(
    self=Get TextLevelActor() -> (TextLevelActor:Text Block Object Reference)
    InText=GetLevelActorDisplayName!(self:Music Party Editor Level Actor Entry Object Reference=Get EntryObjectCasted()) -> (DisplayName:Text)
   )
   > n1 = Is Valid(InputObject=Get Music Prop!(self:Music Party Editor Level Actor Entry Object Reference=Get EntryObjectCasted()) -> (Music Prop:Base Class: Music Party Prop Base Object Reference))
   SetOpacity!(execute=@n1.Is Not Valid, self:Text Object Reference=Get TextLevelActor(), InOpacity:float=0.5) -> (then)
   n2 = Length(TargetArray=Get Parents!(self:Music Party Editor Level Actor Entry Object Reference=Get EntryObjectCasted()) -> (Parents:Array of Actor Object References))
   > n3 = Set Padding!(
    execute
    Padding:Margin Structure=Select(
     Option 0:Margin Structure=Make Margin(Left=15, Top=0, Right=0, Bottom=0)
     Option 1:Margin Structure=Make Margin(Left=30, Top=0, Right=0, Bottom=0)
     Index="integer > integer"(A=@n2)
    ) -> (ReturnValue:Margin Structure)
    self:Border Object Reference=Get LevelActor() -> (LevelActor:Border Object Reference)
   ) -> (then, Output_Get:Margin Structure)
   n4 = Get MusicController!(self:Music Party Editor Level Actor Entry Object Reference=Get EntryObjectCasted()) -> (MusicController:Actor Class Reference)
   n5 = Select(
    Option 0:String
    Option 1:String=Get Class Display Name(Class=@n4)
    Index=IsValidClass!(Class:Object Class Reference=@n4) -> (ReturnValue:bool)
   ) -> (ReturnValue:String)
   "SetText (Text)"(
    execute=[@n3.then, @n13.then]
    self=Get TextMusicController() -> (TextMusicController:Text Block Object Reference)
    InText="To Text (String)"(InString=Split(SourceString=@n5, InStr=_C, SearchDir=FromEnd))
   )
   n6 = Get MusicSpeaker!(self:Music Party Editor Level Actor Entry Object Reference=Get EntryObjectCasted()) -> (MusicSpeaker:Music Party Speaker Object Reference)
   n7 = GetDisplayName(Object=@n6)
   > "SetText (Text)"(
    self=Get TextSpeaker() -> (TextSpeaker:Text Block Object Reference)
    InText=Select(Option 0:Text, Option 1:Text="To Text (String)"(InString=@n7), Index=IsValid(Object=@n6)) -> (ReturnValue:Text)
   )
   n8 = Length(TargetArray:Array of Actor Soft Object References=Get SpeakerActors!(self:Music Party Editor Level Actor Entry Object Reference=Get EntryObjectCasted()) -> (SpeakerActors:Array of Actor Soft Object References))
   > SetVisibility(
    self=Get CollapseCheckbox!() -> (CollapseCheckbox)
    InVisibility=Select(Option 0:ESlateVisibility Enum=Hidden, Option 1:ESlateVisibility Enum=Visible, Index="integer > integer"(A=@n8)) -> (ReturnValue:ESlateVisibility Enum)
   )
   n9 = Get Music Prop(self=Get EntryObjectCasted())
   n10 = Get bAutoActivate!(self:Actor Component Object Reference=@n9) -> (bAutoActivate:bool)
   n11 = Get Parents(self=Get EntryObjectCasted())
   n12 = AND Boolean!(
    A=OR Boolean(A=Get BypassMusic!(self:Base Class: Music Party Prop Base Object Reference=@n9) -> (BypassMusic:bool), B=NOT Boolean(A=@n10))
    B="integer > integer"(A=Length(TargetArray:Array of Actor Object References=@n11))
   ) -> (ReturnValue)
   SetOpacity(execute=@n1.Is Valid, self=Get TextLevelActor(), InOpacity=SelectFloat(A=0.75, B=1, bPickA=@n12))
   > n13 = Set Padding(
    Padding=Select(
     Option 0:Margin Structure=Make Margin(Left=15, Top=0, Right=0, Bottom=0)
     Option 1:Margin Structure=Make Margin(Left=30, Top=0, Right=0, Bottom=0)
     Index=@n12
    ) -> (ReturnValue:Margin Structure)
    self=Get LevelActor()
   )
  ]}
 SetupTitleRow {nodes [
   SetupTitleRow() {is=title K2Node_FunctionEntry}
   n1 = Get TitleText!(self:Music Party Editor Level Actor Entry Object Reference=Get EntryObjectCasted()) -> (TitleText:Text)
   > "SetText (Text)"(self=Get TextLevelActor(), InText=@n1)
   n2 = "To String (Text)"(InText=@n1)
   n3 = "Equal, Case Insensitive (String)"(A=@n2, B=Silenced Components)
   > Set ToolTipText(
    ToolTipText=SelectText(
     A="LOCTEXT(\"These Level Actors are not associated with any Music Prop Components\")"
     B=SelectText(A="LOCTEXT(\"These Level Actors have Music Prop Components but are silenced\")", bPickA=@n3)
     bPickA="Equal, Case Insensitive (String)"(A=@n2, B=Unassigned Level Actors)
    )
    None=Self-Reference()
   )
   n4 = Get TextLevelActor()
   n5 = Break Slate Color!(SlateColor:"Slate Color Structure (by ref)"=Get ColorAndOpacity!(self:Text Object Reference=@n4) -> (ColorAndOpacity:Slate Color Structure)) -> (
    SpecifiedColor_R:float=~
    SpecifiedColor_G:float=~
    SpecifiedColor_B:float=~
    SpecifiedColor_A:float=~
    ColorUseRule:ESlateColorStylingMode Enum
   ) {is=title K2Node_BreakStruct}
   > SetColorAndOpacity(
    self=@n4
    InColorAndOpacity=Make Slate Color!(
     SpecifiedColor_R:float=@n5
     SpecifiedColor_G:float=@n5.SpecifiedColor_G
     SpecifiedColor_B:float=@n5.SpecifiedColor_B
     SpecifiedColor_A:float=1
     ColorUseRule=@n5.ColorUseRule
    ) -> (SlateColor)
   )
   > Set Padding(Padding=Make Margin(Left=0, Top=0, Right=0, Bottom=0), self=Get LevelActor())
   > SetColumnFill(self=Get GridPanel(), ColumnIndex=0, Coefficient=1)
   > "SetText (Text)"(self=Get TextMusicController())
   > "SetText (Text)"(self=Get TextSpeaker())
   > SetVisibility(self=Get CollapseCheckbox(), InVisibility=Visible)
  ]}
]
variables [
 EntryObject {type=Object, default=null}
 EntryObjectCasted {type=MusicPartyEditor_LevelActorEntry_C, default=null}
 In Padding {type=Margin, default=<ltrb 0 0 0 0>}
 isConditionallyHovered {type=boolean}
 Parent {type=EUW_MusicPartyEditor_C, default=null}
 Selected {type=boolean}
]
widgets {
 info {inheritedWidgetCount=0, parentClass=&EditorUtilityWidget, rootWidgetClass=&CanvasPanel, widgetCount=14}
 widgets [
  CanvasPanel_178 {widgetClassPath=&CanvasPanel, widget=CanvasPanel_178}
  VerticalBox_142 {
   widgetClassPath=&VerticalBox
   parent=CanvasPanel_178
   widget=VerticalBox_142
   slot {
    class=CanvasPanelSlot
    properties {bAutoSize=true, layoutData {alignment=<0 0>, anchors {maximum=<1 1>, minimum=<0 0>}, offsets=<ltrb 0 0 0 0>}}
   }
  }
  HorizontalBox_82 {widgetClassPath=&HorizontalBox, parent=VerticalBox_142, widget=HorizontalBox_82, slot {class=VerticalBoxSlot}}
  ColorBorder {
   widgetClassPath=&Border
   bIsVariable=true
   parent=HorizontalBox_82
   widget=ColorBorder
   properties {brushColor=<rgba 1 1 1 0>, padding=<ltrb 0 0.5 0 0.5>}
   slot {class=HorizontalBoxSlot, properties {size {sizeRule=Fill, value=1}}}
  }
  GridPanel {
   widgetClassPath=&GridPanel
   bIsVariable=true
   parent=ColorBorder
   widget=GridPanel
   properties {columnFill [0.33,0.33,0.33]}
   slot {class=BorderSlot, properties {padding=<ltrb 0 0.5 0 0.5>}}
  }
  LevelActor {
   widgetClassPath=&Border
   bIsVariable=true
   parent=GridPanel
   widget=LevelActor
   properties {brushColor=<rgba 0 0 0 0.2>, padding=<ltrb 1 1 1 1>}
   slot {class=GridSlot}
  }
  HorizontalBox_0 {
   widgetClassPath=&HorizontalBox
   parent=LevelActor
   widget=HorizontalBox_0
   slot {class=BorderSlot, properties {padding=<ltrb 1 1 1 1>}}
  }
  CollapseCheckbox {
   widgetClassPath=/Script/UMG.CheckBox
   bIsVariable=true
   parent=HorizontalBox_0
   widget=CollapseCheckbox
   properties {
    checkedState=Checked
    renderTransform {angle=90, scale=<1 1>, shear=<0 0>, translation=<0 1>}
    renderTransformPivot=<0.45 0.5>
    widgetStyle {
     checkBoxType=CheckBox
     backgroundHoveredImage {
      bIsDynamicallyLoaded=false
      drawAs=NoDrawType
      imageType=NoImage
      mirroring=NoMirror
      resourceName=None
      resourceObject=null
      tiling=NoTile
      imageSize=<0 0>
      margin=<ltrb 0 0 0 0>
      outlineSettings {
       bUseBrushTransparency=false
       roundingType=HalfHeightRadius
       width=0
       color {colorUseRule=UseColor_Specified, specifiedColor=<rgba 0 0 0 0>}
       cornerRadii=<0 0 0 0>
      }
      tintColor {colorUseRule=UseColor_Specified, specifiedColor=<rgba 1 1 1 1>}
      uVRegion {bIsValid=false, max=<0 0>, min=<0 0>}
     }
     backgroundImage {
      bIsDynamicallyLoaded=false
      drawAs=NoDrawType
      imageType=NoImage
      mirroring=NoMirror
      resourceName=None
      resourceObject=null
      tiling=NoTile
      imageSize=<0 0>
      margin=<ltrb 0 0 0 0>
      outlineSettings {
       bUseBrushTransparency=false
       roundingType=HalfHeightRadius
       width=0
       color {colorUseRule=UseColor_Specified, specifiedColor=<rgba 0 0 0 0>}
       cornerRadii=<0 0 0 0>
      }
      tintColor {colorUseRule=UseColor_Specified, specifiedColor=<rgba 1 1 1 1>}
      uVRegion {bIsValid=false, max=<0 0>, min=<0 0>}
     }
     backgroundPressedImage {
      bIsDynamicallyLoaded=false
      drawAs=NoDrawType
      imageType=NoImage
      mirroring=NoMirror
      resourceName=None
      resourceObject=null
      tiling=NoTile
      imageSize=<0 0>
      margin=<ltrb 0 0 0 0>
      outlineSettings {
       bUseBrushTransparency=false
       roundingType=HalfHeightRadius
       width=0
       color {colorUseRule=UseColor_Specified, specifiedColor=<rgba 0 0 0 0>}
       cornerRadii=<0 0 0 0>
      }
      tintColor {colorUseRule=UseColor_Specified, specifiedColor=<rgba 1 1 1 1>}
      uVRegion {bIsValid=false, max=<0 0>, min=<0 0>}
     }
     borderBackgroundColor {colorUseRule=UseColor_Specified, specifiedColor=<rgba 1 1 1 1>}
     checkedForeground {colorUseRule=UseColor_Foreground, specifiedColor=<rgba 1 0 1 1>}
     checkedHoveredForeground {colorUseRule=UseColor_Foreground, specifiedColor=<rgba 1 0 1 1>}
     checkedHoveredImage {
      bIsDynamicallyLoaded=false
      drawAs=Image
      imageType=FullColor
      mirroring=NoMirror
      resourceName=../../../Engine/Content/Slate/Common/CheckBox.png
      resourceObject=/FortUILibrary/Foundation/Buttons/Materials/Arrows/M_Button_CycleArrow_Right_Hovered
      tiling=NoTile
      imageSize=<16 16>
      margin=<ltrb 0 0 0 0>
      outlineSettings {
       bUseBrushTransparency=false
       roundingType=HalfHeightRadius
       width=0
       color {colorUseRule=UseColor_Specified, specifiedColor=<rgba 0 0 0 0>}
       cornerRadii=<0 0 0 0>
      }
      tintColor {colorUseRule=UseColor_Specified, specifiedColor=<rgba 1 1 1 1>}
      uVRegion {bIsValid=false, max=<0 0>, min=<0 0>}
     }
     checkedImage {
      bIsDynamicallyLoaded=false
      drawAs=Image
      imageType=FullColor
      mirroring=NoMirror
      resourceName=../../../Engine/Content/Slate/Common/CheckBox.png
      resourceObject=/FortUILibrary/Foundation/Buttons/Materials/Arrows/M_Button_CycleArrow_Right_Base
      tiling=NoTile
      imageSize=<16 16>
      margin=<ltrb 0 0 0 0>
      outlineSettings {
       bUseBrushTransparency=false
       roundingType=HalfHeightRadius
       width=0
       color {colorUseRule=UseColor_Specified, specifiedColor=<rgba 0 0 0 0>}
       cornerRadii=<0 0 0 0>
      }
      tintColor {colorUseRule=UseColor_Specified, specifiedColor=<rgba 1 1 1 1>}
      uVRegion {bIsValid=false, max=<0 0>, min=<0 0>}
     }
     checkedPressedForeground {colorUseRule=UseColor_Foreground, specifiedColor=<rgba 1 0 1 1>}
     checkedPressedImage {
      bIsDynamicallyLoaded=false
      drawAs=Image
      imageType=FullColor
      mirroring=NoMirror
      resourceName=../../../Engine/Content/Slate/Common/CheckBox.png
      resourceObject=/FortUILibrary/Foundation/Buttons/Materials/Arrows/M_Button_CycleArrow_Right_Pressed
      tiling=NoTile
      imageSize=<16 16>
      margin=<ltrb 0 0 0 0>
      outlineSettings {
       bUseBrushTransparency=false
       roundingType=HalfHeightRadius
       width=0
       color {colorUseRule=UseColor_Specified, specifiedColor=<rgba 0 0 0 0>}
       cornerRadii=<0 0 0 0>
      }
      tintColor {colorUseRule=UseColor_Specified, specifiedColor=<rgba 1 1 1 1>}
      uVRegion {bIsValid=false, max=<0 0>, min=<0 0>}
     }
     checkedSlateSound {resourceObject=null}
     foregroundColor {colorUseRule=UseColor_Foreground, specifiedColor=<rgba 1 0 1 1>}
     hoveredForeground {colorUseRule=UseColor_Foreground, specifiedColor=<rgba 1 0 1 1>}
     hoveredSlateSound {resourceObject=null}
     padding=<ltrb 2 0 0 0>
     pressedForeground {colorUseRule=UseColor_Foreground, specifiedColor=<rgba 1 0 1 1>}
     uncheckedHoveredImage {
      bIsDynamicallyLoaded=false
      drawAs=Image
      imageType=FullColor
      mirroring=NoMirror
      resourceName=../../../Engine/Content/Slate/Common/CheckBox.png
      resourceObject=&M_Button_CycleArrow_Right_Hovered
      tiling=NoTile
      imageSize=<16 16>
      margin=<ltrb 0 0 0 0>
      outlineSettings {
       bUseBrushTransparency=false
       roundingType=HalfHeightRadius
       width=0
       color {colorUseRule=UseColor_Specified, specifiedColor=<rgba 0 0 0 0>}
       cornerRadii=<0 0 0 0>
      }
      tintColor {colorUseRule=UseColor_Specified, specifiedColor=<rgba 1 1 1 1>}
      uVRegion {bIsValid=false, max=<0 0>, min=<0 0>}
     }
     uncheckedImage {
      bIsDynamicallyLoaded=false
      drawAs=Image
      imageType=FullColor
      mirroring=NoMirror
      resourceName=../../../Engine/Content/Slate/Common/CheckBox.png
      resourceObject=&M_Button_CycleArrow_Right_Base
      tiling=NoTile
      imageSize=<16 16>
      margin=<ltrb 0 0 0 0>
      outlineSettings {
       bUseBrushTransparency=false
       roundingType=HalfHeightRadius
       width=0
       color {colorUseRule=UseColor_Specified, specifiedColor=<rgba 0 0 0 0>}
       cornerRadii=<0 0 0 0>
      }
      tintColor {colorUseRule=UseColor_Specified, specifiedColor=<rgba 1 1 1 1>}
      uVRegion {bIsValid=false, max=<0 0>, min=<0 0>}
     }
     uncheckedPressedImage {
      bIsDynamicallyLoaded=false
      drawAs=Image
      imageType=FullColor
      mirroring=NoMirror
      resourceName=../../../Engine/Content/Slate/Common/CheckBox.png
      resourceObject=&M_Button_CycleArrow_Right_Pressed
      tiling=NoTile
      imageSize=<16 16>
      margin=<ltrb 0 0 0 0>
      outlineSettings {
       bUseBrushTransparency=false
       roundingType=HalfHeightRadius
       width=0
       color {colorUseRule=UseColor_Specified, specifiedColor=<rgba 0 0 0 0>}
       cornerRadii=<0 0 0 0>
      }
      tintColor {colorUseRule=UseColor_Specified, specifiedColor=<rgba 1 1 1 1>}
      uVRegion {bIsValid=false, max=<0 0>, min=<0 0>}
     }
     uncheckedSlateSound {resourceObject=null}
     undeterminedForeground {colorUseRule=UseColor_Foreground, specifiedColor=<rgba 1 0 1 1>}
     undeterminedHoveredImage {
      bIsDynamicallyLoaded=false
      drawAs=Image
      imageType=FullColor
      mirroring=NoMirror
      resourceName=../../../Engine/Content/Slate/Common/CheckBox_Undetermined_Hovered.png
      resourceObject=null
      tiling=NoTile
      imageSize=<16 16>
      margin=<ltrb 0 0 0 0>
      outlineSettings {
       bUseBrushTransparency=false
       roundingType=HalfHeightRadius
       width=0
       color {colorUseRule=UseColor_Specified, specifiedColor=<rgba 0 0 0 0>}
       cornerRadii=<0 0 0 0>
      }
      tintColor {colorUseRule=UseColor_Specified, specifiedColor=<rgba 1 1 1 1>}
      uVRegion {bIsValid=false, max=<0 0>, min=<0 0>}
     }
     undeterminedImage {
      bIsDynamicallyLoaded=false
      drawAs=Image
      imageType=FullColor
      mirroring=NoMirror
      resourceName=../../../Engine/Content/Slate/Common/CheckBox_Undetermined.png
      resourceObject=null
      tiling=NoTile
      imageSize=<16 16>
      margin=<ltrb 0 0 0 0>
      outlineSettings {
       bUseBrushTransparency=false
       roundingType=HalfHeightRadius
       width=0
       color {colorUseRule=UseColor_Specified, specifiedColor=<rgba 0 0 0 0>}
       cornerRadii=<0 0 0 0>
      }
      tintColor {colorUseRule=UseColor_Specified, specifiedColor=<rgba 1 1 1 1>}
      uVRegion {bIsValid=false, max=<0 0>, min=<0 0>}
     }
     undeterminedPressedImage {
      bIsDynamicallyLoaded=false
      drawAs=Image
      imageType=FullColor
      mirroring=NoMirror
      resourceName=../../../Engine/Content/Slate/Common/CheckBox_Undetermined_Hovered.png
      resourceObject=null
      tiling=NoTile
      imageSize=<16 16>
      margin=<ltrb 0 0 0 0>
      outlineSettings {
       bUseBrushTransparency=false
       roundingType=HalfHeightRadius
       width=0
       color {colorUseRule=UseColor_Specified, specifiedColor=<rgba 0 0 0 0>}
       cornerRadii=<0 0 0 0>
      }
      tintColor {colorUseRule=UseColor_Specified, specifiedColor=<rgba 0.5 0.5 0.5 1>}
      uVRegion {bIsValid=false, max=<0 0>, min=<0 0>}
     }
    }
   }
   slot {class=HorizontalBoxSlot, properties {padding=<ltrb 2.5 0 0 0>}}
  }
  TextAddChild {
   widgetClassPath=&UEFN_TextBlock
   bIsVariable=true
   parent=HorizontalBox_0
   widget=TextAddChild
   properties {
    clipping=ClipToBounds
    text="Make Child -->     "
    visibility=Collapsed
    font {
     bForceMonospaced=false
     bMaterialIsStencil=false
     fontMaterial=null
     fontObject=&NotoSans
     letterSpacing=5
     monospacedWidth=1
     size=10.5
     skewAmount=0
     typefaceFontName=NotoSans-Bold
     outlineSettings {
      bApplyOutlineToDropShadows=false
      bMiteredCorners=false
      bSeparateFillAlpha=false
      outlineMaterial=null
      outlineSize=0
      outlineColor=<rgba 0 0 0 1>
     }
    }
   }
   slot {class=HorizontalBoxSlot, properties {horizontalAlignment=HAlign_Left, verticalAlignment=VAlign_Center, padding=<ltrb 2.5 0 0 0>}}
  }
  TextLevelActor {
   widgetClassPath=&UEFN_TextBlock
   bIsVariable=true
   parent=HorizontalBox_0
   widget=TextLevelActor
   properties {
    clipping=ClipToBounds
    font {
     bForceMonospaced=false
     bMaterialIsStencil=false
     fontMaterial=null
     fontObject=&NotoSans
     letterSpacing=5
     monospacedWidth=1
     size=10
     skewAmount=0
     typefaceFontName=NotoSans-Regular
     outlineSettings {
      bApplyOutlineToDropShadows=false
      bMiteredCorners=false
      bSeparateFillAlpha=false
      outlineMaterial=null
      outlineSize=0
      outlineColor=<rgba 0 0 0 1>
     }
    }
   }
   slot {class=HorizontalBoxSlot, properties {horizontalAlignment=HAlign_Left, verticalAlignment=VAlign_Center, padding=<ltrb 2.5 0 0 0>}}
  }
  MusicController {
   widgetClassPath=&Border
   parent=GridPanel
   widget=MusicController
   properties {horizontalAlignment=HAlign_Left, verticalAlignment=VAlign_Center, brushColor=<rgba 0 0 0 0.2>, padding=<ltrb 1 1 1 1>}
   slot {class=GridSlot, properties {column=1}}
  }
  TextMusicController {
   widgetClassPath=&UEFN_TextBlock
   bIsVariable=true
   parent=MusicController
   widget=TextMusicController
   properties {
    clipping=ClipToBounds
    font {
     bForceMonospaced=false
     bMaterialIsStencil=false
     fontMaterial=null
     fontObject=&NotoSans
     letterSpacing=5
     monospacedWidth=1
     size=10
     skewAmount=0
     typefaceFontName=NotoSans-Regular
     outlineSettings {
      bApplyOutlineToDropShadows=false
      bMiteredCorners=false
      bSeparateFillAlpha=false
      outlineMaterial=null
      outlineSize=0
      outlineColor=<rgba 0 0 0 1>
     }
    }
   }
   slot {class=BorderSlot, properties {horizontalAlignment=HAlign_Left, verticalAlignment=VAlign_Center, padding=<ltrb 1 1 1 1>}}
  }
  Speaker {
   widgetClassPath=&Border
   parent=GridPanel
   widget=Speaker
   properties {horizontalAlignment=HAlign_Left, verticalAlignment=VAlign_Center, brushColor=<rgba 0 0 0 0.2>, padding=<ltrb 1 1 1 1>}
   slot {class=GridSlot, properties {column=2}}
  }
  TextSpeaker {
   widgetClassPath=&UEFN_TextBlock
   bIsVariable=true
   parent=Speaker
   widget=TextSpeaker
   properties {
    clipping=ClipToBounds
    font {
     bForceMonospaced=false
     bMaterialIsStencil=false
     fontMaterial=null
     fontObject=&NotoSans
     letterSpacing=5
     monospacedWidth=1
     size=10
     skewAmount=0
     typefaceFontName=NotoSans-Regular
     outlineSettings {
      bApplyOutlineToDropShadows=false
      bMiteredCorners=false
      bSeparateFillAlpha=false
      outlineMaterial=null
      outlineSize=0
      outlineColor=<rgba 0 0 0 1>
     }
    }
   }
   slot {class=BorderSlot, properties {horizontalAlignment=HAlign_Left, verticalAlignment=VAlign_Center, padding=<ltrb 1 1 1 1>}}
  }
 ]
}
