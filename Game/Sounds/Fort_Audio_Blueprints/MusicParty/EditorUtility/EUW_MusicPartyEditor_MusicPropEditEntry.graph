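# Editor utility list entry for the Music Party editor that edits a music prop component, or several at once through a multi-edit object.
# It binds details views to the prop, speaker asset, controller class defaults and, in single edit, audio component.
# Multi-edit gathers properties from the selected components and copies changes back inside undo transactions.
# During PIE, speaker, controller and prop edits update audio in the running world.
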WidgetBlueprint /Game/Sounds/Fort_Audio_Blueprints/MusicParty/EditorUtility/EUW_MusicPartyEditor_MusicPropEditEntry
graphs [
 CheckboxAdjustment {nodes [
   n1 = CheckboxAdjustment!() -> (then, Checkbox:Check Box Object Reference, WidgetToShow:Widget Object Reference) {is=title K2Node_FunctionEntry}
   n2 = IsChecked(self=@n1)
   > SetVisibility(self=@n1.WidgetToShow, InVisibility=Select(Option 0=Collapsed, Option 1=Visible, Index=@n2))
   > SetRenderTransformAngle!(execute, self:Widget Object Reference=@n1, Angle:float=Select(Option 0:float=0, Option 1:float=90, Index=@n2) -> (ReturnValue:float)) -> (then)
  ]}
 DecibelsToLinear {nodes [
   n1 = DecibelsToLinear!() -> (then, Decibels:double=~) {is=title K2Node_FunctionEntry}
   > n2 = Branch(Condition="float >= float"(A=@n1, B=-60))
   > Return Node!(execute, Linear:double=Power!(Base:double=10, Exp:double=float / float(A=@n1, B=20)) -> (ReturnValue:double) {is=title K2Node_CallFunction})
   Return Node(execute=@n2.else, Linear=0)
  ]}
 EventGraph {nodes [
   n1 = "On Property Changed (MusicPropDetailView)"!() -> (OutputDelegate, then, PropertyName:Name=None) {is=title K2Node_ComponentBoundEvent}
   > n2 = Branch(Condition=Get isPIE() -> (isPIE:bool))
   OnPropertyChanged!(execute=@n2.else, self, PropertyName:Name=@n1.PropertyName, RefreshList:bool=true) -> (then)
   "On Check State Changed (CollapseCheckbox_Speaker)"!() -> (OutputDelegate, then, bIsChecked:bool) {is=title K2Node_ComponentBoundEvent}
   > CheckboxAdjustment!(
    execute
    self
    Checkbox:Check Box Object Reference=Get CollapseCheckbox_Speaker() -> (CollapseCheckbox_Speaker:Check Box Object Reference)
    WidgetToShow:Widget Object Reference=Get SpeakerDetailView() -> (SpeakerDetailView:Details View Object Reference)
   ) -> (then) {is=title K2Node_CallFunction}
   "On Check State Changed (CollapseCheckbox_Controller)"!() -> (OutputDelegate, then, bIsChecked:bool) {is=title K2Node_ComponentBoundEvent}
   > CheckboxAdjustment(
    Checkbox=Get CollapseCheckbox_Controller() -> (CollapseCheckbox_Controller:Check Box Object Reference)
    WidgetToShow=Get Controller_Container() -> (Controller_Container:Vertical Box Object Reference)
   )
   n3 = "On Property Changed (DepricatedAssets)"!() -> (OutputDelegate, then, PropertyName:Name=None) {is=title K2Node_ComponentBoundEvent}
   > OnPropertyChanged(PropertyName=@n3.PropertyName, RefreshList=true)
   n4 = "On Property Changed (MultiEditControllerClass)"!() -> (OutputDelegate, then, PropertyName:Name=None) {is=title K2Node_ComponentBoundEvent}
   > OnPropertyChanged(PropertyName=@n4.PropertyName, RefreshList=true)
   n5 = "On Property Changed (MultiEditorSpeakerView)"!() -> (OutputDelegate, then, PropertyName:Name=None) {is=title K2Node_ComponentBoundEvent}
   > OnPropertyChanged(PropertyName=@n5.PropertyName, RefreshList=true)
   n6 = "On Property Changed (DepricatedBehavior)"!() -> (OutputDelegate, then, PropertyName:Name=None) {is=title K2Node_ComponentBoundEvent}
   > OnPropertyChanged(PropertyName=@n6.PropertyName, RefreshList=true)
   n7 = "On Property Changed (MusicPropPIEView)"!() -> (OutputDelegate, then, PropertyName:Name=None) {is=title K2Node_ComponentBoundEvent}
   > Branch(Condition=Get isPIE())
   > OnPropertyChanged(PropertyName=@n7.PropertyName, RefreshList=true)
   > n8 = Get!(execute:Exec) -> (then:Exec, else:Exec, Parent:EUW Music Party Editor Object Reference)
   n9 = Get EntryObjectPropCasted() -> (EntryObjectPropCasted:Base Class: Music Party Prop Base Object Reference)
   n10 = Get EntryObjectMultiPropCasted() -> (EntryObjectMultiPropCasted:Music Party Editor Music Prop Multi Edit Object Reference)
   > PIE_UpdatePropComponents!(
    execute
    self:EUW Music Party Editor Object Reference=@n8
    PropComponents:Array of Base Class: Music Party Prop Base Object References=Select(
     Option 0:Array of Base Class: Music Party Prop Base Object References=Make Array("[0]":Base Class: Music Party Prop Base Object Reference=@n9) -> (Array:Array of Base Class: Music Party Prop Base Object References)
     Option 1:Array of Base Class: Music Party Prop Base Object References=Get PropComponents!(self:Music Party Editor Music Prop Multi Edit Object Reference=@n10) -> (PropComponents:Array of Base Class: Music Party Prop Base Object References)
     Index=Get isMultiEdit() -> (isMultiEdit:bool)
    ) -> (ReturnValue:Array of Base Class: Music Party Prop Base Object References)
   ) -> (then)
   "On Property Changed (SpeakerDetailView)"!() -> (OutputDelegate, then, PropertyName:Name=None) {is=title K2Node_ComponentBoundEvent}
   > n11 = Branch(Condition=Get IgnorePropertyChanges() -> (IgnorePropertyChanges:bool))
   Branch(execute=@n11.else, Condition=Get isPIE())
   > PIE_OnSpeakerDetailsUpdated()
   "On Property Changed (PIE_ControllerDetailView)"!() -> (OutputDelegate, then, PropertyName:Name=None) {is=title K2Node_ComponentBoundEvent}
   > Branch(Condition=Get isPIE())
   > PIE_OnControllerUpdated()
   ClearReferences() {is=title K2Node_CustomEvent}
   > Set EntryObject(EntryObject:Object Reference) -> (Output_Get:Object Reference)
   > Set EntryObjectMultiPropCasted(EntryObjectMultiPropCasted:Music Party Editor Music Prop Multi Edit Object Reference) -> (Output_Get:Music Party Editor Music Prop Multi Edit Object Reference)
   > Set EntryObjectPropCasted(EntryObjectPropCasted:Base Class: Music Party Prop Base Object Reference) -> (Output_Get:Base Class: Music Party Prop Base Object Reference)
   > InitializeAudioComponentVariables()
   > InitializeDetailViews()
   n12 = Event OnListItemObjectSet()
   > Set EntryObject(EntryObject=@n12.ListItemObject)
   > Set IgnorePropertyChanges(IgnorePropertyChanges:bool=true) -> (Output_Get:bool)
   > n13 = Cast To Actor!(execute, Object=@n12.ListItemObject) -> (then, CastFailed, AsActor:Actor Object Reference) {is=title K2Node_DynamicCast}
   n14 = GetComponentByClass!(
    self:Actor Object Reference=@n13
    ComponentClass:Actor Component Class Reference=/Game/Sounds/Fort_Audio_Blueprints/MusicParty/MusicPartyPropComponentBase.MusicPartyPropComponentBase_C
   ) -> (ReturnValue:Base Class: Music Party Prop Base Object Reference)
   > Is Valid(InputObject=@n14)
   > Set EntryObject(EntryObject=@n14)
   > Set EntryObjectPropCasted(EntryObjectPropCasted=@n14)
   > Set EntryObjectMultiPropCasted()
   > Set isMultiEdit(isMultiEdit:bool=false) -> (Output_Get:bool)
   > InitializeDetailViews()
   > n15 = InitializeSinglePropView!(execute, self, ListItemObject:Base Class: Music Party Prop Base Object Reference=Get EntryObjectPropCasted()) -> (then)
   CheckboxAdjustment(execute=[@n15.then, @n17.then], Checkbox=Get CollapseCheckbox_Speaker(), WidgetToShow=Get SpeakerDetailView())
   > CheckboxAdjustment(Checkbox=Get CollapseCheckbox_Controller(), WidgetToShow=Get Controller_Container())
   n16 = Cast To MusicPartyEditor_MusicProp_MultiEdit!(execute=@n13.CastFailed, Object=Get EntryObject() -> (EntryObject:Object Reference)) -> (then, CastFailed, AsMusic Party Editor Music Prop Multi Edit:Music Party Editor Music Prop Multi Edit Object Reference) {is=title K2Node_DynamicCast}
   > Set EntryObjectMultiPropCasted(EntryObjectMultiPropCasted=@n16)
   > Set EntryObjectPropCasted()
   > Set isMultiEdit(isMultiEdit=true)
   > InitializeDetailViews()
   > n17 = InitializeMultiPropView!(execute, self, ListItemObject:Music Party Editor Music Prop Multi Edit Object Reference=Get EntryObjectMultiPropCasted()) -> (then)
  ]}
 InitializeAudioComponentVariables {nodes [
   InitializeAudioComponentVariables() {is=title K2Node_FunctionEntry}
   > SetVisibility(self=Get AudioComponentRelativeLocation() -> (AudioComponentRelativeLocation:Single Property View Object Reference), InVisibility=Visible)
   > SetVisibility(self=Get AudioComponentRelativeRotation() -> (AudioComponentRelativeRotation:Single Property View Object Reference), InVisibility=Visible)
   > n1 = Branch(Condition=Get isMultiEdit())
   SetVisibility(execute=[@n4.Is Not Valid, @n1.then, @n2.else], self=Get AudioComponentRelativeLocation(), InVisibility=Collapsed)
   > SetVisibility(self=Get AudioComponentRelativeRotation(), InVisibility=Collapsed)
   > SetObject(self=Get AudioComponentRelativeLocation())
   > SetObject(self=Get AudioComponentRelativeRotation())
   > SetObject(self=Get AudioComponentOverrides() -> (AudioComponentOverrides:Details View Object Reference))
   > SetVisibility(self=Get AudioComponentOverrides(), InVisibility=Collapsed)
   > SetVisibility(self=Get AudioComponentTitle() -> (AudioComponentTitle:Text Block Object Reference), InVisibility=Collapsed)
   n2 = Get!(execute=@n1.else) -> (then, else, EntryObjectPropCasted:Base Class: Music Party Prop Base Object Reference)
   n3 = GetComponentByClass(self=GetOwner(self=@n2), ComponentClass=/Script/Engine.AudioComponent) -> (ReturnValue:Audio Component Object Reference)
   > n4 = Is Valid(InputObject=@n3)
   n5 = Get AudioComponentRelativeLocation()
   > SetObject(self=@n5, NewObject=@n3)
   > SetPropertyName!(execute, self:Single Property View Object Reference=@n5, NewPropertyName:Name=RelativeLocation) -> (then)
   n6 = Get AudioComponentRelativeRotation()
   > SetObject(self=@n6, NewObject=@n3)
   > SetPropertyName(self=@n6, NewPropertyName=RelativeRotation)
   > SetObject(self=Get AudioComponentOverrides(), NewObject=@n3)
   > SetVisibility(self=Get AudioComponentOverrides(), InVisibility=Visible)
   > SetVisibility(self=Get AudioComponentTitle(), InVisibility=Visible)
  ]}
 InitializeControllerProperties {nodes [
   n1 = InitializeControllerProperties!() -> (then, Controller:Actor Class Reference) {is=title K2Node_FunctionEntry}
   > n2 = Set ControllerClass(ControllerClass:Actor Class Reference=@n1) -> (Output_Get:Actor Class Reference)
   > n3 = GetClassDefaultObject!(execute, Class:Object Class Reference=@n2) -> (then, ReturnValue:Object Reference)
   > n4 = Is Valid(InputObject=@n3)
   SetVisibility(execute=@n4.Is Not Valid, self=Get ControllerDetails_Border() -> (ControllerDetails_Border:Border Object Reference), InVisibility=Collapsed)
   > SetObject(self=Get ControllerDetailView() -> (ControllerDetailView:Details View Object Reference))
   > SetObject(self=Get PIE_ControllerDetailView() -> (PIE_ControllerDetailView:Details View Object Reference))
   SetObject(execute=@n4.Is Valid, self=Get ControllerDetailView(), NewObject=@n3)
   > SetVisibility(self=Get ControllerDetails_Border(), InVisibility=Visible)
   > SetObject(self=Get PIE_ControllerDetailView(), NewObject=@n3)
  ]}
 InitializeDetailViews {nodes [
   InitializeDetailViews() {is=title K2Node_FunctionEntry}
   > SetObject(self=Get MusicPropPIEView() -> (MusicPropPIEView:Details View Object Reference), NewObject=Get EntryObject())
   > SetObject(self=Get MusicPropDetailView() -> (MusicPropDetailView:Details View Object Reference), NewObject=Get EntryObject())
   > SetObject(self=Get DepricatedAssets() -> (DepricatedAssets:Details View Object Reference), NewObject=Get EntryObject())
   > SetObject(self=Get DepricatedBehavior() -> (DepricatedBehavior:Details View Object Reference), NewObject=Get EntryObject())
   > SetObject(self=Get MultiEditControllerClass() -> (MultiEditControllerClass:Single Property View Object Reference), NewObject=Get EntryObject())
   > SetObject(self=Get MultiEditorSpeakerView() -> (MultiEditorSpeakerView:Details View Object Reference), NewObject=Get EntryObject())
   > n1 = Branch(Condition=Get isMultiEdit())
   n2 = Get PropComponentType!(self:Music Party Editor Music Prop Multi Edit Object Reference=Get EntryObjectMultiPropCasted()) -> (PropComponentType:EMusicPartyPropComponentType Enum=NewEnumerator0)
   > SetVisibility(
    self=Get DepricatedAssets()
    InVisibility=Select!(
     NewEnumerator0:ESlateVisibility Enum=Collapsed
     NewEnumerator1:ESlateVisibility Enum=Collapsed
     NewEnumerator2:ESlateVisibility Enum=Collapsed
     NewEnumerator3:ESlateVisibility Enum=Visible
     Index:EMusicPartyPropComponentType Enum=@n2
    ) -> (ReturnValue:ESlateVisibility Enum)
   )
   > SetVisibility(
    self=Get DepricatedBehavior()
    InVisibility=Select(NewEnumerator0=Collapsed, NewEnumerator1=Collapsed, NewEnumerator2=Collapsed, NewEnumerator3=Visible, Index=@n2)
   )
   > SetVisibility(
    self=Get MultiEditorSpeakerView()
    InVisibility=Select(NewEnumerator0=Visible, NewEnumerator1=Visible, NewEnumerator2=Visible, NewEnumerator3=Collapsed, Index=@n2)
   )
   n3 = Get MultiEditControllerClass()
   > SetVisibility(self=@n3, InVisibility=Visible)
   > SetPropertyName(
    self=@n3
    NewPropertyName=Select(
     NewEnumerator0:Name=ControllerClass
     NewEnumerator1:Name=RadioControllerClass
     NewEnumerator2:Name=ActorClass
     NewEnumerator3:Name=ActorClass
     Index=@n2
    ) -> (ReturnValue:Name)
   )
   SetVisibility(execute=@n1.else, self=Get MultiEditorSpeakerView(), InVisibility=Collapsed)
   > SetVisibility(self=Get MultiEditControllerClass(), InVisibility=Collapsed)
   > SetVisibility(self=Get DepricatedAssets(), InVisibility=Collapsed)
   > SetVisibility(self=Get DepricatedBehavior(), InVisibility=Collapsed)
  ]}
 InitializeMultiPropView {nodes [
   n1 = InitializeMultiPropView!() -> (then, ListItemObject:Music Party Editor Music Prop Multi Edit Object Reference) {is=title K2Node_FunctionEntry}
   > SetObject(self=Get MusicPropDetailView(), NewObject=@n1)
   > n2 = For Each Loop(Array:Array of Base Class: Music Party Prop Base Object References=Get PropComponents(self=@n1)) -> (Array Element:Base Class: Music Party Prop Base Object Reference)
   "SetText (Text)"(
    execute=@n2.Completed
    self=Get TextTitle() -> (TextTitle:Text Block Object Reference)
    InText=Format Text!(Format="LOCTEXT(\"Multi-Edit for {ActorList}\")", ActorList:String=Get TitleString() -> (TitleString:String)) -> (Result)
   )
   > InitializeAudioComponentVariables() {is=title K2Node_CallFunction}
   n3 = "integer > integer"(A=@n2.Array Index)
   n4 = GetOwner(self=@n2)
   Set TitleString(
    execute=@n2.LoopBody
    TitleString:String=Append!(A=Get TitleString(), B=SelectString(A=", ", bPickA=@n3), C:String=GetDisplayName(Object=@n4)) -> (ReturnValue)
   ) -> (Output_Get:String)
  ]}
 InitializeSinglePropView {nodes [
   InitializeSinglePropView!() -> (then, ListItemObject:Base Class: Music Party Prop Base Object Reference) {is=title K2Node_FunctionEntry}
   n1 = GetOwner(self=Get EntryObjectPropCasted())
   > Set Text(Text="To Text (String)"(InString=GetActorLabel(self=@n1, bCreateIfNone=true)), self=Get TextTitle())
   > InitializeAudioComponentVariables()
  ]}
 InitializeSpeakerProperties {nodes [
   n1 = InitializeSpeakerProperties!() -> (then, SpeakerAsset:Music Party Speaker Object Reference) {is=title K2Node_FunctionEntry}
   > n2 = Is Valid(InputObject=@n1)
   SetObject(execute=@n2.Is Not Valid, self=Get SpeakerDetailView())
   > SetVisibility(self=Get SpeakerDetails_Border() -> (SpeakerDetails_Border:Border Object Reference), InVisibility=Collapsed)
   SetObject(execute=@n2.Is Valid, self=Get SpeakerDetailView(), NewObject=@n1)
   > SetVisibility(self=Get SpeakerDetails_Border(), InVisibility=Visible)
  ]}
 OnPostInitialize {nodes [
   n1 = OnPostInitialize!() -> (then, Parent:EUW Music Party Editor Object Reference) {is=title K2Node_FunctionEntry}
   > Set Parent(Parent:EUW Music Party Editor Object Reference=@n1) -> (Output_Get:EUW Music Party Editor Object Reference)
   > n2 = Branch(Condition=Get isMultiEdit())
   > n3 = SetMultiEditVariables()
   Set IgnorePropertyChanges(execute=[@n5.then, @n3.then], IgnorePropertyChanges=false)
   n4 = GetAndTransferPropComponentVariables!(
    execute=@n2.else
    self:EUW Music Party Editor Object Reference=Get Parent() -> (Parent:EUW Music Party Editor Object Reference)
    MusicPartyPropComponent:Base Class: Music Party Prop Base Object Reference=Get EntryObjectPropCasted()
    CopyToNewComponent:bool=false
    NewProp:Base Class: Music Party Prop Base Object Reference
    CopyMusicController:bool=false
   ) -> (
    then
    Controller:Actor Class Reference
    SpeakerActors:Array of Actor Soft Object References
    MusicEventTag:Gameplay Tag Structure
    GroupName:Name=None
    MusicPartySpeaker:Music Party Speaker Object Reference
   )
   > InitializeControllerProperties!(execute, self, Controller:Actor Class Reference=@n4) -> (then) {is=title K2Node_CallFunction}
   > n5 = InitializeSpeakerProperties!(execute, self, SpeakerAsset:Music Party Speaker Object Reference=@n4.MusicPartySpeaker) -> (then) {is=title K2Node_CallFunction}
  ]}
 OnPropertyChanged {nodes [
   n1 = OnPropertyChanged!() -> (then, PropertyName:Name, RefreshList:bool=true) {is=title K2Node_FunctionEntry}
   > n2 = Branch(Condition=Get IgnorePropertyChanges())
   n3 = Branch(execute=@n2.else, Condition=Get isMultiEdit())
   > Get!(execute) -> (then, else, Parent:EUW Music Party Editor Object Reference)
   > TransactionStart!(
    execute
    self
    Description:Text=Format Text!(Format="LOCTEXT(\"Multi Edit for MusicPartyPropComponent Property {Name}\")", Name:Name=@n1) -> (Result)
   ) -> (then)
   > n4 = For Each Loop(Array=Get PropComponents(self=Get EntryObjectMultiPropCasted()))
   TransactionEnd(execute=@n4.Completed)
   > Branch(Condition=Get RefreshList() -> (RefreshList:bool))
   > n5 = Get()
   > Refresh Level Actor List!(execute, self:EUW Music Party Editor Object Reference=@n5, CheckAllLevelActors:bool=false, DelayRefresh:bool=true) -> (then)
   Is Valid(exec=@n4.LoopBody, InputObject=@n4)
   > n6 = Set CurrentPropComponent(CurrentPropComponent:Base Class: Music Party Prop Base Object Reference=@n4) -> (Output_Get:Base Class: Music Party Prop Base Object Reference)
   > TransactionAdd!(execute, self, Object:Object Reference=@n6) -> (then)
   n7 = Get PropertyName() -> (PropertyName:Name=None)
   > n8 = Branch(Condition=OR Boolean(A="Equal (Name)"(A=@n7, B=RadioControllerClass), B="Equal (Name)"(A=@n7, B=ActorClass)))
   > CopyProperty!(
    execute
    self:EUW Music Party Editor Object Reference=Get Parent()
    InputName:Name=Get PropertyName() -> (PropertyName=None)
    OutputName:Name=ControllerClass
    InputObject:Object Reference=Get EntryObjectMultiPropCasted()
    OutputObject:Object Reference=Get CurrentPropComponent() -> (CurrentPropComponent:Base Class: Music Party Prop Base Object Reference)
   ) -> (then, Success:bool)
   n9 = Get PropertyName() -> (PropertyName=None)
   CopyProperty(
    execute=@n8.else
    self=Get Parent()
    InputName=@n9
    OutputName=@n9
    InputObject=Get EntryObjectMultiPropCasted()
    OutputObject=Get CurrentPropComponent()
   )
   Branch(execute=@n3.else, Condition=Get RefreshList())
   > n10 = Get()
   > Refresh Level Actor List(self=@n10, CheckAllLevelActors=false, DelayRefresh=true)
  ]}
 PIE_OnControllerUpdated {nodes [
   PIE_OnControllerUpdated() {is=title K2Node_FunctionEntry}
   n1 = GetObject!(self:Property View Base Object Reference=Get PIE_ControllerDetailView()) -> (ReturnValue:Object Reference)
   > n2 = Cast To MusicPartyController_Wave!(execute, Object=@n1) -> (then, CastFailed, AsMusic Party Controller Wave:Music Party Controller Wave Object Reference) {is=title K2Node_DynamicCast}
   > Set MusicToPlay(MusicToPlay:Sound Base Object Reference=Random Array Item!(TargetArray:Array of Sound Wave Object References=Get MusicToPlay!(self:Music Party Controller Wave Object Reference=@n2) -> (MusicToPlay:Array of Sound Wave Object References)) -> (OutItem:Sound Wave Object Reference, OutIndex:int=0) {is=title K2Node_CallArrayFunction}) -> (Output_Get:Sound Base Object Reference)
   > n3 = Set VolumeAdjust(VolumeAdjust:double=Get DbAdjustment!(self:Music Party Controller Wave Object Reference=@n2) -> (DbAdjustment:double)) -> (Output_Get:double)
   n4 = Cast To Actor Class!(execute:Exec=[@n11.then, @n3.then], Class:Object Class Reference=Get Class(Object=@n1) -> (ReturnValue:Object Class Reference)) -> (then:Exec, CastFailed:Exec, AsActor:Actor Class Reference) {is=title K2Node_ClassDynamicCast}
   > n5 = GetAllActorsOfClass!(
    execute
    WorldContextObject:Object Reference=GetCurrentWorld!(self:EUW Music Party Editor Object Reference=Get Parent()) -> (Object:Object Reference, isPIE:bool)
    ActorClass:Actor Class Reference=@n4
   ) -> (then, OutActors:Array of Actor Object References)
   > n6 = For Each Loop(Array:Array of Actor Object References=@n5) -> (Array Element:Actor Object Reference)
   > n7 = Cast To MusicPartyController_Wave(Object=@n6)
   n8 = Get Audio!(self:Music Party Controller Wave Object Reference=@n7) -> (Audio:Audio Component Object Reference)
   > Set Object Array Parameter!(
    execute
    self:Audio Parameter Controller Interface Interface=@n8
    InName:Name=Waves
    InValue:Array of Object References=Make Array("[0]":Object Reference=Get MusicToPlay!() -> (MusicToPlay:Sound Base Object Reference)) -> (Array:Array of Object References)
   ) -> (then) {is=title K2Node_CallFunction}
   > AdjustVolume!(
    execute
    self:Audio Component Object Reference=@n8
    AdjustVolumeDuration:float=0
    AdjustVolumeLevel:float=DecibelsToLinear!(Decibels:double=Get VolumeAdjust() -> (VolumeAdjust:double)) -> (Linear:double) {is=title K2Node_CallFunction}
    FadeCurve:EAudioFaderCurve Enum=Linear
   ) -> (then)
   n9 = Cast To MusicPartyController_Default_Cue!(execute=@n7.CastFailed, Object=@n6) -> (then, CastFailed, AsMusic Party Controller Default Cue:Music Party Controller Default Cue Object Reference) {is=title K2Node_DynamicCast}
   > SetSound!(
    execute
    self:Audio Component Object Reference=Get Audio(self:Music Party Controller Cue Object Reference=@n9)
    NewSound:Sound Base Object Reference=Get MusicToPlay()
   ) -> (then)
   n10 = Cast To MusicPartyController_Default_Cue(execute=@n2.CastFailed, Object=@n1)
   > n11 = Set MusicToPlay(MusicToPlay=Get MusicToPlay!(self:Music Party Controller Cue Object Reference=@n10) -> (MusicToPlay))
  ]}
 PIE_OnSpeakerDetailsUpdated {nodes [
   PIE_OnSpeakerDetailsUpdated() {is=title K2Node_FunctionEntry}
   > n1 = Cast To MusicPartySpeaker!(execute, Object=GetObject(self=Get SpeakerDetailView())) -> (then, CastFailed, AsMusic Party Speaker:Music Party Speaker Object Reference) {is=title K2Node_DynamicCast}
   > n2 = FindAllComponents!(
    execute
    WorldContextObject:Object Reference=GetCurrentWorld(self=Get Parent())
    ComponentClass:Actor Component Class Reference=/Game/Sounds/Fort_Audio_Blueprints/MusicParty/MusicPartyPropComponent_SpeakerAssets.MusicPartyPropComponent_SpeakerAssets_C
   ) -> (then, ReturnValue:Array of Actor Component Object References)
   > n3 = For Each Loop(Array:Array of Actor Component Object References=@n2) -> (Array Element:Actor Component Object Reference)
   > n4 = Cast To MusicPartyPropComponent_SpeakerAssets!(execute, Object=@n3) -> (then, CastFailed, AsMusic Party Prop Component Speaker Assets:Music Party Prop Component Speaker Assets Object Reference) {is=title K2Node_DynamicCast}
   > Branch(Condition="Equal (Object)"(
     A=Get SpeakerAsset!(self:Music Party Prop Component Speaker Assets Object Reference=@n4) -> (SpeakerAsset:Music Party Speaker Object Reference)
     B=@n1
    ))
   n5 = GetComponentByClass(self=GetOwner(self=@n4), ComponentClass=&AudioComponent)
   > Is Valid(InputObject=@n5)
   > SetSoundToPlaySettings!(execute, self:Base Class: Music Party Prop Base Object Reference=@n4, AudioComponent:Audio Component Object Reference=@n5) -> (then)
  ]}
 PIE_Update {nodes [
   n1 = PIE_Update!() -> (then, isPIE:bool=~) {is=title K2Node_FunctionEntry}
   > Set isPIE(isPIE:bool=@n1) -> (Output_Get:bool)
   n2 = Get PIELock_Container() -> (PIELock_Container:Vertical Box Object Reference)
   > SetVisibility(
    self=@n2
    InVisibility=Select!(Option 0:ESlateVisibility Enum=Visible, Option 1:ESlateVisibility Enum=HitTestInvisible, Index:bool=Get isPIE()) -> (ReturnValue:ESlateVisibility Enum)
   )
   n3 = Select(Option 0=Collapsed, Option 1=Visible, Index=Get isPIE())
   > SetVisibility(self=Get MusicPropPIEView(), InVisibility=@n3)
   > SetVisibility(self=Get PIE_ControllerDetailView(), InVisibility=@n3)
   > SetVisibility(self=Get ControllerDetailView(), InVisibility=Select(Option 0=Visible, Option 1=Collapsed, Index=false))
   > SetRenderOpacity(self=@n2, InOpacity=SelectFloat(A=0.35, B=1, bPickA=Get isPIE()))
  ]}
 SetArrayofVariables {nodes [
   n1 = SetArrayofVariables!() -> (then, isFirst:bool=~, Array:Array of Names, PropComponent:Base Class: Music Party Prop Base Object Reference) {is=title K2Node_FunctionEntry}
   > n2 = For Each Loop(Array:Array of Names=@n1.Array) -> (Array Element:Name)
   > Set CurrentName(CurrentName:Name=@n2) -> (Output_Get:Name=None)
   n3 = Get CurrentName() -> (CurrentName:Name=None)
   n4 = "Equal (Name)"(A=@n3, B=ActorClass)
   n5 = "Equal (Name)"(A=@n3, B=RadioControllerClass)
   > Set InputName(InputName:Name=SelectName!(A:Name=ControllerClass, B:Name=@n2, bPickA:bool=OR Boolean(A=@n5, B=@n4)) -> (ReturnValue:Name=None)) -> (Output_Get:Name=None)
   > n6 = Branch(Condition=Get isFirst() -> (isFirst:bool))
   > CheckMultiEditProperties!(
    execute
    self:EUW Music Party Editor Object Reference=Get Parent()
    InputName:Name=Get InputName() -> (InputName:Name=None)
    OutputName:Name=Get CurrentName() -> (CurrentName=None)
    InputObject:Object Reference=Get PropComponent() -> (PropComponent:Base Class: Music Party Prop Base Object Reference)
    OutputObject:Object Reference=Get EntryObjectMultiPropCasted()
   ) -> (then, MatchFound:bool)
   CopyProperty(
    execute=@n6.else
    self=Get Parent()
    InputName=Get InputName() -> (InputName=None)
    OutputName=Get CurrentName() -> (CurrentName=None)
    InputObject=Get PropComponent()
    OutputObject=Get EntryObjectMultiPropCasted()
   )
   Reroute Node(InputPin:bool=@n1) -> (OutputPin:bool)
  ]}
 SetMultiEditVariables {nodes [
   SetMultiEditVariables() {is=title K2Node_FunctionEntry}
   > n1 = For Each Loop(Array:Array of Base Class: Music Party Prop Base Object References=Get PropComponents(self=Get EntryObjectMultiPropCasted())) -> (Array Element:Base Class: Music Party Prop Base Object Reference)
   n2 = For Each Loop with Break!(
    Exec:Exec=@n1.Completed
    Array:Array of Actor Object References=Get LevelActors!(self:Music Party Editor Music Prop Multi Edit Object Reference=Get EntryObjectMultiPropCasted()) -> (LevelActors:Array of Actor Object References)
    Break:Exec=@n13.then
   ) -> (LoopBody:Exec, Array Element:Actor Object Reference, Array Index:int, Completed:Exec)
   n3 = Get EntryObjectMultiPropCasted()
   Set RelativeLocation!(
    execute=@n2.Completed
    RelativeLocation:Vector=Get CurrentRelativeLocation() -> (CurrentRelativeLocation:Vector)
    self:Music Party Editor Music Prop Multi Edit Object Reference=@n3
   ) -> (then, Output_Get:Vector)
   > Set RelativeRotation!(
    execute
    RelativeRotation:Rotator=Get CurrentRelativeRotation() -> (CurrentRelativeRotation:Rotator)
    self:Music Party Editor Music Prop Multi Edit Object Reference=@n3
   ) -> (then, Output_Get:Rotator)
   n4 = Get EntryObjectMultiPropCasted()
   > InitializeSpeakerProperties(SpeakerAsset=Get SpeakerAsset(self:Music Party Editor Music Prop Multi Edit Object Reference=@n4))
   > n5 = Switch on EMusicPartyPropComponentType!(execute, Selection:EMusicPartyPropComponentType Enum=Get PropComponentType(self=@n4) -> (PropComponentType=NewEnumerator0)) -> (NewEnumerator0:Exec, NewEnumerator1:Exec, NewEnumerator2:Exec, NewEnumerator3:Exec) {is=title K2Node_SwitchEnum}
   n6 = Get EntryObjectMultiPropCasted()
   > InitializeControllerProperties(Controller=Get ControllerClass!(self:Music Party Editor Music Prop Multi Edit Object Reference=@n6) -> (ControllerClass:Music Party Controller Base Class Reference))
   InitializeControllerProperties(
    execute=@n5.NewEnumerator1
    Controller=Get RadioControllerClass!(self:Music Party Editor Music Prop Multi Edit Object Reference=@n6) -> (RadioControllerClass:Fort Media Streaming Radio Class Reference)
   )
   InitializeControllerProperties(
    execute=[@n5.NewEnumerator2, @n5.NewEnumerator3]
    Controller=Get ActorClass!(self:Music Party Editor Music Prop Multi Edit Object Reference=@n6) -> (ActorClass:Actor Class Reference)
   )
   n7 = Branch(execute=@n2.LoopBody, Condition="integer > integer"(A=@n2.Array Index))
   n8 = GetComponentByClass(self=@n2, ComponentClass=&AudioComponent)
   n9 = Get RelativeLocation!(self:Scene Component Object Reference=@n8) -> (RelativeLocation:Vector)
   n10 = Get CurrentRelativeLocation()
   n11 = Get RelativeRotation!(self:Scene Component Object Reference=@n8) -> (RelativeRotation:Rotator)
   > n12 = Branch(Condition=AND Boolean(
     A="Equal (Vector)"!(A:Vector=@n10, B:Vector=@n9, ErrorTolerance:float) -> (ReturnValue:bool=~) {is=title K2Node_PromotableOperator}
     B="Equal (Rotator)"!(A:Rotator=@n11, B:Rotator=@n11, ErrorTolerance:float) -> (ReturnValue:bool=~) {is=title K2Node_PromotableOperator}
    ))
   Set CurrentRelativeLocation(execute=@n12.else, CurrentRelativeLocation:Vector="0, 0, 0") -> (Output_Get:Vector)
   > n13 = Set CurrentRelativeRotation(CurrentRelativeRotation:Rotator="0, 0, 0") -> (Output_Get:Rotator)
   Set CurrentRelativeLocation(execute=@n7.else, CurrentRelativeLocation=@n9)
   > Set CurrentRelativeRotation(CurrentRelativeRotation=@n11)
   Is Valid(exec=@n1.LoopBody, InputObject=@n1)
   > Set CurrentPropComponent(CurrentPropComponent=@n1)
   n14 = "integer > integer"(A=@n1.Array Index)
   > n15 = Branch(Condition=@n14)
   > n16 = GetAndTransferPropComponentVariables(self=Get Parent(), MusicPartyPropComponent=Get CurrentPropComponent(), CopyToNewComponent=false, CopyMusicController=false) -> (GroupName=None)
   n17 = Get MusicEventTag!(self:Music Party Editor Music Prop Multi Edit Object Reference=Get EntryObjectMultiPropCasted()) -> (MusicEventTag:Gameplay Tag Structure)
   > n18 = Set MusicEventTag!(
    execute
    MusicEventTag:Gameplay Tag Structure=Select(
     Option 0:Gameplay Tag Structure
     Option 1:Gameplay Tag Structure=@n17
     Index=MatchesTag!(TagOne:Gameplay Tag Structure=@n16.MusicEventTag, TagTwo:Gameplay Tag Structure=@n17, bExactMatch:bool=false) -> (ReturnValue:bool)
    ) -> (ReturnValue:Gameplay Tag Structure)
    self:Music Party Editor Music Prop Multi Edit Object Reference=Get EntryObjectMultiPropCasted()
   ) -> (then, Output_Get:Gameplay Tag Structure)
   SetArrayofVariables!(
    execute=[@n18.then, @n21.then]
    self
    isFirst:bool=@n14
    Array:Array of Names=Get SharedVariableNames() -> (SharedVariableNames:Array of Names)
    PropComponent:Base Class: Music Party Prop Base Object Reference=Get CurrentPropComponent()
   ) -> (then) {is=title K2Node_CallFunction}
   n19 = Get EntryObjectMultiPropCasted()
   > SetArrayofVariables(
    isFirst=@n14
    Array=Select!(
     NewEnumerator0:Array of Names=Get ControllerVariableNames() -> (ControllerVariableNames:Array of Names)
     NewEnumerator1:Array of Names=Get RadioVariableNames() -> (RadioVariableNames:Array of Names)
     NewEnumerator2:Array of Names=Get ActorVariableNames() -> (ActorVariableNames:Array of Names)
     NewEnumerator3:Array of Names=Get DepricatedNames() -> (DepricatedNames:Array of Names)
     Index:EMusicPartyPropComponentType Enum=Get PropComponentType(self=@n19) -> (PropComponentType=NewEnumerator0)
    ) -> (ReturnValue:Array of Names)
    PropComponent=Get CurrentPropComponent()
   )
   n20 = GetAndTransferPropComponentVariables(
    execute=@n15.else
    self=Get Parent()
    MusicPartyPropComponent=Get CurrentPropComponent()
    CopyToNewComponent=false
    CopyMusicController=false
   ) -> (GroupName=None)
   > n21 = Set MusicEventTag(MusicEventTag=@n20.MusicEventTag, self=Get EntryObjectMultiPropCasted())
  ]}
 TransactionAdd {nodes [
   n1 = TransactionAdd!() -> (then, Object:Object Reference) {is=title K2Node_FunctionEntry}
   > TransactObject(Object=@n1)
  ]}
 TransactionEnd {nodes [
   TransactionEnd() {is=title K2Node_FunctionEntry}
   > EndTransaction()
  ]}
 TransactionStart {nodes [
   n1 = TransactionStart!() -> (then, Description:Text) {is=title K2Node_FunctionEntry}
   n2 = Self-Reference()
   > BeginTransaction(Context=SoundLibraryEditor_EditEntry, Description=@n1, PrimaryObject=@n2)
   > TransactObject(Object=@n2)
  ]}
]
variables [
 ControllerClass {type=Class, default=null}
 EntryObject {type=Object, default=null}
 EntryObjectMultiPropCasted {type=MusicPartyEditor_MusicProp_MultiEdit_C, default=null}
 EntryObjectPropCasted {type=MusicPartyPropComponentBase_C, default=null}
 IgnorePropertyChanges {type=boolean}
 isMultiEdit {type=boolean}
 isPIE {type=boolean}
 Parent {type=EUW_MusicPartyEditor_C, default=null}
 ShowAttenuation {type=boolean}
]
widgets {
 info {inheritedWidgetCount=0, parentClass=&EditorUtilityWidget, rootWidgetClass=&CanvasPanel, widgetCount=31}
 widgets [
  CanvasPanel_178 {widgetClassPath=&CanvasPanel, widget=CanvasPanel_178}
  Border_2 {
   widgetClassPath=&Border
   parent=CanvasPanel_178
   widget=Border_2
   properties {
    horizontalAlignment=HAlign_Left
    background {
     bIsDynamicallyLoaded=false
     drawAs=Box
     imageType=NoImage
     mirroring=NoMirror
     resourceName=None
     resourceObject=null
     tiling=NoTile
     imageSize=<32 32>
     margin=<ltrb 0 0 0 0>
     outlineSettings {
      bUseBrushTransparency=false
      roundingType=HalfHeightRadius
      width=0
      color {colorUseRule=UseColor_Specified, specifiedColor=<rgba 0 0 0 0>}
      cornerRadii=<0 0 0 0>
     }
     tintColor {colorUseRule=UseColor_Specified, specifiedColor=<rgba 1 1 1 1>}
     uVRegion {bIsValid=false, max=<0 0>, min=<0 0>}
    }
    brushColor=<rgba 0 0 0 0.2>
    padding=<ltrb 0 2 4 2>
   }
   slot {
    class=CanvasPanelSlot
    properties {bAutoSize=true, layoutData {alignment=<0 0>, anchors {maximum=<0 0>, minimum=<0 0>}, offsets=<ltrb 0 0 400 325>}}
   }
  }
  VerticalBox_142 {
   widgetClassPath=&VerticalBox
   parent=Border_2
   widget=VerticalBox_142
   slot {class=BorderSlot, properties {horizontalAlignment=HAlign_Left, padding=<ltrb 0 2 4 2>}}
  }
  TextTitle {
   widgetClassPath=&UEFN_TextBlock
   bIsVariable=true
   parent=VerticalBox_142
   widget=TextTitle
   properties {
    bWrapWithInvalidationPanel=true
    clipping=ClipToBounds
    text=Level Actor Detail
    wrapTextAt=400
    font {
     bForceMonospaced=false
     bMaterialIsStencil=false
     fontMaterial=null
     fontObject=&NotoSans
     letterSpacing=5
     monospacedWidth=1
     size=12
     skewAmount=0
     typefaceFontName=NotoSans-Bold
     outlineSettings {
      bApplyOutlineToDropShadows=false
      bMiteredCorners=false
      bSeparateFillAlpha=false
      outlineMaterial=null
      outlineSize=0
      outlineColor=<rgba 0 0 0 1>
     }
    }
   }
   slot {class=VerticalBoxSlot, properties {size {sizeRule=Fill, value=1}}}
  }
  MusicPropPIEView {
   widgetClassPath=&DetailsView
   parent=VerticalBox_142
   widget=MusicPropPIEView
   properties {columnWidth=0.75, visibility=Collapsed, categoriesToShow [Assets], propertiesToShow [SpeakerAsset, InstanceDbVolumeAdjustment]}
   slot {class=VerticalBoxSlot, properties {horizontalAlignment=HAlign_Left}}
  }
  PIELock_Container {widgetClassPath=&VerticalBox, bIsVariable=true, parent=VerticalBox_142, widget=PIELock_Container, slot {class=VerticalBoxSlot}}
  AudioComponentRelativeLocation {
   widgetClassPath=&SinglePropertyView
   parent=PIELock_Container
   widget=AudioComponentRelativeLocation
   properties {visibility=Collapsed}
   slot {class=VerticalBoxSlot, properties {size {sizeRule=Automatic, value=1.33}}}
  }
  AudioComponentRelativeRotation {
   widgetClassPath=&SinglePropertyView
   parent=PIELock_Container
   widget=AudioComponentRelativeRotation
   properties {visibility=Collapsed}
   slot {class=VerticalBoxSlot}
  }
  DepricatedAssets {
   widgetClassPath=&DetailsView
   parent=PIELock_Container
   widget=DepricatedAssets
   properties {columnWidth=0.75, categoriesToShow [DepricatedAssets]}
   slot {class=VerticalBoxSlot}
  }
  MusicPropDetailView {
   widgetClassPath=&DetailsView
   parent=PIELock_Container
   widget=MusicPropDetailView
   properties {columnWidth=0.75, categoriesToShow [Assets,Behavior,Activation]}
   slot {class=VerticalBoxSlot}
  }
  Border_94 {
   widgetClassPath=&Border
   parent=PIELock_Container
   widget=Border_94
   properties {brushColor=<rgba 0.0176 0.0176 0.0176 1>, padding=<ltrb 20 0 0 0>}
   slot {class=VerticalBoxSlot, properties {size {sizeRule=Automatic, value=0.75}}}
  }
  MultiEditControllerClass {
   widgetClassPath=&SinglePropertyView
   parent=Border_94
   widget=MultiEditControllerClass
   properties {visibility=Collapsed}
   slot {class=BorderSlot, properties {padding=<ltrb 20 0 0 0>}}
  }
  MultiEditorSpeakerView {
   widgetClassPath=&DetailsView
   parent=PIELock_Container
   widget=MultiEditorSpeakerView
   properties {columnWidth=0.75, categoriesToShow [SpeakerAsset]}
   slot {class=VerticalBoxSlot}
  }
  DepricatedBehavior {
   widgetClassPath=&DetailsView
   parent=PIELock_Container
   widget=DepricatedBehavior
   properties {columnWidth=0.75, categoriesToShow [DepricatedBehavior]}
   slot {class=VerticalBoxSlot}
  }
  AudioComponentTitle {
   widgetClassPath=&UEFN_TextBlock
   bIsVariable=true
   parent=VerticalBox_142
   widget=AudioComponentTitle
   properties {
    bWrapWithInvalidationPanel=true
    clipping=ClipToBounds
    text=Audio Component Overrides
    wrapTextAt=400
    font {
     bForceMonospaced=false
     bMaterialIsStencil=false
     fontMaterial=null
     fontObject=&NotoSans
     letterSpacing=5
     monospacedWidth=1
     size=9
     skewAmount=0
     typefaceFontName=NotoSans-Bold
     outlineSettings {
      bApplyOutlineToDropShadows=false
      bMiteredCorners=false
      bSeparateFillAlpha=false
      outlineMaterial=null
      outlineSize=0
      outlineColor=<rgba 0 0 0 1>
     }
    }
   }
   slot {class=VerticalBoxSlot, properties {size {sizeRule=Fill, value=1}}}
  }
  AudioComponentOverrides {
   widgetClassPath=&DetailsView
   parent=VerticalBox_142
   widget=AudioComponentOverrides
   properties {columnWidth=0.75, categoriesToShow [Attenuation], propertiesToShow [SourceEffectChain]}
   slot {class=VerticalBoxSlot, properties {horizontalAlignment=HAlign_Left}}
  }
  SpeakerDetails_Border {
   widgetClassPath=&Border
   bIsVariable=true
   parent=VerticalBox_142
   widget=SpeakerDetails_Border
   properties {
    horizontalAlignment=HAlign_Left
    visibility=Collapsed
    background {
     bIsDynamicallyLoaded=false
     drawAs=Box
     imageType=NoImage
     mirroring=NoMirror
     resourceName=None
     resourceObject=null
     tiling=NoTile
     imageSize=<32 32>
     margin=<ltrb 0 0 0 0>
     outlineSettings {
      bUseBrushTransparency=false
      roundingType=HalfHeightRadius
      width=0
      color {colorUseRule=UseColor_Specified, specifiedColor=<rgba 0 0 0 0>}
      cornerRadii=<0 0 0 0>
     }
     tintColor {colorUseRule=UseColor_Specified, specifiedColor=<rgba 1 1 1 1>}
     uVRegion {bIsValid=false, max=<0 0>, min=<0 0>}
    }
    brushColor=<rgba 0 0 0 0.2>
    padding=<ltrb 0 2 4 2>
   }
   slot {class=VerticalBoxSlot, properties {padding=<ltrb 0 15 0 0>}}
  }
  VerticalBox {
   widgetClassPath=&VerticalBox
   parent=SpeakerDetails_Border
   widget=VerticalBox
   slot {class=BorderSlot, properties {horizontalAlignment=HAlign_Left, padding=<ltrb 0 2 4 2>}}
  }
  HorizontalBox_0 {
   widgetClassPath=&HorizontalBox
   parent=VerticalBox
   widget=HorizontalBox_0
   slot {class=VerticalBoxSlot, properties {padding=<ltrb 1 1 1 1>}}
  }
  CollapseCheckbox_Speaker {
   widgetClassPath=/Script/UMG.CheckBox
   bIsVariable=true
   parent=HorizontalBox_0
   widget=CollapseCheckbox_Speaker
   properties {
    renderTransform {angle=0, scale=<1 1>, shear=<0 0>, translation=<0 1>}
    renderTransformPivot=<0.45 0.5>
    widgetStyle {
     checkBoxType=CheckBox
     backgroundHoveredImage {
      bIsDynamicallyLoaded=false
      drawAs=NoDrawType
      imageType=NoImage
      mirroring=NoMirror
      resourceName=None
      resourceObject=null
      tiling=NoTile
      imageSize=<0 0>
      margin=<ltrb 0 0 0 0>
      outlineSettings {
       bUseBrushTransparency=false
       roundingType=HalfHeightRadius
       width=0
       color {colorUseRule=UseColor_Specified, specifiedColor=<rgba 0 0 0 0>}
       cornerRadii=<0 0 0 0>
      }
      tintColor {colorUseRule=UseColor_Specified, specifiedColor=<rgba 1 1 1 1>}
      uVRegion {bIsValid=false, max=<0 0>, min=<0 0>}
     }
     backgroundImage {
      bIsDynamicallyLoaded=false
      drawAs=NoDrawType
      imageType=NoImage
      mirroring=NoMirror
      resourceName=None
      resourceObject=null
      tiling=NoTile
      imageSize=<0 0>
      margin=<ltrb 0 0 0 0>
      outlineSettings {
       bUseBrushTransparency=false
       roundingType=HalfHeightRadius
       width=0
       color {colorUseRule=UseColor_Specified, specifiedColor=<rgba 0 0 0 0>}
       cornerRadii=<0 0 0 0>
      }
      tintColor {colorUseRule=UseColor_Specified, specifiedColor=<rgba 1 1 1 1>}
      uVRegion {bIsValid=false, max=<0 0>, min=<0 0>}
     }
     backgroundPressedImage {
      bIsDynamicallyLoaded=false
      drawAs=NoDrawType
      imageType=NoImage
      mirroring=NoMirror
      resourceName=None
      resourceObject=null
      tiling=NoTile
      imageSize=<0 0>
      margin=<ltrb 0 0 0 0>
      outlineSettings {
       bUseBrushTransparency=false
       roundingType=HalfHeightRadius
       width=0
       color {colorUseRule=UseColor_Specified, specifiedColor=<rgba 0 0 0 0>}
       cornerRadii=<0 0 0 0>
      }
      tintColor {colorUseRule=UseColor_Specified, specifiedColor=<rgba 1 1 1 1>}
      uVRegion {bIsValid=false, max=<0 0>, min=<0 0>}
     }
     borderBackgroundColor {colorUseRule=UseColor_Specified, specifiedColor=<rgba 1 1 1 1>}
     checkedForeground {colorUseRule=UseColor_Foreground, specifiedColor=<rgba 1 0 1 1>}
     checkedHoveredForeground {colorUseRule=UseColor_Foreground, specifiedColor=<rgba 1 0 1 1>}
     checkedHoveredImage {
      bIsDynamicallyLoaded=false
      drawAs=Image
      imageType=FullColor
      mirroring=NoMirror
      resourceName=../../../Engine/Content/Slate/Common/CheckBox.png
      resourceObject=/FortUILibrary/Foundation/Buttons/Materials/Arrows/M_Button_CycleArrow_Right_Hovered
      tiling=NoTile
      imageSize=<16 16>
      margin=<ltrb 0 0 0 0>
      outlineSettings {
       bUseBrushTransparency=false
       roundingType=HalfHeightRadius
       width=0
       color {colorUseRule=UseColor_Specified, specifiedColor=<rgba 0 0 0 0>}
       cornerRadii=<0 0 0 0>
      }
      tintColor {colorUseRule=UseColor_Specified, specifiedColor=<rgba 1 1 1 1>}
      uVRegion {bIsValid=false, max=<0 0>, min=<0 0>}
     }
     checkedImage {
      bIsDynamicallyLoaded=false
      drawAs=Image
      imageType=FullColor
      mirroring=NoMirror
      resourceName=../../../Engine/Content/Slate/Common/CheckBox.png
      resourceObject=/FortUILibrary/Foundation/Buttons/Materials/Arrows/M_Button_CycleArrow_Right_Base
      tiling=NoTile
      imageSize=<16 16>
      margin=<ltrb 0 0 0 0>
      outlineSettings {
       bUseBrushTransparency=false
       roundingType=HalfHeightRadius
       width=0
       color {colorUseRule=UseColor_Specified, specifiedColor=<rgba 0 0 0 0>}
       cornerRadii=<0 0 0 0>
      }
      tintColor {colorUseRule=UseColor_Specified, specifiedColor=<rgba 1 1 1 1>}
      uVRegion {bIsValid=false, max=<0 0>, min=<0 0>}
     }
     checkedPressedForeground {colorUseRule=UseColor_Foreground, specifiedColor=<rgba 1 0 1 1>}
     checkedPressedImage {
      bIsDynamicallyLoaded=false
      drawAs=Image
      imageType=FullColor
      mirroring=NoMirror
      resourceName=../../../Engine/Content/Slate/Common/CheckBox.png
      resourceObject=/FortUILibrary/Foundation/Buttons/Materials/Arrows/M_Button_CycleArrow_Right_Pressed
      tiling=NoTile
      imageSize=<16 16>
      margin=<ltrb 0 0 0 0>
      outlineSettings {
       bUseBrushTransparency=false
       roundingType=HalfHeightRadius
       width=0
       color {colorUseRule=UseColor_Specified, specifiedColor=<rgba 0 0 0 0>}
       cornerRadii=<0 0 0 0>
      }
      tintColor {colorUseRule=UseColor_Specified, specifiedColor=<rgba 1 1 1 1>}
      uVRegion {bIsValid=false, max=<0 0>, min=<0 0>}
     }
     checkedSlateSound {resourceObject=null}
     foregroundColor {colorUseRule=UseColor_Foreground, specifiedColor=<rgba 1 0 1 1>}
     hoveredForeground {colorUseRule=UseColor_Foreground, specifiedColor=<rgba 1 0 1 1>}
     hoveredSlateSound {resourceObject=null}
     padding=<ltrb 2 0 0 0>
     pressedForeground {colorUseRule=UseColor_Foreground, specifiedColor=<rgba 1 0 1 1>}
     uncheckedHoveredImage {
      bIsDynamicallyLoaded=false
      drawAs=Image
      imageType=FullColor
      mirroring=NoMirror
      resourceName=../../../Engine/Content/Slate/Common/CheckBox.png
      resourceObject=&M_Button_CycleArrow_Right_Hovered
      tiling=NoTile
      imageSize=<16 16>
      margin=<ltrb 0 0 0 0>
      outlineSettings {
       bUseBrushTransparency=false
       roundingType=HalfHeightRadius
       width=0
       color {colorUseRule=UseColor_Specified, specifiedColor=<rgba 0 0 0 0>}
       cornerRadii=<0 0 0 0>
      }
      tintColor {colorUseRule=UseColor_Specified, specifiedColor=<rgba 1 1 1 1>}
      uVRegion {bIsValid=false, max=<0 0>, min=<0 0>}
     }
     uncheckedImage {
      bIsDynamicallyLoaded=false
      drawAs=Image
      imageType=FullColor
      mirroring=NoMirror
      resourceName=../../../Engine/Content/Slate/Common/CheckBox.png
      resourceObject=&M_Button_CycleArrow_Right_Base
      tiling=NoTile
      imageSize=<16 16>
      margin=<ltrb 0 0 0 0>
      outlineSettings {
       bUseBrushTransparency=false
       roundingType=HalfHeightRadius
       width=0
       color {colorUseRule=UseColor_Specified, specifiedColor=<rgba 0 0 0 0>}
       cornerRadii=<0 0 0 0>
      }
      tintColor {colorUseRule=UseColor_Specified, specifiedColor=<rgba 1 1 1 1>}
      uVRegion {bIsValid=false, max=<0 0>, min=<0 0>}
     }
     uncheckedPressedImage {
      bIsDynamicallyLoaded=false
      drawAs=Image
      imageType=FullColor
      mirroring=NoMirror
      resourceName=../../../Engine/Content/Slate/Common/CheckBox.png
      resourceObject=&M_Button_CycleArrow_Right_Pressed
      tiling=NoTile
      imageSize=<16 16>
      margin=<ltrb 0 0 0 0>
      outlineSettings {
       bUseBrushTransparency=false
       roundingType=HalfHeightRadius
       width=0
       color {colorUseRule=UseColor_Specified, specifiedColor=<rgba 0 0 0 0>}
       cornerRadii=<0 0 0 0>
      }
      tintColor {colorUseRule=UseColor_Specified, specifiedColor=<rgba 1 1 1 1>}
      uVRegion {bIsValid=false, max=<0 0>, min=<0 0>}
     }
     uncheckedSlateSound {resourceObject=null}
     undeterminedForeground {colorUseRule=UseColor_Foreground, specifiedColor=<rgba 1 0 1 1>}
     undeterminedHoveredImage {
      bIsDynamicallyLoaded=false
      drawAs=Image
      imageType=FullColor
      mirroring=NoMirror
      resourceName=../../../Engine/Content/Slate/Common/CheckBox_Undetermined_Hovered.png
      resourceObject=null
      tiling=NoTile
      imageSize=<16 16>
      margin=<ltrb 0 0 0 0>
      outlineSettings {
       bUseBrushTransparency=false
       roundingType=HalfHeightRadius
       width=0
       color {colorUseRule=UseColor_Specified, specifiedColor=<rgba 0 0 0 0>}
       cornerRadii=<0 0 0 0>
      }
      tintColor {colorUseRule=UseColor_Specified, specifiedColor=<rgba 1 1 1 1>}
      uVRegion {bIsValid=false, max=<0 0>, min=<0 0>}
     }
     undeterminedImage {
      bIsDynamicallyLoaded=false
      drawAs=Image
      imageType=FullColor
      mirroring=NoMirror
      resourceName=../../../Engine/Content/Slate/Common/CheckBox_Undetermined.png
      resourceObject=null
      tiling=NoTile
      imageSize=<16 16>
      margin=<ltrb 0 0 0 0>
      outlineSettings {
       bUseBrushTransparency=false
       roundingType=HalfHeightRadius
       width=0
       color {colorUseRule=UseColor_Specified, specifiedColor=<rgba 0 0 0 0>}
       cornerRadii=<0 0 0 0>
      }
      tintColor {colorUseRule=UseColor_Specified, specifiedColor=<rgba 1 1 1 1>}
      uVRegion {bIsValid=false, max=<0 0>, min=<0 0>}
     }
     undeterminedPressedImage {
      bIsDynamicallyLoaded=false
      drawAs=Image
      imageType=FullColor
      mirroring=NoMirror
      resourceName=../../../Engine/Content/Slate/Common/CheckBox_Undetermined_Hovered.png
      resourceObject=null
      tiling=NoTile
      imageSize=<16 16>
      margin=<ltrb 0 0 0 0>
      outlineSettings {
       bUseBrushTransparency=false
       roundingType=HalfHeightRadius
       width=0
       color {colorUseRule=UseColor_Specified, specifiedColor=<rgba 0 0 0 0>}
       cornerRadii=<0 0 0 0>
      }
      tintColor {colorUseRule=UseColor_Specified, specifiedColor=<rgba 0.5 0.5 0.5 1>}
      uVRegion {bIsValid=false, max=<0 0>, min=<0 0>}
     }
    }
   }
   slot {class=HorizontalBoxSlot, properties {padding=<ltrb 2.5 0 0 0>}}
  }
  SpeakerDetails_Text {
   widgetClassPath=&UEFN_TextBlock
   parent=HorizontalBox_0
   widget=SpeakerDetails_Text
   properties {
    clipping=ClipToBounds
    text="Expand To Edit Speaker Asset (Will Effect all Speakers using this Asset)"
    font {
     bForceMonospaced=false
     bMaterialIsStencil=false
     fontMaterial=null
     fontObject=&NotoSans
     letterSpacing=5
     monospacedWidth=1
     size=9
     skewAmount=0
     typefaceFontName=NotoSans-Regular
     outlineSettings {
      bApplyOutlineToDropShadows=false
      bMiteredCorners=false
      bSeparateFillAlpha=false
      outlineMaterial=null
      outlineSize=0
      outlineColor=<rgba 0 0 0 1>
     }
    }
   }
   slot {class=HorizontalBoxSlot, properties {horizontalAlignment=HAlign_Left, verticalAlignment=VAlign_Center, padding=<ltrb 2.5 0 0 0>}}
  }
  SpeakerDetailView {
   widgetClassPath=&DetailsView
   parent=VerticalBox
   widget=SpeakerDetailView
   properties {columnWidth=0.75, visibility=Collapsed, categoriesToShow [SpeakerSettings, Assets]}
   slot {class=VerticalBoxSlot, properties {horizontalAlignment=HAlign_Left, padding=<ltrb 10 0 0 0>, size {sizeRule=Fill, value=0.75}}}
  }
  ControllerDetails_Border {
   widgetClassPath=&Border
   bIsVariable=true
   parent=VerticalBox_142
   widget=ControllerDetails_Border
   properties {
    horizontalAlignment=HAlign_Left
    visibility=Collapsed
    background {
     bIsDynamicallyLoaded=false
     drawAs=Box
     imageType=NoImage
     mirroring=NoMirror
     resourceName=None
     resourceObject=null
     tiling=NoTile
     imageSize=<32 32>
     margin=<ltrb 0 0 0 0>
     outlineSettings {
      bUseBrushTransparency=false
      roundingType=HalfHeightRadius
      width=0
      color {colorUseRule=UseColor_Specified, specifiedColor=<rgba 0 0 0 0>}
      cornerRadii=<0 0 0 0>
     }
     tintColor {colorUseRule=UseColor_Specified, specifiedColor=<rgba 1 1 1 1>}
     uVRegion {bIsValid=false, max=<0 0>, min=<0 0>}
    }
    brushColor=<rgba 0 0 0 0.2>
    padding=<ltrb 0 2 4 2>
   }
   slot {class=VerticalBoxSlot, properties {padding=<ltrb 0 15 0 0>}}
  }
  VerticalBox_1 {
   widgetClassPath=&VerticalBox
   parent=ControllerDetails_Border
   widget=VerticalBox_1
   slot {class=BorderSlot, properties {horizontalAlignment=HAlign_Left, padding=<ltrb 0 2 4 2>}}
  }
  HorizontalBox {
   widgetClassPath=&HorizontalBox
   parent=VerticalBox_1
   widget=HorizontalBox
   slot {class=VerticalBoxSlot, properties {padding=<ltrb 1 1 1 1>}}
  }
  CollapseCheckbox_Controller {
   widgetClassPath=&CheckBox
   bIsVariable=true
   parent=HorizontalBox
   widget=CollapseCheckbox_Controller
   properties {
    renderTransform {angle=0, scale=<1 1>, shear=<0 0>, translation=<0 1>}
    renderTransformPivot=<0.45 0.5>
    widgetStyle {
     checkBoxType=CheckBox
     backgroundHoveredImage {
      bIsDynamicallyLoaded=false
      drawAs=NoDrawType
      imageType=NoImage
      mirroring=NoMirror
      resourceName=None
      resourceObject=null
      tiling=NoTile
      imageSize=<0 0>
      margin=<ltrb 0 0 0 0>
      outlineSettings {
       bUseBrushTransparency=false
       roundingType=HalfHeightRadius
       width=0
       color {colorUseRule=UseColor_Specified, specifiedColor=<rgba 0 0 0 0>}
       cornerRadii=<0 0 0 0>
      }
      tintColor {colorUseRule=UseColor_Specified, specifiedColor=<rgba 1 1 1 1>}
      uVRegion {bIsValid=false, max=<0 0>, min=<0 0>}
     }
     backgroundImage {
      bIsDynamicallyLoaded=false
      drawAs=NoDrawType
      imageType=NoImage
      mirroring=NoMirror
      resourceName=None
      resourceObject=null
      tiling=NoTile
      imageSize=<0 0>
      margin=<ltrb 0 0 0 0>
      outlineSettings {
       bUseBrushTransparency=false
       roundingType=HalfHeightRadius
       width=0
       color {colorUseRule=UseColor_Specified, specifiedColor=<rgba 0 0 0 0>}
       cornerRadii=<0 0 0 0>
      }
      tintColor {colorUseRule=UseColor_Specified, specifiedColor=<rgba 1 1 1 1>}
      uVRegion {bIsValid=false, max=<0 0>, min=<0 0>}
     }
     backgroundPressedImage {
      bIsDynamicallyLoaded=false
      drawAs=NoDrawType
      imageType=NoImage
      mirroring=NoMirror
      resourceName=None
      resourceObject=null
      tiling=NoTile
      imageSize=<0 0>
      margin=<ltrb 0 0 0 0>
      outlineSettings {
       bUseBrushTransparency=false
       roundingType=HalfHeightRadius
       width=0
       color {colorUseRule=UseColor_Specified, specifiedColor=<rgba 0 0 0 0>}
       cornerRadii=<0 0 0 0>
      }
      tintColor {colorUseRule=UseColor_Specified, specifiedColor=<rgba 1 1 1 1>}
      uVRegion {bIsValid=false, max=<0 0>, min=<0 0>}
     }
     borderBackgroundColor {colorUseRule=UseColor_Specified, specifiedColor=<rgba 1 1 1 1>}
     checkedForeground {colorUseRule=UseColor_Foreground, specifiedColor=<rgba 1 0 1 1>}
     checkedHoveredForeground {colorUseRule=UseColor_Foreground, specifiedColor=<rgba 1 0 1 1>}
     checkedHoveredImage {
      bIsDynamicallyLoaded=false
      drawAs=Image
      imageType=FullColor
      mirroring=NoMirror
      resourceName=../../../Engine/Content/Slate/Common/CheckBox.png
      resourceObject=&M_Button_CycleArrow_Right_Hovered
      tiling=NoTile
      imageSize=<16 16>
      margin=<ltrb 0 0 0 0>
      outlineSettings {
       bUseBrushTransparency=false
       roundingType=HalfHeightRadius
       width=0
       color {colorUseRule=UseColor_Specified, specifiedColor=<rgba 0 0 0 0>}
       cornerRadii=<0 0 0 0>
      }
      tintColor {colorUseRule=UseColor_Specified, specifiedColor=<rgba 1 1 1 1>}
      uVRegion {bIsValid=false, max=<0 0>, min=<0 0>}
     }
     checkedImage {
      bIsDynamicallyLoaded=false
      drawAs=Image
      imageType=FullColor
      mirroring=NoMirror
      resourceName=../../../Engine/Content/Slate/Common/CheckBox.png
      resourceObject=&M_Button_CycleArrow_Right_Base
      tiling=NoTile
      imageSize=<16 16>
      margin=<ltrb 0 0 0 0>
      outlineSettings {
       bUseBrushTransparency=false
       roundingType=HalfHeightRadius
       width=0
       color {colorUseRule=UseColor_Specified, specifiedColor=<rgba 0 0 0 0>}
       cornerRadii=<0 0 0 0>
      }
      tintColor {colorUseRule=UseColor_Specified, specifiedColor=<rgba 1 1 1 1>}
      uVRegion {bIsValid=false, max=<0 0>, min=<0 0>}
     }
     checkedPressedForeground {colorUseRule=UseColor_Foreground, specifiedColor=<rgba 1 0 1 1>}
     checkedPressedImage {
      bIsDynamicallyLoaded=false
      drawAs=Image
      imageType=FullColor
      mirroring=NoMirror
      resourceName=../../../Engine/Content/Slate/Common/CheckBox.png
      resourceObject=&M_Button_CycleArrow_Right_Pressed
      tiling=NoTile
      imageSize=<16 16>
      margin=<ltrb 0 0 0 0>
      outlineSettings {
       bUseBrushTransparency=false
       roundingType=HalfHeightRadius
       width=0
       color {colorUseRule=UseColor_Specified, specifiedColor=<rgba 0 0 0 0>}
       cornerRadii=<0 0 0 0>
      }
      tintColor {colorUseRule=UseColor_Specified, specifiedColor=<rgba 1 1 1 1>}
      uVRegion {bIsValid=false, max=<0 0>, min=<0 0>}
     }
     checkedSlateSound {resourceObject=null}
     foregroundColor {colorUseRule=UseColor_Foreground, specifiedColor=<rgba 1 0 1 1>}
     hoveredForeground {colorUseRule=UseColor_Foreground, specifiedColor=<rgba 1 0 1 1>}
     hoveredSlateSound {resourceObject=null}
     padding=<ltrb 2 0 0 0>
     pressedForeground {colorUseRule=UseColor_Foreground, specifiedColor=<rgba 1 0 1 1>}
     uncheckedHoveredImage {
      bIsDynamicallyLoaded=false
      drawAs=Image
      imageType=FullColor
      mirroring=NoMirror
      resourceName=../../../Engine/Content/Slate/Common/CheckBox.png
      resourceObject=&M_Button_CycleArrow_Right_Hovered
      tiling=NoTile
      imageSize=<16 16>
      margin=<ltrb 0 0 0 0>
      outlineSettings {
       bUseBrushTransparency=false
       roundingType=HalfHeightRadius
       width=0
       color {colorUseRule=UseColor_Specified, specifiedColor=<rgba 0 0 0 0>}
       cornerRadii=<0 0 0 0>
      }
      tintColor {colorUseRule=UseColor_Specified, specifiedColor=<rgba 1 1 1 1>}
      uVRegion {bIsValid=false, max=<0 0>, min=<0 0>}
     }
     uncheckedImage {
      bIsDynamicallyLoaded=false
      drawAs=Image
      imageType=FullColor
      mirroring=NoMirror
      resourceName=../../../Engine/Content/Slate/Common/CheckBox.png
      resourceObject=&M_Button_CycleArrow_Right_Base
      tiling=NoTile
      imageSize=<16 16>
      margin=<ltrb 0 0 0 0>
      outlineSettings {
       bUseBrushTransparency=false
       roundingType=HalfHeightRadius
       width=0
       color {colorUseRule=UseColor_Specified, specifiedColor=<rgba 0 0 0 0>}
       cornerRadii=<0 0 0 0>
      }
      tintColor {colorUseRule=UseColor_Specified, specifiedColor=<rgba 1 1 1 1>}
      uVRegion {bIsValid=false, max=<0 0>, min=<0 0>}
     }
     uncheckedPressedImage {
      bIsDynamicallyLoaded=false
      drawAs=Image
      imageType=FullColor
      mirroring=NoMirror
      resourceName=../../../Engine/Content/Slate/Common/CheckBox.png
      resourceObject=&M_Button_CycleArrow_Right_Pressed
      tiling=NoTile
      imageSize=<16 16>
      margin=<ltrb 0 0 0 0>
      outlineSettings {
       bUseBrushTransparency=false
       roundingType=HalfHeightRadius
       width=0
       color {colorUseRule=UseColor_Specified, specifiedColor=<rgba 0 0 0 0>}
       cornerRadii=<0 0 0 0>
      }
      tintColor {colorUseRule=UseColor_Specified, specifiedColor=<rgba 1 1 1 1>}
      uVRegion {bIsValid=false, max=<0 0>, min=<0 0>}
     }
     uncheckedSlateSound {resourceObject=null}
     undeterminedForeground {colorUseRule=UseColor_Foreground, specifiedColor=<rgba 1 0 1 1>}
     undeterminedHoveredImage {
      bIsDynamicallyLoaded=false
      drawAs=Image
      imageType=FullColor
      mirroring=NoMirror
      resourceName=../../../Engine/Content/Slate/Common/CheckBox_Undetermined_Hovered.png
      resourceObject=null
      tiling=NoTile
      imageSize=<16 16>
      margin=<ltrb 0 0 0 0>
      outlineSettings {
       bUseBrushTransparency=false
       roundingType=HalfHeightRadius
       width=0
       color {colorUseRule=UseColor_Specified, specifiedColor=<rgba 0 0 0 0>}
       cornerRadii=<0 0 0 0>
      }
      tintColor {colorUseRule=UseColor_Specified, specifiedColor=<rgba 1 1 1 1>}
      uVRegion {bIsValid=false, max=<0 0>, min=<0 0>}
     }
     undeterminedImage {
      bIsDynamicallyLoaded=false
      drawAs=Image
      imageType=FullColor
      mirroring=NoMirror
      resourceName=../../../Engine/Content/Slate/Common/CheckBox_Undetermined.png
      resourceObject=null
      tiling=NoTile
      imageSize=<16 16>
      margin=<ltrb 0 0 0 0>
      outlineSettings {
       bUseBrushTransparency=false
       roundingType=HalfHeightRadius
       width=0
       color {colorUseRule=UseColor_Specified, specifiedColor=<rgba 0 0 0 0>}
       cornerRadii=<0 0 0 0>
      }
      tintColor {colorUseRule=UseColor_Specified, specifiedColor=<rgba 1 1 1 1>}
      uVRegion {bIsValid=false, max=<0 0>, min=<0 0>}
     }
     undeterminedPressedImage {
      bIsDynamicallyLoaded=false
      drawAs=Image
      imageType=FullColor
      mirroring=NoMirror
      resourceName=../../../Engine/Content/Slate/Common/CheckBox_Undetermined_Hovered.png
      resourceObject=null
      tiling=NoTile
      imageSize=<16 16>
      margin=<ltrb 0 0 0 0>
      outlineSettings {
       bUseBrushTransparency=false
       roundingType=HalfHeightRadius
       width=0
       color {colorUseRule=UseColor_Specified, specifiedColor=<rgba 0 0 0 0>}
       cornerRadii=<0 0 0 0>
      }
      tintColor {colorUseRule=UseColor_Specified, specifiedColor=<rgba 0.5 0.5 0.5 1>}
      uVRegion {bIsValid=false, max=<0 0>, min=<0 0>}
     }
    }
   }
   slot {class=HorizontalBoxSlot, properties {padding=<ltrb 2.5 0 0 0>}}
  }
  ControllerDetails {
   widgetClassPath=&UEFN_TextBlock
   bIsVariable=true
   parent=HorizontalBox
   widget=ControllerDetails
   properties {
    clipping=ClipToBounds
    text=Expand to Edit Controller Asset
    font {
     bForceMonospaced=false
     bMaterialIsStencil=false
     fontMaterial=null
     fontObject=&NotoSans
     letterSpacing=5
     monospacedWidth=1
     size=10
     skewAmount=0
     typefaceFontName=NotoSans-Regular
     outlineSettings {
      bApplyOutlineToDropShadows=false
      bMiteredCorners=false
      bSeparateFillAlpha=false
      outlineMaterial=null
      outlineSize=0
      outlineColor=<rgba 0 0 0 1>
     }
    }
   }
   slot {class=HorizontalBoxSlot, properties {horizontalAlignment=HAlign_Left, verticalAlignment=VAlign_Center, padding=<ltrb 2.5 0 0 0>}}
  }
  Controller_Container {widgetClassPath=&VerticalBox, bIsVariable=true, parent=VerticalBox_1, widget=Controller_Container, slot {class=VerticalBoxSlot}}
  ControllerDetailView {
   widgetClassPath=&DetailsView
   parent=Controller_Container
   widget=ControllerDetailView
   properties {columnWidth=0.75, visibility=Collapsed, categoriesToShow [Audio], propertiesToShow [AudioBus]}
   slot {class=VerticalBoxSlot, properties {horizontalAlignment=HAlign_Left, padding=<ltrb 10 0 0 0>}}
  }
  PIE_ControllerDetailView {
   widgetClassPath=&DetailsView
   parent=Controller_Container
   widget=PIE_ControllerDetailView
   properties {columnWidth=0.75, visibility=Collapsed, propertiesToShow [MusicToPlay,DbAdjustment,AudioBus]}
   slot {class=VerticalBoxSlot, properties {horizontalAlignment=HAlign_Left, padding=<ltrb 10 0 0 0>}}
  }
  VerticalBox_59 {widgetClassPath=&VerticalBox, parent=VerticalBox_142, widget=VerticalBox_59, slot {class=VerticalBoxSlot}}
 ]
}
