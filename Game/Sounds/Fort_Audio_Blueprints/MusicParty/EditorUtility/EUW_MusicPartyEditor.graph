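# Editor utility widget for managing Music Party audio in a level, including adding, converting and removing prop components and creating controller and speaker assets.
# Scanned actors become color-coded list entries by music group, plus unassigned and silenced sections, with separate editor and PIE views.
# Prop conversions transfer variables from old components, while Python helpers copy and compare properties.
# Tick validates selections, draws optional attenuation shapes and detects PIE changes.
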EditorUtilityWidgetBlueprint /Game/Sounds/Fort_Audio_Blueprints/MusicParty/EditorUtility/EUW_MusicPartyEditor
graphs [
 Actions {nodes [
   "On Clicked (AddMusicPropComponent)"() {is=title K2Node_ComponentBoundEvent}
   > PropEdits_BeginAction!(execute, self, ActionName:Name=Add Music Prop Component) -> (then)
   > Unbind all Events from Execute Action!(execute:Exec, self:EUW Music Party Editor Object Reference) -> (then:Exec) {is=title K2Node_ClearDelegate}
   > Bind Event to Execute Action(self:EUW Music Party Editor Object Reference=Self-Reference(), Delegate=Create Event()) {is=title K2Node_AddDelegate}
   n1 = "On Selection Changed (NewAsset_Controller_ComboBox)"!() -> (OutputDelegate, then, SelectedItem:String, SelectionType:ESelectInfo Enum) {is=title K2Node_ComponentBoundEvent}
   > NewAsset_BeginNewControllerAction!(execute, self, ControllerType:String=@n1.SelectedItem) -> (then)
   "On Clicked (CreateNewSpeakerObject)"() {is=title K2Node_ComponentBoundEvent}
   > NewAsset_BeginNewSpeakerAction()
   "On Clicked (NewAsset_ConfirmButton)"() {is=title K2Node_ComponentBoundEvent}
   > n2 = Get!(execute:Exec) -> (then:Exec, else:Exec, NewAsset_Object:Music Party Editor New Asset Base Object Reference)
   > n3 = ConfirmAction!(
    execute
    self:Music Party Editor New Asset Base Object Reference=@n2
    ExtendSpawn:bool=IsChecked(self=Get NewAsset_AutoGenerateAudioBus() -> (NewAsset_AutoGenerateAudioBus:Editor Utility Check Box Object Reference))
   ) -> (then, NewObjectPath:String)
   > n4 = Branch(Condition="Not Equal Exactly (String)"(A=@n3))
   > n5 = GetClassDefaultObjectFromPath!(execute, ClassAssetPath:String=@n3) -> (then, ReturnValue:Object Reference)
   > n6 = SetVariables!(execute, self:Music Party Editor New Asset Base Object Reference=@n2, ClassDefaults:Object Reference=@n5) -> (then)
   CancelAction(execute=[@n6.then, @n4.else, @n2.else])
   "On Clicked (RemovePropComponent_Button)"() {is=title K2Node_ComponentBoundEvent}
   > Remove Prop Component From Selected Entries!(execute, self, SilenceIfCantRemove:bool=false) -> (then, ChildSpeakerActors:Array of Actor Soft Object References)
   > Refresh Level Actor List!(execute, self, CheckAllLevelActors:bool=false, DelayRefresh:bool=false) -> (then)
   "On Clicked (NewAsset_CancelButton)"() {is=title K2Node_ComponentBoundEvent}
   > n7 = Get()
   > n8 = CancelAction(self:Music Party Editor New Asset Base Object Reference=@n7)
   CancelAction(execute=[@n8.then, @n7.else], self:Self Object Reference)
   n9 = "On Check State Changed (NewAsset_AutoGenerateAudioBus)"!() -> (OutputDelegate, then, bIsChecked:bool) {is=title K2Node_ComponentBoundEvent}
   > CheckAssetExtendedDetails!(execute, self, bIsChecked:bool=@n9.bIsChecked) -> (then)
   "On Clicked (ConvertPropComponent_Button)"() {is=title K2Node_ComponentBoundEvent}
   > ConvertProp_BeginAction!(
    execute
    self
    LevelActorEntries:Array of Music Party Editor Level Actor Entry Object References=Get SelectedEntries() -> (SelectedEntries:Array of Music Party Editor Level Actor Entry Object References)
    ClassToConvertTo:EMusicPartyPropComponentType Enum=Get ConvertProp_Category() -> (ConvertProp_Category:EMusicPartyPropComponentType Enum=NewEnumerator0)
   ) -> (then)
   "On Clicked (MusicPropAction_CancelButton)"() {is=title K2Node_ComponentBoundEvent}
   > CancelAction()
   "On Clicked (MusicPropAction_ConfirmButton)"() {is=title K2Node_ComponentBoundEvent}
   > ConfirmAction!(execute, self:Self Object Reference) -> (then)
   n10 = "On Selection Changed (PropTypeComboBox)"!() -> (OutputDelegate, then, SelectedItem:String, SelectionType:ESelectInfo Enum) {is=title K2Node_ComponentBoundEvent}
   > DelayUntilNextTick!(execute, WorldContextObject:Object Reference=GetCurrentWorld!() -> (Object:Object Reference, isPIE:bool)) -> (then)
   > PropEdits_OnPropTypeSelectionChange!(execute, self, Name:Name=String To Name(InString=@n10.SelectedItem)) -> (then)
   "On Clicked (CreateNewMusicController)"() {is=title K2Node_ComponentBoundEvent}
   > NewAsset_BeginNewControllerAction(ControllerType=GetSelectedOption!(self:"Combo Box (String) Object Reference"=Get NewAsset_Controller_ComboBox() -> (NewAsset_Controller_ComboBox:Editor Utility Combo Box String Object Reference)) -> (ReturnValue:String))
   Delay Refresh Selection() {is=title K2Node_CustomEvent}
   > ClearListItems(self=Get MusicEditorListView() -> (MusicEditorListView:Common List View Object Reference))
   > ClearListItems(self=Get MusicEditorPIEView() -> (MusicEditorPIEView:Common List View Object Reference))
   n11 = GetCurrentWorld()
   > Delay!(execute, WorldContextObject:Object Reference=@n11, Duration=0.1) -> (then)
   > n12 = Branch(Condition=@n11.isPIE)
   > RefreshSelectedEntriesPIE()
   RefreshSelectedEntries(execute=@n12.else)
  ]}
 ActorMatchesSearchTerms {nodes [
   n1 = ActorMatchesSearchTerms!() -> (then, Actor:Actor Object Reference) {is=title K2Node_FunctionEntry}
   > Set Label(Label:String=GetActorLabel(self=@n1, bCreateIfNone=false)) -> (Output_Get:String)
   > n2 = For Each Loop(Array=Get ActorSearchTerms() -> (ActorSearchTerms:Array of Strings))
   Return Node!(execute=@n2.Completed, MatchFound:bool=false)
   > Branch(Condition=Contains(SearchIn=Get Label() -> (Label:String), Substring=@n2))
   > Return Node(MatchFound=true)
  ]}
 AddProp_BeginAction {nodes [
   AddProp_BeginAction() {is=title K2Node_FunctionEntry}
   > n1 = For Each Loop(Array:Array of Music Party Editor Level Actor Entry Object References=Get SelectedEntries()) -> (Array Element:Music Party Editor Level Actor Entry Object Reference)
   Refresh Level Actor List(execute=@n1.Completed, CheckAllLevelActors=false, DelayRefresh=false)
   Branch(
    execute=@n1.LoopBody
    Condition=NOR Boolean!(
     A:bool=Get hasPropComponent!(self:Music Party Editor Level Actor Entry Object Reference=@n1) -> (hasPropComponent:bool)
     B:bool=Get isTitleRow!(self:Music Party Editor Level Actor Entry Object Reference=@n1) -> (isTitleRow:bool)
    ) -> (ReturnValue:bool)
   )
   n2 = GetLevelActor!(self:Music Party Editor Level Actor Entry Object Reference=@n1) -> (LevelActor:Actor Object Reference)
   > Is Valid(InputObject=@n2)
   > Set CurrentLevelActor(CurrentLevelActor:Actor Object Reference=@n2) -> (Output_Get:Actor Object Reference)
   > n3 = Is Valid(InputObject=GetComponentByClass!(self:Actor Object Reference=@n2, ComponentClass:Actor Component Class Reference=/Script/Engine.AudioComponent) -> (ReturnValue:Audio Component Object Reference))
   n4 = PropEdits_AddComponentByClass!(
    execute=@n3.Is Not Valid
    self
    Actor:Actor Object Reference=Get CurrentLevelActor() -> (CurrentLevelActor:Actor Object Reference)
    Class:Actor Component Class Reference=&AudioComponent
   ) -> (then)
   n5 = Get PropTypeComboBox() -> (PropTypeComboBox:Editor Utility Combo Box String Object Reference)
   n6 = Switch on Name!(execute:Exec=[@n4.then, @n3.Is Valid], Selection:Name=String To Name(InString=GetSelectedOption(self=@n5))) -> (Default:Exec, Controller Component:Exec, Radio Component:Exec, Actor Component:Exec, Deprecated Component:Exec) {is=title K2Node_SwitchName, pinNames [Controller Component, Radio Component, Actor Component, Deprecated Component]}
   n7 = Set NewActionPropComponent(
    execute=@n6.Actor Component
    NewActionPropComponent:Base Class: Music Party Prop Base Class Reference=/Game/Sounds/Fort_Audio_Blueprints/MusicParty/MusicPartyPropComponent_Actor.MusicPartyPropComponent_Actor_C
   ) -> (Output_Get:Base Class: Music Party Prop Base Class Reference)
   n8 = Get NewActionPropComponent() -> (NewActionPropComponent:Base Class: Music Party Prop Base Class Reference)
   PropEdits_AddComponentByClass(execute=[@n12.then, @n10.then, @n11.then, @n7.then], Actor=Get CurrentLevelActor(), Class=@n8)
   n9 = GetComponentByClass(
    self=Get ActionActor() -> (ActionActor:Actor Object Reference)
    ComponentClass=/Game/Sounds/Fort_Audio_Blueprints/MusicParty/MusicPartyPropComponentBase.MusicPartyPropComponentBase_C
   ) -> (ReturnValue:Base Class: Music Party Prop Base Object Reference)
   > Is Valid(InputObject=@n9)
   > GetAndTransferPropComponentVariables!(
    execute
    self
    MusicPartyPropComponent:Base Class: Music Party Prop Base Object Reference=@n9
    CopyToNewComponent:bool=true
    NewProp:Base Class: Music Party Prop Base Object Reference=GetComponentByClass(self=Get CurrentLevelActor(), ComponentClass=@n8)
    CopyMusicController:bool=true
   ) -> (
    then
    Controller:Actor Class Reference
    SpeakerActors:Array of Actor Soft Object References
    MusicEventTag:Gameplay Tag Structure
    GroupName:Name=None
    MusicPartySpeaker:Music Party Speaker Object Reference
   )
   n10 = Set NewActionPropComponent(
    execute=@n6.Controller Component
    NewActionPropComponent=/Game/Sounds/Fort_Audio_Blueprints/MusicParty/MusicPartyPropComponent_Controller.MusicPartyPropComponent_Controller_C
   )
   n11 = Set NewActionPropComponent(
    execute=@n6.Deprecated Component
    NewActionPropComponent=/Game/Sounds/Fort_Audio_Blueprints/MusicParty/MusicPartyPropComponent.MusicPartyPropComponent_C
   )
   n12 = Set NewActionPropComponent(
    execute=@n6.Radio Component
    NewActionPropComponent=/Game/Sounds/Fort_Audio_Blueprints/MusicParty/MusicPartyPropComponent_Radio.MusicPartyPropComponent_Radio_C
   )
  ]}
 CancelAction {nodes [
   CancelAction() {is=title K2Node_FunctionEntry}
   > Unbind all Events from Execute Action()
   > SetActiveWidgetIndex(self=Get WidgetSwitcher() -> (WidgetSwitcher:Widget Switcher Object Reference), Index=0)
   > n1 = Get!(execute) -> (then, else, ActionActor:Actor Object Reference)
   > Destroy Actor(self=@n1)
  ]}
 CheckAllLevelActors {nodes [
   CheckAllLevelActors() {is=title K2Node_FunctionEntry}
   > n1 = Branch(Condition=NOT Boolean(A=GetCurrentWorld().isPIE))
   > n2 = GetAllLevelActors(self=Get EditorActorSubsystem())
   > Clear(TargetArray:Array of Actor Object References=Get MonitoredLevelActors() -> (MonitoredLevelActors:Array of Actor Object References))
   > n3 = Set LevelActors(LevelActors:Array of Actor Object References=@n2) -> (Output_Get:Array of Actor Object References)
   > n4 = For Each Loop(Array:Array of Actor Object References=@n3) -> (Array Element:Actor Object Reference)
   Clear(execute=@n4.Completed, TargetArray=Get MonitoredPIEActors() -> (MonitoredPIEActors:Array of Actor Object References))
   > Clear(TargetArray:Array of Music Party Editor Level Actor Entry Object References=Get PIE_SelectedEntries() -> (PIE_SelectedEntries:Array of Music Party Editor Level Actor Entry Object References))
   > Set LastSelectedPIEWidget(LastSelectedPIEWidget:Object Reference) -> (Output_Get:Object Reference)
   > ClearListItems(self=Get MusicEditorPIEView())
   > ClearListItems(self=Get PIEActorEntries() -> (PIEActorEntries:List View Object Reference))
   > SetVisibility(self=Get PIE_ActorEditor() -> (PIE_ActorEditor:Border Object Reference), InVisibility=Collapsed)
   > SetVisibility(self=Get PIEView() -> (PIEView:Border Object Reference), InVisibility=Collapsed)
   > SetVisibility(self=Get Editor_ActorEditor() -> (Editor_ActorEditor:Scroll Box Object Reference), InVisibility=Visible)
   > SetVisibility(self=Get EditorContainer() -> (EditorContainer:Vertical Box Object Reference), InVisibility=Visible)
   Should Actor be Monitored!(exec=@n4.LoopBody, Actor:Actor Object Reference=@n4) -> (True:Exec, False:Exec) {is=title K2Node_MacroInstance}
   > Add Unique(TargetArray:Array of Actor Object References=Get MonitoredLevelActors(), NewItem:"Actor Object Reference (by ref)"=@n4)
   n5 = GetAllActorsOfClass!(execute=@n1.else, WorldContextObject:Object Reference=GetCurrentWorld(), ActorClass:Actor Class Reference=&Actor) -> (then, OutActors:Array of Actor Object References)
   > n6 = For Each Loop(Array=@n5)
   SetVisibility(execute=@n6.Completed, self=Get PIE_ActorEditor(), InVisibility=Visible)
   > SetVisibility(self=Get PIEView(), InVisibility=Visible)
   > SetVisibility(self=Get Editor_ActorEditor(), InVisibility=Collapsed)
   > SetVisibility(self=Get EditorContainer(), InVisibility=Collapsed)
   > ClearLevelEditorReferences()
   Should Actor be Monitored(exec=@n6.LoopBody, Actor=@n6)
   > Add Unique(TargetArray=Get MonitoredPIEActors(), NewItem=@n6)
  ]}
 CheckMultiEditProperties {nodes [
   n1 = CheckMultiEditProperties!() -> (then, InputName:Name, OutputName:Name, InputObject:Object Reference, OutputObject:Object Reference) {is=title K2Node_FunctionEntry}
   > Execute Python Script!(
    execute
    PythonScript="input_param = input_object.get_editor_property(input_name)\r\noutput_param = output_object.get_editor_property(output_name)\r\nmatchfound= input_param==output_param\r\nprint(input_param)\r\nprint(output_param)\r\nprint(matchfound)\r\nif not matchfound:\r\n\tprint(\"No Match Found for Multi-Edit Parameter\")\r\n\toutput_object.set_editor_property(output_name, \"\")"
    input_object:Object Reference=@n1.InputObject
    output_object:Object Reference=@n1.OutputObject
    input_name:Name=@n1
    output_name:Name=@n1.OutputName
   ) -> (then, ReturnValue) {inputs [input_object, output_object, input_name, output_name]}
   > Return Node(MatchFound=false)
  ]}
 CheckSelectedActors {nodes [
   CheckSelectedActors() {is=title K2Node_FunctionEntry}
   > n1 = Set ArrayToTest(ArrayToTest:Array of Music Party Editor Level Actor Entry Object References=Select(
     Option 0:Array of Music Party Editor Level Actor Entry Object References=Get SelectedEntries()
     Option 1:Array of Music Party Editor Level Actor Entry Object References=Get PIE_SelectedEntries()
     Index=Get isPIE() -> (isPIE:bool)
    ) -> (ReturnValue:Array of Music Party Editor Level Actor Entry Object References)) -> (Output_Get:Array of Music Party Editor Level Actor Entry Object References)
   > n2 = For Each Loop(Array:Array of Music Party Editor Level Actor Entry Object References=@n1) -> (Array Element:Music Party Editor Level Actor Entry Object Reference)
   Branch(execute=@n2.Completed, Condition=Get ForceRefresh() -> (ForceRefresh:bool))
   > Refresh Level Actor List(CheckAllLevelActors=true, DelayRefresh=false)
   n3 = Is Valid(exec=@n2.LoopBody, InputObject=@n2)
   Remove Item!(
    execute=[@n3.Is Not Valid, @n7.then]
    TargetArray:Array of Music Party Editor Level Actor Entry Object References=Select(Option 0=Get SelectedEntries(), Option 1=Get PIE_SelectedEntries(), Index=Get isPIE())
    Item:"Music Party Editor Level Actor Entry Object Reference (by ref)"=@n2
   ) -> (then, ReturnValue:bool)
   n4 = GetLevelActor(self=@n2)
   n5 = Is Valid(exec=@n3.Is Valid, InputObject=@n4)
   n6 = Branch(execute=@n5.Is Not Valid, Condition=Get isTitleRow(self=@n2))
   n7 = Set ForceRefresh(execute=@n6.else, ForceRefresh:bool=true) -> (Output_Get:bool)
   n8 = Set CurrentLevelActor(execute=@n5.Is Valid, CurrentLevelActor=@n4)
   > Branch(Condition=Get isShowingAttenuation() -> (isShowingAttenuation:bool))
   > Branch(Condition=Is Empty!(TargetArray:Array of Actor Object References=Get Parents!(self:Music Party Editor Level Actor Entry Object Reference=@n2) -> (Parents:Array of Actor Object References)) -> (ReturnValue:bool))
   n9 = GetComponentByClass(self=@n8, ComponentClass=&AudioComponent) -> (ReturnValue:Audio Component Object Reference)
   > Is Valid(InputObject=@n9)
   > DrawAttenuation!(
    execute
    self
    Audio:Audio Component Object Reference=@n9
    MusicProp:Base Class: Music Party Prop Base Object Reference=GetComponentByClass(self=@n8, ComponentClass=&MusicPartyPropComponentBase_C) -> (ReturnValue:Base Class: Music Party Prop Base Object Reference)
   ) -> (then)
  ]}
 ClearLevelEditorReferences {nodes [
   ClearLevelEditorReferences() {is=title K2Node_FunctionEntry}
   > Clear(TargetArray:Array of Actor Object References=Get LevelActors() -> (LevelActors:Array of Actor Object References))
   > Clear(TargetArray=Get MonitoredLevelActors())
   > Set LastSelectedWidget(LastSelectedWidget:Object Reference) -> (Output_Get:Object Reference)
   > Clear(TargetArray:Array of Music Party Editor Level Actor Entry Object References=Get SelectedEntries())
   > ClearListItems(self=Get MusicEditorListView())
   > ClearListItems(self=Get LevelActorEntries() -> (LevelActorEntries:List View Object Reference))
  ]}
 ConfirmAction {nodes [
   ConfirmAction() {is=title K2Node_FunctionEntry}
   > Call Execute Action!(execute:Exec, self:EUW Music Party Editor Object Reference) -> (then:Exec) {is=title K2Node_CallDelegate}
   > CancelAction() {is=title K2Node_CallFunction}
  ]}
 ConvertProp_BeginAction {nodes [
   n1 = ConvertProp_BeginAction!() -> (
    then
    LevelActorEntries:Array of Music Party Editor Level Actor Entry Object References
    ClassToConvertTo:EMusicPartyPropComponentType Enum=NewEnumerator0
   ) {is=title K2Node_FunctionEntry}
   > TransactionStart!(execute, self, Description:Text="LOCTEXT(\"Convert Prop Components\")") -> (then)
   > n2 = For Each Loop(Array=@n1)
   TransactionEnd(execute=@n2.Completed)
   > Refresh Level Actor List(CheckAllLevelActors=false, DelayRefresh=true)
   Branch(execute=@n2.LoopBody, Condition=Get hasPropComponent(self=@n2))
   n3 = GetLevelActor(self=@n2)
   > Set LevelActorToAct(LevelActorToAct:Actor Object Reference=@n3) -> (Output_Get:Actor Object Reference)
   > n4 = Set OldPropComponent(OldPropComponent:Base Class: Music Party Prop Base Object Reference=GetComponentByClass(self=@n3, ComponentClass=&MusicPartyPropComponentBase_C)) -> (Output_Get:Base Class: Music Party Prop Base Object Reference)
   > Is Valid(InputObject=@n4)
   n5 = Get OldPropComponent() -> (OldPropComponent:Base Class: Music Party Prop Base Object Reference)
   > n6 = Prop Edits Can Component be Deleted!(
    execute
    self
    Component:Object Reference=@n5
    LevelActor:Actor Object Reference=Get LevelActorToAct() -> (LevelActorToAct:Actor Object Reference)
   ) -> (then, CanBeReplaced:bool)
   > n7 = Branch(Condition=@n6)
   > n8 = Switch on EMusicPartyPropComponentType!(
    execute
    Selection:EMusicPartyPropComponentType Enum=Get ClassToConvertTo() -> (ClassToConvertTo:EMusicPartyPropComponentType Enum=NewEnumerator0)
   ) -> (NewEnumerator0:Exec, NewEnumerator1:Exec, NewEnumerator2:Exec, NewEnumerator3:Exec) {is=title K2Node_SwitchEnum}
   n9 = Get OldPropComponent()
   > n10 = Cast To MusicPartyPropComponent_Controller!(execute, Object=@n9) -> (then, CastFailed, AsMusic Party Prop Component Controller:Music Party Prop Component Controller Object Reference) {is=title K2Node_DynamicCast}
   n11 = Get LevelActorToAct()
   PropEdits_AddComponentByClass(
    execute=@n10.CastFailed
    Actor=@n11
    Class=&MusicPartyPropComponent_Controller_C
   )
   > Set NewComponent(NewComponent:Base Class: Music Party Prop Base Object Reference=GetComponentByClass(
     self=@n11
     ComponentClass=&MusicPartyPropComponent_Controller_C
    ) -> (ReturnValue:Music Party Prop Component Controller Object Reference)) -> (Output_Get:Base Class: Music Party Prop Base Object Reference)
   > n12 = GetAndTransferPropComponentVariables(
    MusicPartyPropComponent=Get OldPropComponent()
    CopyToNewComponent=true
    NewProp=Get NewComponent() -> (NewComponent:Base Class: Music Party Prop Base Object Reference)
    CopyMusicController=true
   ) -> (GroupName=None)
   Prop Edits Remove Component!(
    execute=[@n12.then, @n15.then, @n19.then]
    self
    Component:Object Reference=Get OldPropComponent()
    LevelActor:Actor Object Reference=Get LevelActorToAct()
   ) -> (then, Success:bool)
   n13 = Cast To MusicPartyPropComponent_Radio!(execute=@n8.NewEnumerator1, Object=@n9) -> (then, CastFailed, AsMusic Party Prop Component Radio:Music Party Prop Component Radio Object Reference) {is=title K2Node_DynamicCast}
   n14 = Get LevelActorToAct()
   PropEdits_AddComponentByClass(
    execute=@n13.CastFailed
    Actor=@n14
    Class=&MusicPartyPropComponent_Radio_C
   )
   > Set NewComponent(NewComponent=GetComponentByClass(self=@n14, ComponentClass=&MusicPartyPropComponent_Radio_C) -> (ReturnValue:Music Party Prop Component Radio Object Reference))
   > n15 = GetAndTransferPropComponentVariables(MusicPartyPropComponent=Get OldPropComponent(), CopyToNewComponent=true, NewProp=Get NewComponent(), CopyMusicController=false) -> (GroupName=None)
   n16 = Cast To MusicPartyPropComponent_Actor!(execute=@n8.NewEnumerator2, Object=@n9) -> (then, CastFailed, AsMusic Party Prop Component Actor:Music Party Prop Component Actor Object Reference) {is=title K2Node_DynamicCast}
   n17 = Get LevelActorToAct()
   PropEdits_AddComponentByClass(
    execute=@n16.CastFailed
    Actor=@n17
    Class=&MusicPartyPropComponent_Actor_C
   )
   > n18 = Set NewComponent(NewComponent=GetComponentByClass(self=@n17, ComponentClass=&MusicPartyPropComponent_Actor_C) -> (ReturnValue:Music Party Prop Component Actor Object Reference))
   n19 = GetAndTransferPropComponentVariables(
    execute=[@n22.then, @n18.then]
    MusicPartyPropComponent=Get OldPropComponent()
    CopyToNewComponent=true
    NewProp=Get NewComponent()
    CopyMusicController=true
   ) -> (GroupName=None)
   n20 = Cast To MusicPartyPropComponent!(execute=@n8.NewEnumerator3, Object=@n9) -> (then, CastFailed, As_MusicPartyPropComponent_Legacy:" Music Party Prop Component Legacy Object Reference") {is=title K2Node_DynamicCast}
   n21 = Get LevelActorToAct()
   PropEdits_AddComponentByClass(execute=@n20.CastFailed, Actor=@n21, Class=&MusicPartyPropComponent_C)
   > n22 = Set NewComponent(NewComponent=GetComponentByClass(self=@n21, ComponentClass=&MusicPartyPropComponent_C) -> (ReturnValue:" Music Party Prop Component Legacy Object Reference"))
   LogString(
    execute=@n7.else
    InString=Append!(
     A="Warning: The MusicPropComponent "
     B=GetDisplayName(Object=@n5)
     C:String=" cannot be deleted on this instance, so must be changed Manually "
    ) -> (ReturnValue)
   )
  ]}
 CopyProperty {nodes [
   n1 = CopyProperty!() -> (then, InputName:Name, OutputName:Name, InputObject:Object Reference, OutputObject:Object Reference) {is=title K2Node_FunctionEntry}
   > n2 = Execute Python Script!(
    execute
    PythonScript="new_param = input_object.get_editor_property(input_name)\r\noutput_object.set_editor_property(output_name, new_param)\r\noutput_success=true\r\n"
    input_name=@n1
    output_name=@n1.OutputName
    input_object=@n1.InputObject
    output_object=@n1.OutputObject
   ) -> (then, ReturnValue, output_success:bool=~) {inputs [input_name, output_name, input_object, output_object], outputs [output_success]}
   > Return Node!(execute, Success:bool=@n2.output_success)
  ]}
 DrawAttenuation {nodes [
   n1 = DrawAttenuation!() -> (then, Audio:Audio Component Object Reference, MusicProp:Base Class: Music Party Prop Base Object Reference) {is=title K2Node_FunctionEntry}
   > Branch(Condition=AND Boolean(A=IsValid(Object=@n1), B=IsValid(Object=@n1.MusicProp)))
   n2 = NOT Boolean(A=IsVirtualized!(self:Audio Component Object Reference=@n1) -> (ReturnValue:bool))
   n3 = Get bIsAudible!(self:Base Class: Music Party Prop Base Object Reference=@n1.MusicProp) -> (bIsAudible:bool)
   > Set Color(Color:Linear Color Structure=Select(
     Option 0:Linear Color Structure=(R=0,G=0.196,B=0.3281,A=1)
     Option 1:Linear Color Structure=(R=0.9951,G=0.276,B=1,A=1)
     Index=AND Boolean(A=@n2, B=@n3)
    ) -> (ReturnValue:Linear Color Structure)) -> (Output_Get:Linear Color Structure)
   > n4 = Get Attenuation Settings To Apply!(execute, self:Audio Component Object Reference=Get Audio() -> (Audio:Audio Component Object Reference)) -> (then, OutAttenuationSettings:Sound Attenuation Settings Structure, ReturnValue:bool) {is=title K2Node_CallFunction}
   > n5 = Set Attenuation(Attenuation:Sound Attenuation Settings Structure=@n4) -> (Output_Get:Sound Attenuation Settings Structure)
   n6 = Break Sound Attenuation Settings!(SoundAttenuationSettings:"Sound Attenuation Settings Structure (by ref)"=@n5) -> (AttenuationShape:EAttenuationShape Enum, AttenuationShapeExtents:Vector=~, FalloffDistance:float=~) {is=title K2Node_BreakStruct}
   > Set ShapeExtents(ShapeExtents:Vector=@n6.AttenuationShapeExtents) -> (Output_Get:Vector)
   n7 = BreakVector(InVec=@n6.AttenuationShapeExtents)
   > Set x(x:double=@n7) -> (Output_Get:double)
   > Set y(y:double=@n7.Y) -> (Output_Get:double)
   > Set z(z:double=@n7.Z) -> (Output_Get:double)
   > Set FalloffDistance(FalloffDistance:float=@n6.FalloffDistance) -> (Output_Get:float)
   > Set AudioLocation(AudioLocation:Vector=Get World Location!(self:Scene Component Object Reference=Get Audio()) -> (ReturnValue:Vector) {is=title K2Node_CallFunction}) -> (Output_Get:Vector)
   > n8 = Switch on EAttenuationShape!(execute, Selection:EAttenuationShape Enum=@n6) -> (Sphere:Exec, Capsule:Exec, Box:Exec, Cone:Exec) {is=title K2Node_SwitchEnum}
   n9 = Get World Rotation!(self:Scene Component Object Reference=Get Audio()) -> (ReturnValue:Rotator) {is=title K2Node_CallFunction}
   n10 = Get AudioLocation() -> (AudioLocation:Vector)
   n11 = Get Audio()
   n12 = Get Color() -> (Color:Linear Color Structure)
   DrawDebugBox!(
    execute=@n8.Box
    WorldContextObject:Object Reference=@n11
    Center:Vector=@n10
    Extent:Vector=Get ShapeExtents() -> (ShapeExtents:Vector)
    LineColor:Linear Color Structure=@n12
    Rotation:Rotator=@n9
    Duration:float=0
    Thickness:float=4
    DepthPriority:EDrawDebugSceneDepthPriorityGroup Enum=World
   ) -> (then)
   > DrawDebugBox(
    WorldContextObject=@n11
    Center=@n10
    Extent=vector + vector(A=Get ShapeExtents(), B:float=Get FalloffDistance() -> (FalloffDistance:float))
    LineColor=@n12
    Rotation=@n9
    Duration=0
    Thickness=4
    DepthPriority=World
   )
   n13 = Get World Rotation(self=Get Audio())
   n14 = Get Color()
   n15 = Get Audio()
   n16 = Get AudioLocation()
   DrawDebugCapsule!(
    execute=@n8.Capsule
    WorldContextObject:Object Reference=@n15
    Center:Vector=@n16
    HalfHeight:float=Get x() -> (x:double)
    Radius:float=Get y() -> (y:double)
    Rotation:Rotator=@n13
    LineColor:Linear Color Structure=@n14
    Duration:float=0
    Thickness:float=4
    DepthPriority:EDrawDebugSceneDepthPriorityGroup Enum=World
   ) -> (then)
   n17 = Get FalloffDistance()
   > DrawDebugCapsule(
    WorldContextObject=@n15
    Center=@n16
    HalfHeight=float + float(A=Get x(), B=@n17)
    Radius=float + float(A=Get y(), B=@n17)
    Rotation=@n13
    LineColor=@n14
    Duration=0
    Thickness=4
    DepthPriority=World
   )
   n18 = Get Audio()
   n19 = Get Rotation X Vector!(InRot:Rotator=Get World Rotation(self=@n18)) -> (ReturnValue:Vector) {is=title K2Node_CallFunction}
   Draw Debug Cone!(
    execute=@n8.Cone
    WorldContextObject:Object Reference=@n18
    Origin:Vector=Get AudioLocation()
    Direction:Vector=@n19
    Length:float=Get x()
    AngleWidth:float=Get y()
    AngleHeight:float=Get z() -> (z:double)
    NumSides:int=12
    LineColor:Linear Color Structure=Get Color()
    Duration:float=0
    Thickness:float=4
    DepthPriority:EDrawDebugSceneDepthPriorityGroup Enum=World
   ) -> (then)
   > Draw Debug Cone(
    WorldContextObject=@n18
    Origin=Get AudioLocation()
    Direction=@n19
    Length=float + float(A=Get x(), B=Get FalloffDistance())
    AngleWidth=Get y()
    AngleHeight=Get z()
    NumSides=12
    LineColor=Get Color()
    Duration=0
    Thickness=4
    DepthPriority=World
   )
   n20 = Break Sound Attenuation Settings!(SoundAttenuationSettings=Get Attenuation() -> (Attenuation:Sound Attenuation Settings Structure)) -> (ConeSphereRadius:float=~, ConeSphereFalloffDistance:float=~)
   > DrawDebugSphere!(
    execute
    WorldContextObject:Object Reference=@n18
    Center:Vector=Get AudioLocation()
    Radius:float=@n20
    Segments:int=12
    LineColor:Linear Color Structure=Get Color()
    Duration:float=0
    Thickness:float=4
    DepthPriority:EDrawDebugSceneDepthPriorityGroup Enum=World
   ) -> (then)
   > DrawDebugSphere(
    WorldContextObject=@n18
    Center=Get AudioLocation()
    Radius=float + float(A=@n20, B=@n20.ConeSphereFalloffDistance)
    Segments=12
    LineColor=Get Color()
    Duration=0
    Thickness=4
    DepthPriority=World
   )
   n21 = Get Audio()
   n22 = Get AudioLocation()
   DrawDebugSphere(
    execute=@n8.Sphere
    WorldContextObject=@n21
    Center=@n22
    Radius=Get x()
    Segments=12
    LineColor=Get Color()
    Duration=0
    Thickness=4
    DepthPriority=World
   )
   > DrawDebugSphere(
    WorldContextObject=@n21
    Center=@n22
    Radius=float + float(A=Get x(), B=Get FalloffDistance())
    Segments=12
    LineColor=Get Color()
    Duration=0
    Thickness=4
    DepthPriority=World
   )
  ]}
 EventGraph {nodes [
   n1 = "On Entry Initialized (LevelActorEntries)"!() -> (OutputDelegate, then, Item:Object Reference, Widget:User Widget Object Reference) {is=title K2Node_ComponentBoundEvent}
   > OnActorEntryInitialized!(
    execute
    self
    TargetArray:Array of Music Party Editor Level Actor Entry Object References=Get SelectedEntries()
    Entry:Object Reference=@n1.Item
    Widget:Object Reference=@n1.Widget
   ) -> (then)
   n2 = "On Entry Initialized (MusicEditorPIEView)"!() -> (OutputDelegate, then, Item:Object Reference, Widget:User Widget Object Reference) {is=title K2Node_ComponentBoundEvent}
   > OnEditEntryInitialized!(execute, self, Widget:Object Reference=@n2.Widget) -> (then)
   n3 = "On Check State Changed (ShowAttenuation)"!() -> (OutputDelegate, then, bIsChecked:bool) {is=title K2Node_ComponentBoundEvent}
   > Set isShowingAttenuation(isShowingAttenuation:bool=@n3.bIsChecked) -> (Output_Get:bool)
   n4 = "On Entry Initialized (MusicEditorListView)"!() -> (OutputDelegate, then, Item:Object Reference, Widget:User Widget Object Reference) {is=title K2Node_ComponentBoundEvent}
   > OnEditEntryInitialized(Widget=@n4.Widget)
   K2Node_ComponentBoundEvent_3 = CheckAssetExtendedDetails!() -> (OutputDelegate, then, bIsChecked:bool=~) {is=title K2Node_CustomEvent}
   > SetVisibility(
    self=Get NewAsset_ExtendedDetails() -> (NewAsset_ExtendedDetails:Details View Object Reference)
    InVisibility=Select(Option 0:ESlateVisibility Enum=Visible, Option 1:ESlateVisibility Enum=Collapsed, Index=@K2Node_ComponentBoundEvent_3.bIsChecked) -> (ReturnValue:ESlateVisibility Enum)
   )
   n5 = "On Entry Initialized (PIEActorEntries)"!() -> (OutputDelegate, then, Item:Object Reference, Widget:User Widget Object Reference) {is=title K2Node_ComponentBoundEvent}
   > OnActorEntryInitialized(TargetArray=Get PIE_SelectedEntries(), Entry=@n5.Item, Widget=@n5.Widget)
   "On Property Changed (FindActorSearchTerms)"!() -> (OutputDelegate, then, PropertyName:Name=None) {is=title K2Node_ComponentBoundEvent}
   > Refresh Level Actor List(CheckAllLevelActors=true, DelayRefresh=false)
   "On Item Is Hovered Changed (LevelActorEntries)"!() -> (OutputDelegate, then, Item:Object Reference, bIsHovered:bool) {is=title K2Node_ComponentBoundEvent}
   > n6 = For Each Loop(Array=Get SelectedEntries())
   > n7 = Branch(Condition=Get isTitleRow(self=@n6))
   OnSelectionChanged!(
    execute=@n7.else
    self:EUW Music Party Editor Level Actor Entry Object Reference=Get Widget!(self:Music Party Editor Level Actor Entry Object Reference=@n6) -> (Widget:EUW Music Party Editor Level Actor Entry Object Reference)
    Selected:bool=true
   ) -> (then)
   "On Clicked (ForceRefreshButton)"() {is=title K2Node_ComponentBoundEvent}
   > Refresh Level Actor List(CheckAllLevelActors=true, DelayRefresh=false)
   Event Destruct()
   > CancelAction()
   Event Construct()
   > Refresh Level Actor List(CheckAllLevelActors=true, DelayRefresh=false)
   > SetObject(self=Get ConvertProp_Enum() -> (ConvertProp_Enum:Single Property View Object Reference), NewObject=Self-Reference())
   Event Tick()
   > CheckSelectedActors() {is=title K2Node_CallFunction}
   n8 = GetCurrentWorld()
   > Branch(Condition="Not Equal (Boolean)"(A=@n8.isPIE, B=Get isPIE()) -> (ReturnValue=~))
   > PIE_Update!(execute, self, isPIE:bool=@n8.isPIE) -> (then)
   > Refresh Level Actor List(CheckAllLevelActors=true, DelayRefresh=true)
   Event PreConstruct()
   > SetObject(self=Get FindActorSearchTerms() -> (FindActorSearchTerms:Details View Object Reference), NewObject=Self-Reference())
  ]}
 ExecuteAction {nodes [ExecuteAction() {is=title K2Node_FunctionEntry}]}
 GenerateActorList {nodes [
   n1 = GenerateActorList!() -> (
    then
    SelectedEntriesToEvaluate:Array of Music Party Editor Level Actor Entry Object References
    ListView:List View Object Reference
    ActorsToEvaluate:Array of Actor Object References
   ) {is=title K2Node_FunctionEntry}
   > n2 = For Each Loop(Array=@n1)
   ClearListItems(execute=@n2.Completed, self=Get ListView() -> (ListView:List View Object Reference))
   > n3 = For Each Loop(Array:Array of Actor Object References=Get ActorsToEvaluate() -> (ActorsToEvaluate:Array of Actor Object References)) -> (Array Element:Actor Object Reference)
   Clear(execute=@n3.Completed, TargetArray:Array of Names=Get MusicGroups() -> (MusicGroups:Array of Names))
   > n4 = Values!(
    execute
    TargetMap:Map of Music Party Editor Level Actor Entry Object References to Names=Get LevelActorsWithComponents() -> (LevelActorsWithComponents:Map of Music Party Editor Level Actor Entry Object References to Names)
   ) -> (then, Values:Array of Names)
   > n5 = For Each Loop(Array:Array of Names=@n4) -> (Array Element:Name)
   SortNameArray!(
    execute=@n5.Completed
    TargetArray:Array of Names=Get MusicGroups()
    bStableSort:bool=false
    bLexicalSort:bool=true
    SortOrder:EArraySortOrder Enum=Ascending
   ) -> (then)
   > n6 = For Each Loop(Array=Get MusicGroups())
   n7 = Construct Object from Class(
    execute=@n6.Completed
    Class=&/MusicPartyEditor_LevelActorEntry.MusicPartyEditor_LevelActorEntry_C
   ) -> (ReturnValue:Music Party Editor Level Actor Entry Object Reference)
   > n8 = Set CurrentCategoryGroupEntry(CurrentCategoryGroupEntry:Music Party Editor Level Actor Entry Object Reference=@n7) -> (Output_Get:Music Party Editor Level Actor Entry Object Reference)
   > SetVariables!(
    execute
    self:Music Party Editor Level Actor Entry Object Reference=@n8
    LevelActor:Actor Object Reference
    IsTitleRow:bool=true
    "Optional: TitleText":Text="LOCTEXT(\"Unassigned Level Actors\")"
    Color:Linear Color Structure=(R=1,G=1,B=1,A=1)
    MusicController:Actor Class Reference
    SpeakerActors:Array of Actor Soft Object References
    EventTag:Gameplay Tag Structure
    MusicGroupName:Name=None
   ) -> (then, _HasComponent:bool, _GroupName:Name=None)
   > Add Unique(
    TargetArray:Array of Music Party Editor Level Actor Entry Object References=Get AllMusicPropEntryObjects() -> (AllMusicPropEntryObjects:Array of Music Party Editor Level Actor Entry Object References)
    NewItem:"Music Party Editor Level Actor Entry Object Reference (by ref)"=Get CurrentCategoryGroupEntry() -> (CurrentCategoryGroupEntry:Music Party Editor Level Actor Entry Object Reference)
   )
   > n9 = Values(TargetMap:Map of Actor Object References to Music Party Editor Level Actor Entry Object References=Get LevelActorsWithNoComponent() -> (LevelActorsWithNoComponent:Map of Actor Object References to Music Party Editor Level Actor Entry Object References)) -> (Values:Array of Music Party Editor Level Actor Entry Object References)
   > n10 = For Each Loop(Array:Array of Music Party Editor Level Actor Entry Object References=@n9) -> (Array Element:Music Party Editor Level Actor Entry Object Reference)
   > Add Unique(TargetArray=Get AllMusicPropEntryObjects(), NewItem=@n10)
   > Add Unique(
    TargetArray=Get ChildWidgets!(self:Music Party Editor Level Actor Entry Object Reference=Get CurrentCategoryGroupEntry()) -> (ChildWidgets:Array of Music Party Editor Level Actor Entry Object References)
    NewItem=@n10
   )
   n11 = Construct Object from Class(
    execute=@n10.Completed
    Class=&MusicPartyEditor_LevelActorEntry_C
   )
   > n12 = Set CurrentCategoryGroupEntry(CurrentCategoryGroupEntry=@n11)
   > SetVariables(self=@n12, IsTitleRow=true, "Optional: TitleText"="LOCTEXT(\"Silenced Components\")", Color=(R=1,G=1,B=1,A=1), MusicGroupName=None) -> (_GroupName=None)
   > Add Unique(TargetArray=Get AllMusicPropEntryObjects(), NewItem=Get CurrentCategoryGroupEntry())
   > n13 = Values(TargetMap=Get SilencedLevelActors() -> (SilencedLevelActors:Map of Actor Object References to Music Party Editor Level Actor Entry Object References))
   > n14 = For Each Loop(Array=@n13)
   n15 = Get AllMusicPropEntryObjects()
   Set List Items!(execute=@n14.Completed, self:List View Object Reference=Get ListView(), InListItems:Array of Object References=@n15) -> (then) {is=title K2Node_CallFunction}
   > n16 = For Each Loop(Array=@n15)
   Return Node!(
    execute=@n16.Completed
    NewSelectedEntries:Array of Music Party Editor Level Actor Entry Object References=Get UpdatedSelectedEntries() -> (UpdatedSelectedEntries:Array of Music Party Editor Level Actor Entry Object References)
   )
   n17 = GetLevelActor(self=@n16)
   > Is Valid(InputObject=@n17)
   n18 = Get OldSelectedEntries() -> (OldSelectedEntries:Array of Actor Object References)
   > Branch(Condition="integer >= integer"(A=Find Item!(TargetArray:Array of Actor Object References=@n18, ItemToFind:"Actor Object Reference (by ref)"=@n17) -> (ReturnValue)))
   > Add(
    TargetArray:Array of Music Party Editor Level Actor Entry Object References=Get UpdatedSelectedEntries()
    NewItem:"Music Party Editor Level Actor Entry Object Reference (by ref)"=@n16
   ) {is=title K2Node_CallArrayFunction}
   Is Valid(exec=@n14.LoopBody, InputObject=@n14)
   > Add Unique(TargetArray=Get AllMusicPropEntryObjects(), NewItem=@n14)
   > Add Unique(TargetArray=Get ChildWidgets(self=Get CurrentCategoryGroupEntry()), NewItem=@n14)
   Set Selected Color(
    execute=@n6.LoopBody
    Selected Color:Linear Color Structure=Increment Colors!(Int:int=@n6.Array Index) -> (ReturnValue:Linear Color Structure) {is=title K2Node_MacroInstance}
   ) -> (Output_Get:Linear Color Structure)
   > Set CurrentMusicGroup(CurrentMusicGroup:Name=@n6) -> (Output_Get:Name=None)
   > n19 = Construct Object from Class(Class=&MusicPartyEditor_LevelActorEntry_C)
   > n20 = Set CurrentCategoryGroupEntry(CurrentCategoryGroupEntry=@n19)
   > SetVariables(
    self=@n20
    IsTitleRow=true
    "Optional: TitleText"="To Text (Name)"!(InName:Name=Get CurrentMusicGroup() -> (CurrentMusicGroup:Name=None)) -> (ReturnValue:Text)
    Color=Get Selected Color() -> (Selected Color:Linear Color Structure)
    MusicGroupName=Get CurrentMusicGroup() -> (CurrentMusicGroup=None)
   ) -> (_GroupName=None)
   > Add(TargetArray=Get AllMusicPropEntryObjects(), NewItem=Get CurrentCategoryGroupEntry())
   > n21 = Values(TargetMap:Map of Music Party Editor Level Actor Entry Object References to Names=Get LevelActorsWithComponents()) -> (Values:Array of Names)
   > n22 = For Each Loop(Array:Array of Names=@n21) -> (Array Element:Name)
   > Branch(Condition="Equal (Name)"(A=@n22, B=Get CurrentMusicGroup() -> (CurrentMusicGroup=None)))
   > n23 = Keys!(execute, TargetMap:Map of Music Party Editor Level Actor Entry Object References to Names=Get LevelActorsWithComponents()) -> (then, Keys:Array of Music Party Editor Level Actor Entry Object References)
   > n24 = Set CurrentLevelActorEntry(CurrentLevelActorEntry:Music Party Editor Level Actor Entry Object Reference="Get (a copy)"(Array:Array of Music Party Editor Level Actor Entry Object References=@n23, Dimension 1=@n22.Array Index) -> (Output:Music Party Editor Level Actor Entry Object Reference)) -> (Output_Get:Music Party Editor Level Actor Entry Object Reference)
   > Set AssignedColor!(execute, AssignedColor:Linear Color Structure=Get Selected Color(), self:Music Party Editor Level Actor Entry Object Reference=@n24) -> (then, Output_Get:Linear Color Structure)
   n25 = Get CurrentLevelActorEntry() -> (CurrentLevelActorEntry:Music Party Editor Level Actor Entry Object Reference)
   > Add Unique(TargetArray=Get AllMusicPropEntryObjects(), NewItem=@n25)
   > Add Unique(TargetArray=Get ChildWidgets(self=Get CurrentCategoryGroupEntry()), NewItem=@n25)
   > n26 = For Each Loop(Array:Array of Actor Soft Object References=Get SpeakerActors!(self:Music Party Editor Level Actor Entry Object Reference=Get CurrentLevelActorEntry()) -> (SpeakerActors:Array of Actor Soft Object References)) -> (Array Element:Actor Soft Object Reference)
   n27 = Resolve Soft Reference!(Input:Actor Soft Object Reference=@n26) -> (Output:Actor Object Reference)
   n28 = Find(
    TargetMap:Map of Actor Object References to Music Party Editor Level Actor Entry Object References=Get LevelActorsWithNoComponent()
    Key:"Actor Object Reference (by ref)"=@n27
   ) -> (Value:Music Party Editor Level Actor Entry Object Reference)
   > n29 = Branch(Condition=@n28.ReturnValue)
   > n30 = Set CurrentLevelActorChildEntry(CurrentLevelActorChildEntry:Music Party Editor Level Actor Entry Object Reference=@n28) -> (Output_Get:Music Party Editor Level Actor Entry Object Reference)
   Add Unique(
    execute=[@n35.then, @n30.then]
    TargetArray=Get AllMusicPropEntryObjects()
    NewItem=Get CurrentLevelActorChildEntry() -> (CurrentLevelActorChildEntry:Music Party Editor Level Actor Entry Object Reference)
   )
   n31 = GetLevelActor(self=Get CurrentLevelActorEntry())
   > Add Unique(
    TargetArray:Array of Actor Object References=Get Parents(self=Get CurrentLevelActorChildEntry())
    NewItem:"Actor Object Reference (by ref)"=@n31
   )
   n32 = Get CurrentLevelActorChildEntry()
   > Set AssignedColor(AssignedColor=Get Selected Color(), self=@n32)
   > Add Unique(
    TargetArray:Array of Music Party Editor Level Actor Entry Object References=Get ChildWidgets(self=Get CurrentLevelActorEntry())
    NewItem:"Music Party Editor Level Actor Entry Object Reference (by ref)"=@n32
   )
   n33 = Find(TargetMap=Get SilencedLevelActors(), Key=@n27)
   n34 = Branch(execute=@n29.else, Condition=@n33.ReturnValue)
   > n35 = Set CurrentLevelActorChildEntry(CurrentLevelActorChildEntry=@n33)
   n36 = Construct Object from Class(
    execute=@n34.else
    Class=&MusicPartyEditor_LevelActorEntry_C
   )
   > n37 = Set CurrentLevelActorChildEntry(CurrentLevelActorChildEntry=@n36)
   > SetVariables(self=[@n36, @n37], LevelActor=@n27, IsTitleRow=false, Color=Get Selected Color(), MusicGroupName=None) -> (_GroupName=None)
   > Add Unique(
    TargetArray:Array of Actor Object References=Get Parents(self=Get CurrentLevelActorChildEntry())
    NewItem:"Actor Object Reference (by ref)"=@n31
   )
   n38 = Get CurrentLevelActorChildEntry()
   > Add(TargetArray=Get AllMusicPropEntryObjects(), NewItem=@n38)
   > Add Unique(
    TargetArray:Array of Music Party Editor Level Actor Entry Object References=Get ChildWidgets(self=Get CurrentLevelActorEntry())
    NewItem:"Music Party Editor Level Actor Entry Object Reference (by ref)"=@n38
   )
   Add Unique(execute=@n5.LoopBody, TargetArray:Array of Names=Get MusicGroups(), NewItem:"Name (by ref)"=@n5)
   n39 = Is Valid(exec=@n3.LoopBody, InputObject=@n3)
   Remove Item(execute=@n39.Is Not Valid, TargetArray:Array of Actor Object References=Get ActorsToEvaluate(), Item:"Actor Object Reference (by ref)"=@n3)
   Set CurrentLevelActor(execute=@n39.Is Valid, CurrentLevelActor=@n3)
   > n40 = Construct Object from Class(Class=&MusicPartyEditor_LevelActorEntry_C)
   > Set CurrentLevelActorEntry(CurrentLevelActorEntry=@n40)
   n41 = GetComponentByClass(
    self=Get CurrentLevelActor()
    ComponentClass=&MusicPartyPropComponentBase_C
   ) -> (ReturnValue:Base Class: Music Party Prop Base Object Reference)
   > n42 = Is Valid(InputObject=@n41)
   SetVariables(
    execute=@n42.Is Not Valid
    self=Get CurrentLevelActorEntry()
    LevelActor=Get CurrentLevelActor()
    IsTitleRow=false
    Color=(R=1,G=1,B=1,A=1)
    MusicGroupName=None
   ) -> (_GroupName=None)
   > Add!(
    execute
    TargetMap:Map of Actor Object References to Music Party Editor Level Actor Entry Object References=Get LevelActorsWithNoComponent()
    Key:"Actor Object Reference (by ref)"=Get CurrentLevelActor()
    Value:"Music Party Editor Level Actor Entry Object Reference (by ref)"=Get CurrentLevelActorEntry()
   ) -> (then) {is=title K2Node_CallFunction}
   n43 = GetAndTransferPropComponentVariables(execute=@n42.Is Valid, MusicPartyPropComponent=@n41, CopyToNewComponent=false, CopyMusicController=true) -> (GroupName=None)
   > n44 = SetVariables(
    self=Get CurrentLevelActorEntry()
    LevelActor=Get CurrentLevelActor()
    IsTitleRow=false
    Color=(R=1,G=1,B=1,A=1)
    MusicController=@n43
    SpeakerActors=@n43.SpeakerActors
    EventTag=@n43.MusicEventTag
    MusicGroupName=@n43.GroupName
   ) -> (_GroupName=None)
   n45 = Get BypassMusic!(self:Base Class: Music Party Prop Base Object Reference=@n41) -> (BypassMusic:bool)
   > n46 = Branch(Condition=AND Boolean(A=Get bAutoActivate!(self:Actor Component Object Reference=@n41) -> (bAutoActivate:bool), B=NOT Boolean(A=@n45)))
   > Add(
    TargetMap:Map of Music Party Editor Level Actor Entry Object References to Names=Get LevelActorsWithComponents()
    Key:"Music Party Editor Level Actor Entry Object Reference (by ref)"=Get CurrentLevelActorEntry()
    Value:"Name (by ref)"=@n44._GroupName
   )
   Add(
    execute=@n46.else
    TargetMap:Map of Actor Object References to Music Party Editor Level Actor Entry Object References=Get SilencedLevelActors()
    Key:"Actor Object Reference (by ref)"=Get CurrentLevelActor()
    Value:"Music Party Editor Level Actor Entry Object Reference (by ref)"=Get CurrentLevelActorEntry()
   )
   n47 = GetLevelActor(self=@n2)
   Is Valid(exec=@n2.LoopBody, InputObject=@n47)
   > Add Unique(TargetArray:Array of Actor Object References=Get OldSelectedEntries(), NewItem:"Actor Object Reference (by ref)"=@n47)
  ]}
 GetAndTransferPropComponentVariables {nodes [
   n1 = GetAndTransferPropComponentVariables!() -> (
    then
    MusicPartyPropComponent:Base Class: Music Party Prop Base Object Reference
    CopyToNewComponent:bool=~
    NewProp:Base Class: Music Party Prop Base Object Reference
    CopyMusicController:bool=true
   ) {is=title K2Node_FunctionEntry}
   n2 = Get NewProp() -> (NewProp:Base Class: Music Party Prop Base Object Reference)
   > Set _CopyToComponent(_CopyToComponent:bool=AND Boolean(A=Get CopyToNewComponent() -> (CopyToNewComponent:bool), B=IsValid(Object=@n2))) -> (Output_Get:bool)
   > Is Valid(InputObject=@n1)
   > n3 = GetEditorProperty!(execute:Exec, Object:Object Reference=@n1, PropertyName:Name=MusicEventTag) -> (then:Exec, PropertyValue:Gameplay Tag Structure, ReturnValue:bool)
   > n4 = Set _GameplayTag(_GameplayTag:Gameplay Tag Structure=@n3) -> (Output_Get:Gameplay Tag Structure)
   > n5 = Branch(Condition=AND Boolean(A=@n3.ReturnValue, B=Get _CopyToComponent() -> (_CopyToComponent:bool)))
   > n6 = SetEditorProperty(Object=Get NewProp(), PropertyName=MusicEventTag, PropertyValue:Gameplay Tag Structure=@n4)
   n7 = GetEditorProperty(execute=[@n5.else, @n6.then], Object=@n1, PropertyName=MusicEventGroupName) -> (PropertyValue:Name)
   > n8 = Set _GroupName(_GroupName:Name=@n7) -> (Output_Get:Name=None)
   > n9 = Branch(Condition=AND Boolean(A=@n7.ReturnValue, B=Get _CopyToComponent()))
   > n10 = SetEditorProperty(Object=Get NewProp(), PropertyName=MusicEventGroupName, PropertyValue:Name=@n8)
   n11 = GetEditorProperty(execute=[@n9.else, @n10.then], Object=@n1, PropertyName=ControllerClass) -> (PropertyValue:Actor Class Reference)
   > n12 = Set _Controller(_Controller:Actor Class Reference=@n11) -> (Output_Get:Actor Class Reference)
   > n13 = Branch(Condition=AND Boolean!(A=@n11.ReturnValue, B=Get _CopyToComponent(), C:bool=Get CopyMusicController() -> (CopyMusicController:bool)) -> (ReturnValue))
   > n14 = SetEditorProperty(Object=Get NewProp(), PropertyName=ControllerClass, PropertyValue:Actor Class Reference=@n12)
   n15 = GetEditorProperty(execute=[@n13.else, @n14.then], Object=@n1, PropertyName=SpeakerActors) -> (PropertyValue:Array of Actor Soft Object References)
   > n16 = Set _SpeakerActors(_SpeakerActors:Array of Actor Soft Object References=@n15) -> (Output_Get:Array of Actor Soft Object References)
   > n17 = Branch(Condition=AND Boolean!(A=@n15.ReturnValue, B=Get _CopyToComponent()) -> (ReturnValue))
   > n18 = SetEditorProperty(Object=Get NewProp(), PropertyName=SpeakerActors, PropertyValue:Array of Actor Soft Object References=@n16)
   n19 = Cast To MusicPartyPropComponent(
    execute=[@n17.else, @n18.then]
    Object=Get MusicPartyPropComponent() -> (MusicPartyPropComponent:Base Class: Music Party Prop Base Object Reference)
   )
   > Return Node!(
    execute
    Controller:Actor Class Reference=Get _Controller() -> (_Controller:Actor Class Reference)
    SpeakerActors:Array of Actor Soft Object References=Get _SpeakerActors() -> (_SpeakerActors:Array of Actor Soft Object References)
    MusicEventTag:Gameplay Tag Structure=Get _GameplayTag() -> (_GameplayTag:Gameplay Tag Structure)
    GroupName:Name=Get _GroupName() -> (_GroupName:Name=None)
    MusicPartySpeaker:Music Party Speaker Object Reference=Get _MusicPartySpeaker() -> (_MusicPartySpeaker:Music Party Speaker Object Reference)
   )
   n20 = GetEditorProperty(execute=@n19.CastFailed, Object=@n1, PropertyName=SpeakerAsset) -> (PropertyValue:Music Party Speaker Object Reference)
   > n21 = Set _MusicPartySpeaker(_MusicPartySpeaker:Music Party Speaker Object Reference=@n20) -> (Output_Get:Music Party Speaker Object Reference)
   > n22 = Branch(Condition=AND Boolean(A=@n20.ReturnValue, B=Get _CopyToComponent()))
   > n23 = SetEditorProperty(Object=Get NewProp(), PropertyName=SpeakerAsset, PropertyValue:Music Party Speaker Object Reference=@n21)
   Return Node(
    execute=[@n22.else, @n23.then]
    Controller=Get _Controller()
    SpeakerActors=Get _SpeakerActors()
    MusicEventTag=Get _GameplayTag()
    GroupName=Get _GroupName() -> (_GroupName=None)
    MusicPartySpeaker=Get _MusicPartySpeaker()
   )
  ]}
 GetCurrentWorld {nodes [
   GetCurrentWorld() {is=title K2Node_FunctionEntry}
   > n1 = GetPIEWorlds!(execute, bIncludeDedicatedServer:bool=false) -> (then, ReturnValue:Array of World Object References)
   > n2 = Branch(Condition="integer > integer"(A=Length(TargetArray:Array of World Object References=@n1)))
   n3 = "Get (a copy)"(Array:Array of World Object References=@n1) -> (Output:World Object Reference)
   > n4 = Is Valid(InputObject=@n3)
   n5 = GetEditorWorld(execute=[@n2.else, @n4.Is Not Valid], self=Get UnrealEditorSubsystem())
   > Set WorldObject(WorldObject:Object Reference=@n5) -> (Output_Get:Object Reference)
   > Return Node!(execute, Object:Object Reference=Get WorldObject() -> (WorldObject:Object Reference), isPIE:bool=false)
   n6 = Set WorldObject(execute=@n4.Is Valid, WorldObject=@n3)
   > Return Node(Object=@n6, isPIE=true)
  ]}
 IncrementColors {nodes [
   n1 = "% (Integer)"(A=Inputs!() -> (Int:int), B=2) -> (ReturnValue=0)
   Outputs!(ReturnValue:Linear Color Structure=Select(
     Option 0:Linear Color Structure=(R=0.8271,G=0.4136,B=1,A=1)
     Option 1:Linear Color Structure=(R=0,G=0.6823,B=0.6267,A=1)
     Index="integer > integer"(A=@n1)
    ) -> (ReturnValue:Linear Color Structure))
  ]}
 NewAsset_BeginNewControllerAction {nodes [
   n1 = NewAsset_BeginNewControllerAction!() -> (then, ControllerType:String) {is=title K2Node_FunctionEntry}
   > SetActiveWidgetIndex(self=Get WidgetSwitcher(), Index=2)
   > n2 = Construct Object from Class(Class=Find(
     TargetMap:Map of Strings to Music Party Editor New Asset Base Class References=Get ControllerToClass() -> (ControllerToClass:Map of Strings to Music Party Editor New Asset Base Class References)
     Key:"String (by ref)"=@n1
    ) -> (Value:Music Party Editor New Asset Base Class Reference)) -> (ReturnValue:Music Party Editor New Asset Base Object Reference)
   > n3 = Set NewAsset_Object(NewAsset_Object:Music Party Editor New Asset Base Object Reference=@n2) -> (Output_Get:Music Party Editor New Asset Base Object Reference)
   > n4 = GetSelectedAssets()
   > n5 = Branch(Condition="integer > integer"(A=Length(TargetArray:Array of Object References=@n4)))
   > n6 = Set Path!(
    execute
    Path:String=Get Object Path String(Object="Get (a copy)"(Array:Array of Object References=@n4) -> (Output:Object Reference))
    self:Music Party Editor New Asset Base Object Reference=@n3
   ) -> (then, Output_Get:String)
   SetObject(
    execute=[@n6.then, @n5.else]
    self=Get NewAsset_Details() -> (NewAsset_Details:Details View Object Reference)
    NewObject=Get NewAsset_Object() -> (NewAsset_Object:Music Party Editor New Asset Base Object Reference)
   )
   > SetObject(self=Get NewAsset_ExtendedDetails(), NewObject=Get NewAsset_Object())
   > SetVisibility(self=Get NewAsset_Controller_ComboBox(), InVisibility=Visible)
   > Set Text(Text="LOCTEXT(\"Create New Controller\")", self=Get NewAsset_Text() -> (NewAsset_Text:Text Block Object Reference))
   > n7 = Cast To MusicPartyEditor_NewAsset_WaveController!(execute, Object=Get NewAsset_Object()) -> (then, CastFailed, AsMusic Party Editor New Asset Wave Controller:Music Party Editor New Asset Wave Controller Object Reference) {is=title K2Node_DynamicCast}
   > SetVisibility(
    self=Get NewAsset_AutoGenerateAudioBus_Container() -> (NewAsset_AutoGenerateAudioBus_Container:Horizontal Box Object Reference)
    InVisibility=Visible
   )
   > CheckAssetExtendedDetails(bIsChecked=IsChecked(self=Get NewAsset_AutoGenerateAudioBus())) {is=title K2Node_CallFunction}
   SetVisibility(execute=@n7.CastFailed, self=Get NewAsset_AutoGenerateAudioBus_Container(), InVisibility=Collapsed)
   > CheckAssetExtendedDetails(bIsChecked=true)
  ]}
 NewAsset_BeginNewSpeakerAction {nodes [
   NewAsset_BeginNewSpeakerAction() {is=title K2Node_FunctionEntry}
   > SetActiveWidgetIndex(self=Get WidgetSwitcher(), Index=2)
   > CheckAssetExtendedDetails(bIsChecked=true)
   > n1 = Construct Object from Class(Class=&/MusicPartyEditor_NewAsset_SpeakerAsset.MusicPartyEditor_NewAsset_SpeakerAsset_C) -> (ReturnValue:Music Party Editor New Asset Speaker Asset Object Reference)
   > Set NewAsset_Object(NewAsset_Object=@n1)
   > n2 = GetSelectedAssets()
   > n3 = Branch(Condition="integer > integer"(A=Length(TargetArray=@n2)))
   > n4 = Set Path(Path=Get Object Path String(Object="Get (a copy)"(Array=@n2)), self=Get NewAsset_Object())
   SetObject(execute=[@n4.then, @n3.else], self=Get NewAsset_Details(), NewObject=Get NewAsset_Object())
   > SetObject(self=Get NewAsset_ExtendedDetails(), NewObject=Get NewAsset_Object())
   > SetVisibility(self=Get NewAsset_AutoGenerateAudioBus_Container(), InVisibility=Collapsed)
   > SetVisibility(self=Get NewAsset_Controller_ComboBox(), InVisibility=Collapsed)
   > Set Text(Text="LOCTEXT(\"Create New Speaker Asset\")", self=Get NewAsset_Text())
   > SetVisibility(self=Get NewAsset_ExtendedDetails(), InVisibility=Collapsed)
  ]}
 OnActorEntryInitialized {nodes [
   n1 = OnActorEntryInitialized!() -> (then, TargetArray:Array of Music Party Editor Level Actor Entry Object References, Entry:Object Reference, Widget:Object Reference) {is=title K2Node_FunctionEntry}
   > n2 = Cast To EUW_MusicPartyEditor_LevelActorEntry!(execute, Object=@n1.Widget) -> (then, CastFailed, AsEUW Music Party Editor Level Actor Entry:EUW Music Party Editor Level Actor Entry Object Reference) {is=title K2Node_DynamicCast}
   > OnPostInitialize!(execute, self:EUW Music Party Editor Level Actor Entry Object Reference=@n2, Parent:EUW Music Party Editor Object Reference=Self-Reference()) -> (then)
   > n3 = Cast To MusicPartyEditor_LevelActorEntry!(execute, Object=@n1.Entry) -> (then, CastFailed, AsMusic Party Editor Level Actor Entry:Music Party Editor Level Actor Entry Object Reference) {is=title K2Node_DynamicCast}
   > n4 = Branch(Condition="integer >= integer"(A=Find Item(
      TargetArray:Array of Music Party Editor Level Actor Entry Object References=@n1
      ItemToFind:"Music Party Editor Level Actor Entry Object Reference (by ref)"=@n3
     )))
   > OnSelectionChanged(self=@n2, Selected=true)
   OnSelectionChanged(execute=@n4.else, self=@n2, Selected=false)
  ]}
 OnEditEntryInitialized {nodes [
   n1 = OnEditEntryInitialized!() -> (then, Widget:Object Reference) {is=title K2Node_FunctionEntry}
   > n2 = Cast To EUW_MusicPartyEditor_MusicPropEditEntry!(execute, Object=@n1) -> (then, CastFailed, AsEUW Music Party Editor Music Prop Edit Entry:EUW Music Party Editor Music Prop Edit Entry Object Reference) {is=title K2Node_DynamicCast}
   > OnPostInitialize(self:EUW Music Party Editor Music Prop Edit Entry Object Reference=@n2, Parent=Self-Reference())
   > PIE_Update(self:EUW Music Party Editor Music Prop Edit Entry Object Reference=@n2, isPIE=Get isPIE())
  ]}
 OnEntryClicked {nodes [
   OnEntryClicked!() -> (then, Widget:EUW Music Party Editor Level Actor Entry Object Reference, DoubleClick:bool=~, ForceAdditiveSelect:bool=~) {is=title K2Node_FunctionEntry}
   n1 = GetModifierKeysState!() -> (ReturnValue:Modifier Keys State Structure)
   > Set isControlDown(isControlDown:bool=Is Control Down!(KeysState:"Modifier Keys State Structure (by ref)"=@n1) -> (ReturnValue:bool)) -> (Output_Get:bool)
   > Set isShiftDown(isShiftDown:bool=Is Shift Down!(KeysState:"Modifier Keys State Structure (by ref)"=@n1) -> (ReturnValue:bool)) -> (Output_Get:bool)
   > Set DisplayedEntryWidgets(DisplayedEntryWidgets:Array of Object References=GetDisplayedEntryWidgets!(self:List View Base Object Reference=Get LevelActorEntries()) -> (ReturnValue:Array of User Widget Object References)) -> (Output_Get:Array of Object References)
   > n2 = Sequence!(execute) -> (then_0, then_1, then_2:Exec)
   n3 = Branch(execute=@n2.then_1, Condition=Get isShiftDown() -> (isShiftDown:bool))
   > n4 = Get!(execute) -> (then, else, LastSelectedWidget:Object Reference)
   > n5 = Set StartInt(StartInt:int=Find Item(
     TargetArray:Array of Object References=Get DisplayedEntryWidgets() -> (DisplayedEntryWidgets:Array of Object References)
     ItemToFind:"Object Reference (by ref)"=@n4
    )) -> (Output_Get:int=0)
   n6 = Get!(execute=[@n11.then, @n5.then]) -> (then, else, Widget:EUW Music Party Editor Level Actor Entry Object Reference)
   > Set EndInt(EndInt:int=Find Item(TargetArray=Get DisplayedEntryWidgets(), ItemToFind=@n6)) -> (Output_Get:int=0)
   > n7 = For Each Loop(Array:Array of Object References=Get DisplayedEntryWidgets()) -> (Array Element:Object Reference)
   Set LastSelectedWidget(execute=@n7.Completed, LastSelectedWidget=Get Widget!() -> (Widget))
   n8 = Get StartInt() -> (StartInt:int=0)
   n9 = Get EndInt() -> (EndInt:int=0)
   Branch(
    execute=@n7.LoopBody
    Condition="In Range (Integer)"!(
     Value:int=@n7.Array Index
     Min:int="Min (Integer)"(A=@n8, B=@n9) -> (ReturnValue=0)
     Max:int="Max (Integer)"(A=@n8, B=@n9)
     InclusiveMin:bool=true
     InclusiveMax:bool=true
    ) -> (ReturnValue:bool)
   )
   > n10 = Cast To EUW_MusicPartyEditor_LevelActorEntry(Object=@n7)
   > OnSelectionChanged(self=@n10, Selected=true)
   > Add Unique(
    TargetArray:Array of Music Party Editor Level Actor Entry Object References=Get SelectedEntries()
    NewItem:"Music Party Editor Level Actor Entry Object Reference (by ref)"=Get EntryObjectCasted!(self:EUW Music Party Editor Level Actor Entry Object Reference=@n10) -> (EntryObjectCasted:Music Party Editor Level Actor Entry Object Reference)
   )
   n11 = Set StartInt(execute=@n4.else, StartInt=0) -> (Output_Get=0)
   n12 = Get(execute=@n3.else)
   > OnSelectionChanged(self=@n12, Selected=true)
   > Set LastSelectedWidget(LastSelectedWidget=Get Widget())
   > Add Unique(TargetArray=Get SelectedEntries(), NewItem=Get EntryObjectCasted(self=Get Widget()))
   Set LastSelectedWidget(execute=@n12.else)
   RefreshSelectedEntries(execute=@n2.then_2)
   > Branch(Condition=IsChecked(self=Get SelectionFollowsWidget() -> (SelectionFollowsWidget:Editor Utility Check Box Object Reference)))
   > n13 = For Each Loop(Array=Get SelectedEntries())
   n14 = Get EditorActorSubsystem()
   SetSelectedLevelActors(execute=@n13.Completed, self=@n14, ActorsToSelect=Get SelectedActors() -> (SelectedActors:Array of Actor Object References))
   > ExecuteConsoleCommand(WorldContextObject=@n14, Command=CAMERA ALIGN)
   n15 = Cast To MusicPartyEditor_LevelActorEntry(execute=@n13.LoopBody, Object=@n13)
   n16 = GetLevelActor(self=@n15)
   > Is Valid(InputObject=@n16)
   > Add Unique(TargetArray:Array of Actor Object References=Get SelectedActors(), NewItem:"Actor Object Reference (by ref)"=@n16)
   n17 = Branch(
    execute=@n2.then_0
    Condition=OR Boolean!(A=Get isControlDown() -> (isControlDown:bool), B=Get isShiftDown(), C:bool=Get ForceAdditiveSelect() -> (ForceAdditiveSelect:bool)) -> (ReturnValue)
   )
   Clear(execute=@n17.else, TargetArray:Array of Music Party Editor Level Actor Entry Object References=Get SelectedEntries())
   > n18 = For Each Loop(Array=Get DisplayedEntryWidgets())
   > n19 = Cast To EUW_MusicPartyEditor_LevelActorEntry(Object=@n18)
   > OnSelectionChanged(self=@n19, Selected=false)
  ]}
 OnMouseButtonDown {nodes [
   n1 = OnMouseButtonDown!() -> (then, MyGeometry:Geometry Structure, MouseEvent:"Pointer Event Structure (by ref)") {is=title K2Node_FunctionEntry}
   > n2 = Parent: OnMouseButtonDown!(execute, MyGeometry:Geometry Structure=@n1, MouseEvent:"Pointer Event Structure (by ref)"=@n1.MouseEvent) -> (then, ReturnValue:Event Reply Structure) {is=title K2Node_CallParentFunction}
   > Set MouseButtonIsDown(MouseButtonIsDown:bool=true) -> (Output_Get:bool)
   > Return Node!(execute, ReturnValue:Event Reply Structure=@n2)
  ]}
 OnMouseButtonUp {nodes [
   n1 = OnMouseButtonUp!() -> (then, MyGeometry:Geometry Structure, MouseEvent:"Pointer Event Structure (by ref)") {is=title K2Node_FunctionEntry}
   > n2 = Parent: OnMouseButtonUp!(execute, MyGeometry:Geometry Structure=@n1, MouseEvent:"Pointer Event Structure (by ref)"=@n1.MouseEvent) -> (then, ReturnValue:Event Reply Structure) {is=title K2Node_CallParentFunction}
   > Set MouseButtonIsDown(MouseButtonIsDown=false)
   > Return Node(ReturnValue=@n2)
  ]}
 OnPIEEntryClicked {nodes [
   OnPIEEntryClicked!() -> (then, Widget:EUW Music Party Editor Level Actor Entry Object Reference, DoubleClick:bool=~, ForceAdditiveSelect:bool=~) {is=title K2Node_FunctionEntry}
   n1 = GetModifierKeysState()
   > Set isControlDown(isControlDown=Is Control Down(KeysState=@n1))
   > Set isShiftDown(isShiftDown=Is Shift Down(KeysState=@n1))
   > Set DisplayedEntryWidgets(DisplayedEntryWidgets=GetDisplayedEntryWidgets(self=Get PIEActorEntries()))
   > n2 = Sequence()
   n3 = Branch(execute=@n2.then_1, Condition=Get isShiftDown())
   > n4 = Get!(execute) -> (then, else, LastSelectedWidget:Object Reference)
   > n5 = Set StartInt(StartInt=Find Item(TargetArray=Get DisplayedEntryWidgets(), ItemToFind=@n4)) -> (Output_Get=0)
   n6 = Get!(execute=[@n11.then, @n5.then]) -> (then, else, Widget:EUW Music Party Editor Level Actor Entry Object Reference)
   > Set EndInt(EndInt=Find Item(TargetArray=Get DisplayedEntryWidgets(), ItemToFind=@n6)) -> (Output_Get=0)
   > n7 = For Each Loop(Array=Get DisplayedEntryWidgets())
   Set LastSelectedPIEWidget(execute=@n7.Completed, LastSelectedPIEWidget=Get Widget())
   n8 = Get StartInt() -> (StartInt=0)
   n9 = Get EndInt() -> (EndInt=0)
   Branch(
    execute=@n7.LoopBody
    Condition="In Range (Integer)"(
     Value=@n7.Array Index
     Min="Min (Integer)"(A=@n8, B=@n9) -> (ReturnValue=0)
     Max="Max (Integer)"(A=@n8, B=@n9)
     InclusiveMin=true
     InclusiveMax=true
    )
   )
   > n10 = Cast To EUW_MusicPartyEditor_LevelActorEntry(Object=@n7)
   > OnSelectionChanged(self=@n10, Selected=true)
   > Add Unique(
    TargetArray:Array of Music Party Editor Level Actor Entry Object References=Get PIE_SelectedEntries()
    NewItem:"Music Party Editor Level Actor Entry Object Reference (by ref)"=Get EntryObjectCasted(self=@n10)
   )
   n11 = Set StartInt(execute=@n4.else, StartInt=0) -> (Output_Get=0)
   n12 = Get(execute=@n3.else)
   > OnSelectionChanged(self=@n12, Selected=true)
   > Set LastSelectedPIEWidget(LastSelectedPIEWidget=Get Widget())
   > Add Unique(TargetArray=Get PIE_SelectedEntries(), NewItem=Get EntryObjectCasted(self=Get Widget()))
   Set LastSelectedPIEWidget(execute=@n12.else)
   RefreshSelectedEntriesPIE(execute=@n2.then_2)
   n13 = Branch(execute=@n2.then_0, Condition=OR Boolean(A=Get isControlDown(), B=Get isShiftDown(), C=Get ForceAdditiveSelect()))
   Clear(execute=@n13.else, TargetArray=Get PIE_SelectedEntries())
   > n14 = For Each Loop(Array=Get DisplayedEntryWidgets())
   > n15 = Cast To EUW_MusicPartyEditor_LevelActorEntry(Object=@n14)
   > OnSelectionChanged(self=@n15, Selected=false)
  ]}
 OnPIESessionEnded {nodes [
   OnPIESessionEnded!() -> (then, bIsSimulating:bool=~) {is=title K2Node_FunctionEntry}
   > ClearListItems(self=Get PIEActorEntries())
   > Clear(TargetArray=Get PIE_SelectedEntries())
   > Clear(TargetArray:Array of Actor Object References=Get MonitoredPIEActors())
   n1 = Get MusicEditorPIEView()
   > ClearListItems(self=@n1)
   > n2 = For Each Loop(Array=GetDisplayedEntryWidgets(self=@n1))
   Set LastSelectedPIEWidget(execute=@n2.Completed)
   n3 = Cast To EUW_MusicPartyEditor_MusicPropEditEntry(execute=@n2.LoopBody, Object=@n2)
   > ClearReferences(self:EUW Music Party Editor Music Prop Edit Entry Object Reference=@n3)
  ]}
 PIE_Update {nodes [
   n1 = PIE_Update!() -> (then, isPIE:bool=~) {is=title K2Node_FunctionEntry}
   > n2 = Set isPIE(isPIE:bool=@n1) -> (Output_Get:bool)
   > n3 = Branch(Condition=@n2)
   n4 = Create Event()
   > Bind Event to On End PIE(self:Editor Utility Subsystem Object Reference=Get EditorUtilitySubsystem(), Delegate=@n4) {is=title K2Node_AddDelegate}
   Unbind Event from On End PIE(execute=@n3.else, self:Editor Utility Subsystem Object Reference=Get EditorUtilitySubsystem(), Delegate=@n4) {is=title K2Node_RemoveDelegate}
  ]}
 PIE_UpdatePropComponents {nodes [
   PIE_UpdatePropComponents!() -> (then, PropComponents:Array of Base Class: Music Party Prop Base Object References) {is=title K2Node_FunctionEntry}
   n1 = GetCurrentWorld() {is=title K2Node_CallFunction}
   > n2 = Branch(Condition=@n1.isPIE)
   n3 = For Each Loop(
    Exec=[@n2.then, @n5.then]
    Array:Array of Base Class: Music Party Prop Base Object References=Get PropComponents() -> (PropComponents:Array of Base Class: Music Party Prop Base Object References)
   ) -> (Array Element:Base Class: Music Party Prop Base Object Reference)
   > Is Valid(InputObject=@n3)
   > Set PropComponentRef(PropComponentRef:Base Class: Music Party Prop Base Object Reference=@n3) -> (Output_Get:Base Class: Music Party Prop Base Object Reference)
   n4 = Get PropComponentRef() -> (PropComponentRef:Base Class: Music Party Prop Base Object Reference)
   > CallSetupRuntimeVariables(self:Base Class: Music Party Prop Base Object Reference=@n4)
   > Setup Auto Activate(self:Base Class: Music Party Prop Base Object Reference=@n4)
   n5 = Set CurrentWorld(CurrentWorld:Object Reference=@n1) -> (Output_Get:Object Reference)
  ]}
 PerformHoverAction {nodes [
   n1 = PerformHoverAction!() -> (then, LevelActorEntry:Music Party Editor Level Actor Entry Object Reference) {is=title K2Node_FunctionEntry}
   > n2 = Branch(Condition=Get isTitleRow(self=@n1))
   n3 = Get TitleText!(self:Music Party Editor Level Actor Entry Object Reference=@n1) -> (TitleText:Text)
   > n4 = Branch(Condition="Equal, Case Insensitive (Text)"!(A:"Text (by ref)"=@n3, B:"Text (by ref)"=MakeLiteralText(Value="LOCTEXT(\"Unassigned Level Actors\")")) -> (ReturnValue:bool))
   > Remove Prop Component From Selected Entries(SilenceIfCantRemove=false)
   > Refresh Level Actor List(CheckAllLevelActors=false, DelayRefresh=false)
   n5 = Branch(execute=@n4.else, Condition="Equal, Case Insensitive (Text)"(A=@n3, B=MakeLiteralText(Value="LOCTEXT(\"Silenced Components\")")))
   > n6 = For Each Loop(Array:Array of Music Party Editor Level Actor Entry Object References=Get SelectedEntries()) -> (Array Element:Music Party Editor Level Actor Entry Object Reference)
   n7 = GetLevelActor(self=@n6)
   > Prop Edits Silence Component!(
    execute
    self
    Component:Object Reference=GetComponentByClass(self=@n7, ComponentClass=&MusicPartyPropComponentBase_C)
    LevelActor:Actor Object Reference=@n7
   ) -> (then, Success:bool)
   n8 = Get TitleText(self=Get LevelActorEntry() -> (LevelActorEntry:Music Party Editor Level Actor Entry Object Reference))
   Set NewGroupName(execute=@n5.else, NewGroupName:Name=String To Name(InString="To String (Text)"(InText=@n8))) -> (Output_Get:Name=None)
   > TransactionStart(Description="LOCTEXT(\"Updating Music Group Name\")")
   > n9 = For Each Loop(Array=Get SelectedEntries())
   TransactionEnd(execute=@n9.Completed)
   > Refresh Level Actor List(CheckAllLevelActors=false, DelayRefresh=false)
   Set LoopedSelectedEntry(execute=@n9.LoopBody, LoopedSelectedEntry:Music Party Editor Level Actor Entry Object Reference=@n9) -> (Output_Get:Music Party Editor Level Actor Entry Object Reference)
   > Branch(Condition=Get hasPropComponent(self=@n9))
   n10 = Get Music Prop!(self:Music Party Editor Level Actor Entry Object Reference=@n9) -> (Music Prop:Base Class: Music Party Prop Base Object Reference)
   > TransactionAdd!(execute, self, Object:Object Reference=@n10) -> (then)
   > SetEditorProperty(Object=@n10, PropertyName=MusicEventGroupName, PropertyValue:Name=Get NewGroupName() -> (NewGroupName:Name=None))
   > Branch(Condition=Get BypassMusic(self=@n10))
   > Set BypassMusic!(
    execute
    BypassMusic:bool=false
    self:Base Class: Music Party Prop Base Object Reference=Get Music Prop(self=Get LoopedSelectedEntry() -> (LoopedSelectedEntry:Music Party Editor Level Actor Entry Object Reference))
   ) -> (then, Output_Get:bool)
   n11 = Get Parents(self=Get LoopedSelectedEntry())
   > Branch(Condition="integer > integer"(A=Length(TargetArray:Array of Actor Object References=@n11)))
   > n12 = Set ParentsToRemove(ParentsToRemove:Array of Actor Object References=Get Parents(self=Get LoopedSelectedEntry())) -> (Output_Get:Array of Actor Object References)
   > n13 = For Each Loop(Array:Array of Actor Object References=@n12) -> (Array Element:Actor Object Reference)
   Clear(execute=@n13.Completed, TargetArray=Get Parents(self=Get LoopedSelectedEntry()))
   n14 = For Each Loop(
    Exec=@n13.LoopBody
    Array:Array of Actor Component Object References=Get Components By Class!(self:Actor Object Reference=@n13, ComponentClass:Actor Component Class Reference=/Script/Engine.ActorComponent) -> (ReturnValue:Array of Actor Component Object References)
   ) -> (Array Element:Actor Component Object Reference)
   > n15 = GetEditorProperty(Object=@n14, PropertyName=SpeakerActors) -> (PropertyValue:Array of Actor Soft Object References)
   > n16 = Set RemovableSpeakerActors(RemovableSpeakerActors:Array of Actor Soft Object References=@n15) -> (Output_Get:Array of Actor Soft Object References)
   n17 = Get LoopedSelectedEntry()
   > Remove Item(
    TargetArray:Array of Actor Soft Object References=@n16
    Item:"Actor Soft Object Reference (by ref)"=Make Soft Reference!(Input:Actor Object Reference=GetLevelActor(self=@n17)) -> (Output:Actor Soft Object Reference)
   )
   > SetEditorProperty(
    Object=@n14
    PropertyName=SpeakerActors
    PropertyValue:Array of Actor Soft Object References=Get RemovableSpeakerActors() -> (RemovableSpeakerActors:Array of Actor Soft Object References)
   )
   Branch(execute=@n2.else, Condition=Get hasPropComponent(self=Get LevelActorEntry()))
   > n18 = Remove Prop Component From Selected Entries(SilenceIfCantRemove=true)
   > TransactionStart(Description="LOCTEXT(\"Making Child Actors\")")
   > Set NewSpeakerActors(NewSpeakerActors:Array of Actor Soft Object References=Get SpeakerActors(self=Get LevelActorEntry())) -> (Output_Get:Array of Actor Soft Object References)
   > n19 = For Each Loop(Array:Array of Actor Soft Object References=@n18) -> (Array Element:Actor Soft Object Reference)
   n20 = For Each Loop(Exec=@n19.Completed, Array:Array of Music Party Editor Level Actor Entry Object References=Get SelectedEntries()) -> (Array Element:Music Party Editor Level Actor Entry Object Reference)
   > n21 = Branch(Condition=Get isTitleRow(self=@n20))
   n22 = GetLevelActor(self=@n20)
   n23 = GetComponentByClass(self=@n22, ComponentClass=&MusicPartyPropComponentBase_C)
   n24 = Is Valid(exec=@n21.else, InputObject=@n23)
   n25 = Set CurrentSoftReference(execute=[@n24.Is Not Valid, @n32.then], CurrentSoftReference:Actor Soft Object Reference=Make Soft Reference(Input=@n22)) -> (Output_Get:Actor Soft Object Reference)
   > Add Unique(
    TargetArray:Array of Actor Soft Object References=Get NewSpeakerActors() -> (NewSpeakerActors:Array of Actor Soft Object References)
    NewItem:"Actor Soft Object Reference (by ref)"=@n25
   )
   > n26 = For Each Loop(Array:Array of Actor Object References=Get Parents(self=@n20)) -> (Array Element:Actor Object Reference)
   n27 = GetComponentByClass(self=@n26, ComponentClass=&MusicPartyPropComponentBase_C)
   > Is Valid(InputObject=@n27)
   > TransactionAdd(Object=@n27)
   > n28 = GetEditorProperty(Object=@n27, PropertyName=SpeakerActors)
   > n29 = Set RemovableSpeakerActors(RemovableSpeakerActors=@n28)
   > Remove Item(TargetArray=@n29, Item=Get CurrentSoftReference() -> (CurrentSoftReference:Actor Soft Object Reference))
   > SetEditorProperty(Object=@n27, PropertyName=SpeakerActors, PropertyValue=Get RemovableSpeakerActors())
   n30 = Cast To MusicPartyPropComponent(execute=@n24.Is Valid, Object=@n23)
   n31 = Get LevelActorEntry()
   > Set MusicEventGroupName!(
    execute
    MusicEventGroupName:Name=Get GroupName!(self:Music Party Editor Level Actor Entry Object Reference=@n31) -> (GroupName:Name=None)
    self:" Music Party Prop Component Legacy Object Reference"=@n30
   ) -> (then, Output_Get:Name=None)
   > n32 = Set MusicEventTag!(
    execute
    MusicEventTag:Gameplay Tag Structure=Get MusicEventTag!(self:Music Party Editor Level Actor Entry Object Reference=@n31) -> (MusicEventTag:Gameplay Tag Structure)
    self:" Music Party Prop Component Legacy Object Reference"=@n30
   ) -> (then, Output_Get:Gameplay Tag Structure)
   TransactionAdd(execute=@n20.Completed, Object=Get Music Prop(self=Get LevelActorEntry()))
   > SetEditorProperty(Object=Get Music Prop(self=Get LevelActorEntry()), PropertyName=SpeakerActors, PropertyValue=Get NewSpeakerActors())
   > TransactionEnd()
   > Refresh Level Actor List(CheckAllLevelActors=false, DelayRefresh=false)
   Add Unique(execute=@n19.LoopBody, TargetArray=Get NewSpeakerActors(), NewItem=@n19)
  ]}
 Prop Edits Can Component be Deleted {nodes [
   n1 = Prop Edits Can Component be Deleted!() -> (then, Component:Object Reference, LevelActor:Actor Object Reference) {is=title K2Node_FunctionEntry}
   > n2 = Gather Subobject Data For Instance(self=Get SubobjectDataSubsystem(), Context=@n1.LevelActor)
   > n3 = Set SubObjectDataStructures(SubObjectDataStructures:Array of Subobject Data Handle Structures=@n2) -> (Output_Get:Array of Subobject Data Handle Structures)
   > n4 = For Each Loop(Array:Array of Subobject Data Handle Structures=@n3) -> (Array Element:Subobject Data Handle Structure)
   Return Node!(execute=@n4.Completed, CanBeReplaced:bool=false)
   n5 = GetData(DataHandle=@n4)
   > Branch(Condition="Equal (Object)"(
     A=Get Component() -> (Component:Object Reference)
     B=GetObject!(Data:"Subobject Data Structure (by ref)"=@n5, bEvenIfPendingKill:bool=false) -> (ReturnValue:Object Reference)
    ))
   > Return Node(CanBeReplaced=CanDelete!(Data:"Subobject Data Structure (by ref)"=@n5) -> (ReturnValue:bool))
  ]}
 Prop Edits Remove Component {nodes [
   n1 = Prop Edits Remove Component!() -> (then, Component:Object Reference, LevelActor:Actor Object Reference) {is=title K2Node_FunctionEntry}
   > n2 = Gather Subobject Data For Instance(self=Get SubobjectDataSubsystem(), Context=@n1.LevelActor)
   > n3 = Set SubObjectDataStructures(SubObjectDataStructures=@n2)
   > n4 = For Each Loop(Array=@n3)
   Return Node!(execute=@n4.Completed, Success:bool=false)
   n5 = GetData(DataHandle=@n4)
   > Branch(Condition="Equal (Object)"(A=Get Component(), B=GetObject(Data=@n5, bEvenIfPendingKill=false)))
   > Branch(Condition=CanDelete(Data=@n5))
   > Delete Subobject from Instance!(
    execute
    self:Subobject Data Subsystem Object Reference=Get SubobjectDataSubsystem()
    ContextHandle:"Subobject Data Handle Structure (by ref)"="Get (a copy)"(Array:Array of Subobject Data Handle Structures=@n3) -> (Output:Subobject Data Handle Structure)
    SubobjectToDelete:"Subobject Data Handle Structure (by ref)"=@n4
   ) -> (then, ReturnValue:int=0)
   > Return Node(Success=true)
  ]}
 Prop Edits Silence Component {nodes [
   n1 = Prop Edits Silence Component!() -> (then, Component:Object Reference, LevelActor:Actor Object Reference) {is=title K2Node_FunctionEntry}
   > n2 = Cast To MusicPartyPropComponentBase!(execute, Object=@n1) -> (then, CastFailed, "AsBase Class: MusicPartyProp_Base":Base Class: Music Party Prop Base Object Reference) {is=title K2Node_DynamicCast}
   > Set BypassMusic(BypassMusic=true, self=@n2)
   > n3 = Cast To MusicPartyPropComponent(Object=@n1)
   > Set SoundToPlay!(execute, SoundToPlay:Sound Base Object Reference, self:" Music Party Prop Component Legacy Object Reference"=@n3) -> (then, Output_Get:Sound Base Object Reference)
   > Set SpeakerActors!(
    execute
    SpeakerActors:Array of Building Actor Soft Object References=Get EmptyArray() -> (EmptyArray:Array of Building Actor Soft Object References)
    self:" Music Party Prop Component Legacy Object Reference"=@n3
   ) -> (then, Output_Get:Array of Building Actor Soft Object References)
   > Return Node(Success=true)
   n4 = Cast To MusicPartyPropComponent_SpeakerAssets!(execute=@n3.CastFailed, Object=@n2) -> (then, CastFailed, AsMusic Party Prop Component Speaker Assets:Music Party Prop Component Speaker Assets Object Reference) {is=title K2Node_DynamicCast}
   > Set SpeakerActors(SpeakerActors=Get EmptyArray(), self:Music Party Prop Component Speaker Assets Object Reference=@n4)
   Return Node(Success=false)
  ]}
 PropEdits_AddComponentByClass {nodes [
   n1 = PropEdits_AddComponentByClass!() -> (then, Actor:Actor Object Reference, Class:Actor Component Class Reference) {is=title K2Node_FunctionEntry}
   > Is Valid(InputObject=@n1)
   > TransactObject(Object=@n1)
   > n2 = Gather Subobject Data For Instance(self=Get SubobjectDataSubsystem(), Context=@n1)
   > Branch(Condition="integer > integer"(A=Length(TargetArray:Array of Subobject Data Handle Structures=@n2)))
   > AddNewSubobject!(
    execute
    self:Subobject Data Subsystem Object Reference=Get SubobjectDataSubsystem()
    Params_ParentHandle:Subobject Data Handle Structure="Get (a ref)"(Array:Array of Subobject Data Handle Structures=@n2, Dimension 1=0) -> (Output:"Subobject Data Handle Structure (by ref)")
    Params_NewClass:Object Class Reference=Get Class() -> (Class:Actor Component Class Reference) {is=title K2Node_VariableGet}
    Params_BlueprintContext:Blueprint Object Reference
    Params_bSkipMarkBlueprintModified:bool=False
    Params_bConformTransformToParent:bool=True
   ) -> (then, FailReason:Text, ReturnValue:Subobject Data Handle Structure)
  ]}
 PropEdits_BeginAction {nodes [
   n1 = PropEdits_BeginAction!() -> (then, ActionName:Name) {is=title K2Node_FunctionEntry}
   > CancelAction()
   > Set Text(Text="To Text (Name)"(InName=@n1), self=Get ActionText() -> (ActionText:Text Block Object Reference))
   n2 = Get PropTypeComboBox()
   > PropEdits_OnPropTypeSelectionChange(Name=String To Name(InString=GetSelectedOption(self=@n2)))
   > SetActiveWidgetIndex(self=Get WidgetSwitcher(), Index=1)
  ]}
 PropEdits_OnPropTypeSelectionChange {nodes [
   n1 = PropEdits_OnPropTypeSelectionChange!() -> (then, Name:Name) {is=title K2Node_FunctionEntry}
   > n2 = Switch on Name(Selection=@n1) {pinNames [Controller Component, Radio Component, Actor Component, Deprecated Component]}
   n3 = Set NewActionPropComponent(
    execute=@n2.Actor Component
    NewActionPropComponent=&MusicPartyPropComponent_Actor_C
   )
   n4 = Get!(execute=[@n16.then, @n15.then, @n17.then, @n3.then]) -> (then, else, ActionActor:Actor Object Reference)
   n5 = GetComponentByClass(self=@n4, ComponentClass=&MusicPartyPropComponentBase_C)
   > n6 = Is Valid(InputObject=@n5)
   n7 = Get ActionActor()
   n8 = Get NewActionPropComponent()
   PropEdits_AddComponentByClass(execute=[@n6.Is Not Valid, @n12.then, @n14.then], Actor=@n7, Class=@n8) {is=title K2Node_CallFunction}
   > n9 = Set NewMusicPartyPropComponent(NewMusicPartyPropComponent:Base Class: Music Party Prop Base Object Reference=GetComponentByClass(self=@n7, ComponentClass=@n8)) -> (Output_Get:Base Class: Music Party Prop Base Object Reference)
   > SetObject(self=Get MusicPropActionSettings() -> (MusicPropActionSettings:Details View Object Reference), NewObject=@n9)
   > n10 = Get!(execute) -> (then, else, OldMusicPartyPropComponent:Base Class: Music Party Prop Base Object Reference)
   > GetAndTransferPropComponentVariables(
    MusicPartyPropComponent=@n10
    CopyToNewComponent=true
    NewProp=Get NewMusicPartyPropComponent() -> (NewMusicPartyPropComponent:Base Class: Music Party Prop Base Object Reference)
    CopyMusicController=true
   ) -> (GroupName=None) {is=title K2Node_CallFunction}
   n11 = Get OldMusicPartyPropComponent() -> (OldMusicPartyPropComponent:Base Class: Music Party Prop Base Object Reference)
   > Destroy Component!(execute, self:Actor Component Object Reference=@n11, Object:Object Reference=@n11) -> (then)
   n12 = Set OldMusicPartyPropComponent(execute=@n6.Is Valid, OldMusicPartyPropComponent:Base Class: Music Party Prop Base Object Reference=@n5) -> (Output_Get:Base Class: Music Party Prop Base Object Reference)
   n13 = SpawnActorFromClass!(execute=@n4.else, ActorClass:Actor Class Reference=&Actor, Location:Vector="0, 0, 0", Rotation:Rotator, bTransient:bool=true) -> (then, ReturnValue:Actor Object Reference)
   > n14 = Set ActionActor(ActionActor:Actor Object Reference=@n13) -> (Output_Get:Actor Object Reference)
   n15 = Set NewActionPropComponent(
    execute=@n2.Controller Component
    NewActionPropComponent=&MusicPartyPropComponent_Controller_C
   )
   n16 = Set NewActionPropComponent(
    execute=@n2.Deprecated Component
    NewActionPropComponent=&MusicPartyPropComponent_C
   )
   n17 = Set NewActionPropComponent(
    execute=@n2.Radio Component
    NewActionPropComponent=&MusicPartyPropComponent_Radio_C
   )
  ]}
 Refresh Level Actor List {nodes [
   n1 = Refresh Level Actor List!() -> (then, CheckAllLevelActors:bool=~, DelayRefresh:bool=~) {is=title K2Node_FunctionEntry}
   > n2 = Branch(Condition=@n1)
   > n3 = CheckAllLevelActors() {is=title K2Node_CallFunction}
   n4 = Branch(execute=[@n3.then, @n2.else], Condition=Get isPIE())
   > n5 = GenerateActorList!(
    execute
    self
    SelectedEntriesToEvaluate:Array of Music Party Editor Level Actor Entry Object References=Get PIE_SelectedEntries()
    ListView:List View Object Reference=Get PIEActorEntries()
    ActorsToEvaluate:Array of Actor Object References=Get MonitoredPIEActors()
   ) -> (then, NewSelectedEntries:Array of Music Party Editor Level Actor Entry Object References) {is=title K2Node_CallFunction}
   > n6 = Set PIE_SelectedEntries(PIE_SelectedEntries:Array of Music Party Editor Level Actor Entry Object References=@n5) -> (Output_Get:Array of Music Party Editor Level Actor Entry Object References)
   n7 = Branch(execute=[@n10.then, @n6.then], Condition=Get DelayRefresh() -> (DelayRefresh:bool))
   > Delay Refresh Selection() {is=title K2Node_CallFunction}
   n8 = Branch(execute=@n7.else, Condition=Get isPIE())
   > RefreshSelectedEntriesPIE()
   RefreshSelectedEntries(execute=@n8.else)
   n9 = GenerateActorList(
    execute=@n4.else
    SelectedEntriesToEvaluate=Get SelectedEntries()
    ListView=Get LevelActorEntries()
    ActorsToEvaluate=Get MonitoredLevelActors()
   )
   > n10 = Set SelectedEntries(SelectedEntries:Array of Music Party Editor Level Actor Entry Object References=@n9) -> (Output_Get:Array of Music Party Editor Level Actor Entry Object References)
  ]}
 RefreshSelectedEntries {nodes [
   RefreshSelectedEntries() {is=title K2Node_FunctionEntry}
   > ClearListItems(self=Get MusicEditorListView())
   > n1 = For Each Loop(Array:Array of Music Party Editor Level Actor Entry Object References=Get SelectedEntries()) -> (Array Element:Music Party Editor Level Actor Entry Object Reference)
   n2 = Values(
    execute=@n1.Completed
    TargetMap:Map of EMusicPartyPropComponentType Enums to Object References=Get TypeToObjects() -> (TypeToObjects:Map of EMusicPartyPropComponentType Enums to Object References=NewEnumerator0)
   ) -> (Values:Array of Object References)
   > Set List Items(self=Get MusicEditorListView(), InListItems=@n2)
   n3 = Get SelectedEntries()
   n4 = Select(
    Option 0:ESlateVisibility Enum=Collapsed
    Option 1:ESlateVisibility Enum=Get AddComponentVisibility() -> (AddComponentVisibility:ESlateVisibility Enum=Visible)
    Index="integer > integer"(A=Length(TargetArray:Array of Music Party Editor Level Actor Entry Object References=@n3))
   ) -> (ReturnValue:ESlateVisibility Enum)
   > SetVisibility(self=Get AddMusicPropComponent() -> (AddMusicPropComponent:Button Object Reference), InVisibility=@n4)
   > SetVisibility(
    self=Get ConvertPropComponent_Container() -> (ConvertPropComponent_Container:Horizontal Box Object Reference)
    InVisibility=Select!(
     Visible:ESlateVisibility Enum=Collapsed
     Collapsed:ESlateVisibility Enum=Visible
     Hidden:ESlateVisibility Enum=Visible
     HitTestInvisible:ESlateVisibility Enum=Visible
     SelfHitTestInvisible:ESlateVisibility Enum=Visible
     Index:ESlateVisibility Enum=@n4
    ) -> (ReturnValue)
   )
   > Branch(Condition=Get RequireRefresh() -> (RequireRefresh:bool))
   > Refresh Level Actor List(CheckAllLevelActors=true, DelayRefresh=false) {is=title K2Node_CallFunction}
   n5 = Is Valid(exec=@n1.LoopBody, InputObject=@n1)
   > n6 = Set CurrentLevelActorEntry(CurrentLevelActorEntry=@n1)
   n7 = GetLevelActor(self=@n6)
   > n8 = Is Valid(InputObject=@n7)
   n9 = Branch(execute=@n8.Is Not Valid, Condition=Get isTitleRow(self=Get CurrentLevelActorEntry()))
   n10 = Set RequireRefresh(execute=@n9.else, RequireRefresh:bool=true) -> (Output_Get:bool)
   Remove Item(
    execute=[@n10.then, @n5.Is Not Valid]
    TargetArray:Array of Music Party Editor Level Actor Entry Object References=Get SelectedEntries()
    Item:"Music Party Editor Level Actor Entry Object Reference (by ref)"=@n1
   )
   n11 = Get Parents(self=@n6)
   n12 = Branch(execute=@n8.Is Valid, Condition=Is Empty(TargetArray=@n11))
   > RemoveChildSpeakerActorComponent!(execute, self, SpeakerActor:Actor Object Reference=@n7) -> (then)
   n13 = Get CurrentLevelActorEntry()
   > Set CurrentMusicProp(CurrentMusicProp:Base Class: Music Party Prop Base Object Reference=Get Music Prop(self=@n13)) -> (Output_Get:Base Class: Music Party Prop Base Object Reference)
   n14 = Get hasPropComponent(self=@n13)
   > n15 = Branch(Condition=OR Boolean!(A=Get isTitleRow(self=@n13), B=NOT Boolean(A=@n14)) -> (ReturnValue))
   Set AddComponentVisibility(execute=@n15.else, AddComponentVisibility:ESlateVisibility Enum=Collapsed) -> (Output_Get:ESlateVisibility Enum=Visible)
   n16 = Get CurrentLevelActorEntry()
   n17 = Get MusicPropType!(self:Music Party Editor Level Actor Entry Object Reference=@n16) -> (MusicPropType:EMusicPartyPropComponentType Enum=NewEnumerator0)
   n18 = Find(
    TargetMap:Map of EMusicPartyPropComponentType Enums to Object References=Get TypeToObjects() -> (TypeToObjects=NewEnumerator0)
    Key:"EMusicPartyPropComponentType Enum (by ref)"=@n17
   ) -> (Value:Object Reference)
   > n19 = Branch(Condition=@n18.ReturnValue)
   > n20 = Cast To MusicPartyEditor_MusicProp_MultiEdit!(execute, Object=@n18) -> (then, CastFailed, AsMusic Party Editor Music Prop Multi Edit:Music Party Editor Music Prop Multi Edit Object Reference) {is=title K2Node_DynamicCast}
   n21 = Get CurrentLevelActorEntry()
   > AddNewPropComponent!(
    execute
    self:Music Party Editor Music Prop Multi Edit Object Reference=@n20
    PropType:EMusicPartyPropComponentType Enum=Get MusicPropType(self=@n21) -> (MusicPropType=NewEnumerator0)
    PropComponent:Base Class: Music Party Prop Base Object Reference=Get CurrentMusicProp() -> (CurrentMusicProp:Base Class: Music Party Prop Base Object Reference)
    LevelActor:Actor Object Reference=GetLevelActor(self=@n21)
   ) -> (then)
   n22 = Construct Object from Class(
    execute=@n20.CastFailed
    Class=&/MusicPartyEditor_MusicProp_MultiEdit.MusicPartyEditor_MusicProp_MultiEdit_C
   ) -> (ReturnValue:Music Party Editor Music Prop Multi Edit Object Reference)
   n23 = Get CurrentLevelActorEntry()
   > n24 = Cast To Actor!(
    execute
    Object=Find(TargetMap=Get TypeToObjects() -> (TypeToObjects=NewEnumerator0), Key=Get MusicPropType(self=@n23) -> (MusicPropType=NewEnumerator0))
   ) -> (then, CastFailed, AsActor:Actor Object Reference) {is=title K2Node_DynamicCast}
   n25 = Get CurrentLevelActorEntry()
   n26 = Get MusicPropType(self=@n25) -> (MusicPropType=NewEnumerator0)
   > AddNewPropComponent(
    self=@n22
    PropType=@n26
    PropComponent=GetComponentByClass(self=@n24, ComponentClass=&MusicPartyPropComponentBase_C)
    LevelActor=@n24
   )
   > AddNewPropComponent(self=@n22, PropType=@n26, PropComponent=Get CurrentMusicProp(), LevelActor=GetLevelActor(self=@n25))
   > Add(
    TargetMap:Map of EMusicPartyPropComponentType Enums to Object References=Get TypeToObjects() -> (TypeToObjects=NewEnumerator0)
    Key:"EMusicPartyPropComponentType Enum (by ref)"=@n26
    Value:"Object Reference (by ref)"=@n22
   )
   Add(execute=@n19.else, TargetMap=Get TypeToObjects() -> (TypeToObjects=NewEnumerator0), Key=@n17, Value=GetLevelActor(self=@n16))
   SetupChildSpeakerActorTracking!(execute=@n12.else, self, NewSpeakerActors:"Actor Object Reference (by ref)"=@n7, Parents:Array of Actor Object References=@n11) -> (then)
  ]}
 RefreshSelectedEntriesPIE {nodes [
   RefreshSelectedEntriesPIE() {is=title K2Node_FunctionEntry}
   > ClearListItems(self=Get MusicEditorPIEView())
   > n1 = For Each Loop(Array=Get PIE_SelectedEntries())
   n2 = Values(execute=@n1.Completed, TargetMap=Get TypeToObjects() -> (TypeToObjects=NewEnumerator0))
   > Set List Items(self=Get MusicEditorPIEView(), InListItems=@n2)
   n3 = Is Valid(exec=@n1.LoopBody, InputObject=@n1)
   Remove Item(execute=@n3.Is Not Valid, TargetArray=Get PIE_SelectedEntries(), Item=@n1)
   n4 = Set CurrentLevelActorEntry(execute=@n3.Is Valid, CurrentLevelActorEntry=@n1)
   > Set CurrentMusicProp(CurrentMusicProp=Get Music Prop(self=@n4))
   n5 = Get hasPropComponent(self=@n4)
   > n6 = Branch(Condition=OR Boolean(A=Get isTitleRow(self=@n4), B=NOT Boolean(A=@n5)))
   Set AddComponentVisibility(execute=@n6.else, AddComponentVisibility=Collapsed) -> (Output_Get=Visible)
   n7 = Find(
    TargetMap=Get TypeToObjects() -> (TypeToObjects=NewEnumerator0)
    Key=Get MusicPropType(self=Get CurrentLevelActorEntry()) -> (MusicPropType=NewEnumerator0)
   )
   > n8 = Branch(Condition=@n7.ReturnValue)
   > n9 = Cast To MusicPartyEditor_MusicProp_MultiEdit(Object=@n7)
   n10 = Get CurrentLevelActorEntry()
   > AddNewPropComponent(
    self=@n9
    PropType=Get MusicPropType(self=@n10) -> (MusicPropType=NewEnumerator0)
    PropComponent=Get CurrentMusicProp()
    LevelActor=GetLevelActor(self=@n10)
   )
   n11 = Construct Object from Class(
    execute=@n9.CastFailed
    Class=&MusicPartyEditor_MusicProp_MultiEdit_C
   )
   n12 = Get MusicPropType(self=Get CurrentLevelActorEntry()) -> (MusicPropType=NewEnumerator0)
   > n13 = Cast To Actor(Object=Find(TargetMap=Get TypeToObjects() -> (TypeToObjects=NewEnumerator0), Key=@n12))
   > AddNewPropComponent(
    self=@n11
    PropType=@n12
    PropComponent=GetComponentByClass(self=@n13, ComponentClass=&MusicPartyPropComponentBase_C)
    LevelActor=@n13
   )
   n14 = Get CurrentLevelActorEntry()
   n15 = Get MusicPropType(self=@n14) -> (MusicPropType=NewEnumerator0)
   > AddNewPropComponent(self=@n11, PropType=@n15, PropComponent=Get CurrentMusicProp(), LevelActor=GetLevelActor(self=@n14))
   > Add(TargetMap=Get TypeToObjects() -> (TypeToObjects=NewEnumerator0), Key=@n15, Value=@n11)
   n16 = Get CurrentLevelActorEntry()
   Add(
    execute=@n8.else
    TargetMap=Get TypeToObjects() -> (TypeToObjects=NewEnumerator0)
    Key=Get MusicPropType(self=@n16) -> (MusicPropType=NewEnumerator0)
    Value=GetLevelActor(self=@n16)
   )
  ]}
 Remove Prop Component From Selected Entries {nodes [
   Remove Prop Component From Selected Entries!() -> (then, SilenceIfCantRemove:bool=~) {is=title K2Node_FunctionEntry}
   > TransactionStart(Description="LOCTEXT(\"RemovePropComponents\")")
   > n1 = For Each Loop(Array=Get SelectedEntries())
   TransactionEnd(execute=@n1.Completed)
   > Return Node!(
    execute
    ChildSpeakerActors:Array of Actor Soft Object References=Get ChildSpeakerActorArray() -> (ChildSpeakerActorArray:Array of Actor Soft Object References)
   )
   n2 = GetLevelActor(self=@n1)
   n3 = Set Current_LevelActor(execute=@n1.LoopBody, Current_LevelActor:Actor Object Reference=@n2) -> (Output_Get:Actor Object Reference)
   > Is Valid(InputObject=@n3)
   > Set CurrentAudioComponent(CurrentAudioComponent:Audio Component Object Reference=GetComponentByClass(self=@n2, ComponentClass=&AudioComponent) -> (ReturnValue:Audio Component Object Reference)) -> (Output_Get:Audio Component Object Reference)
   > n4 = Set MusicPropComponent(MusicPropComponent:Base Class: Music Party Prop Base Object Reference=GetComponentByClass(self=@n2, ComponentClass=&MusicPartyPropComponentBase_C) -> (ReturnValue:Base Class: Music Party Prop Base Object Reference)) -> (Output_Get:Base Class: Music Party Prop Base Object Reference)
   > Is Valid(InputObject=@n4)
   > n5 = Prop Edits Can Component be Deleted(
    Component=Get MusicPropComponent() -> (MusicPropComponent:Base Class: Music Party Prop Base Object Reference)
    LevelActor=Get Current_LevelActor() -> (Current_LevelActor:Actor Object Reference)
   ) {is=title K2Node_CallFunction}
   > n6 = Branch(Condition=@n5)
   > n7 = GetEditorProperty(Object=Get MusicPropComponent(), PropertyName=SpeakerActors)
   > n8 = For Each Loop(Array:Array of Actor Soft Object References=@n7) -> (Array Element:Actor Soft Object Reference)
   TransactionAdd(execute=@n8.Completed, Object=Get Current_LevelActor())
   > Prop Edits Remove Component(Component=Get MusicPropComponent(), LevelActor=Get Current_LevelActor()) {is=title K2Node_CallFunction}
   > n9 = Get!(execute) -> (then, else, CurrentAudioComponent:Audio Component Object Reference)
   > n10 = Prop Edits Can Component be Deleted(Component=@n9, LevelActor=Get Current_LevelActor())
   > Branch(Condition=@n10)
   > TransactionAdd(Object=Get Current_LevelActor())
   > Prop Edits Remove Component(Component=Get CurrentAudioComponent() -> (CurrentAudioComponent:Audio Component Object Reference), LevelActor=Get Current_LevelActor())
   Add Unique(execute=@n8.LoopBody, TargetArray=Get ChildSpeakerActorArray(), NewItem=@n8)
   Branch(execute=@n6.else, Condition=Get SilenceIfCantRemove() -> (SilenceIfCantRemove:bool))
   > n11 = GetEditorProperty(Object=Get MusicPropComponent(), PropertyName=SpeakerActors)
   > n12 = For Each Loop(Array=@n11)
   TransactionAdd(execute=@n12.Completed, Object=Get Current_LevelActor())
   > Prop Edits Silence Component(Component=Get MusicPropComponent(), LevelActor=Get Current_LevelActor()) {is=title K2Node_CallFunction}
   Add Unique(execute=@n12.LoopBody, TargetArray=Get ChildSpeakerActorArray(), NewItem=@n12)
  ]}
 RemoveChildSpeakerActorComponent {nodes [
   n1 = RemoveChildSpeakerActorComponent!() -> (then, SpeakerActor:Actor Object Reference) {is=title K2Node_FunctionEntry}
   n2 = GetComponentByClass(self=@n1, ComponentClass=/Game/Sounds/Fort_Audio_Blueprints/MusicParty/MusicParty_ChildSpeakerComponent.MusicParty_ChildSpeakerComponent_C) -> (ReturnValue:Music Party Child Speaker Component Object Reference)
   > Is Valid(InputObject=@n2)
   > n3 = Prop Edits Remove Component(Component=@n2, LevelActor=@n1)
   > Branch(Condition=@n3)
   > PrintLog!(execute, Log:Text="LOCTEXT(\"Component Removed\")", __WorldContext:Object Reference) -> (then)
  ]}
 SetupChildSpeakerActorTracking {nodes [
   n1 = SetupChildSpeakerActorTracking!() -> (then, NewSpeakerActors:"Actor Object Reference (by ref)", Parents:Array of Actor Object References) {is=title K2Node_FunctionEntry}
   > Set NewChildSpeakerActor(NewChildSpeakerActor:Actor Object Reference=@n1) -> (Output_Get:Actor Object Reference)
   n2 = GetComponentByClass(
    self="Get (a copy)"(Array:Array of Actor Object References=@n1.Parents) -> (Output:Actor Object Reference)
    ComponentClass=&MusicPartyPropComponentBase_C
   ) -> (ReturnValue:Base Class: Music Party Prop Base Object Reference)
   > Is Valid(InputObject=@n2)
   > Set MusicPartyPropComponent(MusicPartyPropComponent:Base Class: Music Party Prop Base Object Reference=@n2) -> (Output_Get:Base Class: Music Party Prop Base Object Reference)
   > n3 = Is Valid(InputObject=GetComponentByClass(
     self=Get NewChildSpeakerActor() -> (NewChildSpeakerActor:Actor Object Reference)
     ComponentClass=&MusicParty_ChildSpeakerComponent_C
    ) -> (ReturnValue:Music Party Child Speaker Component Object Reference))
   n4 = Get MusicPartyChildSpeakerComponent() -> (MusicPartyChildSpeakerComponent:Actor Component Class Reference)
   PropEdits_AddComponentByClass(execute=@n3.Is Not Valid, Actor=Get NewChildSpeakerActor(), Class=@n4)
   > n5 = Cast To MusicParty_ChildSpeakerComponent!(execute, Object=GetComponentByClass(self=Get NewChildSpeakerActor(), ComponentClass=@n4) -> (ReturnValue:Actor Component Object Reference)) -> (then, CastFailed, AsMusic Party Child Speaker Component:Music Party Child Speaker Component Object Reference) {is=title K2Node_DynamicCast}
   > Add(
    TargetArray:Array of Music Party Child Speaker Component Object References=Get ChildSpeakerComponents!(self:Base Class: Music Party Prop Base Object Reference=Get MusicPartyPropComponent()) -> (ChildSpeakerComponents:Array of Music Party Child Speaker Component Object References)
    NewItem:"Music Party Child Speaker Component Object Reference (by ref)"=@n5
   ) {is=title K2Node_CallArrayFunction}
  ]}
 ShouldActorBeMonitored {nodes [
   n1 = Inputs!() -> (exec:Exec, Actor:Actor Object Reference)
   > n2 = Is Valid(InputObject=GetComponentByClass(self=@n1, ComponentClass=&MusicPartyPropComponentBase_C) -> (ReturnValue:Base Class: Music Party Prop Base Object Reference))
   n3 = Branch(
    execute=@n2.Is Not Valid
    Condition=ActorMatchesSearchTerms!(Actor:Actor Object Reference=@n1) -> (MatchFound:bool) {is=title K2Node_CallFunction}
   )
   Outputs!(True:Exec=[@n3.then, @n2.Is Valid], False:Exec=@n3.else)
  ]}
 TransactionAdd {nodes [
   n1 = TransactionAdd!() -> (then, Object:Object Reference) {is=title K2Node_FunctionEntry}
   > TransactObject(Object=@n1)
  ]}
 TransactionEnd {nodes [
   TransactionEnd() {is=title K2Node_FunctionEntry}
   > EndTransaction()
  ]}
 TransactionStart {nodes [
   n1 = TransactionStart!() -> (then, Description:Text) {is=title K2Node_FunctionEntry}
   n2 = Self-Reference()
   > BeginTransaction(Context=MusicPartyEditor, Description=@n1, PrimaryObject=@n2)
   > TransactObject(Object=@n2)
  ]}
]
variables [
 ActionActor {type=Actor, default=null}
 ActorSearchTerms {type="array<string>", default [Speaker,DJ,Monitor,Computer,Screen,Vending]}
 ConvertProp_Category {type=EMusicPartyPropComponentType, default=Controller Component}
 EditorWorldObjectRef {type=Object, default=null}
 ExecuteAction {default=()}
 HoveredEntry {type=MusicPartyEditor_LevelActorEntry_C, default=null}
 isPIE {type=boolean}
 isShowingAttenuation {type=boolean}
 LastSelectedPIEWidget {type=Object, default=null}
 LastSelectedWidget {type=Object, default=null}
 LevelActors {type="array<Actor>"}
 MonitoredLevelActors {type="array<Actor>"}
 MonitoredPIEActors {type="array<Actor>"}
 MouseButtonIsDown {type=boolean}
 NewAsset_Object {type=MusicPartyEditor_NewAsset_Base_C, default=null}
 PIE_SelectedEntries {type="array<MusicPartyEditor_LevelActorEntry_C>", default []}
 SelectedEntries {type="array<MusicPartyEditor_LevelActorEntry_C>", default []}
]
widgets {
 info {rootWidgetClass=&CanvasPanel, widgetCount=76}
 widgets [
  CanvasPanel_0 {widgetClassPath=&CanvasPanel, widget=CanvasPanel_0}
  SizeBox_0 {
   widgetClassPath=&SizeBox
   parent=CanvasPanel_0
   widget=SizeBox_0
   slot {class=CanvasPanelSlot, properties {layoutData {alignment=<0 0>, anchors {maximum=<1 1>, minimum=<0 0>}, offsets=<ltrb 0 0 0 0>}}}
  }
  Border_154 {
   widgetClassPath=&Border
   parent=SizeBox_0
   widget=Border_154
   properties {
    background {
     bIsDynamicallyLoaded=false
     drawAs=Box
     imageType=NoImage
     mirroring=NoMirror
     resourceName=None
     resourceObject=null
     tiling=NoTile
     imageSize=<32 32>
     margin=<ltrb 0 0 0 0>
     outlineSettings {
      bUseBrushTransparency=false
      roundingType=HalfHeightRadius
      width=0
      color {colorUseRule=UseColor_Specified, specifiedColor=<rgba 0 0 0 0>}
      cornerRadii=<0 0 0 0>
     }
     tintColor {colorUseRule=UseColor_Specified, specifiedColor=<rgba 1 1 1 1>}
     uVRegion {bIsValid=false, max=<0 0>, min=<0 0>}
    }
    brushColor=<rgba 0 0 0 0.2>
    padding=<ltrb 0 0 0 0>
   }
   slot {class=SizeBoxSlot, properties {padding=<ltrb 2.5 2.5 2.5 2.5>}}
  }
  HorizontalBox_233 {
   widgetClassPath=&HorizontalBox
   parent=Border_154
   widget=HorizontalBox_233
   slot {class=BorderSlot, properties {padding=<ltrb 0 0 0 0>}}
  }
  VerticalBox_1 {
   widgetClassPath=&VerticalBox
   parent=HorizontalBox_233
   widget=VerticalBox_1
   slot {class=HorizontalBoxSlot, properties {padding=<ltrb 5 5 5 5>, size {sizeRule=Fill, value=0.75}}}
  }
  Border {
   widgetClassPath=&Border
   parent=VerticalBox_1
   widget=Border
   properties {
    background {
     bIsDynamicallyLoaded=false
     drawAs=Box
     imageType=NoImage
     mirroring=NoMirror
     resourceName=None
     resourceObject=null
     tiling=NoTile
     imageSize=<32 32>
     margin=<ltrb 0 0 0 0>
     outlineSettings {
      bUseBrushTransparency=false
      roundingType=HalfHeightRadius
      width=0
      color {colorUseRule=UseColor_Specified, specifiedColor=<rgba 0 0 0 0>}
      cornerRadii=<0 0 0 0>
     }
     tintColor {colorUseRule=UseColor_Specified, specifiedColor=<rgba 1 1 1 1>}
     uVRegion {bIsValid=false, max=<0 0>, min=<0 0>}
    }
    brushColor=<rgba 0 0 0 0.2>
    padding=<ltrb 0 0 0 0>
   }
   slot {class=VerticalBoxSlot, properties {verticalAlignment=VAlign_Center, size {sizeRule=Automatic, value=0.05}}}
  }
  VerticalBox_210 {widgetClassPath=&VerticalBox, parent=Border, widget=VerticalBox_210, slot {class=BorderSlot, properties {padding=<ltrb 0 0 0 0>}}}
  HorizontalBox_283 {widgetClassPath=&HorizontalBox, parent=VerticalBox_210, widget=HorizontalBox_283, slot {class=VerticalBoxSlot}}
  SelectionFollows {
   widgetClassPath=&HorizontalBox
   parent=HorizontalBox_283
   widget=SelectionFollows
   properties {toolTipText=Should view snap to level actors?}
   slot {class=HorizontalBoxSlot}
  }
  SelectionFollowsLabel {
   widgetClassPath=&TextBlock
   parent=SelectionFollows
   widget=SelectionFollowsLabel
   properties {
    text=Viewport Selection Follows Widget
    font {
     bForceMonospaced=false
     bMaterialIsStencil=false
     fontMaterial=null
     fontObject=&NotoSans
     letterSpacing=5
     monospacedWidth=1
     size=7.5
     skewAmount=0
     typefaceFontName=NotoSans-Regular
     outlineSettings {
      bApplyOutlineToDropShadows=false
      bMiteredCorners=false
      bSeparateFillAlpha=false
      outlineMaterial=null
      outlineSize=0
      outlineColor=<rgba 0 0 0 1>
     }
    }
   }
   slot {class=HorizontalBoxSlot, properties {verticalAlignment=VAlign_Center, padding=<ltrb 10 0 0 0>}}
  }
  SelectionFollowsWidget {
   widgetClassPath=&EditorUtilityCheckBox
   bIsVariable=true
   parent=SelectionFollows
   widget=SelectionFollowsWidget
   properties {checkedState=Checked}
   slot {class=HorizontalBoxSlot, properties {padding=<ltrb 5 5 5 5>}}
  }
  ForceRefreshButton {
   widgetClassPath=&Button
   parent=HorizontalBox_283
   widget=ForceRefreshButton
   properties {
    toolTipText=Get and Refresh all Loaded Level Actors
    widgetStyle {
     clickedSlateSound {resourceObject=null}
     disabled {
      bIsDynamicallyLoaded=false
      drawAs=Box
      imageType=NoImage
      mirroring=NoMirror
      resourceName=None
      resourceObject=null
      tiling=NoTile
      imageSize=<0 0>
      margin=<ltrb 0 0 0 0>
      outlineSettings {
       bUseBrushTransparency=false
       roundingType=FixedRadius
       width=1
       color {colorUseRule=UseColor_ColorTable, specifiedColor=<rgba 0 0 0 0>}
       cornerRadii=<4 4 4 4>
      }
      tintColor {colorUseRule=UseColor_Specified, specifiedColor=<rgba 0.0395 0.0395 0.0395 1>}
      uVRegion {bIsValid=false, max=<0 0>, min=<0 0>}
     }
     disabledForeground {colorUseRule=UseColor_ColorTable, specifiedColor=<rgba 0 0 0 0>}
     hovered {
      bIsDynamicallyLoaded=false
      drawAs=Box
      imageType=NoImage
      mirroring=NoMirror
      resourceName=None
      resourceObject=null
      tiling=NoTile
      imageSize=<32 32>
      margin=<ltrb 0 0 0 0>
      outlineSettings {
       bUseBrushTransparency=false
       roundingType=FixedRadius
       width=1
       color {colorUseRule=UseColor_ColorTable, specifiedColor=<rgba 0 0 0 0>}
       cornerRadii=<4 4 4 4>
      }
      tintColor {colorUseRule=UseColor_Specified, specifiedColor=<rgba 0.0953 0.0953 0.0953 1>}
      uVRegion {bIsValid=false, max=<0 0>, min=<0 0>}
     }
     hoveredForeground {colorUseRule=UseColor_ColorTable, specifiedColor=<rgba 0 0 0 0>}
     hoveredSlateSound {resourceObject=null}
     normal {
      bIsDynamicallyLoaded=false
      drawAs=Box
      imageType=NoImage
      mirroring=NoMirror
      resourceName=None
      resourceObject=null
      tiling=NoTile
      imageSize=<32 32>
      margin=<ltrb 0 0 0 0>
      outlineSettings {
       bUseBrushTransparency=false
       roundingType=FixedRadius
       width=1
       color {colorUseRule=UseColor_ColorTable, specifiedColor=<rgba 0 0 0 0>}
       cornerRadii=<4 4 4 4>
      }
      tintColor {colorUseRule=UseColor_Specified, specifiedColor=<rgba 0.0284 0.0284 0.0284 1>}
      uVRegion {bIsValid=false, max=<0 0>, min=<0 0>}
     }
     normalForeground {colorUseRule=UseColor_ColorTable, specifiedColor=<rgba 0 0 0 0>}
     normalPadding=<ltrb 12 1.5 12 1.5>
     pressed {
      bIsDynamicallyLoaded=false
      drawAs=Box
      imageType=NoImage
      mirroring=NoMirror
      resourceName=None
      resourceObject=null
      tiling=NoTile
      imageSize=<32 32>
      margin=<ltrb 0 0 0 0>
      outlineSettings {
       bUseBrushTransparency=false
       roundingType=FixedRadius
       width=1
       color {colorUseRule=UseColor_ColorTable, specifiedColor=<rgba 0 0 0 0>}
       cornerRadii=<4 4 4 4>
      }
      tintColor {colorUseRule=UseColor_Specified, specifiedColor=<rgba 0.0284 0.0284 0.0284 1>}
      uVRegion {bIsValid=false, max=<0 0>, min=<0 0>}
     }
     pressedForeground {colorUseRule=UseColor_ColorTable, specifiedColor=<rgba 0 0 0 0>}
     pressedPadding=<ltrb 12 2.5 12 0.5>
     pressedSlateSound {resourceObject=null}
    }
   }
   slot {class=HorizontalBoxSlot, properties {padding=<ltrb 0.5 0 0.5 0>, size {sizeRule=Fill, value=1}}}
  }
  ForceUpdateText {
   widgetClassPath=&UEFN_TextBlock
   parent=ForceRefreshButton
   widget=ForceUpdateText
   properties {
    clipping=ClipToBounds
    text=Force Refresh of Level Actors
    font {
     bForceMonospaced=false
     bMaterialIsStencil=false
     fontMaterial=null
     fontObject=&NotoSans
     letterSpacing=5
     monospacedWidth=1
     size=10
     skewAmount=0
     typefaceFontName=NotoSans-Regular
     outlineSettings {
      bApplyOutlineToDropShadows=false
      bMiteredCorners=false
      bSeparateFillAlpha=false
      outlineMaterial=null
      outlineSize=0
      outlineColor=<rgba 0 0 0 1>
     }
    }
   }
   slot {class=ButtonSlot, properties {padding=<ltrb 0 0 0 0>}}
  }
  HorizontalBox_108 {
   widgetClassPath=&HorizontalBox
   parent=VerticalBox_210
   widget=HorizontalBox_108
   slot {class=VerticalBoxSlot, properties {verticalAlignment=VAlign_Center, padding=<ltrb 5 5 5 5>}}
  }
  CreateNewMusicController {
   widgetClassPath=&Button
   parent=HorizontalBox_108
   widget=CreateNewMusicController
   properties {
    toolTipText=Creates a New Music Controller Blueprint
    widgetStyle {
     clickedSlateSound {resourceObject=null}
     disabled {
      bIsDynamicallyLoaded=false
      drawAs=Box
      imageType=NoImage
      mirroring=NoMirror
      resourceName=None
      resourceObject=null
      tiling=NoTile
      imageSize=<0 0>
      margin=<ltrb 0 0 0 0>
      outlineSettings {
       bUseBrushTransparency=false
       roundingType=FixedRadius
       width=1
       color {colorUseRule=UseColor_ColorTable, specifiedColor=<rgba 0 0 0 0>}
       cornerRadii=<4 4 4 4>
      }
      tintColor {colorUseRule=UseColor_Specified, specifiedColor=<rgba 0.0395 0.0395 0.0395 1>}
      uVRegion {bIsValid=false, max=<0 0>, min=<0 0>}
     }
     disabledForeground {colorUseRule=UseColor_ColorTable, specifiedColor=<rgba 0 0 0 0>}
     hovered {
      bIsDynamicallyLoaded=false
      drawAs=Box
      imageType=NoImage
      mirroring=NoMirror
      resourceName=None
      resourceObject=null
      tiling=NoTile
      imageSize=<32 32>
      margin=<ltrb 0 0 0 0>
      outlineSettings {
       bUseBrushTransparency=false
       roundingType=FixedRadius
       width=1
       color {colorUseRule=UseColor_ColorTable, specifiedColor=<rgba 0 0 0 0>}
       cornerRadii=<4 4 4 4>
      }
      tintColor {colorUseRule=UseColor_Specified, specifiedColor=<rgba 0.0953 0.0953 0.0953 1>}
      uVRegion {bIsValid=false, max=<0 0>, min=<0 0>}
     }
     hoveredForeground {colorUseRule=UseColor_ColorTable, specifiedColor=<rgba 0 0 0 0>}
     hoveredSlateSound {resourceObject=null}
     normal {
      bIsDynamicallyLoaded=false
      drawAs=Box
      imageType=NoImage
      mirroring=NoMirror
      resourceName=None
      resourceObject=null
      tiling=NoTile
      imageSize=<32 32>
      margin=<ltrb 0 0 0 0>
      outlineSettings {
       bUseBrushTransparency=false
       roundingType=FixedRadius
       width=1
       color {colorUseRule=UseColor_ColorTable, specifiedColor=<rgba 0 0 0 0>}
       cornerRadii=<4 4 4 4>
      }
      tintColor {colorUseRule=UseColor_Specified, specifiedColor=<rgba 0.0284 0.0284 0.0284 1>}
      uVRegion {bIsValid=false, max=<0 0>, min=<0 0>}
     }
     normalForeground {colorUseRule=UseColor_ColorTable, specifiedColor=<rgba 0 0 0 0>}
     normalPadding=<ltrb 12 1.5 12 1.5>
     pressed {
      bIsDynamicallyLoaded=false
      drawAs=Box
      imageType=NoImage
      mirroring=NoMirror
      resourceName=None
      resourceObject=null
      tiling=NoTile
      imageSize=<32 32>
      margin=<ltrb 0 0 0 0>
      outlineSettings {
       bUseBrushTransparency=false
       roundingType=FixedRadius
       width=1
       color {colorUseRule=UseColor_ColorTable, specifiedColor=<rgba 0 0 0 0>}
       cornerRadii=<4 4 4 4>
      }
      tintColor {colorUseRule=UseColor_Specified, specifiedColor=<rgba 0.0284 0.0284 0.0284 1>}
      uVRegion {bIsValid=false, max=<0 0>, min=<0 0>}
     }
     pressedForeground {colorUseRule=UseColor_ColorTable, specifiedColor=<rgba 0 0 0 0>}
     pressedPadding=<ltrb 12 2.5 12 0.5>
     pressedSlateSound {resourceObject=null}
    }
   }
   slot {class=HorizontalBoxSlot, properties {padding=<ltrb 0.5 0 0.5 0>, size {sizeRule=Fill, value=1}}}
  }
  Text_4 {
   widgetClassPath=&UEFN_TextBlock
   parent=CreateNewMusicController
   widget=Text_4
   properties {
    clipping=ClipToBounds
    text=Create New Music Controller
    font {
     bForceMonospaced=false
     bMaterialIsStencil=false
     fontMaterial=null
     fontObject=&NotoSans
     letterSpacing=5
     monospacedWidth=1
     size=10
     skewAmount=0
     typefaceFontName=NotoSans-Regular
     outlineSettings {
      bApplyOutlineToDropShadows=false
      bMiteredCorners=false
      bSeparateFillAlpha=false
      outlineMaterial=null
      outlineSize=0
      outlineColor=<rgba 0 0 0 1>
     }
    }
   }
   slot {class=ButtonSlot, properties {padding=<ltrb 0 0 0 0>}}
  }
  CreateNewSpeakerObject {
   widgetClassPath=&Button
   parent=HorizontalBox_108
   widget=CreateNewSpeakerObject
   properties {
    toolTipText=Creates a New Speaker Object Blueprint
    widgetStyle {
     clickedSlateSound {resourceObject=null}
     disabled {
      bIsDynamicallyLoaded=false
      drawAs=Box
      imageType=NoImage
      mirroring=NoMirror
      resourceName=None
      resourceObject=null
      tiling=NoTile
      imageSize=<0 0>
      margin=<ltrb 0 0 0 0>
      outlineSettings {
       bUseBrushTransparency=false
       roundingType=FixedRadius
       width=1
       color {colorUseRule=UseColor_ColorTable, specifiedColor=<rgba 0 0 0 0>}
       cornerRadii=<4 4 4 4>
      }
      tintColor {colorUseRule=UseColor_Specified, specifiedColor=<rgba 0.0395 0.0395 0.0395 1>}
      uVRegion {bIsValid=false, max=<0 0>, min=<0 0>}
     }
     disabledForeground {colorUseRule=UseColor_ColorTable, specifiedColor=<rgba 0 0 0 0>}
     hovered {
      bIsDynamicallyLoaded=false
      drawAs=Box
      imageType=NoImage
      mirroring=NoMirror
      resourceName=None
      resourceObject=null
      tiling=NoTile
      imageSize=<32 32>
      margin=<ltrb 0 0 0 0>
      outlineSettings {
       bUseBrushTransparency=false
       roundingType=FixedRadius
       width=1
       color {colorUseRule=UseColor_ColorTable, specifiedColor=<rgba 0 0 0 0>}
       cornerRadii=<4 4 4 4>
      }
      tintColor {colorUseRule=UseColor_Specified, specifiedColor=<rgba 0.0953 0.0953 0.0953 1>}
      uVRegion {bIsValid=false, max=<0 0>, min=<0 0>}
     }
     hoveredForeground {colorUseRule=UseColor_ColorTable, specifiedColor=<rgba 0 0 0 0>}
     hoveredSlateSound {resourceObject=null}
     normal {
      bIsDynamicallyLoaded=false
      drawAs=Box
      imageType=NoImage
      mirroring=NoMirror
      resourceName=None
      resourceObject=null
      tiling=NoTile
      imageSize=<32 32>
      margin=<ltrb 0 0 0 0>
      outlineSettings {
       bUseBrushTransparency=false
       roundingType=FixedRadius
       width=1
       color {colorUseRule=UseColor_ColorTable, specifiedColor=<rgba 0 0 0 0>}
       cornerRadii=<4 4 4 4>
      }
      tintColor {colorUseRule=UseColor_Specified, specifiedColor=<rgba 0.0284 0.0284 0.0284 1>}
      uVRegion {bIsValid=false, max=<0 0>, min=<0 0>}
     }
     normalForeground {colorUseRule=UseColor_ColorTable, specifiedColor=<rgba 0 0 0 0>}
     normalPadding=<ltrb 12 1.5 12 1.5>
     pressed {
      bIsDynamicallyLoaded=false
      drawAs=Box
      imageType=NoImage
      mirroring=NoMirror
      resourceName=None
      resourceObject=null
      tiling=NoTile
      imageSize=<32 32>
      margin=<ltrb 0 0 0 0>
      outlineSettings {
       bUseBrushTransparency=false
       roundingType=FixedRadius
       width=1
       color {colorUseRule=UseColor_ColorTable, specifiedColor=<rgba 0 0 0 0>}
       cornerRadii=<4 4 4 4>
      }
      tintColor {colorUseRule=UseColor_Specified, specifiedColor=<rgba 0.0284 0.0284 0.0284 1>}
      uVRegion {bIsValid=false, max=<0 0>, min=<0 0>}
     }
     pressedForeground {colorUseRule=UseColor_ColorTable, specifiedColor=<rgba 0 0 0 0>}
     pressedPadding=<ltrb 12 2.5 12 0.5>
     pressedSlateSound {resourceObject=null}
    }
   }
   slot {class=HorizontalBoxSlot, properties {padding=<ltrb 0.5 0 0.5 0>, size {sizeRule=Fill, value=1}}}
  }
  Text_2 {
   widgetClassPath=&UEFN_TextBlock
   parent=CreateNewSpeakerObject
   widget=Text_2
   properties {
    clipping=ClipToBounds
    text=Create New Speaker Object
    font {
     bForceMonospaced=false
     bMaterialIsStencil=false
     fontMaterial=null
     fontObject=&NotoSans
     letterSpacing=5
     monospacedWidth=1
     size=10
     skewAmount=0
     typefaceFontName=NotoSans-Regular
     outlineSettings {
      bApplyOutlineToDropShadows=false
      bMiteredCorners=false
      bSeparateFillAlpha=false
      outlineMaterial=null
      outlineSize=0
      outlineColor=<rgba 0 0 0 1>
     }
    }
   }
   slot {class=ButtonSlot, properties {padding=<ltrb 0 0 0 0>}}
  }
  ConvertPropComponent_Container {
   widgetClassPath=&HorizontalBox
   bIsVariable=true
   parent=VerticalBox_210
   widget=ConvertPropComponent_Container
   properties {visibility=Visible}
   slot {class=VerticalBoxSlot, properties {verticalAlignment=VAlign_Center, padding=<ltrb 0 0.5 5 0.5>}}
  }
  ConvertProp_Enum {
   widgetClassPath=&SinglePropertyView
   parent=ConvertPropComponent_Container
   widget=ConvertProp_Enum
   properties {propertyName=ConvertProp_Category}
   slot {class=HorizontalBoxSlot}
  }
  ConvertPropComponent_Button {
   widgetClassPath=&Button
   parent=ConvertPropComponent_Container
   widget=ConvertPropComponent_Button
   properties {
    toolTipText="Convert currently selected Music Prop Component to a different type (if not assigned in BP)"
    widgetStyle {
     clickedSlateSound {resourceObject=null}
     disabled {
      bIsDynamicallyLoaded=false
      drawAs=Box
      imageType=NoImage
      mirroring=NoMirror
      resourceName=None
      resourceObject=null
      tiling=NoTile
      imageSize=<0 0>
      margin=<ltrb 0 0 0 0>
      outlineSettings {
       bUseBrushTransparency=false
       roundingType=FixedRadius
       width=1
       color {colorUseRule=UseColor_ColorTable, specifiedColor=<rgba 0 0 0 0>}
       cornerRadii=<4 4 4 4>
      }
      tintColor {colorUseRule=UseColor_Specified, specifiedColor=<rgba 0.0395 0.0395 0.0395 1>}
      uVRegion {bIsValid=false, max=<0 0>, min=<0 0>}
     }
     disabledForeground {colorUseRule=UseColor_ColorTable, specifiedColor=<rgba 0 0 0 0>}
     hovered {
      bIsDynamicallyLoaded=false
      drawAs=Box
      imageType=NoImage
      mirroring=NoMirror
      resourceName=None
      resourceObject=null
      tiling=NoTile
      imageSize=<32 32>
      margin=<ltrb 0 0 0 0>
      outlineSettings {
       bUseBrushTransparency=false
       roundingType=FixedRadius
       width=1
       color {colorUseRule=UseColor_ColorTable, specifiedColor=<rgba 0 0 0 0>}
       cornerRadii=<4 4 4 4>
      }
      tintColor {colorUseRule=UseColor_Specified, specifiedColor=<rgba 0.0953 0.0953 0.0953 1>}
      uVRegion {bIsValid=false, max=<0 0>, min=<0 0>}
     }
     hoveredForeground {colorUseRule=UseColor_ColorTable, specifiedColor=<rgba 0 0 0 0>}
     hoveredSlateSound {resourceObject=null}
     normal {
      bIsDynamicallyLoaded=false
      drawAs=Box
      imageType=NoImage
      mirroring=NoMirror
      resourceName=None
      resourceObject=null
      tiling=NoTile
      imageSize=<32 32>
      margin=<ltrb 0 0 0 0>
      outlineSettings {
       bUseBrushTransparency=false
       roundingType=FixedRadius
       width=1
       color {colorUseRule=UseColor_ColorTable, specifiedColor=<rgba 0 0 0 0>}
       cornerRadii=<4 4 4 4>
      }
      tintColor {colorUseRule=UseColor_Specified, specifiedColor=<rgba 0.0284 0.0284 0.0284 1>}
      uVRegion {bIsValid=false, max=<0 0>, min=<0 0>}
     }
     normalForeground {colorUseRule=UseColor_ColorTable, specifiedColor=<rgba 0 0 0 0>}
     normalPadding=<ltrb 12 1.5 12 1.5>
     pressed {
      bIsDynamicallyLoaded=false
      drawAs=Box
      imageType=NoImage
      mirroring=NoMirror
      resourceName=None
      resourceObject=null
      tiling=NoTile
      imageSize=<32 32>
      margin=<ltrb 0 0 0 0>
      outlineSettings {
       bUseBrushTransparency=false
       roundingType=FixedRadius
       width=1
       color {colorUseRule=UseColor_ColorTable, specifiedColor=<rgba 0 0 0 0>}
       cornerRadii=<4 4 4 4>
      }
      tintColor {colorUseRule=UseColor_Specified, specifiedColor=<rgba 0.0284 0.0284 0.0284 1>}
      uVRegion {bIsValid=false, max=<0 0>, min=<0 0>}
     }
     pressedForeground {colorUseRule=UseColor_ColorTable, specifiedColor=<rgba 0 0 0 0>}
     pressedPadding=<ltrb 12 2.5 12 0.5>
     pressedSlateSound {resourceObject=null}
    }
   }
   slot {class=HorizontalBoxSlot, properties {padding=<ltrb 0.5 0 0.5 0>, size {sizeRule=Fill, value=1}}}
  }
  Text_9 {
   widgetClassPath=&UEFN_TextBlock
   parent=ConvertPropComponent_Button
   widget=Text_9
   properties {
    clipping=ClipToBounds
    text=Convert Music Prop Component
    font {
     bForceMonospaced=false
     bMaterialIsStencil=false
     fontMaterial=null
     fontObject=&NotoSans
     letterSpacing=5
     monospacedWidth=1
     size=10.5
     skewAmount=0
     typefaceFontName=NotoSans-Regular
     outlineSettings {
      bApplyOutlineToDropShadows=false
      bMiteredCorners=false
      bSeparateFillAlpha=false
      outlineMaterial=null
      outlineSize=0
      outlineColor=<rgba 0 0 0 1>
     }
    }
   }
   slot {class=ButtonSlot, properties {padding=<ltrb 0 0 0 0>}}
  }
  RemovePropComponent_Button {
   widgetClassPath=&Button
   parent=ConvertPropComponent_Container
   widget=RemovePropComponent_Button
   properties {
    toolTipText="Remoeves currently selected Music Prop Component to a different type (if not assigned in BP)"
    widgetStyle {
     clickedSlateSound {resourceObject=null}
     disabled {
      bIsDynamicallyLoaded=false
      drawAs=Box
      imageType=NoImage
      mirroring=NoMirror
      resourceName=None
      resourceObject=null
      tiling=NoTile
      imageSize=<0 0>
      margin=<ltrb 0 0 0 0>
      outlineSettings {
       bUseBrushTransparency=false
       roundingType=FixedRadius
       width=1
       color {colorUseRule=UseColor_ColorTable, specifiedColor=<rgba 0 0 0 0>}
       cornerRadii=<4 4 4 4>
      }
      tintColor {colorUseRule=UseColor_Specified, specifiedColor=<rgba 0.0395 0.0395 0.0395 1>}
      uVRegion {bIsValid=false, max=<0 0>, min=<0 0>}
     }
     disabledForeground {colorUseRule=UseColor_ColorTable, specifiedColor=<rgba 0 0 0 0>}
     hovered {
      bIsDynamicallyLoaded=false
      drawAs=Box
      imageType=NoImage
      mirroring=NoMirror
      resourceName=None
      resourceObject=null
      tiling=NoTile
      imageSize=<32 32>
      margin=<ltrb 0 0 0 0>
      outlineSettings {
       bUseBrushTransparency=false
       roundingType=FixedRadius
       width=1
       color {colorUseRule=UseColor_ColorTable, specifiedColor=<rgba 0 0 0 0>}
       cornerRadii=<4 4 4 4>
      }
      tintColor {colorUseRule=UseColor_Specified, specifiedColor=<rgba 0.0953 0.0953 0.0953 1>}
      uVRegion {bIsValid=false, max=<0 0>, min=<0 0>}
     }
     hoveredForeground {colorUseRule=UseColor_ColorTable, specifiedColor=<rgba 0 0 0 0>}
     hoveredSlateSound {resourceObject=null}
     normal {
      bIsDynamicallyLoaded=false
      drawAs=Box
      imageType=NoImage
      mirroring=NoMirror
      resourceName=None
      resourceObject=null
      tiling=NoTile
      imageSize=<32 32>
      margin=<ltrb 0 0 0 0>
      outlineSettings {
       bUseBrushTransparency=false
       roundingType=FixedRadius
       width=1
       color {colorUseRule=UseColor_ColorTable, specifiedColor=<rgba 0 0 0 0>}
       cornerRadii=<4 4 4 4>
      }
      tintColor {colorUseRule=UseColor_Specified, specifiedColor=<rgba 0.0284 0.0284 0.0284 1>}
      uVRegion {bIsValid=false, max=<0 0>, min=<0 0>}
     }
     normalForeground {colorUseRule=UseColor_ColorTable, specifiedColor=<rgba 0 0 0 0>}
     normalPadding=<ltrb 12 1.5 12 1.5>
     pressed {
      bIsDynamicallyLoaded=false
      drawAs=Box
      imageType=NoImage
      mirroring=NoMirror
      resourceName=None
      resourceObject=null
      tiling=NoTile
      imageSize=<32 32>
      margin=<ltrb 0 0 0 0>
      outlineSettings {
       bUseBrushTransparency=false
       roundingType=FixedRadius
       width=1
       color {colorUseRule=UseColor_ColorTable, specifiedColor=<rgba 0 0 0 0>}
       cornerRadii=<4 4 4 4>
      }
      tintColor {colorUseRule=UseColor_Specified, specifiedColor=<rgba 0.0284 0.0284 0.0284 1>}
      uVRegion {bIsValid=false, max=<0 0>, min=<0 0>}
     }
     pressedForeground {colorUseRule=UseColor_ColorTable, specifiedColor=<rgba 0 0 0 0>}
     pressedPadding=<ltrb 12 2.5 12 0.5>
     pressedSlateSound {resourceObject=null}
    }
   }
   slot {class=HorizontalBoxSlot, properties {padding=<ltrb 0.5 0 0.5 0>, size {sizeRule=Fill, value=1}}}
  }
  Text_5 {
   widgetClassPath=&UEFN_TextBlock
   parent=RemovePropComponent_Button
   widget=Text_5
   properties {
    clipping=ClipToBounds
    text=Remove Music Prop Component
    font {
     bForceMonospaced=false
     bMaterialIsStencil=false
     fontMaterial=null
     fontObject=&NotoSans
     letterSpacing=5
     monospacedWidth=1
     size=10.5
     skewAmount=0
     typefaceFontName=NotoSans-Regular
     outlineSettings {
      bApplyOutlineToDropShadows=false
      bMiteredCorners=false
      bSeparateFillAlpha=false
      outlineMaterial=null
      outlineSize=0
      outlineColor=<rgba 0 0 0 1>
     }
    }
   }
   slot {class=ButtonSlot, properties {padding=<ltrb 0 0 0 0>}}
  }
  AddMusicPropComponent {
   widgetClassPath=&Button
   parent=VerticalBox_210
   widget=AddMusicPropComponent
   properties {
    toolTipText=Adds a Music Prop Component to the selected Assets
    visibility=Collapsed
    widgetStyle {
     clickedSlateSound {resourceObject=null}
     disabled {
      bIsDynamicallyLoaded=false
      drawAs=Box
      imageType=NoImage
      mirroring=NoMirror
      resourceName=None
      resourceObject=null
      tiling=NoTile
      imageSize=<0 0>
      margin=<ltrb 0 0 0 0>
      outlineSettings {
       bUseBrushTransparency=false
       roundingType=FixedRadius
       width=1
       color {colorUseRule=UseColor_ColorTable, specifiedColor=<rgba 0 0 0 0>}
       cornerRadii=<4 4 4 4>
      }
      tintColor {colorUseRule=UseColor_Specified, specifiedColor=<rgba 0.0395 0.0395 0.0395 1>}
      uVRegion {bIsValid=false, max=<0 0>, min=<0 0>}
     }
     disabledForeground {colorUseRule=UseColor_ColorTable, specifiedColor=<rgba 0 0 0 0>}
     hovered {
      bIsDynamicallyLoaded=false
      drawAs=Box
      imageType=NoImage
      mirroring=NoMirror
      resourceName=None
      resourceObject=null
      tiling=NoTile
      imageSize=<32 32>
      margin=<ltrb 0 0 0 0>
      outlineSettings {
       bUseBrushTransparency=false
       roundingType=FixedRadius
       width=1
       color {colorUseRule=UseColor_ColorTable, specifiedColor=<rgba 0 0 0 0>}
       cornerRadii=<4 4 4 4>
      }
      tintColor {colorUseRule=UseColor_Specified, specifiedColor=<rgba 0.0953 0.0953 0.0953 1>}
      uVRegion {bIsValid=false, max=<0 0>, min=<0 0>}
     }
     hoveredForeground {colorUseRule=UseColor_ColorTable, specifiedColor=<rgba 0 0 0 0>}
     hoveredSlateSound {resourceObject=null}
     normal {
      bIsDynamicallyLoaded=false
      drawAs=Box
      imageType=NoImage
      mirroring=NoMirror
      resourceName=None
      resourceObject=null
      tiling=NoTile
      imageSize=<32 32>
      margin=<ltrb 0 0 0 0>
      outlineSettings {
       bUseBrushTransparency=false
       roundingType=FixedRadius
       width=1
       color {colorUseRule=UseColor_ColorTable, specifiedColor=<rgba 0 0 0 0>}
       cornerRadii=<4 4 4 4>
      }
      tintColor {colorUseRule=UseColor_Specified, specifiedColor=<rgba 0.0284 0.0284 0.0284 1>}
      uVRegion {bIsValid=false, max=<0 0>, min=<0 0>}
     }
     normalForeground {colorUseRule=UseColor_ColorTable, specifiedColor=<rgba 0 0 0 0>}
     normalPadding=<ltrb 12 1.5 12 1.5>
     pressed {
      bIsDynamicallyLoaded=false
      drawAs=Box
      imageType=NoImage
      mirroring=NoMirror
      resourceName=None
      resourceObject=null
      tiling=NoTile
      imageSize=<32 32>
      margin=<ltrb 0 0 0 0>
      outlineSettings {
       bUseBrushTransparency=false
       roundingType=FixedRadius
       width=1
       color {colorUseRule=UseColor_ColorTable, specifiedColor=<rgba 0 0 0 0>}
       cornerRadii=<4 4 4 4>
      }
      tintColor {colorUseRule=UseColor_Specified, specifiedColor=<rgba 0.0284 0.0284 0.0284 1>}
      uVRegion {bIsValid=false, max=<0 0>, min=<0 0>}
     }
     pressedForeground {colorUseRule=UseColor_ColorTable, specifiedColor=<rgba 0 0 0 0>}
     pressedPadding=<ltrb 12 2.5 12 0.5>
     pressedSlateSound {resourceObject=null}
    }
   }
   slot {class=VerticalBoxSlot, properties {padding=<ltrb 0.5 0 0.5 0>, size {sizeRule=Fill, value=1}}}
  }
  Text {
   widgetClassPath=&UEFN_TextBlock
   parent=AddMusicPropComponent
   widget=Text
   properties {
    clipping=ClipToBounds
    text=Add Music Prop Component To Selected
    font {
     bForceMonospaced=false
     bMaterialIsStencil=false
     fontMaterial=null
     fontObject=&NotoSans
     letterSpacing=5
     monospacedWidth=1
     size=12
     skewAmount=0
     typefaceFontName=NotoSans-Regular
     outlineSettings {
      bApplyOutlineToDropShadows=false
      bMiteredCorners=false
      bSeparateFillAlpha=false
      outlineMaterial=null
      outlineSize=0
      outlineColor=<rgba 0 0 0 1>
     }
    }
   }
   slot {class=ButtonSlot, properties {padding=<ltrb 0 0 0 0>}}
  }
  FindActorSearchTerms {
   widgetClassPath=&DetailsView
   parent=VerticalBox_210
   widget=FindActorSearchTerms
   properties {toolTipText=Add or remove search terms to find potential unassigned speakers in a level, categoriesToShow [SearchFilters]}
   slot {class=VerticalBoxSlot, properties {padding=<ltrb 0 2 0 2>}}
  }
  EditorContainer {
   widgetClassPath=&VerticalBox
   bIsVariable=true
   parent=VerticalBox_1
   widget=EditorContainer
   properties {toolTipText=Editable loaded Music Prop Actors}
   slot {class=VerticalBoxSlot, properties {size {sizeRule=Fill, value=1}}}
  }
  EditorLevelActorView {
   widgetClassPath=&UEFN_TextBlock
   bIsVariable=true
   parent=EditorContainer
   widget=EditorLevelActorView
   properties {
    clipping=ClipToBounds
    text=Level Actor Editor View
    font {
     bForceMonospaced=false
     bMaterialIsStencil=false
     fontMaterial=null
     fontObject=&NotoSans
     letterSpacing=5
     monospacedWidth=1
     size=12
     skewAmount=0
     typefaceFontName=NotoSans-Bold
     outlineSettings {
      bApplyOutlineToDropShadows=false
      bMiteredCorners=false
      bSeparateFillAlpha=false
      outlineMaterial=null
      outlineSize=0
      outlineColor=<rgba 0 0 0 1>
     }
    }
   }
   slot {class=VerticalBoxSlot}
  }
  ScrollBox_1 {
   widgetClassPath=&ScrollBox
   parent=EditorContainer
   widget=ScrollBox_1
   properties {
    widgetBarStyle {
     thickness=8
     draggedThumbImage {
      bIsDynamicallyLoaded=false
      drawAs=RoundedBox
      imageType=NoImage
      mirroring=NoMirror
      resourceName=None
      resourceObject=null
      tiling=NoTile
      imageSize=<0 0>
      margin=<ltrb 0 0 0 0>
      outlineSettings {
       bUseBrushTransparency=false
       roundingType=FixedRadius
       width=0
       color {colorUseRule=UseColor_Specified, specifiedColor=<rgba 0 0 0 0>}
       cornerRadii=<4 4 4 4>
      }
      tintColor {colorUseRule=UseColor_Specified, specifiedColor=<rgba 0.2159 0.2159 0.2159 1>}
      uVRegion {bIsValid=false, max=<0 0>, min=<0 0>}
     }
     horizontalBackgroundImage {
      bIsDynamicallyLoaded=false
      drawAs=NoDrawType
      imageType=NoImage
      mirroring=NoMirror
      resourceName=None
      resourceObject=null
      tiling=NoTile
      imageSize=<0 0>
      margin=<ltrb 0 0 0 0>
      outlineSettings {
       bUseBrushTransparency=false
       roundingType=HalfHeightRadius
       width=0
       color {colorUseRule=UseColor_Specified, specifiedColor=<rgba 0 0 0 0>}
       cornerRadii=<0 0 0 0>
      }
      tintColor {colorUseRule=UseColor_Specified, specifiedColor=<rgba 1 1 1 1>}
      uVRegion {bIsValid=false, max=<0 0>, min=<0 0>}
     }
     horizontalBottomSlotImage {
      bIsDynamicallyLoaded=false
      drawAs=NoDrawType
      imageType=NoImage
      mirroring=NoMirror
      resourceName=None
      resourceObject=null
      tiling=NoTile
      imageSize=<0 0>
      margin=<ltrb 0 0 0 0>
      outlineSettings {
       bUseBrushTransparency=false
       roundingType=HalfHeightRadius
       width=0
       color {colorUseRule=UseColor_Specified, specifiedColor=<rgba 0 0 0 0>}
       cornerRadii=<0 0 0 0>
      }
      tintColor {colorUseRule=UseColor_Specified, specifiedColor=<rgba 1 1 1 1>}
      uVRegion {bIsValid=false, max=<0 0>, min=<0 0>}
     }
     horizontalTopSlotImage {
      bIsDynamicallyLoaded=false
      drawAs=NoDrawType
      imageType=NoImage
      mirroring=NoMirror
      resourceName=None
      resourceObject=null
      tiling=NoTile
      imageSize=<0 0>
      margin=<ltrb 0 0 0 0>
      outlineSettings {
       bUseBrushTransparency=false
       roundingType=HalfHeightRadius
       width=0
       color {colorUseRule=UseColor_Specified, specifiedColor=<rgba 0 0 0 0>}
       cornerRadii=<0 0 0 0>
      }
      tintColor {colorUseRule=UseColor_Specified, specifiedColor=<rgba 1 1 1 1>}
      uVRegion {bIsValid=false, max=<0 0>, min=<0 0>}
     }
     hoveredThumbImage {
      bIsDynamicallyLoaded=false
      drawAs=RoundedBox
      imageType=NoImage
      mirroring=NoMirror
      resourceName=None
      resourceObject=null
      tiling=NoTile
      imageSize=<0 0>
      margin=<ltrb 0 0 0 0>
      outlineSettings {
       bUseBrushTransparency=false
       roundingType=FixedRadius
       width=0
       color {colorUseRule=UseColor_Specified, specifiedColor=<rgba 0 0 0 0>}
       cornerRadii=<4 4 4 4>
      }
      tintColor {colorUseRule=UseColor_Specified, specifiedColor=<rgba 0.2159 0.2159 0.2159 1>}
      uVRegion {bIsValid=false, max=<0 0>, min=<0 0>}
     }
     normalThumbImage {
      bIsDynamicallyLoaded=false
      drawAs=RoundedBox
      imageType=NoImage
      mirroring=NoMirror
      resourceName=None
      resourceObject=null
      tiling=NoTile
      imageSize=<0 0>
      margin=<ltrb 0 0 0 0>
      outlineSettings {
       bUseBrushTransparency=false
       roundingType=FixedRadius
       width=0
       color {colorUseRule=UseColor_Specified, specifiedColor=<rgba 0 0 0 0>}
       cornerRadii=<4 4 4 4>
      }
      tintColor {colorUseRule=UseColor_Specified, specifiedColor=<rgba 0.0953 0.0953 0.0953 1>}
      uVRegion {bIsValid=false, max=<0 0>, min=<0 0>}
     }
     verticalBackgroundImage {
      bIsDynamicallyLoaded=false
      drawAs=NoDrawType
      imageType=NoImage
      mirroring=NoMirror
      resourceName=None
      resourceObject=null
      tiling=NoTile
      imageSize=<0 0>
      margin=<ltrb 0 0 0 0>
      outlineSettings {
       bUseBrushTransparency=false
       roundingType=HalfHeightRadius
       width=0
       color {colorUseRule=UseColor_Specified, specifiedColor=<rgba 0 0 0 0>}
       cornerRadii=<0 0 0 0>
      }
      tintColor {colorUseRule=UseColor_Specified, specifiedColor=<rgba 1 1 1 1>}
      uVRegion {bIsValid=false, max=<0 0>, min=<0 0>}
     }
     verticalBottomSlotImage {
      bIsDynamicallyLoaded=false
      drawAs=NoDrawType
      imageType=NoImage
      mirroring=NoMirror
      resourceName=None
      resourceObject=null
      tiling=NoTile
      imageSize=<0 0>
      margin=<ltrb 0 0 0 0>
      outlineSettings {
       bUseBrushTransparency=false
       roundingType=HalfHeightRadius
       width=0
       color {colorUseRule=UseColor_Specified, specifiedColor=<rgba 0 0 0 0>}
       cornerRadii=<0 0 0 0>
      }
      tintColor {colorUseRule=UseColor_Specified, specifiedColor=<rgba 1 1 1 1>}
      uVRegion {bIsValid=false, max=<0 0>, min=<0 0>}
     }
     verticalTopSlotImage {
      bIsDynamicallyLoaded=false
      drawAs=NoDrawType
      imageType=NoImage
      mirroring=NoMirror
      resourceName=None
      resourceObject=null
      tiling=NoTile
      imageSize=<0 0>
      margin=<ltrb 0 0 0 0>
      outlineSettings {
       bUseBrushTransparency=false
       roundingType=HalfHeightRadius
       width=0
       color {colorUseRule=UseColor_Specified, specifiedColor=<rgba 0 0 0 0>}
       cornerRadii=<0 0 0 0>
      }
      tintColor {colorUseRule=UseColor_Specified, specifiedColor=<rgba 1 1 1 1>}
      uVRegion {bIsValid=false, max=<0 0>, min=<0 0>}
     }
    }
    widgetStyle {
     barThickness=8
     bottomShadowBrush {
      bIsDynamicallyLoaded=false
      drawAs=Box
      imageType=FullColor
      mirroring=NoMirror
      resourceName=../../../Engine/Content/Slate/Common/ScrollBoxShadowBottom.png
      resourceObject=null
      tiling=NoTile
      imageSize=<16 8>
      margin=<ltrb 0.5 0 0.5 1>
      outlineSettings {
       bUseBrushTransparency=false
       roundingType=HalfHeightRadius
       width=0
       color {colorUseRule=UseColor_Specified, specifiedColor=<rgba 0 0 0 0>}
       cornerRadii=<0 0 0 0>
      }
      tintColor {colorUseRule=UseColor_Specified, specifiedColor=<rgba 1 1 1 1>}
      uVRegion {bIsValid=false, max=<0 0>, min=<0 0>}
     }
     horizontalScrolledContentPadding=<ltrb 0 0 1 0>
     leftShadowBrush {
      bIsDynamicallyLoaded=false
      drawAs=Box
      imageType=FullColor
      mirroring=NoMirror
      resourceName=../../../Engine/Content/Slate/Common/ScrollBoxShadowLeft.png
      resourceObject=null
      tiling=NoTile
      imageSize=<8 16>
      margin=<ltrb 1 0.5 0 0.5>
      outlineSettings {
       bUseBrushTransparency=false
       roundingType=HalfHeightRadius
       width=0
       color {colorUseRule=UseColor_Specified, specifiedColor=<rgba 0 0 0 0>}
       cornerRadii=<0 0 0 0>
      }
      tintColor {colorUseRule=UseColor_Specified, specifiedColor=<rgba 1 1 1 1>}
      uVRegion {bIsValid=false, max=<0 0>, min=<0 0>}
     }
     rightShadowBrush {
      bIsDynamicallyLoaded=false
      drawAs=Box
      imageType=FullColor
      mirroring=NoMirror
      resourceName=../../../Engine/Content/Slate/Common/ScrollBoxShadowRight.png
      resourceObject=null
      tiling=NoTile
      imageSize=<8 16>
      margin=<ltrb 0 0.5 1 0.5>
      outlineSettings {
       bUseBrushTransparency=false
       roundingType=HalfHeightRadius
       width=0
       color {colorUseRule=UseColor_Specified, specifiedColor=<rgba 0 0 0 0>}
       cornerRadii=<0 0 0 0>
      }
      tintColor {colorUseRule=UseColor_Specified, specifiedColor=<rgba 1 1 1 1>}
      uVRegion {bIsValid=false, max=<0 0>, min=<0 0>}
     }
     topShadowBrush {
      bIsDynamicallyLoaded=false
      drawAs=Box
      imageType=FullColor
      mirroring=NoMirror
      resourceName=../../../Engine/Content/Slate/Common/ScrollBoxShadowTop.png
      resourceObject=null
      tiling=NoTile
      imageSize=<16 8>
      margin=<ltrb 0.5 1 0.5 0>
      outlineSettings {
       bUseBrushTransparency=false
       roundingType=HalfHeightRadius
       width=0
       color {colorUseRule=UseColor_Specified, specifiedColor=<rgba 0 0 0 0>}
       cornerRadii=<0 0 0 0>
      }
      tintColor {colorUseRule=UseColor_Specified, specifiedColor=<rgba 1 1 1 1>}
      uVRegion {bIsValid=false, max=<0 0>, min=<0 0>}
     }
     verticalScrolledContentPadding=<ltrb 0 0 0 1>
    }
   }
   slot {class=VerticalBoxSlot, properties {size {sizeRule=Fill, value=1}}}
  }
  LevelActorEntries {
   widgetClassPath=/Script/UMG.ListView
   bIsVariable=true
   parent=ScrollBox_1
   widget=LevelActorEntries
   properties {
    entryWidgetClass=&/EUW_MusicPartyEditor_LevelActorEntry.EUW_MusicPartyEditor_LevelActorEntry_C
    selectionMode=None
    scrollBarStyle {
     thickness=8
     draggedThumbImage {
      bIsDynamicallyLoaded=false
      drawAs=RoundedBox
      imageType=NoImage
      mirroring=NoMirror
      resourceName=None
      resourceObject=null
      tiling=NoTile
      imageSize=<0 0>
      margin=<ltrb 0 0 0 0>
      outlineSettings {
       bUseBrushTransparency=false
       roundingType=FixedRadius
       width=0
       color {colorUseRule=UseColor_Specified, specifiedColor=<rgba 0 0 0 0>}
       cornerRadii=<4 4 4 4>
      }
      tintColor {colorUseRule=UseColor_Specified, specifiedColor=<rgba 0.2159 0.2159 0.2159 1>}
      uVRegion {bIsValid=false, max=<0 0>, min=<0 0>}
     }
     horizontalBackgroundImage {
      bIsDynamicallyLoaded=false
      drawAs=NoDrawType
      imageType=NoImage
      mirroring=NoMirror
      resourceName=None
      resourceObject=null
      tiling=NoTile
      imageSize=<0 0>
      margin=<ltrb 0 0 0 0>
      outlineSettings {
       bUseBrushTransparency=false
       roundingType=HalfHeightRadius
       width=0
       color {colorUseRule=UseColor_Specified, specifiedColor=<rgba 0 0 0 0>}
       cornerRadii=<0 0 0 0>
      }
      tintColor {colorUseRule=UseColor_Specified, specifiedColor=<rgba 1 1 1 1>}
      uVRegion {bIsValid=false, max=<0 0>, min=<0 0>}
     }
     horizontalBottomSlotImage {
      bIsDynamicallyLoaded=false
      drawAs=NoDrawType
      imageType=NoImage
      mirroring=NoMirror
      resourceName=None
      resourceObject=null
      tiling=NoTile
      imageSize=<0 0>
      margin=<ltrb 0 0 0 0>
      outlineSettings {
       bUseBrushTransparency=false
       roundingType=HalfHeightRadius
       width=0
       color {colorUseRule=UseColor_Specified, specifiedColor=<rgba 0 0 0 0>}
       cornerRadii=<0 0 0 0>
      }
      tintColor {colorUseRule=UseColor_Specified, specifiedColor=<rgba 1 1 1 1>}
      uVRegion {bIsValid=false, max=<0 0>, min=<0 0>}
     }
     horizontalTopSlotImage {
      bIsDynamicallyLoaded=false
      drawAs=NoDrawType
      imageType=NoImage
      mirroring=NoMirror
      resourceName=None
      resourceObject=null
      tiling=NoTile
      imageSize=<0 0>
      margin=<ltrb 0 0 0 0>
      outlineSettings {
       bUseBrushTransparency=false
       roundingType=HalfHeightRadius
       width=0
       color {colorUseRule=UseColor_Specified, specifiedColor=<rgba 0 0 0 0>}
       cornerRadii=<0 0 0 0>
      }
      tintColor {colorUseRule=UseColor_Specified, specifiedColor=<rgba 1 1 1 1>}
      uVRegion {bIsValid=false, max=<0 0>, min=<0 0>}
     }
     hoveredThumbImage {
      bIsDynamicallyLoaded=false
      drawAs=RoundedBox
      imageType=NoImage
      mirroring=NoMirror
      resourceName=None
      resourceObject=null
      tiling=NoTile
      imageSize=<0 0>
      margin=<ltrb 0 0 0 0>
      outlineSettings {
       bUseBrushTransparency=false
       roundingType=FixedRadius
       width=0
       color {colorUseRule=UseColor_Specified, specifiedColor=<rgba 0 0 0 0>}
       cornerRadii=<4 4 4 4>
      }
      tintColor {colorUseRule=UseColor_Specified, specifiedColor=<rgba 0.2159 0.2159 0.2159 1>}
      uVRegion {bIsValid=false, max=<0 0>, min=<0 0>}
     }
     normalThumbImage {
      bIsDynamicallyLoaded=false
      drawAs=RoundedBox
      imageType=NoImage
      mirroring=NoMirror
      resourceName=None
      resourceObject=null
      tiling=NoTile
      imageSize=<0 0>
      margin=<ltrb 0 0 0 0>
      outlineSettings {
       bUseBrushTransparency=false
       roundingType=FixedRadius
       width=0
       color {colorUseRule=UseColor_Specified, specifiedColor=<rgba 0 0 0 0>}
       cornerRadii=<4 4 4 4>
      }
      tintColor {colorUseRule=UseColor_Specified, specifiedColor=<rgba 0.0953 0.0953 0.0953 1>}
      uVRegion {bIsValid=false, max=<0 0>, min=<0 0>}
     }
     verticalBackgroundImage {
      bIsDynamicallyLoaded=false
      drawAs=NoDrawType
      imageType=NoImage
      mirroring=NoMirror
      resourceName=None
      resourceObject=null
      tiling=NoTile
      imageSize=<0 0>
      margin=<ltrb 0 0 0 0>
      outlineSettings {
       bUseBrushTransparency=false
       roundingType=HalfHeightRadius
       width=0
       color {colorUseRule=UseColor_Specified, specifiedColor=<rgba 0 0 0 0>}
       cornerRadii=<0 0 0 0>
      }
      tintColor {colorUseRule=UseColor_Specified, specifiedColor=<rgba 1 1 1 1>}
      uVRegion {bIsValid=false, max=<0 0>, min=<0 0>}
     }
     verticalBottomSlotImage {
      bIsDynamicallyLoaded=false
      drawAs=NoDrawType
      imageType=NoImage
      mirroring=NoMirror
      resourceName=None
      resourceObject=null
      tiling=NoTile
      imageSize=<0 0>
      margin=<ltrb 0 0 0 0>
      outlineSettings {
       bUseBrushTransparency=false
       roundingType=HalfHeightRadius
       width=0
       color {colorUseRule=UseColor_Specified, specifiedColor=<rgba 0 0 0 0>}
       cornerRadii=<0 0 0 0>
      }
      tintColor {colorUseRule=UseColor_Specified, specifiedColor=<rgba 1 1 1 1>}
      uVRegion {bIsValid=false, max=<0 0>, min=<0 0>}
     }
     verticalTopSlotImage {
      bIsDynamicallyLoaded=false
      drawAs=NoDrawType
      imageType=NoImage
      mirroring=NoMirror
      resourceName=None
      resourceObject=null
      tiling=NoTile
      imageSize=<0 0>
      margin=<ltrb 0 0 0 0>
      outlineSettings {
       bUseBrushTransparency=false
       roundingType=HalfHeightRadius
       width=0
       color {colorUseRule=UseColor_Specified, specifiedColor=<rgba 0 0 0 0>}
       cornerRadii=<0 0 0 0>
      }
      tintColor {colorUseRule=UseColor_Specified, specifiedColor=<rgba 1 1 1 1>}
      uVRegion {bIsValid=false, max=<0 0>, min=<0 0>}
     }
    }
   }
   slot {class=ScrollBoxSlot}
  }
  PIEView {
   widgetClassPath=&Border
   bIsVariable=true
   parent=VerticalBox_1
   widget=PIEView
   properties {toolTipText=PIE View for Testing: Changes will not Save, brushColor=<rgba 0.022 0.0461 0.0521 1>, padding=<ltrb 0 0 0 0>}
   slot {class=VerticalBoxSlot, properties {size {sizeRule=Fill, value=1}}}
  }
  PIEContainer {widgetClassPath=&VerticalBox, parent=PIEView, widget=PIEContainer, slot {class=BorderSlot, properties {padding=<ltrb 0 0 0 0>}}}
  PIEText {
   widgetClassPath=&UEFN_TextBlock
   parent=PIEContainer
   widget=PIEText
   properties {
    clipping=ClipToBounds
    text="PIE View (Changes Will Not Save)"
    font {
     bForceMonospaced=false
     bMaterialIsStencil=false
     fontMaterial=null
     fontObject=&NotoSans
     letterSpacing=5
     monospacedWidth=1
     size=12
     skewAmount=0
     typefaceFontName=NotoSans-Bold
     outlineSettings {
      bApplyOutlineToDropShadows=false
      bMiteredCorners=false
      bSeparateFillAlpha=false
      outlineMaterial=null
      outlineSize=0
      outlineColor=<rgba 0 0 0 1>
     }
    }
   }
   slot {class=VerticalBoxSlot}
  }
  ScrollBox_2 {
   widgetClassPath=&ScrollBox
   parent=PIEContainer
   widget=ScrollBox_2
   properties {
    toolTipText=PIE Level Actors: Changes Made Here Will Not Save
    widgetBarStyle {
     thickness=8
     draggedThumbImage {
      bIsDynamicallyLoaded=false
      drawAs=RoundedBox
      imageType=NoImage
      mirroring=NoMirror
      resourceName=None
      resourceObject=null
      tiling=NoTile
      imageSize=<0 0>
      margin=<ltrb 0 0 0 0>
      outlineSettings {
       bUseBrushTransparency=false
       roundingType=FixedRadius
       width=0
       color {colorUseRule=UseColor_Specified, specifiedColor=<rgba 0 0 0 0>}
       cornerRadii=<4 4 4 4>
      }
      tintColor {colorUseRule=UseColor_Specified, specifiedColor=<rgba 0.2159 0.2159 0.2159 1>}
      uVRegion {bIsValid=false, max=<0 0>, min=<0 0>}
     }
     horizontalBackgroundImage {
      bIsDynamicallyLoaded=false
      drawAs=NoDrawType
      imageType=NoImage
      mirroring=NoMirror
      resourceName=None
      resourceObject=null
      tiling=NoTile
      imageSize=<0 0>
      margin=<ltrb 0 0 0 0>
      outlineSettings {
       bUseBrushTransparency=false
       roundingType=HalfHeightRadius
       width=0
       color {colorUseRule=UseColor_Specified, specifiedColor=<rgba 0 0 0 0>}
       cornerRadii=<0 0 0 0>
      }
      tintColor {colorUseRule=UseColor_Specified, specifiedColor=<rgba 1 1 1 1>}
      uVRegion {bIsValid=false, max=<0 0>, min=<0 0>}
     }
     horizontalBottomSlotImage {
      bIsDynamicallyLoaded=false
      drawAs=NoDrawType
      imageType=NoImage
      mirroring=NoMirror
      resourceName=None
      resourceObject=null
      tiling=NoTile
      imageSize=<0 0>
      margin=<ltrb 0 0 0 0>
      outlineSettings {
       bUseBrushTransparency=false
       roundingType=HalfHeightRadius
       width=0
       color {colorUseRule=UseColor_Specified, specifiedColor=<rgba 0 0 0 0>}
       cornerRadii=<0 0 0 0>
      }
      tintColor {colorUseRule=UseColor_Specified, specifiedColor=<rgba 1 1 1 1>}
      uVRegion {bIsValid=false, max=<0 0>, min=<0 0>}
     }
     horizontalTopSlotImage {
      bIsDynamicallyLoaded=false
      drawAs=NoDrawType
      imageType=NoImage
      mirroring=NoMirror
      resourceName=None
      resourceObject=null
      tiling=NoTile
      imageSize=<0 0>
      margin=<ltrb 0 0 0 0>
      outlineSettings {
       bUseBrushTransparency=false
       roundingType=HalfHeightRadius
       width=0
       color {colorUseRule=UseColor_Specified, specifiedColor=<rgba 0 0 0 0>}
       cornerRadii=<0 0 0 0>
      }
      tintColor {colorUseRule=UseColor_Specified, specifiedColor=<rgba 1 1 1 1>}
      uVRegion {bIsValid=false, max=<0 0>, min=<0 0>}
     }
     hoveredThumbImage {
      bIsDynamicallyLoaded=false
      drawAs=RoundedBox
      imageType=NoImage
      mirroring=NoMirror
      resourceName=None
      resourceObject=null
      tiling=NoTile
      imageSize=<0 0>
      margin=<ltrb 0 0 0 0>
      outlineSettings {
       bUseBrushTransparency=false
       roundingType=FixedRadius
       width=0
       color {colorUseRule=UseColor_Specified, specifiedColor=<rgba 0 0 0 0>}
       cornerRadii=<4 4 4 4>
      }
      tintColor {colorUseRule=UseColor_Specified, specifiedColor=<rgba 0.2159 0.2159 0.2159 1>}
      uVRegion {bIsValid=false, max=<0 0>, min=<0 0>}
     }
     normalThumbImage {
      bIsDynamicallyLoaded=false
      drawAs=RoundedBox
      imageType=NoImage
      mirroring=NoMirror
      resourceName=None
      resourceObject=null
      tiling=NoTile
      imageSize=<0 0>
      margin=<ltrb 0 0 0 0>
      outlineSettings {
       bUseBrushTransparency=false
       roundingType=FixedRadius
       width=0
       color {colorUseRule=UseColor_Specified, specifiedColor=<rgba 0 0 0 0>}
       cornerRadii=<4 4 4 4>
      }
      tintColor {colorUseRule=UseColor_Specified, specifiedColor=<rgba 0.0953 0.0953 0.0953 1>}
      uVRegion {bIsValid=false, max=<0 0>, min=<0 0>}
     }
     verticalBackgroundImage {
      bIsDynamicallyLoaded=false
      drawAs=NoDrawType
      imageType=NoImage
      mirroring=NoMirror
      resourceName=None
      resourceObject=null
      tiling=NoTile
      imageSize=<0 0>
      margin=<ltrb 0 0 0 0>
      outlineSettings {
       bUseBrushTransparency=false
       roundingType=HalfHeightRadius
       width=0
       color {colorUseRule=UseColor_Specified, specifiedColor=<rgba 0 0 0 0>}
       cornerRadii=<0 0 0 0>
      }
      tintColor {colorUseRule=UseColor_Specified, specifiedColor=<rgba 1 1 1 1>}
      uVRegion {bIsValid=false, max=<0 0>, min=<0 0>}
     }
     verticalBottomSlotImage {
      bIsDynamicallyLoaded=false
      drawAs=NoDrawType
      imageType=NoImage
      mirroring=NoMirror
      resourceName=None
      resourceObject=null
      tiling=NoTile
      imageSize=<0 0>
      margin=<ltrb 0 0 0 0>
      outlineSettings {
       bUseBrushTransparency=false
       roundingType=HalfHeightRadius
       width=0
       color {colorUseRule=UseColor_Specified, specifiedColor=<rgba 0 0 0 0>}
       cornerRadii=<0 0 0 0>
      }
      tintColor {colorUseRule=UseColor_Specified, specifiedColor=<rgba 1 1 1 1>}
      uVRegion {bIsValid=false, max=<0 0>, min=<0 0>}
     }
     verticalTopSlotImage {
      bIsDynamicallyLoaded=false
      drawAs=NoDrawType
      imageType=NoImage
      mirroring=NoMirror
      resourceName=None
      resourceObject=null
      tiling=NoTile
      imageSize=<0 0>
      margin=<ltrb 0 0 0 0>
      outlineSettings {
       bUseBrushTransparency=false
       roundingType=HalfHeightRadius
       width=0
       color {colorUseRule=UseColor_Specified, specifiedColor=<rgba 0 0 0 0>}
       cornerRadii=<0 0 0 0>
      }
      tintColor {colorUseRule=UseColor_Specified, specifiedColor=<rgba 1 1 1 1>}
      uVRegion {bIsValid=false, max=<0 0>, min=<0 0>}
     }
    }
    widgetStyle {
     barThickness=8
     bottomShadowBrush {
      bIsDynamicallyLoaded=false
      drawAs=Box
      imageType=FullColor
      mirroring=NoMirror
      resourceName=../../../Engine/Content/Slate/Common/ScrollBoxShadowBottom.png
      resourceObject=null
      tiling=NoTile
      imageSize=<16 8>
      margin=<ltrb 0.5 0 0.5 1>
      outlineSettings {
       bUseBrushTransparency=false
       roundingType=HalfHeightRadius
       width=0
       color {colorUseRule=UseColor_Specified, specifiedColor=<rgba 0 0 0 0>}
       cornerRadii=<0 0 0 0>
      }
      tintColor {colorUseRule=UseColor_Specified, specifiedColor=<rgba 1 1 1 1>}
      uVRegion {bIsValid=false, max=<0 0>, min=<0 0>}
     }
     horizontalScrolledContentPadding=<ltrb 0 0 1 0>
     leftShadowBrush {
      bIsDynamicallyLoaded=false
      drawAs=Box
      imageType=FullColor
      mirroring=NoMirror
      resourceName=../../../Engine/Content/Slate/Common/ScrollBoxShadowLeft.png
      resourceObject=null
      tiling=NoTile
      imageSize=<8 16>
      margin=<ltrb 1 0.5 0 0.5>
      outlineSettings {
       bUseBrushTransparency=false
       roundingType=HalfHeightRadius
       width=0
       color {colorUseRule=UseColor_Specified, specifiedColor=<rgba 0 0 0 0>}
       cornerRadii=<0 0 0 0>
      }
      tintColor {colorUseRule=UseColor_Specified, specifiedColor=<rgba 1 1 1 1>}
      uVRegion {bIsValid=false, max=<0 0>, min=<0 0>}
     }
     rightShadowBrush {
      bIsDynamicallyLoaded=false
      drawAs=Box
      imageType=FullColor
      mirroring=NoMirror
      resourceName=../../../Engine/Content/Slate/Common/ScrollBoxShadowRight.png
      resourceObject=null
      tiling=NoTile
      imageSize=<8 16>
      margin=<ltrb 0 0.5 1 0.5>
      outlineSettings {
       bUseBrushTransparency=false
       roundingType=HalfHeightRadius
       width=0
       color {colorUseRule=UseColor_Specified, specifiedColor=<rgba 0 0 0 0>}
       cornerRadii=<0 0 0 0>
      }
      tintColor {colorUseRule=UseColor_Specified, specifiedColor=<rgba 1 1 1 1>}
      uVRegion {bIsValid=false, max=<0 0>, min=<0 0>}
     }
     topShadowBrush {
      bIsDynamicallyLoaded=false
      drawAs=Box
      imageType=FullColor
      mirroring=NoMirror
      resourceName=../../../Engine/Content/Slate/Common/ScrollBoxShadowTop.png
      resourceObject=null
      tiling=NoTile
      imageSize=<16 8>
      margin=<ltrb 0.5 1 0.5 0>
      outlineSettings {
       bUseBrushTransparency=false
       roundingType=HalfHeightRadius
       width=0
       color {colorUseRule=UseColor_Specified, specifiedColor=<rgba 0 0 0 0>}
       cornerRadii=<0 0 0 0>
      }
      tintColor {colorUseRule=UseColor_Specified, specifiedColor=<rgba 1 1 1 1>}
      uVRegion {bIsValid=false, max=<0 0>, min=<0 0>}
     }
     verticalScrolledContentPadding=<ltrb 0 0 0 1>
    }
   }
   slot {class=VerticalBoxSlot, properties {size {sizeRule=Fill, value=1}}}
  }
  PIEActorEntries {
   widgetClassPath=&ListView
   bIsVariable=true
   parent=ScrollBox_2
   widget=PIEActorEntries
   properties {
    entryWidgetClass=&EUW_MusicPartyEditor_LevelActorEntry_C
    selectionMode=None
    scrollBarStyle {
     thickness=8
     draggedThumbImage {
      bIsDynamicallyLoaded=false
      drawAs=RoundedBox
      imageType=NoImage
      mirroring=NoMirror
      resourceName=None
      resourceObject=null
      tiling=NoTile
      imageSize=<0 0>
      margin=<ltrb 0 0 0 0>
      outlineSettings {
       bUseBrushTransparency=false
       roundingType=FixedRadius
       width=0
       color {colorUseRule=UseColor_Specified, specifiedColor=<rgba 0 0 0 0>}
       cornerRadii=<4 4 4 4>
      }
      tintColor {colorUseRule=UseColor_Specified, specifiedColor=<rgba 0.2159 0.2159 0.2159 1>}
      uVRegion {bIsValid=false, max=<0 0>, min=<0 0>}
     }
     horizontalBackgroundImage {
      bIsDynamicallyLoaded=false
      drawAs=NoDrawType
      imageType=NoImage
      mirroring=NoMirror
      resourceName=None
      resourceObject=null
      tiling=NoTile
      imageSize=<0 0>
      margin=<ltrb 0 0 0 0>
      outlineSettings {
       bUseBrushTransparency=false
       roundingType=HalfHeightRadius
       width=0
       color {colorUseRule=UseColor_Specified, specifiedColor=<rgba 0 0 0 0>}
       cornerRadii=<0 0 0 0>
      }
      tintColor {colorUseRule=UseColor_Specified, specifiedColor=<rgba 1 1 1 1>}
      uVRegion {bIsValid=false, max=<0 0>, min=<0 0>}
     }
     horizontalBottomSlotImage {
      bIsDynamicallyLoaded=false
      drawAs=NoDrawType
      imageType=NoImage
      mirroring=NoMirror
      resourceName=None
      resourceObject=null
      tiling=NoTile
      imageSize=<0 0>
      margin=<ltrb 0 0 0 0>
      outlineSettings {
       bUseBrushTransparency=false
       roundingType=HalfHeightRadius
       width=0
       color {colorUseRule=UseColor_Specified, specifiedColor=<rgba 0 0 0 0>}
       cornerRadii=<0 0 0 0>
      }
      tintColor {colorUseRule=UseColor_Specified, specifiedColor=<rgba 1 1 1 1>}
      uVRegion {bIsValid=false, max=<0 0>, min=<0 0>}
     }
     horizontalTopSlotImage {
      bIsDynamicallyLoaded=false
      drawAs=NoDrawType
      imageType=NoImage
      mirroring=NoMirror
      resourceName=None
      resourceObject=null
      tiling=NoTile
      imageSize=<0 0>
      margin=<ltrb 0 0 0 0>
      outlineSettings {
       bUseBrushTransparency=false
       roundingType=HalfHeightRadius
       width=0
       color {colorUseRule=UseColor_Specified, specifiedColor=<rgba 0 0 0 0>}
       cornerRadii=<0 0 0 0>
      }
      tintColor {colorUseRule=UseColor_Specified, specifiedColor=<rgba 1 1 1 1>}
      uVRegion {bIsValid=false, max=<0 0>, min=<0 0>}
     }
     hoveredThumbImage {
      bIsDynamicallyLoaded=false
      drawAs=RoundedBox
      imageType=NoImage
      mirroring=NoMirror
      resourceName=None
      resourceObject=null
      tiling=NoTile
      imageSize=<0 0>
      margin=<ltrb 0 0 0 0>
      outlineSettings {
       bUseBrushTransparency=false
       roundingType=FixedRadius
       width=0
       color {colorUseRule=UseColor_Specified, specifiedColor=<rgba 0 0 0 0>}
       cornerRadii=<4 4 4 4>
      }
      tintColor {colorUseRule=UseColor_Specified, specifiedColor=<rgba 0.2159 0.2159 0.2159 1>}
      uVRegion {bIsValid=false, max=<0 0>, min=<0 0>}
     }
     normalThumbImage {
      bIsDynamicallyLoaded=false
      drawAs=RoundedBox
      imageType=NoImage
      mirroring=NoMirror
      resourceName=None
      resourceObject=null
      tiling=NoTile
      imageSize=<0 0>
      margin=<ltrb 0 0 0 0>
      outlineSettings {
       bUseBrushTransparency=false
       roundingType=FixedRadius
       width=0
       color {colorUseRule=UseColor_Specified, specifiedColor=<rgba 0 0 0 0>}
       cornerRadii=<4 4 4 4>
      }
      tintColor {colorUseRule=UseColor_Specified, specifiedColor=<rgba 0.0953 0.0953 0.0953 1>}
      uVRegion {bIsValid=false, max=<0 0>, min=<0 0>}
     }
     verticalBackgroundImage {
      bIsDynamicallyLoaded=false
      drawAs=NoDrawType
      imageType=NoImage
      mirroring=NoMirror
      resourceName=None
      resourceObject=null
      tiling=NoTile
      imageSize=<0 0>
      margin=<ltrb 0 0 0 0>
      outlineSettings {
       bUseBrushTransparency=false
       roundingType=HalfHeightRadius
       width=0
       color {colorUseRule=UseColor_Specified, specifiedColor=<rgba 0 0 0 0>}
       cornerRadii=<0 0 0 0>
      }
      tintColor {colorUseRule=UseColor_Specified, specifiedColor=<rgba 1 1 1 1>}
      uVRegion {bIsValid=false, max=<0 0>, min=<0 0>}
     }
     verticalBottomSlotImage {
      bIsDynamicallyLoaded=false
      drawAs=NoDrawType
      imageType=NoImage
      mirroring=NoMirror
      resourceName=None
      resourceObject=null
      tiling=NoTile
      imageSize=<0 0>
      margin=<ltrb 0 0 0 0>
      outlineSettings {
       bUseBrushTransparency=false
       roundingType=HalfHeightRadius
       width=0
       color {colorUseRule=UseColor_Specified, specifiedColor=<rgba 0 0 0 0>}
       cornerRadii=<0 0 0 0>
      }
      tintColor {colorUseRule=UseColor_Specified, specifiedColor=<rgba 1 1 1 1>}
      uVRegion {bIsValid=false, max=<0 0>, min=<0 0>}
     }
     verticalTopSlotImage {
      bIsDynamicallyLoaded=false
      drawAs=NoDrawType
      imageType=NoImage
      mirroring=NoMirror
      resourceName=None
      resourceObject=null
      tiling=NoTile
      imageSize=<0 0>
      margin=<ltrb 0 0 0 0>
      outlineSettings {
       bUseBrushTransparency=false
       roundingType=HalfHeightRadius
       width=0
       color {colorUseRule=UseColor_Specified, specifiedColor=<rgba 0 0 0 0>}
       cornerRadii=<0 0 0 0>
      }
      tintColor {colorUseRule=UseColor_Specified, specifiedColor=<rgba 1 1 1 1>}
      uVRegion {bIsValid=false, max=<0 0>, min=<0 0>}
     }
    }
   }
   slot {class=ScrollBoxSlot}
  }
  Border_1 {
   widgetClassPath=&Border
   parent=HorizontalBox_233
   widget=Border_1
   properties {
    background {
     bIsDynamicallyLoaded=false
     drawAs=Box
     imageType=NoImage
     mirroring=NoMirror
     resourceName=None
     resourceObject=null
     tiling=NoTile
     imageSize=<32 32>
     margin=<ltrb 0 0 0 0>
     outlineSettings {
      bUseBrushTransparency=false
      roundingType=HalfHeightRadius
      width=0
      color {colorUseRule=UseColor_Specified, specifiedColor=<rgba 0 0 0 0>}
      cornerRadii=<0 0 0 0>
     }
     tintColor {colorUseRule=UseColor_Specified, specifiedColor=<rgba 1 1 1 1>}
     uVRegion {bIsValid=false, max=<0 0>, min=<0 0>}
    }
    brushColor=<rgba 0 0 0 0.2>
    padding=<ltrb 0 0 0 0>
   }
   slot {class=HorizontalBoxSlot, properties {padding=<ltrb 5 5 5 5>, size {sizeRule=Fill, value=0.4}}}
  }
  WidgetSwitcher {
   widgetClassPath=&WidgetSwitcher
   bIsVariable=true
   parent=Border_1
   widget=WidgetSwitcher
   slot {class=BorderSlot, properties {padding=<ltrb 0 0 0 0>}}
  }
  0_EditorView {widgetClassPath=&VerticalBox, parent=WidgetSwitcher, widget=0_EditorView, slot {class=WidgetSwitcherSlot}}
  TextTitle {
   widgetClassPath=&UEFN_TextBlock
   bIsVariable=true
   parent=0_EditorView
   widget=TextTitle
   properties {
    clipping=ClipToBounds
    text=Level Actor Editor View
    font {
     bForceMonospaced=false
     bMaterialIsStencil=false
     fontMaterial=null
     fontObject=&NotoSans
     letterSpacing=5
     monospacedWidth=1
     size=12
     skewAmount=0
     typefaceFontName=NotoSans-Bold
     outlineSettings {
      bApplyOutlineToDropShadows=false
      bMiteredCorners=false
      bSeparateFillAlpha=false
      outlineMaterial=null
      outlineSize=0
      outlineColor=<rgba 0 0 0 1>
     }
    }
   }
   slot {class=VerticalBoxSlot}
  }
  ShowAttenuation_Container {
   widgetClassPath=&HorizontalBox
   parent=0_EditorView
   widget=ShowAttenuation_Container
   properties {toolTipText=Display a Draw Debug of Selected Asset Attenuations}
   slot {class=VerticalBoxSlot}
  }
  ShowAttenuation_Label {
   widgetClassPath=&TextBlock
   parent=ShowAttenuation_Container
   widget=ShowAttenuation_Label
   properties {
    text=Show Attenuation Debug
    font {
     bForceMonospaced=false
     bMaterialIsStencil=false
     fontMaterial=null
     fontObject=&NotoSans
     letterSpacing=5
     monospacedWidth=1
     size=7.5
     skewAmount=0
     typefaceFontName=NotoSans-Regular
     outlineSettings {
      bApplyOutlineToDropShadows=false
      bMiteredCorners=false
      bSeparateFillAlpha=false
      outlineMaterial=null
      outlineSize=0
      outlineColor=<rgba 0 0 0 1>
     }
    }
   }
   slot {class=HorizontalBoxSlot, properties {verticalAlignment=VAlign_Center, padding=<ltrb 10 0 0 0>}}
  }
  ShowAttenuation {
   widgetClassPath=&EditorUtilityCheckBox
   bIsVariable=true
   parent=ShowAttenuation_Container
   widget=ShowAttenuation
   slot {class=HorizontalBoxSlot, properties {padding=<ltrb 5 5 5 5>}}
  }
  Editor_ActorEditor {
   widgetClassPath=&ScrollBox
   bIsVariable=true
   parent=0_EditorView
   widget=Editor_ActorEditor
   properties {
    widgetBarStyle {
     thickness=8
     draggedThumbImage {
      bIsDynamicallyLoaded=false
      drawAs=RoundedBox
      imageType=NoImage
      mirroring=NoMirror
      resourceName=None
      resourceObject=null
      tiling=NoTile
      imageSize=<0 0>
      margin=<ltrb 0 0 0 0>
      outlineSettings {
       bUseBrushTransparency=false
       roundingType=FixedRadius
       width=0
       color {colorUseRule=UseColor_Specified, specifiedColor=<rgba 0 0 0 0>}
       cornerRadii=<4 4 4 4>
      }
      tintColor {colorUseRule=UseColor_Specified, specifiedColor=<rgba 0.2159 0.2159 0.2159 1>}
      uVRegion {bIsValid=false, max=<0 0>, min=<0 0>}
     }
     horizontalBackgroundImage {
      bIsDynamicallyLoaded=false
      drawAs=NoDrawType
      imageType=NoImage
      mirroring=NoMirror
      resourceName=None
      resourceObject=null
      tiling=NoTile
      imageSize=<0 0>
      margin=<ltrb 0 0 0 0>
      outlineSettings {
       bUseBrushTransparency=false
       roundingType=HalfHeightRadius
       width=0
       color {colorUseRule=UseColor_Specified, specifiedColor=<rgba 0 0 0 0>}
       cornerRadii=<0 0 0 0>
      }
      tintColor {colorUseRule=UseColor_Specified, specifiedColor=<rgba 1 1 1 1>}
      uVRegion {bIsValid=false, max=<0 0>, min=<0 0>}
     }
     horizontalBottomSlotImage {
      bIsDynamicallyLoaded=false
      drawAs=NoDrawType
      imageType=NoImage
      mirroring=NoMirror
      resourceName=None
      resourceObject=null
      tiling=NoTile
      imageSize=<0 0>
      margin=<ltrb 0 0 0 0>
      outlineSettings {
       bUseBrushTransparency=false
       roundingType=HalfHeightRadius
       width=0
       color {colorUseRule=UseColor_Specified, specifiedColor=<rgba 0 0 0 0>}
       cornerRadii=<0 0 0 0>
      }
      tintColor {colorUseRule=UseColor_Specified, specifiedColor=<rgba 1 1 1 1>}
      uVRegion {bIsValid=false, max=<0 0>, min=<0 0>}
     }
     horizontalTopSlotImage {
      bIsDynamicallyLoaded=false
      drawAs=NoDrawType
      imageType=NoImage
      mirroring=NoMirror
      resourceName=None
      resourceObject=null
      tiling=NoTile
      imageSize=<0 0>
      margin=<ltrb 0 0 0 0>
      outlineSettings {
       bUseBrushTransparency=false
       roundingType=HalfHeightRadius
       width=0
       color {colorUseRule=UseColor_Specified, specifiedColor=<rgba 0 0 0 0>}
       cornerRadii=<0 0 0 0>
      }
      tintColor {colorUseRule=UseColor_Specified, specifiedColor=<rgba 1 1 1 1>}
      uVRegion {bIsValid=false, max=<0 0>, min=<0 0>}
     }
     hoveredThumbImage {
      bIsDynamicallyLoaded=false
      drawAs=RoundedBox
      imageType=NoImage
      mirroring=NoMirror
      resourceName=None
      resourceObject=null
      tiling=NoTile
      imageSize=<0 0>
      margin=<ltrb 0 0 0 0>
      outlineSettings {
       bUseBrushTransparency=false
       roundingType=FixedRadius
       width=0
       color {colorUseRule=UseColor_Specified, specifiedColor=<rgba 0 0 0 0>}
       cornerRadii=<4 4 4 4>
      }
      tintColor {colorUseRule=UseColor_Specified, specifiedColor=<rgba 0.2159 0.2159 0.2159 1>}
      uVRegion {bIsValid=false, max=<0 0>, min=<0 0>}
     }
     normalThumbImage {
      bIsDynamicallyLoaded=false
      drawAs=RoundedBox
      imageType=NoImage
      mirroring=NoMirror
      resourceName=None
      resourceObject=null
      tiling=NoTile
      imageSize=<0 0>
      margin=<ltrb 0 0 0 0>
      outlineSettings {
       bUseBrushTransparency=false
       roundingType=FixedRadius
       width=0
       color {colorUseRule=UseColor_Specified, specifiedColor=<rgba 0 0 0 0>}
       cornerRadii=<4 4 4 4>
      }
      tintColor {colorUseRule=UseColor_Specified, specifiedColor=<rgba 0.0953 0.0953 0.0953 1>}
      uVRegion {bIsValid=false, max=<0 0>, min=<0 0>}
     }
     verticalBackgroundImage {
      bIsDynamicallyLoaded=false
      drawAs=NoDrawType
      imageType=NoImage
      mirroring=NoMirror
      resourceName=None
      resourceObject=null
      tiling=NoTile
      imageSize=<0 0>
      margin=<ltrb 0 0 0 0>
      outlineSettings {
       bUseBrushTransparency=false
       roundingType=HalfHeightRadius
       width=0
       color {colorUseRule=UseColor_Specified, specifiedColor=<rgba 0 0 0 0>}
       cornerRadii=<0 0 0 0>
      }
      tintColor {colorUseRule=UseColor_Specified, specifiedColor=<rgba 1 1 1 1>}
      uVRegion {bIsValid=false, max=<0 0>, min=<0 0>}
     }
     verticalBottomSlotImage {
      bIsDynamicallyLoaded=false
      drawAs=NoDrawType
      imageType=NoImage
      mirroring=NoMirror
      resourceName=None
      resourceObject=null
      tiling=NoTile
      imageSize=<0 0>
      margin=<ltrb 0 0 0 0>
      outlineSettings {
       bUseBrushTransparency=false
       roundingType=HalfHeightRadius
       width=0
       color {colorUseRule=UseColor_Specified, specifiedColor=<rgba 0 0 0 0>}
       cornerRadii=<0 0 0 0>
      }
      tintColor {colorUseRule=UseColor_Specified, specifiedColor=<rgba 1 1 1 1>}
      uVRegion {bIsValid=false, max=<0 0>, min=<0 0>}
     }
     verticalTopSlotImage {
      bIsDynamicallyLoaded=false
      drawAs=NoDrawType
      imageType=NoImage
      mirroring=NoMirror
      resourceName=None
      resourceObject=null
      tiling=NoTile
      imageSize=<0 0>
      margin=<ltrb 0 0 0 0>
      outlineSettings {
       bUseBrushTransparency=false
       roundingType=HalfHeightRadius
       width=0
       color {colorUseRule=UseColor_Specified, specifiedColor=<rgba 0 0 0 0>}
       cornerRadii=<0 0 0 0>
      }
      tintColor {colorUseRule=UseColor_Specified, specifiedColor=<rgba 1 1 1 1>}
      uVRegion {bIsValid=false, max=<0 0>, min=<0 0>}
     }
    }
    widgetStyle {
     barThickness=8
     bottomShadowBrush {
      bIsDynamicallyLoaded=false
      drawAs=Box
      imageType=FullColor
      mirroring=NoMirror
      resourceName=../../../Engine/Content/Slate/Common/ScrollBoxShadowBottom.png
      resourceObject=null
      tiling=NoTile
      imageSize=<16 8>
      margin=<ltrb 0.5 0 0.5 1>
      outlineSettings {
       bUseBrushTransparency=false
       roundingType=HalfHeightRadius
       width=0
       color {colorUseRule=UseColor_Specified, specifiedColor=<rgba 0 0 0 0>}
       cornerRadii=<0 0 0 0>
      }
      tintColor {colorUseRule=UseColor_Specified, specifiedColor=<rgba 1 1 1 1>}
      uVRegion {bIsValid=false, max=<0 0>, min=<0 0>}
     }
     horizontalScrolledContentPadding=<ltrb 0 0 1 0>
     leftShadowBrush {
      bIsDynamicallyLoaded=false
      drawAs=Box
      imageType=FullColor
      mirroring=NoMirror
      resourceName=../../../Engine/Content/Slate/Common/ScrollBoxShadowLeft.png
      resourceObject=null
      tiling=NoTile
      imageSize=<8 16>
      margin=<ltrb 1 0.5 0 0.5>
      outlineSettings {
       bUseBrushTransparency=false
       roundingType=HalfHeightRadius
       width=0
       color {colorUseRule=UseColor_Specified, specifiedColor=<rgba 0 0 0 0>}
       cornerRadii=<0 0 0 0>
      }
      tintColor {colorUseRule=UseColor_Specified, specifiedColor=<rgba 1 1 1 1>}
      uVRegion {bIsValid=false, max=<0 0>, min=<0 0>}
     }
     rightShadowBrush {
      bIsDynamicallyLoaded=false
      drawAs=Box
      imageType=FullColor
      mirroring=NoMirror
      resourceName=../../../Engine/Content/Slate/Common/ScrollBoxShadowRight.png
      resourceObject=null
      tiling=NoTile
      imageSize=<8 16>
      margin=<ltrb 0 0.5 1 0.5>
      outlineSettings {
       bUseBrushTransparency=false
       roundingType=HalfHeightRadius
       width=0
       color {colorUseRule=UseColor_Specified, specifiedColor=<rgba 0 0 0 0>}
       cornerRadii=<0 0 0 0>
      }
      tintColor {colorUseRule=UseColor_Specified, specifiedColor=<rgba 1 1 1 1>}
      uVRegion {bIsValid=false, max=<0 0>, min=<0 0>}
     }
     topShadowBrush {
      bIsDynamicallyLoaded=false
      drawAs=Box
      imageType=FullColor
      mirroring=NoMirror
      resourceName=../../../Engine/Content/Slate/Common/ScrollBoxShadowTop.png
      resourceObject=null
      tiling=NoTile
      imageSize=<16 8>
      margin=<ltrb 0.5 1 0.5 0>
      outlineSettings {
       bUseBrushTransparency=false
       roundingType=HalfHeightRadius
       width=0
       color {colorUseRule=UseColor_Specified, specifiedColor=<rgba 0 0 0 0>}
       cornerRadii=<0 0 0 0>
      }
      tintColor {colorUseRule=UseColor_Specified, specifiedColor=<rgba 1 1 1 1>}
      uVRegion {bIsValid=false, max=<0 0>, min=<0 0>}
     }
     verticalScrolledContentPadding=<ltrb 0 0 0 1>
    }
   }
   slot {class=VerticalBoxSlot, properties {size {sizeRule=Fill, value=1}}}
  }
  MusicEditorListView {
   widgetClassPath=&CommonListView
   bIsVariable=true
   parent=Editor_ActorEditor
   widget=MusicEditorListView
   properties {
    entryWidgetClass=&/EUW_MusicPartyEditor_MusicPropEditEntry.EUW_MusicPartyEditor_MusicPropEditEntry_C
    selectionMode=None
    scrollBarStyle {
     thickness=0
     draggedThumbImage {
      bIsDynamicallyLoaded=false
      drawAs=RoundedBox
      imageType=NoImage
      mirroring=NoMirror
      resourceName=None
      resourceObject=null
      tiling=NoTile
      imageSize=<0 0>
      margin=<ltrb 0 0 0 0>
      outlineSettings {
       bUseBrushTransparency=false
       roundingType=FixedRadius
       width=0
       color {colorUseRule=UseColor_Specified, specifiedColor=<rgba 0 0 0 0>}
       cornerRadii=<4 4 4 4>
      }
      tintColor {colorUseRule=UseColor_Specified, specifiedColor=<rgba 0.2159 0.2159 0.2159 1>}
      uVRegion {bIsValid=false, max=<0 0>, min=<0 0>}
     }
     horizontalBackgroundImage {
      bIsDynamicallyLoaded=false
      drawAs=NoDrawType
      imageType=NoImage
      mirroring=NoMirror
      resourceName=None
      resourceObject=null
      tiling=NoTile
      imageSize=<0 0>
      margin=<ltrb 0 0 0 0>
      outlineSettings {
       bUseBrushTransparency=false
       roundingType=HalfHeightRadius
       width=0
       color {colorUseRule=UseColor_Specified, specifiedColor=<rgba 0 0 0 0>}
       cornerRadii=<0 0 0 0>
      }
      tintColor {colorUseRule=UseColor_Specified, specifiedColor=<rgba 1 1 1 1>}
      uVRegion {bIsValid=false, max=<0 0>, min=<0 0>}
     }
     horizontalBottomSlotImage {
      bIsDynamicallyLoaded=false
      drawAs=NoDrawType
      imageType=NoImage
      mirroring=NoMirror
      resourceName=None
      resourceObject=null
      tiling=NoTile
      imageSize=<0 0>
      margin=<ltrb 0 0 0 0>
      outlineSettings {
       bUseBrushTransparency=false
       roundingType=HalfHeightRadius
       width=0
       color {colorUseRule=UseColor_Specified, specifiedColor=<rgba 0 0 0 0>}
       cornerRadii=<0 0 0 0>
      }
      tintColor {colorUseRule=UseColor_Specified, specifiedColor=<rgba 1 1 1 1>}
      uVRegion {bIsValid=false, max=<0 0>, min=<0 0>}
     }
     horizontalTopSlotImage {
      bIsDynamicallyLoaded=false
      drawAs=NoDrawType
      imageType=NoImage
      mirroring=NoMirror
      resourceName=None
      resourceObject=null
      tiling=NoTile
      imageSize=<0 0>
      margin=<ltrb 0 0 0 0>
      outlineSettings {
       bUseBrushTransparency=false
       roundingType=HalfHeightRadius
       width=0
       color {colorUseRule=UseColor_Specified, specifiedColor=<rgba 0 0 0 0>}
       cornerRadii=<0 0 0 0>
      }
      tintColor {colorUseRule=UseColor_Specified, specifiedColor=<rgba 1 1 1 1>}
      uVRegion {bIsValid=false, max=<0 0>, min=<0 0>}
     }
     hoveredThumbImage {
      bIsDynamicallyLoaded=false
      drawAs=RoundedBox
      imageType=NoImage
      mirroring=NoMirror
      resourceName=None
      resourceObject=null
      tiling=NoTile
      imageSize=<0 0>
      margin=<ltrb 0 0 0 0>
      outlineSettings {
       bUseBrushTransparency=false
       roundingType=FixedRadius
       width=0
       color {colorUseRule=UseColor_Specified, specifiedColor=<rgba 0 0 0 0>}
       cornerRadii=<4 4 4 4>
      }
      tintColor {colorUseRule=UseColor_Specified, specifiedColor=<rgba 0.2159 0.2159 0.2159 1>}
      uVRegion {bIsValid=false, max=<0 0>, min=<0 0>}
     }
     normalThumbImage {
      bIsDynamicallyLoaded=false
      drawAs=RoundedBox
      imageType=NoImage
      mirroring=NoMirror
      resourceName=None
      resourceObject=null
      tiling=NoTile
      imageSize=<0 0>
      margin=<ltrb 0 0 0 0>
      outlineSettings {
       bUseBrushTransparency=false
       roundingType=FixedRadius
       width=0
       color {colorUseRule=UseColor_Specified, specifiedColor=<rgba 0 0 0 0>}
       cornerRadii=<4 4 4 4>
      }
      tintColor {colorUseRule=UseColor_Specified, specifiedColor=<rgba 0.0953 0.0953 0.0953 1>}
      uVRegion {bIsValid=false, max=<0 0>, min=<0 0>}
     }
     verticalBackgroundImage {
      bIsDynamicallyLoaded=false
      drawAs=NoDrawType
      imageType=NoImage
      mirroring=NoMirror
      resourceName=None
      resourceObject=null
      tiling=NoTile
      imageSize=<0 0>
      margin=<ltrb 0 0 0 0>
      outlineSettings {
       bUseBrushTransparency=false
       roundingType=HalfHeightRadius
       width=0
       color {colorUseRule=UseColor_Specified, specifiedColor=<rgba 0 0 0 0>}
       cornerRadii=<0 0 0 0>
      }
      tintColor {colorUseRule=UseColor_Specified, specifiedColor=<rgba 1 1 1 1>}
      uVRegion {bIsValid=false, max=<0 0>, min=<0 0>}
     }
     verticalBottomSlotImage {
      bIsDynamicallyLoaded=false
      drawAs=NoDrawType
      imageType=NoImage
      mirroring=NoMirror
      resourceName=None
      resourceObject=null
      tiling=NoTile
      imageSize=<0 0>
      margin=<ltrb 0 0 0 0>
      outlineSettings {
       bUseBrushTransparency=false
       roundingType=HalfHeightRadius
       width=0
       color {colorUseRule=UseColor_Specified, specifiedColor=<rgba 0 0 0 0>}
       cornerRadii=<0 0 0 0>
      }
      tintColor {colorUseRule=UseColor_Specified, specifiedColor=<rgba 1 1 1 1>}
      uVRegion {bIsValid=false, max=<0 0>, min=<0 0>}
     }
     verticalTopSlotImage {
      bIsDynamicallyLoaded=false
      drawAs=NoDrawType
      imageType=NoImage
      mirroring=NoMirror
      resourceName=None
      resourceObject=null
      tiling=NoTile
      imageSize=<0 0>
      margin=<ltrb 0 0 0 0>
      outlineSettings {
       bUseBrushTransparency=false
       roundingType=HalfHeightRadius
       width=0
       color {colorUseRule=UseColor_Specified, specifiedColor=<rgba 0 0 0 0>}
       cornerRadii=<0 0 0 0>
      }
      tintColor {colorUseRule=UseColor_Specified, specifiedColor=<rgba 1 1 1 1>}
      uVRegion {bIsValid=false, max=<0 0>, min=<0 0>}
     }
    }
   }
   slot {class=ScrollBoxSlot, properties {padding=<ltrb 0 10 10 10>, size {sizeRule=Fill, value=1}}}
  }
  PIE_ActorEditor {
   widgetClassPath=&Border
   bIsVariable=true
   parent=0_EditorView
   widget=PIE_ActorEditor
   properties {brushColor=<rgba 0.022 0.0461 0.0521 1>}
   slot {class=VerticalBoxSlot, properties {size {sizeRule=Fill, value=1}}}
  }
  VerticalBox_180 {widgetClassPath=&VerticalBox, parent=PIE_ActorEditor, widget=VerticalBox_180, slot {class=BorderSlot}}
  TextTitle_1 {
   widgetClassPath=&UEFN_TextBlock
   bIsVariable=true
   parent=VerticalBox_180
   widget=TextTitle_1
   properties {
    clipping=ClipToBounds
    text="PIE Actor Edit View (Prop Changes Will Not Save)"
    font {
     bForceMonospaced=false
     bMaterialIsStencil=false
     fontMaterial=null
     fontObject=&NotoSans
     letterSpacing=5
     monospacedWidth=1
     size=12
     skewAmount=0
     typefaceFontName=NotoSans-Bold
     outlineSettings {
      bApplyOutlineToDropShadows=false
      bMiteredCorners=false
      bSeparateFillAlpha=false
      outlineMaterial=null
      outlineSize=0
      outlineColor=<rgba 0 0 0 1>
     }
    }
   }
   slot {class=VerticalBoxSlot}
  }
  ScrollBox_3 {
   widgetClassPath=&ScrollBox
   parent=VerticalBox_180
   widget=ScrollBox_3
   properties {
    widgetBarStyle {
     thickness=8
     draggedThumbImage {
      bIsDynamicallyLoaded=false
      drawAs=RoundedBox
      imageType=NoImage
      mirroring=NoMirror
      resourceName=None
      resourceObject=null
      tiling=NoTile
      imageSize=<0 0>
      margin=<ltrb 0 0 0 0>
      outlineSettings {
       bUseBrushTransparency=false
       roundingType=FixedRadius
       width=0
       color {colorUseRule=UseColor_Specified, specifiedColor=<rgba 0 0 0 0>}
       cornerRadii=<4 4 4 4>
      }
      tintColor {colorUseRule=UseColor_Specified, specifiedColor=<rgba 0.2159 0.2159 0.2159 1>}
      uVRegion {bIsValid=false, max=<0 0>, min=<0 0>}
     }
     horizontalBackgroundImage {
      bIsDynamicallyLoaded=false
      drawAs=NoDrawType
      imageType=NoImage
      mirroring=NoMirror
      resourceName=None
      resourceObject=null
      tiling=NoTile
      imageSize=<0 0>
      margin=<ltrb 0 0 0 0>
      outlineSettings {
       bUseBrushTransparency=false
       roundingType=HalfHeightRadius
       width=0
       color {colorUseRule=UseColor_Specified, specifiedColor=<rgba 0 0 0 0>}
       cornerRadii=<0 0 0 0>
      }
      tintColor {colorUseRule=UseColor_Specified, specifiedColor=<rgba 1 1 1 1>}
      uVRegion {bIsValid=false, max=<0 0>, min=<0 0>}
     }
     horizontalBottomSlotImage {
      bIsDynamicallyLoaded=false
      drawAs=NoDrawType
      imageType=NoImage
      mirroring=NoMirror
      resourceName=None
      resourceObject=null
      tiling=NoTile
      imageSize=<0 0>
      margin=<ltrb 0 0 0 0>
      outlineSettings {
       bUseBrushTransparency=false
       roundingType=HalfHeightRadius
       width=0
       color {colorUseRule=UseColor_Specified, specifiedColor=<rgba 0 0 0 0>}
       cornerRadii=<0 0 0 0>
      }
      tintColor {colorUseRule=UseColor_Specified, specifiedColor=<rgba 1 1 1 1>}
      uVRegion {bIsValid=false, max=<0 0>, min=<0 0>}
     }
     horizontalTopSlotImage {
      bIsDynamicallyLoaded=false
      drawAs=NoDrawType
      imageType=NoImage
      mirroring=NoMirror
      resourceName=None
      resourceObject=null
      tiling=NoTile
      imageSize=<0 0>
      margin=<ltrb 0 0 0 0>
      outlineSettings {
       bUseBrushTransparency=false
       roundingType=HalfHeightRadius
       width=0
       color {colorUseRule=UseColor_Specified, specifiedColor=<rgba 0 0 0 0>}
       cornerRadii=<0 0 0 0>
      }
      tintColor {colorUseRule=UseColor_Specified, specifiedColor=<rgba 1 1 1 1>}
      uVRegion {bIsValid=false, max=<0 0>, min=<0 0>}
     }
     hoveredThumbImage {
      bIsDynamicallyLoaded=false
      drawAs=RoundedBox
      imageType=NoImage
      mirroring=NoMirror
      resourceName=None
      resourceObject=null
      tiling=NoTile
      imageSize=<0 0>
      margin=<ltrb 0 0 0 0>
      outlineSettings {
       bUseBrushTransparency=false
       roundingType=FixedRadius
       width=0
       color {colorUseRule=UseColor_Specified, specifiedColor=<rgba 0 0 0 0>}
       cornerRadii=<4 4 4 4>
      }
      tintColor {colorUseRule=UseColor_Specified, specifiedColor=<rgba 0.2159 0.2159 0.2159 1>}
      uVRegion {bIsValid=false, max=<0 0>, min=<0 0>}
     }
     normalThumbImage {
      bIsDynamicallyLoaded=false
      drawAs=RoundedBox
      imageType=NoImage
      mirroring=NoMirror
      resourceName=None
      resourceObject=null
      tiling=NoTile
      imageSize=<0 0>
      margin=<ltrb 0 0 0 0>
      outlineSettings {
       bUseBrushTransparency=false
       roundingType=FixedRadius
       width=0
       color {colorUseRule=UseColor_Specified, specifiedColor=<rgba 0 0 0 0>}
       cornerRadii=<4 4 4 4>
      }
      tintColor {colorUseRule=UseColor_Specified, specifiedColor=<rgba 0.0953 0.0953 0.0953 1>}
      uVRegion {bIsValid=false, max=<0 0>, min=<0 0>}
     }
     verticalBackgroundImage {
      bIsDynamicallyLoaded=false
      drawAs=NoDrawType
      imageType=NoImage
      mirroring=NoMirror
      resourceName=None
      resourceObject=null
      tiling=NoTile
      imageSize=<0 0>
      margin=<ltrb 0 0 0 0>
      outlineSettings {
       bUseBrushTransparency=false
       roundingType=HalfHeightRadius
       width=0
       color {colorUseRule=UseColor_Specified, specifiedColor=<rgba 0 0 0 0>}
       cornerRadii=<0 0 0 0>
      }
      tintColor {colorUseRule=UseColor_Specified, specifiedColor=<rgba 1 1 1 1>}
      uVRegion {bIsValid=false, max=<0 0>, min=<0 0>}
     }
     verticalBottomSlotImage {
      bIsDynamicallyLoaded=false
      drawAs=NoDrawType
      imageType=NoImage
      mirroring=NoMirror
      resourceName=None
      resourceObject=null
      tiling=NoTile
      imageSize=<0 0>
      margin=<ltrb 0 0 0 0>
      outlineSettings {
       bUseBrushTransparency=false
       roundingType=HalfHeightRadius
       width=0
       color {colorUseRule=UseColor_Specified, specifiedColor=<rgba 0 0 0 0>}
       cornerRadii=<0 0 0 0>
      }
      tintColor {colorUseRule=UseColor_Specified, specifiedColor=<rgba 1 1 1 1>}
      uVRegion {bIsValid=false, max=<0 0>, min=<0 0>}
     }
     verticalTopSlotImage {
      bIsDynamicallyLoaded=false
      drawAs=NoDrawType
      imageType=NoImage
      mirroring=NoMirror
      resourceName=None
      resourceObject=null
      tiling=NoTile
      imageSize=<0 0>
      margin=<ltrb 0 0 0 0>
      outlineSettings {
       bUseBrushTransparency=false
       roundingType=HalfHeightRadius
       width=0
       color {colorUseRule=UseColor_Specified, specifiedColor=<rgba 0 0 0 0>}
       cornerRadii=<0 0 0 0>
      }
      tintColor {colorUseRule=UseColor_Specified, specifiedColor=<rgba 1 1 1 1>}
      uVRegion {bIsValid=false, max=<0 0>, min=<0 0>}
     }
    }
    widgetStyle {
     barThickness=8
     bottomShadowBrush {
      bIsDynamicallyLoaded=false
      drawAs=Box
      imageType=FullColor
      mirroring=NoMirror
      resourceName=../../../Engine/Content/Slate/Common/ScrollBoxShadowBottom.png
      resourceObject=null
      tiling=NoTile
      imageSize=<16 8>
      margin=<ltrb 0.5 0 0.5 1>
      outlineSettings {
       bUseBrushTransparency=false
       roundingType=HalfHeightRadius
       width=0
       color {colorUseRule=UseColor_Specified, specifiedColor=<rgba 0 0 0 0>}
       cornerRadii=<0 0 0 0>
      }
      tintColor {colorUseRule=UseColor_Specified, specifiedColor=<rgba 1 1 1 1>}
      uVRegion {bIsValid=false, max=<0 0>, min=<0 0>}
     }
     horizontalScrolledContentPadding=<ltrb 0 0 1 0>
     leftShadowBrush {
      bIsDynamicallyLoaded=false
      drawAs=Box
      imageType=FullColor
      mirroring=NoMirror
      resourceName=../../../Engine/Content/Slate/Common/ScrollBoxShadowLeft.png
      resourceObject=null
      tiling=NoTile
      imageSize=<8 16>
      margin=<ltrb 1 0.5 0 0.5>
      outlineSettings {
       bUseBrushTransparency=false
       roundingType=HalfHeightRadius
       width=0
       color {colorUseRule=UseColor_Specified, specifiedColor=<rgba 0 0 0 0>}
       cornerRadii=<0 0 0 0>
      }
      tintColor {colorUseRule=UseColor_Specified, specifiedColor=<rgba 1 1 1 1>}
      uVRegion {bIsValid=false, max=<0 0>, min=<0 0>}
     }
     rightShadowBrush {
      bIsDynamicallyLoaded=false
      drawAs=Box
      imageType=FullColor
      mirroring=NoMirror
      resourceName=../../../Engine/Content/Slate/Common/ScrollBoxShadowRight.png
      resourceObject=null
      tiling=NoTile
      imageSize=<8 16>
      margin=<ltrb 0 0.5 1 0.5>
      outlineSettings {
       bUseBrushTransparency=false
       roundingType=HalfHeightRadius
       width=0
       color {colorUseRule=UseColor_Specified, specifiedColor=<rgba 0 0 0 0>}
       cornerRadii=<0 0 0 0>
      }
      tintColor {colorUseRule=UseColor_Specified, specifiedColor=<rgba 1 1 1 1>}
      uVRegion {bIsValid=false, max=<0 0>, min=<0 0>}
     }
     topShadowBrush {
      bIsDynamicallyLoaded=false
      drawAs=Box
      imageType=FullColor
      mirroring=NoMirror
      resourceName=../../../Engine/Content/Slate/Common/ScrollBoxShadowTop.png
      resourceObject=null
      tiling=NoTile
      imageSize=<16 8>
      margin=<ltrb 0.5 1 0.5 0>
      outlineSettings {
       bUseBrushTransparency=false
       roundingType=HalfHeightRadius
       width=0
       color {colorUseRule=UseColor_Specified, specifiedColor=<rgba 0 0 0 0>}
       cornerRadii=<0 0 0 0>
      }
      tintColor {colorUseRule=UseColor_Specified, specifiedColor=<rgba 1 1 1 1>}
      uVRegion {bIsValid=false, max=<0 0>, min=<0 0>}
     }
     verticalScrolledContentPadding=<ltrb 0 0 0 1>
    }
   }
   slot {class=VerticalBoxSlot, properties {size {sizeRule=Fill, value=1}}}
  }
  MusicEditorPIEView {
   widgetClassPath=&CommonListView
   bIsVariable=true
   parent=ScrollBox_3
   widget=MusicEditorPIEView
   properties {
    entryWidgetClass=&EUW_MusicPartyEditor_MusicPropEditEntry_C
    selectionMode=None
    scrollBarStyle {
     thickness=0
     draggedThumbImage {
      bIsDynamicallyLoaded=false
      drawAs=RoundedBox
      imageType=NoImage
      mirroring=NoMirror
      resourceName=None
      resourceObject=null
      tiling=NoTile
      imageSize=<0 0>
      margin=<ltrb 0 0 0 0>
      outlineSettings {
       bUseBrushTransparency=false
       roundingType=FixedRadius
       width=0
       color {colorUseRule=UseColor_Specified, specifiedColor=<rgba 0 0 0 0>}
       cornerRadii=<4 4 4 4>
      }
      tintColor {colorUseRule=UseColor_Specified, specifiedColor=<rgba 0.2159 0.2159 0.2159 1>}
      uVRegion {bIsValid=false, max=<0 0>, min=<0 0>}
     }
     horizontalBackgroundImage {
      bIsDynamicallyLoaded=false
      drawAs=NoDrawType
      imageType=NoImage
      mirroring=NoMirror
      resourceName=None
      resourceObject=null
      tiling=NoTile
      imageSize=<0 0>
      margin=<ltrb 0 0 0 0>
      outlineSettings {
       bUseBrushTransparency=false
       roundingType=HalfHeightRadius
       width=0
       color {colorUseRule=UseColor_Specified, specifiedColor=<rgba 0 0 0 0>}
       cornerRadii=<0 0 0 0>
      }
      tintColor {colorUseRule=UseColor_Specified, specifiedColor=<rgba 1 1 1 1>}
      uVRegion {bIsValid=false, max=<0 0>, min=<0 0>}
     }
     horizontalBottomSlotImage {
      bIsDynamicallyLoaded=false
      drawAs=NoDrawType
      imageType=NoImage
      mirroring=NoMirror
      resourceName=None
      resourceObject=null
      tiling=NoTile
      imageSize=<0 0>
      margin=<ltrb 0 0 0 0>
      outlineSettings {
       bUseBrushTransparency=false
       roundingType=HalfHeightRadius
       width=0
       color {colorUseRule=UseColor_Specified, specifiedColor=<rgba 0 0 0 0>}
       cornerRadii=<0 0 0 0>
      }
      tintColor {colorUseRule=UseColor_Specified, specifiedColor=<rgba 1 1 1 1>}
      uVRegion {bIsValid=false, max=<0 0>, min=<0 0>}
     }
     horizontalTopSlotImage {
      bIsDynamicallyLoaded=false
      drawAs=NoDrawType
      imageType=NoImage
      mirroring=NoMirror
      resourceName=None
      resourceObject=null
      tiling=NoTile
      imageSize=<0 0>
      margin=<ltrb 0 0 0 0>
      outlineSettings {
       bUseBrushTransparency=false
       roundingType=HalfHeightRadius
       width=0
       color {colorUseRule=UseColor_Specified, specifiedColor=<rgba 0 0 0 0>}
       cornerRadii=<0 0 0 0>
      }
      tintColor {colorUseRule=UseColor_Specified, specifiedColor=<rgba 1 1 1 1>}
      uVRegion {bIsValid=false, max=<0 0>, min=<0 0>}
     }
     hoveredThumbImage {
      bIsDynamicallyLoaded=false
      drawAs=RoundedBox
      imageType=NoImage
      mirroring=NoMirror
      resourceName=None
      resourceObject=null
      tiling=NoTile
      imageSize=<0 0>
      margin=<ltrb 0 0 0 0>
      outlineSettings {
       bUseBrushTransparency=false
       roundingType=FixedRadius
       width=0
       color {colorUseRule=UseColor_Specified, specifiedColor=<rgba 0 0 0 0>}
       cornerRadii=<4 4 4 4>
      }
      tintColor {colorUseRule=UseColor_Specified, specifiedColor=<rgba 0.2159 0.2159 0.2159 1>}
      uVRegion {bIsValid=false, max=<0 0>, min=<0 0>}
     }
     normalThumbImage {
      bIsDynamicallyLoaded=false
      drawAs=RoundedBox
      imageType=NoImage
      mirroring=NoMirror
      resourceName=None
      resourceObject=null
      tiling=NoTile
      imageSize=<0 0>
      margin=<ltrb 0 0 0 0>
      outlineSettings {
       bUseBrushTransparency=false
       roundingType=FixedRadius
       width=0
       color {colorUseRule=UseColor_Specified, specifiedColor=<rgba 0 0 0 0>}
       cornerRadii=<4 4 4 4>
      }
      tintColor {colorUseRule=UseColor_Specified, specifiedColor=<rgba 0.0953 0.0953 0.0953 1>}
      uVRegion {bIsValid=false, max=<0 0>, min=<0 0>}
     }
     verticalBackgroundImage {
      bIsDynamicallyLoaded=false
      drawAs=NoDrawType
      imageType=NoImage
      mirroring=NoMirror
      resourceName=None
      resourceObject=null
      tiling=NoTile
      imageSize=<0 0>
      margin=<ltrb 0 0 0 0>
      outlineSettings {
       bUseBrushTransparency=false
       roundingType=HalfHeightRadius
       width=0
       color {colorUseRule=UseColor_Specified, specifiedColor=<rgba 0 0 0 0>}
       cornerRadii=<0 0 0 0>
      }
      tintColor {colorUseRule=UseColor_Specified, specifiedColor=<rgba 1 1 1 1>}
      uVRegion {bIsValid=false, max=<0 0>, min=<0 0>}
     }
     verticalBottomSlotImage {
      bIsDynamicallyLoaded=false
      drawAs=NoDrawType
      imageType=NoImage
      mirroring=NoMirror
      resourceName=None
      resourceObject=null
      tiling=NoTile
      imageSize=<0 0>
      margin=<ltrb 0 0 0 0>
      outlineSettings {
       bUseBrushTransparency=false
       roundingType=HalfHeightRadius
       width=0
       color {colorUseRule=UseColor_Specified, specifiedColor=<rgba 0 0 0 0>}
       cornerRadii=<0 0 0 0>
      }
      tintColor {colorUseRule=UseColor_Specified, specifiedColor=<rgba 1 1 1 1>}
      uVRegion {bIsValid=false, max=<0 0>, min=<0 0>}
     }
     verticalTopSlotImage {
      bIsDynamicallyLoaded=false
      drawAs=NoDrawType
      imageType=NoImage
      mirroring=NoMirror
      resourceName=None
      resourceObject=null
      tiling=NoTile
      imageSize=<0 0>
      margin=<ltrb 0 0 0 0>
      outlineSettings {
       bUseBrushTransparency=false
       roundingType=HalfHeightRadius
       width=0
       color {colorUseRule=UseColor_Specified, specifiedColor=<rgba 0 0 0 0>}
       cornerRadii=<0 0 0 0>
      }
      tintColor {colorUseRule=UseColor_Specified, specifiedColor=<rgba 1 1 1 1>}
      uVRegion {bIsValid=false, max=<0 0>, min=<0 0>}
     }
    }
   }
   slot {class=ScrollBoxSlot, properties {padding=<ltrb 0 10 10 10>, size {sizeRule=Fill, value=1}}}
  }
  1_NewMusicPropComponents {widgetClassPath=&VerticalBox, parent=WidgetSwitcher, widget=1_NewMusicPropComponents, slot {class=WidgetSwitcherSlot}}
  MusicPropAction_View {
   widgetClassPath=&ScrollBox
   parent=1_NewMusicPropComponents
   widget=MusicPropAction_View
   properties {
    widgetBarStyle {
     thickness=8
     draggedThumbImage {
      bIsDynamicallyLoaded=false
      drawAs=RoundedBox
      imageType=NoImage
      mirroring=NoMirror
      resourceName=None
      resourceObject=null
      tiling=NoTile
      imageSize=<0 0>
      margin=<ltrb 0 0 0 0>
      outlineSettings {
       bUseBrushTransparency=false
       roundingType=FixedRadius
       width=0
       color {colorUseRule=UseColor_Specified, specifiedColor=<rgba 0 0 0 0>}
       cornerRadii=<4 4 4 4>
      }
      tintColor {colorUseRule=UseColor_Specified, specifiedColor=<rgba 0.2159 0.2159 0.2159 1>}
      uVRegion {bIsValid=false, max=<0 0>, min=<0 0>}
     }
     horizontalBackgroundImage {
      bIsDynamicallyLoaded=false
      drawAs=NoDrawType
      imageType=NoImage
      mirroring=NoMirror
      resourceName=None
      resourceObject=null
      tiling=NoTile
      imageSize=<0 0>
      margin=<ltrb 0 0 0 0>
      outlineSettings {
       bUseBrushTransparency=false
       roundingType=HalfHeightRadius
       width=0
       color {colorUseRule=UseColor_Specified, specifiedColor=<rgba 0 0 0 0>}
       cornerRadii=<0 0 0 0>
      }
      tintColor {colorUseRule=UseColor_Specified, specifiedColor=<rgba 1 1 1 1>}
      uVRegion {bIsValid=false, max=<0 0>, min=<0 0>}
     }
     horizontalBottomSlotImage {
      bIsDynamicallyLoaded=false
      drawAs=NoDrawType
      imageType=NoImage
      mirroring=NoMirror
      resourceName=None
      resourceObject=null
      tiling=NoTile
      imageSize=<0 0>
      margin=<ltrb 0 0 0 0>
      outlineSettings {
       bUseBrushTransparency=false
       roundingType=HalfHeightRadius
       width=0
       color {colorUseRule=UseColor_Specified, specifiedColor=<rgba 0 0 0 0>}
       cornerRadii=<0 0 0 0>
      }
      tintColor {colorUseRule=UseColor_Specified, specifiedColor=<rgba 1 1 1 1>}
      uVRegion {bIsValid=false, max=<0 0>, min=<0 0>}
     }
     horizontalTopSlotImage {
      bIsDynamicallyLoaded=false
      drawAs=NoDrawType
      imageType=NoImage
      mirroring=NoMirror
      resourceName=None
      resourceObject=null
      tiling=NoTile
      imageSize=<0 0>
      margin=<ltrb 0 0 0 0>
      outlineSettings {
       bUseBrushTransparency=false
       roundingType=HalfHeightRadius
       width=0
       color {colorUseRule=UseColor_Specified, specifiedColor=<rgba 0 0 0 0>}
       cornerRadii=<0 0 0 0>
      }
      tintColor {colorUseRule=UseColor_Specified, specifiedColor=<rgba 1 1 1 1>}
      uVRegion {bIsValid=false, max=<0 0>, min=<0 0>}
     }
     hoveredThumbImage {
      bIsDynamicallyLoaded=false
      drawAs=RoundedBox
      imageType=NoImage
      mirroring=NoMirror
      resourceName=None
      resourceObject=null
      tiling=NoTile
      imageSize=<0 0>
      margin=<ltrb 0 0 0 0>
      outlineSettings {
       bUseBrushTransparency=false
       roundingType=FixedRadius
       width=0
       color {colorUseRule=UseColor_Specified, specifiedColor=<rgba 0 0 0 0>}
       cornerRadii=<4 4 4 4>
      }
      tintColor {colorUseRule=UseColor_Specified, specifiedColor=<rgba 0.2159 0.2159 0.2159 1>}
      uVRegion {bIsValid=false, max=<0 0>, min=<0 0>}
     }
     normalThumbImage {
      bIsDynamicallyLoaded=false
      drawAs=RoundedBox
      imageType=NoImage
      mirroring=NoMirror
      resourceName=None
      resourceObject=null
      tiling=NoTile
      imageSize=<0 0>
      margin=<ltrb 0 0 0 0>
      outlineSettings {
       bUseBrushTransparency=false
       roundingType=FixedRadius
       width=0
       color {colorUseRule=UseColor_Specified, specifiedColor=<rgba 0 0 0 0>}
       cornerRadii=<4 4 4 4>
      }
      tintColor {colorUseRule=UseColor_Specified, specifiedColor=<rgba 0.0953 0.0953 0.0953 1>}
      uVRegion {bIsValid=false, max=<0 0>, min=<0 0>}
     }
     verticalBackgroundImage {
      bIsDynamicallyLoaded=false
      drawAs=NoDrawType
      imageType=NoImage
      mirroring=NoMirror
      resourceName=None
      resourceObject=null
      tiling=NoTile
      imageSize=<0 0>
      margin=<ltrb 0 0 0 0>
      outlineSettings {
       bUseBrushTransparency=false
       roundingType=HalfHeightRadius
       width=0
       color {colorUseRule=UseColor_Specified, specifiedColor=<rgba 0 0 0 0>}
       cornerRadii=<0 0 0 0>
      }
      tintColor {colorUseRule=UseColor_Specified, specifiedColor=<rgba 1 1 1 1>}
      uVRegion {bIsValid=false, max=<0 0>, min=<0 0>}
     }
     verticalBottomSlotImage {
      bIsDynamicallyLoaded=false
      drawAs=NoDrawType
      imageType=NoImage
      mirroring=NoMirror
      resourceName=None
      resourceObject=null
      tiling=NoTile
      imageSize=<0 0>
      margin=<ltrb 0 0 0 0>
      outlineSettings {
       bUseBrushTransparency=false
       roundingType=HalfHeightRadius
       width=0
       color {colorUseRule=UseColor_Specified, specifiedColor=<rgba 0 0 0 0>}
       cornerRadii=<0 0 0 0>
      }
      tintColor {colorUseRule=UseColor_Specified, specifiedColor=<rgba 1 1 1 1>}
      uVRegion {bIsValid=false, max=<0 0>, min=<0 0>}
     }
     verticalTopSlotImage {
      bIsDynamicallyLoaded=false
      drawAs=NoDrawType
      imageType=NoImage
      mirroring=NoMirror
      resourceName=None
      resourceObject=null
      tiling=NoTile
      imageSize=<0 0>
      margin=<ltrb 0 0 0 0>
      outlineSettings {
       bUseBrushTransparency=false
       roundingType=HalfHeightRadius
       width=0
       color {colorUseRule=UseColor_Specified, specifiedColor=<rgba 0 0 0 0>}
       cornerRadii=<0 0 0 0>
      }
      tintColor {colorUseRule=UseColor_Specified, specifiedColor=<rgba 1 1 1 1>}
      uVRegion {bIsValid=false, max=<0 0>, min=<0 0>}
     }
    }
    widgetStyle {
     barThickness=8
     bottomShadowBrush {
      bIsDynamicallyLoaded=false
      drawAs=Box
      imageType=FullColor
      mirroring=NoMirror
      resourceName=../../../Engine/Content/Slate/Common/ScrollBoxShadowBottom.png
      resourceObject=null
      tiling=NoTile
      imageSize=<16 8>
      margin=<ltrb 0.5 0 0.5 1>
      outlineSettings {
       bUseBrushTransparency=false
       roundingType=HalfHeightRadius
       width=0
       color {colorUseRule=UseColor_Specified, specifiedColor=<rgba 0 0 0 0>}
       cornerRadii=<0 0 0 0>
      }
      tintColor {colorUseRule=UseColor_Specified, specifiedColor=<rgba 1 1 1 1>}
      uVRegion {bIsValid=false, max=<0 0>, min=<0 0>}
     }
     horizontalScrolledContentPadding=<ltrb 0 0 1 0>
     leftShadowBrush {
      bIsDynamicallyLoaded=false
      drawAs=Box
      imageType=FullColor
      mirroring=NoMirror
      resourceName=../../../Engine/Content/Slate/Common/ScrollBoxShadowLeft.png
      resourceObject=null
      tiling=NoTile
      imageSize=<8 16>
      margin=<ltrb 1 0.5 0 0.5>
      outlineSettings {
       bUseBrushTransparency=false
       roundingType=HalfHeightRadius
       width=0
       color {colorUseRule=UseColor_Specified, specifiedColor=<rgba 0 0 0 0>}
       cornerRadii=<0 0 0 0>
      }
      tintColor {colorUseRule=UseColor_Specified, specifiedColor=<rgba 1 1 1 1>}
      uVRegion {bIsValid=false, max=<0 0>, min=<0 0>}
     }
     rightShadowBrush {
      bIsDynamicallyLoaded=false
      drawAs=Box
      imageType=FullColor
      mirroring=NoMirror
      resourceName=../../../Engine/Content/Slate/Common/ScrollBoxShadowRight.png
      resourceObject=null
      tiling=NoTile
      imageSize=<8 16>
      margin=<ltrb 0 0.5 1 0.5>
      outlineSettings {
       bUseBrushTransparency=false
       roundingType=HalfHeightRadius
       width=0
       color {colorUseRule=UseColor_Specified, specifiedColor=<rgba 0 0 0 0>}
       cornerRadii=<0 0 0 0>
      }
      tintColor {colorUseRule=UseColor_Specified, specifiedColor=<rgba 1 1 1 1>}
      uVRegion {bIsValid=false, max=<0 0>, min=<0 0>}
     }
     topShadowBrush {
      bIsDynamicallyLoaded=false
      drawAs=Box
      imageType=FullColor
      mirroring=NoMirror
      resourceName=../../../Engine/Content/Slate/Common/ScrollBoxShadowTop.png
      resourceObject=null
      tiling=NoTile
      imageSize=<16 8>
      margin=<ltrb 0.5 1 0.5 0>
      outlineSettings {
       bUseBrushTransparency=false
       roundingType=HalfHeightRadius
       width=0
       color {colorUseRule=UseColor_Specified, specifiedColor=<rgba 0 0 0 0>}
       cornerRadii=<0 0 0 0>
      }
      tintColor {colorUseRule=UseColor_Specified, specifiedColor=<rgba 1 1 1 1>}
      uVRegion {bIsValid=false, max=<0 0>, min=<0 0>}
     }
     verticalScrolledContentPadding=<ltrb 0 0 0 1>
    }
   }
   slot {class=VerticalBoxSlot, properties {padding=<ltrb 5 5 5 5>, size {sizeRule=Fill, value=1}}}
  }
  ActionText {
   widgetClassPath=&UEFN_TextBlock
   bIsVariable=true
   parent=MusicPropAction_View
   widget=ActionText
   properties {
    clipping=ClipToBounds
    text=What Music Prop Action?
    font {
     bForceMonospaced=false
     bMaterialIsStencil=false
     fontMaterial=null
     fontObject=&NotoSans
     letterSpacing=5
     monospacedWidth=1
     size=12
     skewAmount=0
     typefaceFontName=NotoSans-Bold
     outlineSettings {
      bApplyOutlineToDropShadows=false
      bMiteredCorners=false
      bSeparateFillAlpha=false
      outlineMaterial=null
      outlineSize=0
      outlineColor=<rgba 0 0 0 1>
     }
    }
   }
   slot {class=ScrollBoxSlot, properties {horizontalAlignment=HAlign_Center, verticalAlignment=VAlign_Center}}
  }
  PropTypeComboBox {
   widgetClassPath=/Script/Blutility.EditorUtilityComboBoxString
   bIsVariable=true
   parent=MusicPropAction_View
   widget=PropTypeComboBox
   properties {
    selectedOption=Controller Component
    toolTipText="Controller Component: Default- Uses a Music Controller Base actor as its Controller. Use this when you just have a single simple loop or sound cue to play.\r\nRadio Component: Uses FortMediaStreamingRadio as its Controller\r\nActor Component: Uses an Actor as its Controller, use if you have custom Logic that requires a specific Actor Class\r\nDeprecated Component: Legacy Version that can use any Actor as its Controller. Does not work with Speaker Presets."
    defaultOptions [Controller Component, Radio Component, Actor Component, Deprecated Component]
    foregroundColor {colorUseRule=UseColor_Specified, specifiedColor=<rgba 0.5271 0.5271 0.5271 1>}
   }
   slot {class=ScrollBoxSlot}
  }
  MusicPropActionSettings {
   widgetClassPath=&DetailsView
   parent=MusicPropAction_View
   widget=MusicPropActionSettings
   properties {categoriesToShow [Assets,Behavior]}
   slot {class=ScrollBoxSlot}
  }
  Border_3 {
   widgetClassPath=&Border
   parent=MusicPropAction_View
   widget=Border_3
   properties {
    background {
     bIsDynamicallyLoaded=false
     drawAs=Box
     imageType=NoImage
     mirroring=NoMirror
     resourceName=None
     resourceObject=null
     tiling=NoTile
     imageSize=<32 32>
     margin=<ltrb 0 0 0 0>
     outlineSettings {
      bUseBrushTransparency=false
      roundingType=HalfHeightRadius
      width=0
      color {colorUseRule=UseColor_Specified, specifiedColor=<rgba 0 0 0 0>}
      cornerRadii=<0 0 0 0>
     }
     tintColor {colorUseRule=UseColor_Specified, specifiedColor=<rgba 1 1 1 1>}
     uVRegion {bIsValid=false, max=<0 0>, min=<0 0>}
    }
    brushColor=<rgba 0 0 0 0.2>
    padding=<ltrb 5 5 5 5>
   }
   slot {class=ScrollBoxSlot}
  }
  HorizontalBox_3 {
   widgetClassPath=&HorizontalBox
   parent=Border_3
   widget=HorizontalBox_3
   slot {class=BorderSlot, properties {padding=<ltrb 5 5 5 5>}}
  }
  MusicPropAction_ConfirmButton {
   widgetClassPath=&Button
   parent=HorizontalBox_3
   widget=MusicPropAction_ConfirmButton
   properties {widgetStyle {
     clickedSlateSound {resourceObject=null}
     disabled {
      bIsDynamicallyLoaded=false
      drawAs=Box
      imageType=NoImage
      mirroring=NoMirror
      resourceName=None
      resourceObject=null
      tiling=NoTile
      imageSize=<0 0>
      margin=<ltrb 0 0 0 0>
      outlineSettings {
       bUseBrushTransparency=false
       roundingType=FixedRadius
       width=1
       color {colorUseRule=UseColor_ColorTable, specifiedColor=<rgba 0 0 0 0>}
       cornerRadii=<4 4 4 4>
      }
      tintColor {colorUseRule=UseColor_Specified, specifiedColor=<rgba 0.0395 0.0395 0.0395 1>}
      uVRegion {bIsValid=false, max=<0 0>, min=<0 0>}
     }
     disabledForeground {colorUseRule=UseColor_ColorTable, specifiedColor=<rgba 0 0 0 0>}
     hovered {
      bIsDynamicallyLoaded=false
      drawAs=Box
      imageType=NoImage
      mirroring=NoMirror
      resourceName=None
      resourceObject=null
      tiling=NoTile
      imageSize=<32 32>
      margin=<ltrb 0 0 0 0>
      outlineSettings {
       bUseBrushTransparency=false
       roundingType=FixedRadius
       width=1
       color {colorUseRule=UseColor_ColorTable, specifiedColor=<rgba 0 0 0 0>}
       cornerRadii=<4 4 4 4>
      }
      tintColor {colorUseRule=UseColor_Specified, specifiedColor=<rgba 0.0953 0.0953 0.0953 1>}
      uVRegion {bIsValid=false, max=<0 0>, min=<0 0>}
     }
     hoveredForeground {colorUseRule=UseColor_ColorTable, specifiedColor=<rgba 0 0 0 0>}
     hoveredSlateSound {resourceObject=null}
     normal {
      bIsDynamicallyLoaded=false
      drawAs=Box
      imageType=NoImage
      mirroring=NoMirror
      resourceName=None
      resourceObject=null
      tiling=NoTile
      imageSize=<32 32>
      margin=<ltrb 0 0 0 0>
      outlineSettings {
       bUseBrushTransparency=false
       roundingType=FixedRadius
       width=1
       color {colorUseRule=UseColor_ColorTable, specifiedColor=<rgba 0 0 0 0>}
       cornerRadii=<4 4 4 4>
      }
      tintColor {colorUseRule=UseColor_Specified, specifiedColor=<rgba 0.0284 0.0284 0.0284 1>}
      uVRegion {bIsValid=false, max=<0 0>, min=<0 0>}
     }
     normalForeground {colorUseRule=UseColor_ColorTable, specifiedColor=<rgba 0 0 0 0>}
     normalPadding=<ltrb 12 1.5 12 1.5>
     pressed {
      bIsDynamicallyLoaded=false
      drawAs=Box
      imageType=NoImage
      mirroring=NoMirror
      resourceName=None
      resourceObject=null
      tiling=NoTile
      imageSize=<32 32>
      margin=<ltrb 0 0 0 0>
      outlineSettings {
       bUseBrushTransparency=false
       roundingType=FixedRadius
       width=1
       color {colorUseRule=UseColor_ColorTable, specifiedColor=<rgba 0 0 0 0>}
       cornerRadii=<4 4 4 4>
      }
      tintColor {colorUseRule=UseColor_Specified, specifiedColor=<rgba 0.0284 0.0284 0.0284 1>}
      uVRegion {bIsValid=false, max=<0 0>, min=<0 0>}
     }
     pressedForeground {colorUseRule=UseColor_ColorTable, specifiedColor=<rgba 0 0 0 0>}
     pressedPadding=<ltrb 12 2.5 12 0.5>
     pressedSlateSound {resourceObject=null}
    }}
   slot {class=HorizontalBoxSlot, properties {padding=<ltrb 0.5 0 0.5 0>, size {sizeRule=Fill, value=1}}}
  }
  Text_1 {
   widgetClassPath=&UEFN_TextBlock
   parent=MusicPropAction_ConfirmButton
   widget=Text_1
   properties {
    clipping=ClipToBounds
    text=Confirm
    font {
     bForceMonospaced=false
     bMaterialIsStencil=false
     fontMaterial=null
     fontObject=&NotoSans
     letterSpacing=5
     monospacedWidth=1
     size=10
     skewAmount=0
     typefaceFontName=NotoSans-Regular
     outlineSettings {
      bApplyOutlineToDropShadows=false
      bMiteredCorners=false
      bSeparateFillAlpha=false
      outlineMaterial=null
      outlineSize=0
      outlineColor=<rgba 0 0 0 1>
     }
    }
   }
   slot {class=ButtonSlot, properties {padding=<ltrb 0 0 0 0>}}
  }
  MusicPropAction_CancelButton {
   widgetClassPath=&Button
   parent=HorizontalBox_3
   widget=MusicPropAction_CancelButton
   properties {widgetStyle {
     clickedSlateSound {resourceObject=null}
     disabled {
      bIsDynamicallyLoaded=false
      drawAs=Box
      imageType=NoImage
      mirroring=NoMirror
      resourceName=None
      resourceObject=null
      tiling=NoTile
      imageSize=<0 0>
      margin=<ltrb 0 0 0 0>
      outlineSettings {
       bUseBrushTransparency=false
       roundingType=FixedRadius
       width=1
       color {colorUseRule=UseColor_ColorTable, specifiedColor=<rgba 0 0 0 0>}
       cornerRadii=<4 4 4 4>
      }
      tintColor {colorUseRule=UseColor_Specified, specifiedColor=<rgba 0.0395 0.0395 0.0395 1>}
      uVRegion {bIsValid=false, max=<0 0>, min=<0 0>}
     }
     disabledForeground {colorUseRule=UseColor_ColorTable, specifiedColor=<rgba 0 0 0 0>}
     hovered {
      bIsDynamicallyLoaded=false
      drawAs=Box
      imageType=NoImage
      mirroring=NoMirror
      resourceName=None
      resourceObject=null
      tiling=NoTile
      imageSize=<32 32>
      margin=<ltrb 0 0 0 0>
      outlineSettings {
       bUseBrushTransparency=false
       roundingType=FixedRadius
       width=1
       color {colorUseRule=UseColor_ColorTable, specifiedColor=<rgba 0 0 0 0>}
       cornerRadii=<4 4 4 4>
      }
      tintColor {colorUseRule=UseColor_Specified, specifiedColor=<rgba 0.0953 0.0953 0.0953 1>}
      uVRegion {bIsValid=false, max=<0 0>, min=<0 0>}
     }
     hoveredForeground {colorUseRule=UseColor_ColorTable, specifiedColor=<rgba 0 0 0 0>}
     hoveredSlateSound {resourceObject=null}
     normal {
      bIsDynamicallyLoaded=false
      drawAs=Box
      imageType=NoImage
      mirroring=NoMirror
      resourceName=None
      resourceObject=null
      tiling=NoTile
      imageSize=<32 32>
      margin=<ltrb 0 0 0 0>
      outlineSettings {
       bUseBrushTransparency=false
       roundingType=FixedRadius
       width=1
       color {colorUseRule=UseColor_ColorTable, specifiedColor=<rgba 0 0 0 0>}
       cornerRadii=<4 4 4 4>
      }
      tintColor {colorUseRule=UseColor_Specified, specifiedColor=<rgba 0.0284 0.0284 0.0284 1>}
      uVRegion {bIsValid=false, max=<0 0>, min=<0 0>}
     }
     normalForeground {colorUseRule=UseColor_ColorTable, specifiedColor=<rgba 0 0 0 0>}
     normalPadding=<ltrb 12 1.5 12 1.5>
     pressed {
      bIsDynamicallyLoaded=false
      drawAs=Box
      imageType=NoImage
      mirroring=NoMirror
      resourceName=None
      resourceObject=null
      tiling=NoTile
      imageSize=<32 32>
      margin=<ltrb 0 0 0 0>
      outlineSettings {
       bUseBrushTransparency=false
       roundingType=FixedRadius
       width=1
       color {colorUseRule=UseColor_ColorTable, specifiedColor=<rgba 0 0 0 0>}
       cornerRadii=<4 4 4 4>
      }
      tintColor {colorUseRule=UseColor_Specified, specifiedColor=<rgba 0.0284 0.0284 0.0284 1>}
      uVRegion {bIsValid=false, max=<0 0>, min=<0 0>}
     }
     pressedForeground {colorUseRule=UseColor_ColorTable, specifiedColor=<rgba 0 0 0 0>}
     pressedPadding=<ltrb 12 2.5 12 0.5>
     pressedSlateSound {resourceObject=null}
    }}
   slot {class=HorizontalBoxSlot, properties {padding=<ltrb 0.5 0 0.5 0>, size {sizeRule=Fill, value=1}}}
  }
  Text_3 {
   widgetClassPath=&UEFN_TextBlock
   parent=MusicPropAction_CancelButton
   widget=Text_3
   properties {
    clipping=ClipToBounds
    text=Cancel
    font {
     bForceMonospaced=false
     bMaterialIsStencil=false
     fontMaterial=null
     fontObject=&NotoSans
     letterSpacing=5
     monospacedWidth=1
     size=10
     skewAmount=0
     typefaceFontName=NotoSans-Regular
     outlineSettings {
      bApplyOutlineToDropShadows=false
      bMiteredCorners=false
      bSeparateFillAlpha=false
      outlineMaterial=null
      outlineSize=0
      outlineColor=<rgba 0 0 0 1>
     }
    }
   }
   slot {class=ButtonSlot, properties {padding=<ltrb 0 0 0 0>}}
  }
  2_CreateNewController {widgetClassPath=&VerticalBox, parent=WidgetSwitcher, widget=2_CreateNewController, slot {class=WidgetSwitcherSlot}}
  ActionView_1 {
   widgetClassPath=&ScrollBox
   parent=2_CreateNewController
   widget=ActionView_1
   properties {
    widgetBarStyle {
     thickness=8
     draggedThumbImage {
      bIsDynamicallyLoaded=false
      drawAs=RoundedBox
      imageType=NoImage
      mirroring=NoMirror
      resourceName=None
      resourceObject=null
      tiling=NoTile
      imageSize=<0 0>
      margin=<ltrb 0 0 0 0>
      outlineSettings {
       bUseBrushTransparency=false
       roundingType=FixedRadius
       width=0
       color {colorUseRule=UseColor_Specified, specifiedColor=<rgba 0 0 0 0>}
       cornerRadii=<4 4 4 4>
      }
      tintColor {colorUseRule=UseColor_Specified, specifiedColor=<rgba 0.2159 0.2159 0.2159 1>}
      uVRegion {bIsValid=false, max=<0 0>, min=<0 0>}
     }
     horizontalBackgroundImage {
      bIsDynamicallyLoaded=false
      drawAs=NoDrawType
      imageType=NoImage
      mirroring=NoMirror
      resourceName=None
      resourceObject=null
      tiling=NoTile
      imageSize=<0 0>
      margin=<ltrb 0 0 0 0>
      outlineSettings {
       bUseBrushTransparency=false
       roundingType=HalfHeightRadius
       width=0
       color {colorUseRule=UseColor_Specified, specifiedColor=<rgba 0 0 0 0>}
       cornerRadii=<0 0 0 0>
      }
      tintColor {colorUseRule=UseColor_Specified, specifiedColor=<rgba 1 1 1 1>}
      uVRegion {bIsValid=false, max=<0 0>, min=<0 0>}
     }
     horizontalBottomSlotImage {
      bIsDynamicallyLoaded=false
      drawAs=NoDrawType
      imageType=NoImage
      mirroring=NoMirror
      resourceName=None
      resourceObject=null
      tiling=NoTile
      imageSize=<0 0>
      margin=<ltrb 0 0 0 0>
      outlineSettings {
       bUseBrushTransparency=false
       roundingType=HalfHeightRadius
       width=0
       color {colorUseRule=UseColor_Specified, specifiedColor=<rgba 0 0 0 0>}
       cornerRadii=<0 0 0 0>
      }
      tintColor {colorUseRule=UseColor_Specified, specifiedColor=<rgba 1 1 1 1>}
      uVRegion {bIsValid=false, max=<0 0>, min=<0 0>}
     }
     horizontalTopSlotImage {
      bIsDynamicallyLoaded=false
      drawAs=NoDrawType
      imageType=NoImage
      mirroring=NoMirror
      resourceName=None
      resourceObject=null
      tiling=NoTile
      imageSize=<0 0>
      margin=<ltrb 0 0 0 0>
      outlineSettings {
       bUseBrushTransparency=false
       roundingType=HalfHeightRadius
       width=0
       color {colorUseRule=UseColor_Specified, specifiedColor=<rgba 0 0 0 0>}
       cornerRadii=<0 0 0 0>
      }
      tintColor {colorUseRule=UseColor_Specified, specifiedColor=<rgba 1 1 1 1>}
      uVRegion {bIsValid=false, max=<0 0>, min=<0 0>}
     }
     hoveredThumbImage {
      bIsDynamicallyLoaded=false
      drawAs=RoundedBox
      imageType=NoImage
      mirroring=NoMirror
      resourceName=None
      resourceObject=null
      tiling=NoTile
      imageSize=<0 0>
      margin=<ltrb 0 0 0 0>
      outlineSettings {
       bUseBrushTransparency=false
       roundingType=FixedRadius
       width=0
       color {colorUseRule=UseColor_Specified, specifiedColor=<rgba 0 0 0 0>}
       cornerRadii=<4 4 4 4>
      }
      tintColor {colorUseRule=UseColor_Specified, specifiedColor=<rgba 0.2159 0.2159 0.2159 1>}
      uVRegion {bIsValid=false, max=<0 0>, min=<0 0>}
     }
     normalThumbImage {
      bIsDynamicallyLoaded=false
      drawAs=RoundedBox
      imageType=NoImage
      mirroring=NoMirror
      resourceName=None
      resourceObject=null
      tiling=NoTile
      imageSize=<0 0>
      margin=<ltrb 0 0 0 0>
      outlineSettings {
       bUseBrushTransparency=false
       roundingType=FixedRadius
       width=0
       color {colorUseRule=UseColor_Specified, specifiedColor=<rgba 0 0 0 0>}
       cornerRadii=<4 4 4 4>
      }
      tintColor {colorUseRule=UseColor_Specified, specifiedColor=<rgba 0.0953 0.0953 0.0953 1>}
      uVRegion {bIsValid=false, max=<0 0>, min=<0 0>}
     }
     verticalBackgroundImage {
      bIsDynamicallyLoaded=false
      drawAs=NoDrawType
      imageType=NoImage
      mirroring=NoMirror
      resourceName=None
      resourceObject=null
      tiling=NoTile
      imageSize=<0 0>
      margin=<ltrb 0 0 0 0>
      outlineSettings {
       bUseBrushTransparency=false
       roundingType=HalfHeightRadius
       width=0
       color {colorUseRule=UseColor_Specified, specifiedColor=<rgba 0 0 0 0>}
       cornerRadii=<0 0 0 0>
      }
      tintColor {colorUseRule=UseColor_Specified, specifiedColor=<rgba 1 1 1 1>}
      uVRegion {bIsValid=false, max=<0 0>, min=<0 0>}
     }
     verticalBottomSlotImage {
      bIsDynamicallyLoaded=false
      drawAs=NoDrawType
      imageType=NoImage
      mirroring=NoMirror
      resourceName=None
      resourceObject=null
      tiling=NoTile
      imageSize=<0 0>
      margin=<ltrb 0 0 0 0>
      outlineSettings {
       bUseBrushTransparency=false
       roundingType=HalfHeightRadius
       width=0
       color {colorUseRule=UseColor_Specified, specifiedColor=<rgba 0 0 0 0>}
       cornerRadii=<0 0 0 0>
      }
      tintColor {colorUseRule=UseColor_Specified, specifiedColor=<rgba 1 1 1 1>}
      uVRegion {bIsValid=false, max=<0 0>, min=<0 0>}
     }
     verticalTopSlotImage {
      bIsDynamicallyLoaded=false
      drawAs=NoDrawType
      imageType=NoImage
      mirroring=NoMirror
      resourceName=None
      resourceObject=null
      tiling=NoTile
      imageSize=<0 0>
      margin=<ltrb 0 0 0 0>
      outlineSettings {
       bUseBrushTransparency=false
       roundingType=HalfHeightRadius
       width=0
       color {colorUseRule=UseColor_Specified, specifiedColor=<rgba 0 0 0 0>}
       cornerRadii=<0 0 0 0>
      }
      tintColor {colorUseRule=UseColor_Specified, specifiedColor=<rgba 1 1 1 1>}
      uVRegion {bIsValid=false, max=<0 0>, min=<0 0>}
     }
    }
    widgetStyle {
     barThickness=8
     bottomShadowBrush {
      bIsDynamicallyLoaded=false
      drawAs=Box
      imageType=FullColor
      mirroring=NoMirror
      resourceName=../../../Engine/Content/Slate/Common/ScrollBoxShadowBottom.png
      resourceObject=null
      tiling=NoTile
      imageSize=<16 8>
      margin=<ltrb 0.5 0 0.5 1>
      outlineSettings {
       bUseBrushTransparency=false
       roundingType=HalfHeightRadius
       width=0
       color {colorUseRule=UseColor_Specified, specifiedColor=<rgba 0 0 0 0>}
       cornerRadii=<0 0 0 0>
      }
      tintColor {colorUseRule=UseColor_Specified, specifiedColor=<rgba 1 1 1 1>}
      uVRegion {bIsValid=false, max=<0 0>, min=<0 0>}
     }
     horizontalScrolledContentPadding=<ltrb 0 0 1 0>
     leftShadowBrush {
      bIsDynamicallyLoaded=false
      drawAs=Box
      imageType=FullColor
      mirroring=NoMirror
      resourceName=../../../Engine/Content/Slate/Common/ScrollBoxShadowLeft.png
      resourceObject=null
      tiling=NoTile
      imageSize=<8 16>
      margin=<ltrb 1 0.5 0 0.5>
      outlineSettings {
       bUseBrushTransparency=false
       roundingType=HalfHeightRadius
       width=0
       color {colorUseRule=UseColor_Specified, specifiedColor=<rgba 0 0 0 0>}
       cornerRadii=<0 0 0 0>
      }
      tintColor {colorUseRule=UseColor_Specified, specifiedColor=<rgba 1 1 1 1>}
      uVRegion {bIsValid=false, max=<0 0>, min=<0 0>}
     }
     rightShadowBrush {
      bIsDynamicallyLoaded=false
      drawAs=Box
      imageType=FullColor
      mirroring=NoMirror
      resourceName=../../../Engine/Content/Slate/Common/ScrollBoxShadowRight.png
      resourceObject=null
      tiling=NoTile
      imageSize=<8 16>
      margin=<ltrb 0 0.5 1 0.5>
      outlineSettings {
       bUseBrushTransparency=false
       roundingType=HalfHeightRadius
       width=0
       color {colorUseRule=UseColor_Specified, specifiedColor=<rgba 0 0 0 0>}
       cornerRadii=<0 0 0 0>
      }
      tintColor {colorUseRule=UseColor_Specified, specifiedColor=<rgba 1 1 1 1>}
      uVRegion {bIsValid=false, max=<0 0>, min=<0 0>}
     }
     topShadowBrush {
      bIsDynamicallyLoaded=false
      drawAs=Box
      imageType=FullColor
      mirroring=NoMirror
      resourceName=../../../Engine/Content/Slate/Common/ScrollBoxShadowTop.png
      resourceObject=null
      tiling=NoTile
      imageSize=<16 8>
      margin=<ltrb 0.5 1 0.5 0>
      outlineSettings {
       bUseBrushTransparency=false
       roundingType=HalfHeightRadius
       width=0
       color {colorUseRule=UseColor_Specified, specifiedColor=<rgba 0 0 0 0>}
       cornerRadii=<0 0 0 0>
      }
      tintColor {colorUseRule=UseColor_Specified, specifiedColor=<rgba 1 1 1 1>}
      uVRegion {bIsValid=false, max=<0 0>, min=<0 0>}
     }
     verticalScrolledContentPadding=<ltrb 0 0 0 1>
    }
   }
   slot {class=VerticalBoxSlot, properties {padding=<ltrb 5 5 5 5>, size {sizeRule=Fill, value=1}}}
  }
  NewAsset_Text {
   widgetClassPath=&UEFN_TextBlock
   bIsVariable=true
   parent=ActionView_1
   widget=NewAsset_Text
   properties {
    clipping=ClipToBounds
    text=Create New Controller
    font {
     bForceMonospaced=false
     bMaterialIsStencil=false
     fontMaterial=null
     fontObject=&NotoSans
     letterSpacing=5
     monospacedWidth=1
     size=12
     skewAmount=0
     typefaceFontName=NotoSans-Bold
     outlineSettings {
      bApplyOutlineToDropShadows=false
      bMiteredCorners=false
      bSeparateFillAlpha=false
      outlineMaterial=null
      outlineSize=0
      outlineColor=<rgba 0 0 0 1>
     }
    }
   }
   slot {class=ScrollBoxSlot, properties {horizontalAlignment=HAlign_Center, verticalAlignment=VAlign_Center}}
  }
  NewAsset_Controller_ComboBox {
   widgetClassPath=&EditorUtilityComboBoxString
   bIsVariable=true
   parent=ActionView_1
   widget=NewAsset_Controller_ComboBox
   properties {
    selectedOption=Sound Wave Controller
    toolTipText="Sound Wave Controller: Random Soundwave from Array is selected and looped through a metasound at runtime.\r\n\r\nSound Cue Controller: Use for a Sound Cue that is already created.\r\n"
    defaultOptions [Sound Wave Controller, Sound Cue Controller]
    foregroundColor {colorUseRule=UseColor_Specified, specifiedColor=<rgba 0.5271 0.5271 0.5271 1>}
   }
   slot {class=ScrollBoxSlot}
  }
  NewAsset_Details {
   widgetClassPath=&DetailsView
   parent=ActionView_1
   widget=NewAsset_Details
   properties {categoriesToShow [Path, Assets, SpeakerSettings]}
   slot {class=ScrollBoxSlot}
  }
  NewAsset_AutoGenerateAudioBus_Container {
   widgetClassPath=&HorizontalBox
   bIsVariable=true
   parent=ActionView_1
   widget=NewAsset_AutoGenerateAudioBus_Container
   properties {toolTipText="Automatically creates and assigns an audio bus with this controller (otherwise you MUST manually assign one)"}
   slot {class=ScrollBoxSlot}
  }
  AutoGenerateAudioBus_Text {
   widgetClassPath=&TextBlock
   parent=NewAsset_AutoGenerateAudioBus_Container
   widget=AutoGenerateAudioBus_Text
   properties {
    text=Auto Generate Audio Bus
    font {
     bForceMonospaced=false
     bMaterialIsStencil=false
     fontMaterial=null
     fontObject=&NotoSans
     letterSpacing=5
     monospacedWidth=1
     size=7.5
     skewAmount=0
     typefaceFontName=NotoSans-Regular
     outlineSettings {
      bApplyOutlineToDropShadows=false
      bMiteredCorners=false
      bSeparateFillAlpha=false
      outlineMaterial=null
      outlineSize=0
      outlineColor=<rgba 0 0 0 1>
     }
    }
   }
   slot {class=HorizontalBoxSlot, properties {verticalAlignment=VAlign_Center, padding=<ltrb 10 0 0 0>}}
  }
  NewAsset_AutoGenerateAudioBus {
   widgetClassPath=&EditorUtilityCheckBox
   bIsVariable=true
   parent=NewAsset_AutoGenerateAudioBus_Container
   widget=NewAsset_AutoGenerateAudioBus
   properties {checkedState=Checked}
   slot {class=HorizontalBoxSlot, properties {padding=<ltrb 5 5 5 5>}}
  }
  NewAsset_ExtendedDetails {
   widgetClassPath=&DetailsView
   parent=ActionView_1
   widget=NewAsset_ExtendedDetails
   properties {categoriesToShow [AudioBus]}
   slot {class=ScrollBoxSlot}
  }
  Border_2 {
   widgetClassPath=&Border
   parent=ActionView_1
   widget=Border_2
   properties {
    background {
     bIsDynamicallyLoaded=false
     drawAs=Box
     imageType=NoImage
     mirroring=NoMirror
     resourceName=None
     resourceObject=null
     tiling=NoTile
     imageSize=<32 32>
     margin=<ltrb 0 0 0 0>
     outlineSettings {
      bUseBrushTransparency=false
      roundingType=HalfHeightRadius
      width=0
      color {colorUseRule=UseColor_Specified, specifiedColor=<rgba 0 0 0 0>}
      cornerRadii=<0 0 0 0>
     }
     tintColor {colorUseRule=UseColor_Specified, specifiedColor=<rgba 1 1 1 1>}
     uVRegion {bIsValid=false, max=<0 0>, min=<0 0>}
    }
    brushColor=<rgba 0 0 0 0.2>
    padding=<ltrb 5 5 5 5>
   }
   slot {class=ScrollBoxSlot}
  }
  HorizontalBox {
   widgetClassPath=&HorizontalBox
   parent=Border_2
   widget=HorizontalBox
   slot {class=BorderSlot, properties {padding=<ltrb 5 5 5 5>}}
  }
  NewAsset_ConfirmButton {
   widgetClassPath=&Button
   parent=HorizontalBox
   widget=NewAsset_ConfirmButton
   properties {widgetStyle {
     clickedSlateSound {resourceObject=null}
     disabled {
      bIsDynamicallyLoaded=false
      drawAs=Box
      imageType=NoImage
      mirroring=NoMirror
      resourceName=None
      resourceObject=null
      tiling=NoTile
      imageSize=<0 0>
      margin=<ltrb 0 0 0 0>
      outlineSettings {
       bUseBrushTransparency=false
       roundingType=FixedRadius
       width=1
       color {colorUseRule=UseColor_ColorTable, specifiedColor=<rgba 0 0 0 0>}
       cornerRadii=<4 4 4 4>
      }
      tintColor {colorUseRule=UseColor_Specified, specifiedColor=<rgba 0.0395 0.0395 0.0395 1>}
      uVRegion {bIsValid=false, max=<0 0>, min=<0 0>}
     }
     disabledForeground {colorUseRule=UseColor_ColorTable, specifiedColor=<rgba 0 0 0 0>}
     hovered {
      bIsDynamicallyLoaded=false
      drawAs=Box
      imageType=NoImage
      mirroring=NoMirror
      resourceName=None
      resourceObject=null
      tiling=NoTile
      imageSize=<32 32>
      margin=<ltrb 0 0 0 0>
      outlineSettings {
       bUseBrushTransparency=false
       roundingType=FixedRadius
       width=1
       color {colorUseRule=UseColor_ColorTable, specifiedColor=<rgba 0 0 0 0>}
       cornerRadii=<4 4 4 4>
      }
      tintColor {colorUseRule=UseColor_Specified, specifiedColor=<rgba 0.0953 0.0953 0.0953 1>}
      uVRegion {bIsValid=false, max=<0 0>, min=<0 0>}
     }
     hoveredForeground {colorUseRule=UseColor_ColorTable, specifiedColor=<rgba 0 0 0 0>}
     hoveredSlateSound {resourceObject=null}
     normal {
      bIsDynamicallyLoaded=false
      drawAs=Box
      imageType=NoImage
      mirroring=NoMirror
      resourceName=None
      resourceObject=null
      tiling=NoTile
      imageSize=<32 32>
      margin=<ltrb 0 0 0 0>
      outlineSettings {
       bUseBrushTransparency=false
       roundingType=FixedRadius
       width=1
       color {colorUseRule=UseColor_ColorTable, specifiedColor=<rgba 0 0 0 0>}
       cornerRadii=<4 4 4 4>
      }
      tintColor {colorUseRule=UseColor_Specified, specifiedColor=<rgba 0.0284 0.0284 0.0284 1>}
      uVRegion {bIsValid=false, max=<0 0>, min=<0 0>}
     }
     normalForeground {colorUseRule=UseColor_ColorTable, specifiedColor=<rgba 0 0 0 0>}
     normalPadding=<ltrb 12 1.5 12 1.5>
     pressed {
      bIsDynamicallyLoaded=false
      drawAs=Box
      imageType=NoImage
      mirroring=NoMirror
      resourceName=None
      resourceObject=null
      tiling=NoTile
      imageSize=<32 32>
      margin=<ltrb 0 0 0 0>
      outlineSettings {
       bUseBrushTransparency=false
       roundingType=FixedRadius
       width=1
       color {colorUseRule=UseColor_ColorTable, specifiedColor=<rgba 0 0 0 0>}
       cornerRadii=<4 4 4 4>
      }
      tintColor {colorUseRule=UseColor_Specified, specifiedColor=<rgba 0.0284 0.0284 0.0284 1>}
      uVRegion {bIsValid=false, max=<0 0>, min=<0 0>}
     }
     pressedForeground {colorUseRule=UseColor_ColorTable, specifiedColor=<rgba 0 0 0 0>}
     pressedPadding=<ltrb 12 2.5 12 0.5>
     pressedSlateSound {resourceObject=null}
    }}
   slot {class=HorizontalBoxSlot, properties {padding=<ltrb 0.5 0 0.5 0>, size {sizeRule=Fill, value=1}}}
  }
  Text_7 {
   widgetClassPath=&UEFN_TextBlock
   parent=NewAsset_ConfirmButton
   widget=Text_7
   properties {
    clipping=ClipToBounds
    text=Confirm
    font {
     bForceMonospaced=false
     bMaterialIsStencil=false
     fontMaterial=null
     fontObject=&NotoSans
     letterSpacing=5
     monospacedWidth=1
     size=10
     skewAmount=0
     typefaceFontName=NotoSans-Regular
     outlineSettings {
      bApplyOutlineToDropShadows=false
      bMiteredCorners=false
      bSeparateFillAlpha=false
      outlineMaterial=null
      outlineSize=0
      outlineColor=<rgba 0 0 0 1>
     }
    }
   }
   slot {class=ButtonSlot, properties {padding=<ltrb 0 0 0 0>}}
  }
  NewAsset_CancelButton {
   widgetClassPath=&Button
   parent=HorizontalBox
   widget=NewAsset_CancelButton
   properties {widgetStyle {
     clickedSlateSound {resourceObject=null}
     disabled {
      bIsDynamicallyLoaded=false
      drawAs=Box
      imageType=NoImage
      mirroring=NoMirror
      resourceName=None
      resourceObject=null
      tiling=NoTile
      imageSize=<0 0>
      margin=<ltrb 0 0 0 0>
      outlineSettings {
       bUseBrushTransparency=false
       roundingType=FixedRadius
       width=1
       color {colorUseRule=UseColor_ColorTable, specifiedColor=<rgba 0 0 0 0>}
       cornerRadii=<4 4 4 4>
      }
      tintColor {colorUseRule=UseColor_Specified, specifiedColor=<rgba 0.0395 0.0395 0.0395 1>}
      uVRegion {bIsValid=false, max=<0 0>, min=<0 0>}
     }
     disabledForeground {colorUseRule=UseColor_ColorTable, specifiedColor=<rgba 0 0 0 0>}
     hovered {
      bIsDynamicallyLoaded=false
      drawAs=Box
      imageType=NoImage
      mirroring=NoMirror
      resourceName=None
      resourceObject=null
      tiling=NoTile
      imageSize=<32 32>
      margin=<ltrb 0 0 0 0>
      outlineSettings {
       bUseBrushTransparency=false
       roundingType=FixedRadius
       width=1
       color {colorUseRule=UseColor_ColorTable, specifiedColor=<rgba 0 0 0 0>}
       cornerRadii=<4 4 4 4>
      }
      tintColor {colorUseRule=UseColor_Specified, specifiedColor=<rgba 0.0953 0.0953 0.0953 1>}
      uVRegion {bIsValid=false, max=<0 0>, min=<0 0>}
     }
     hoveredForeground {colorUseRule=UseColor_ColorTable, specifiedColor=<rgba 0 0 0 0>}
     hoveredSlateSound {resourceObject=null}
     normal {
      bIsDynamicallyLoaded=false
      drawAs=Box
      imageType=NoImage
      mirroring=NoMirror
      resourceName=None
      resourceObject=null
      tiling=NoTile
      imageSize=<32 32>
      margin=<ltrb 0 0 0 0>
      outlineSettings {
       bUseBrushTransparency=false
       roundingType=FixedRadius
       width=1
       color {colorUseRule=UseColor_ColorTable, specifiedColor=<rgba 0 0 0 0>}
       cornerRadii=<4 4 4 4>
      }
      tintColor {colorUseRule=UseColor_Specified, specifiedColor=<rgba 0.0284 0.0284 0.0284 1>}
      uVRegion {bIsValid=false, max=<0 0>, min=<0 0>}
     }
     normalForeground {colorUseRule=UseColor_ColorTable, specifiedColor=<rgba 0 0 0 0>}
     normalPadding=<ltrb 12 1.5 12 1.5>
     pressed {
      bIsDynamicallyLoaded=false
      drawAs=Box
      imageType=NoImage
      mirroring=NoMirror
      resourceName=None
      resourceObject=null
      tiling=NoTile
      imageSize=<32 32>
      margin=<ltrb 0 0 0 0>
      outlineSettings {
       bUseBrushTransparency=false
       roundingType=FixedRadius
       width=1
       color {colorUseRule=UseColor_ColorTable, specifiedColor=<rgba 0 0 0 0>}
       cornerRadii=<4 4 4 4>
      }
      tintColor {colorUseRule=UseColor_Specified, specifiedColor=<rgba 0.0284 0.0284 0.0284 1>}
      uVRegion {bIsValid=false, max=<0 0>, min=<0 0>}
     }
     pressedForeground {colorUseRule=UseColor_ColorTable, specifiedColor=<rgba 0 0 0 0>}
     pressedPadding=<ltrb 12 2.5 12 0.5>
     pressedSlateSound {resourceObject=null}
    }}
   slot {class=HorizontalBoxSlot, properties {padding=<ltrb 0.5 0 0.5 0>, size {sizeRule=Fill, value=1}}}
  }
  Text_8 {
   widgetClassPath=&UEFN_TextBlock
   parent=NewAsset_CancelButton
   widget=Text_8
   properties {
    clipping=ClipToBounds
    text=Cancel
    font {
     bForceMonospaced=false
     bMaterialIsStencil=false
     fontMaterial=null
     fontObject=&NotoSans
     letterSpacing=5
     monospacedWidth=1
     size=10
     skewAmount=0
     typefaceFontName=NotoSans-Regular
     outlineSettings {
      bApplyOutlineToDropShadows=false
      bMiteredCorners=false
      bSeparateFillAlpha=false
      outlineMaterial=null
      outlineSize=0
      outlineColor=<rgba 0 0 0 1>
     }
    }
   }
   slot {class=ButtonSlot, properties {padding=<ltrb 0 0 0 0>}}
  }
 ]
}
