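; Editor utility row entry for the Music Party tool, representing either a title row or a level actor and its music prop component.
; SetVariables stores actor, color and title state, then looks up the actor's MusicPartyPropComponentBase and returns whether one exists, plus the group name.
; A cast chain over Controller, Radio and Actor component subclasses sets the prop type and speaker asset, with a fallback type.
; Getters expose the actor and its display name.
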(fn GetLevelActor
  (return (Variables|InitializedVariables|GetLevelActor)))

(fn GetLevelActorDisplayName
  (bind _levelactor (CallFunction|GetLevelActor))
  (Utilities|IsValid _levelactor
    (:Is%20Valid
      (return (Utilities|Text|ToText(String) (Utilities|GetDisplayName _levelactor))))
    (:Is%20Not%20Valid
      (return ""))))

(fn SetVariables
  (params (LevelActor :type Actor%20Object%20Reference) (IsTitleRow :type Boolean) (Optional:%20TitleText :type Text) (Color :type Linear%20Color%20Structure :default "(R=1.000000,G=1.000000,B=1.000000,A=1.000000)") (MusicController :type Actor%20Class%20Reference) (SpeakerActors :type Array%20of%20Actor%20Soft%20Object%20References) (EventTag :type Gameplay%20Tag%20Structure) (MusicGroupName :type Name))
  (bind _music_prop (Variables|Default|GetMusic%20Prop))
  (Variables|WidgetOrg|SetAssignedColor Color)
  (Variables|InitializedVariables|SetLevelActor LevelActor)
  (bind _output_get (Variables|WidgetOrg|SetisTitleRow IsTitleRow))
  (if _output_get
    (Variables|WidgetOrg|SetTitleText (Variables|GetOptional:%20TitleText))
    (Variables|WidgetOrg|SethasPropComponent false)
    (return false "None")
    (else
      (bind _output_get_2 (Variables|Default|SetMusic%20Prop (Actor|GetComponentbyClass LevelActor "/Game/Sounds/Fort_Audio_Blueprints/MusicParty/MusicPartyPropComponentBase.MusicPartyPropComponentBase_C")))
      (Utilities|IsValid _output_get_2
        (:Is%20Valid
          (Variables|WidgetOrg|SethasPropComponent true)
          (Variables|InitializedVariables|SetMusicController (Variables|InitializedVariables|GetMusicController))
          (Variables|InitializedVariables|SetSpeakerActors (Variables|InitializedVariables|GetSpeakerActors))
          (Variables|InitializedVariables|SetMusicEventTag (Variables|GetEventTag))
          (Variables|InitializedVariables|SetGroupName (Variables|GetMusicGroupName))
          (bind _asmusic_party_prop_component_controller (Utilities|Casting|CastToMusicPartyPropComponent_Controller _music_prop)
            (:then
              (Variables|CalculatedVariables|SetMusicSpeaker (Variables|/Game/Sounds/Fort_Audio_Blueprints/MusicParty/MusicPartyPropComponent_SpeakerAssets.MusicPartyPropComponent_SpeakerAssets_C|GetSpeakerAsset _asmusic_party_prop_component_controller))
              (Variables|CalculatedVariables|SetMusicPropType "NewEnumerator0"))
            (:CastFailed
              (bind _asmusic_party_prop_component_radio (Utilities|Casting|CastToMusicPartyPropComponent_Radio _music_prop)
                (:then
                  (Variables|CalculatedVariables|SetMusicSpeaker (Variables|/Game/Sounds/Fort_Audio_Blueprints/MusicParty/MusicPartyPropComponent_SpeakerAssets.MusicPartyPropComponent_SpeakerAssets_C|GetSpeakerAsset _asmusic_party_prop_component_radio))
                  (Variables|CalculatedVariables|SetMusicPropType "NewEnumerator1"))
                (:CastFailed
                  (bind _asmusic_party_prop_component_actor (Utilities|Casting|CastToMusicPartyPropComponent_Actor _music_prop)
                    (:then
                      (Variables|CalculatedVariables|SetMusicSpeaker (Variables|/Game/Sounds/Fort_Audio_Blueprints/MusicParty/MusicPartyPropComponent_SpeakerAssets.MusicPartyPropComponent_SpeakerAssets_C|GetSpeakerAsset _asmusic_party_prop_component_actor))
                      (Variables|CalculatedVariables|SetMusicPropType "NewEnumerator2"))
                    (:CastFailed
                      (Variables|CalculatedVariables|SetMusicPropType "NewEnumerator3")))))))
          (return true (Variables|GetMusicGroupName)))
        (:Is%20Not%20Valid
          (Variables|WidgetOrg|SethasPropComponent false)
          (return false "None"))))))
