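; Editor utility for multi-editing Music Party props, storing a batch of prop components and their owning level actors.
; AddNewPropComponent records the given prop component type, then adds the component and its actor to transient arrays with AddUnique to avoid duplicates.
; The stored type is a single value that each call overwrites.
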(fn AddNewPropComponent
  (params (PropType :type EMusicPartyPropComponentType%20Enum :default "NewEnumerator0") (PropComponent :type Base%20Class:%20Music%20Party%20Prop%20Base%20Object%20Reference) (LevelActor :type Actor%20Object%20Reference))
  (Variables|Transient|SetPropComponentType PropType)
  (Utilities|Array|AddUnique (Variables|Transient|GetPropComponents) PropComponent)
  (Utilities|Array|AddUnique (Variables|Transient|GetLevelActors) LevelActor))
