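; MusicPartyEditor_NewAsset_Base is an editor utility base Blueprint for creating new assets in the MusicParty audio tooling.
; It declares overridable hooks: a CancelAction event, a ConfirmAction function taking an ExtendSpawn flag, and a SetVariables function taking a ClassDefaults object.
; All three are empty stubs here, with ConfirmAction returning an empty value, so the working logic is not in this base.
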(event Custom|CancelAction)

(fn ConfirmAction
  (params (ExtendSpawn :type Boolean))
  (return ""))

(fn SetVariables
  (params (ClassDefaults :type Object%20Reference)))
