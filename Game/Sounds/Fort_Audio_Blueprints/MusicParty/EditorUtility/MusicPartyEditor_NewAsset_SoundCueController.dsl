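; Editor utility that creates a new MusicParty sound cue controller asset from the default Cue controller template.
; ConfirmAction duplicates the template through the AssetTools dialog using the stored name and path, and returns the new asset's path string.
; SetVariables casts the supplied class defaults to the MusicParty controller base, records a transaction, then writes the selected audio bus and Sound Cue into it.
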(fn ConfirmAction
  (params (ExtendSpawn :type Boolean))
  (bind _returnvalue (EditorScripting|AssetTools|GetAssetTools))
  (bind _returnvalue_2 (EditorScripting|AssetTools|DuplicateAssetwithDialog _returnvalue (Variables|Path|GetName) (Variables|Path|GetPath) "/Game/Sounds/Fort_Audio_Blueprints/MusicParty/MusicPartyController_Default_Cue.MusicPartyController_Default_Cue"))
  (return (Utilities|GetObjectPathString _returnvalue_2)))

(fn SetVariables
  (params (ClassDefaults :type Object%20Reference))
  (bind _asmusicpartycontrollercue (Utilities|Casting|CastToMusicPartyControllerBase ClassDefaults))
  (Transactions|TransactObject _asmusicpartycontrollercue)
  (Variables|/Game/Sounds/Fort_Audio_Blueprints/MusicParty/MusicPartyController_Base.MusicPartyController_Base_C|SetAudioBus (Variables|Assets|GetAudioBus) _asmusicpartycontrollercue)
  (Variables|/Game/Sounds/Fort_Audio_Blueprints/MusicParty/MusicPartyControllerBase.MusicPartyControllerBase_C|SetMusicToPlay (Variables|Assets|GetSound%20Cue) _asmusicpartycontrollercue))
