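; Editor utility that creates a new MusicPartySpeaker asset for the MusicParty audio system.
; ConfirmAction opens a duplicate-with-dialog on the default speaker template, using the configured name and path.
; It casts the copy to MusicPartySpeaker and writes in the utility's damaged and destroyed sounds, filter, attenuation, source preset chain and concurrency settings.
; The ExtendSpawn parameter is unused, and no explicit save follows.
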(fn ConfirmAction
  (params (ExtendSpawn :type Boolean))
  (bind _returnvalue (EditorScripting|AssetTools|GetAssetTools))
  (bind _returnvalue_2 (EditorScripting|AssetTools|DuplicateAssetwithDialog _returnvalue (Variables|Path|GetName) (Variables|Path|GetPath) "/Game/Sounds/Fort_Audio_Blueprints/MusicParty/DefaultSpeakers/MPS_DefaultSpeaker.MPS_DefaultSpeaker"))
  (bind _asmusic_party_speaker (Utilities|Casting|CastToMusicPartySpeaker _returnvalue_2))
  (Variables|/Game/Sounds/Fort_Audio_Blueprints/MusicParty/MusicPartySpeaker.MusicPartySpeaker_C|SetOnSpeakerDamagedSound (Variables|Assets|GetOnSpeakerDamagedSound) _asmusic_party_speaker)
  (Variables|/Game/Sounds/Fort_Audio_Blueprints/MusicParty/MusicPartySpeaker.MusicPartySpeaker_C|SetOnSpeakerDestroyedSound (Variables|Assets|GetOnSpeakerDestroyedSound) _asmusic_party_speaker)
  (Variables|/Game/Sounds/Fort_Audio_Blueprints/MusicParty/MusicPartySpeaker.MusicPartySpeaker_C|SetOnAllSpeakersDestroyedSound (Variables|Assets|GetOnAllSpeakersDestroyedSound) _asmusic_party_speaker)
  (Variables|/Game/Sounds/Fort_Audio_Blueprints/MusicParty/MusicPartySpeaker.MusicPartySpeaker_C|SetLPF (Variables|SpeakerSettings|GetLPF) _asmusic_party_speaker)
  (Variables|/Game/Sounds/Fort_Audio_Blueprints/MusicParty/MusicPartySpeaker.MusicPartySpeaker_C|SetHPF (Variables|SpeakerSettings|GetHPF) _asmusic_party_speaker)
  (Variables|/Game/Sounds/Fort_Audio_Blueprints/MusicParty/MusicPartySpeaker.MusicPartySpeaker_C|SetSoundAttenuation (Variables|SpeakerSettings|GetSoundAttenuation) _asmusic_party_speaker)
  (Variables|/Game/Sounds/Fort_Audio_Blueprints/MusicParty/MusicPartySpeaker.MusicPartySpeaker_C|SetSourcePresetChain (Variables|SpeakerSettings|GetSourcePresetChain) _asmusic_party_speaker)
  (Variables|/Game/Sounds/Fort_Audio_Blueprints/MusicParty/MusicPartySpeaker.MusicPartySpeaker_C|SetConcurrencySet (Variables|SpeakerSettings|GetConcurrencySet) _asmusic_party_speaker))
