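; Editor utility that creates a new Music Party wave controller asset from the default wave controller template.
; ConfirmAction duplicates the template through a dialog and, when ExtendSpawn is set, also duplicates a template audio bus with an "AB_" name derived from the suffix after "B_MusicParty_".
; SetVariables registers the controller's class defaults for transaction, then writes the waves, audio bus and dB volume adjustment into them.
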(fn ConfirmAction
  (params (ExtendSpawn :type Boolean))
  (bind (_lefts _rights _returnvalue) (Utilities|String|Split (Variables|Path|GetName) "B_MusicParty_"))
  (bind _returnvalue_2 (EditorScripting|AssetTools|GetAssetTools))
  (bind _returnvalue_3 (EditorScripting|AssetTools|DuplicateAssetwithDialog _returnvalue_2 (Variables|Path|GetName) (Variables|Path|GetPath) "/Game/Sounds/Fort_Audio_Blueprints/MusicParty/MusicPartyController_Default_Wave.MusicPartyController_Default_Wave"))
  (LocalVariables|Set_NewObjectPath (Utilities|GetObjectPathString _returnvalue_3))
  (if (Variables|GetExtendSpawn)
    (bind _returnvalue_4 (EditorScripting|AssetTools|GetAssetTools))
    (bind _returnvalue_5 (EditorScripting|AssetTools|DuplicateAssetwithDialog _returnvalue_4 (Utilities|String|Append "AB_" (Utilities|String|SelectString _rights (Variables|Path|GetName) _returnvalue)) (Variables|Path|GetPath) "/Game/Sounds/Asteria/POI/NeonCity/mx_NeonCity/DiegeticMusic/NeonCityPresets/MusicParty_MS_AudioBus_01.MusicParty_MS_AudioBus_01"))
    (bind _asaudio_bus (Utilities|Casting|CastToAudioBus _returnvalue_5)
      (:then
        (Variables|AudioBus|SetAudioBus _asaudio_bus)
        (return (LocalVariables|Get_NewObjectPath)))
      (:CastFailed
        (return (LocalVariables|Get_NewObjectPath))))
    (else
      (return (LocalVariables|Get_NewObjectPath)))))

(fn SetVariables
  (params (ClassDefaults :type Object%20Reference))
  (bind _asmusic_party_controller_wave (Utilities|Casting|CastToMusicPartyController_Wave ClassDefaults))
  (Transactions|TransactObject _asmusic_party_controller_wave)
  (Variables|/Game/Sounds/Fort_Audio_Blueprints/MusicParty/MusicPartyController_Wave.MusicPartyController_Wave_C|SetMusicToPlay (Variables|Assets|GetWaves) _asmusic_party_controller_wave)
  (Variables|/Game/Sounds/Fort_Audio_Blueprints/MusicParty/MusicPartyController_Base.MusicPartyController_Base_C|SetAudioBus (Variables|AudioBus|GetAudioBus) _asmusic_party_controller_wave)
  (Variables|/Game/Sounds/Fort_Audio_Blueprints/MusicParty/MusicPartyController_Wave.MusicPartyController_Wave_C|SetDbAdjustment (Variables|Assets|GetVolumeAdjustmentDb) _asmusic_party_controller_wave))
